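# Sidebar chat accept panel, a CommonActivatableWidget showing a group conversation with its joined members and pending invites in a social tree list.
# On activation it initializes the conversation viewmodel from the navigation point's ConversationId, and pops back to the Chat tag if the conversation is removed.
# Conversion functions nest non-empty user lists under collapsible headers after the conversation entry, then fill and expand the tree.
# Clicking items opens player profiles; focus defaults to the first item.
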WidgetBlueprint /Sidebar/Panels/Chat/WBP_Sidebar_ChatAcceptPanel
graphs [
 BP_GetDesiredFocusTarget {nodes [
   Get Desired Focus Target()
   n1 = Get WBP_Sidebar_SocialTreeList() -> (WBP_Sidebar_SocialTreeList:WBP Sidebar Social Tree List Object Reference)
   > n2 = Is Valid(InputObject=Get Selected Item!(self:List View Object Reference=Get TreeView!(self:Fort Polymorphic Tree View Object Reference=@n1) -> (TreeView:Tree View Object Reference)) -> (ReturnValue:Object Reference) {is=title K2Node_CallFunction})
   n3 = Get TreeView(self=Get WBP_Sidebar_SocialTreeList())
   n4 = Set Selected Item!(
    execute=@n2.Is Not Valid
    self:List View Object Reference=@n3
    Item:Object Reference=GetItemAt!(self:List View Object Reference=@n3, Index:int=0) -> (ReturnValue:Object Reference)
   ) -> (then) {is=title K2Node_CallFunction}
   Return Node!(execute=[@n4.then, @n2.Is Valid], ReturnValue:Widget Object Reference=Get TreeView(self=Get WBP_Sidebar_SocialTreeList()))
  ]}
 ConvertMembersAndPendingAsSingleListWithHeaders {nodes [
   ConvertMembersAndPendingAsSingleListWithHeaders!() -> (
    then
    JoinedMembers:Array of Fort Social User Viewmodel Object References
    PendingMembers:Array of Fort Social User Viewmodel Object References
   ) {is=title K2Node_FunctionEntry}
   > n1 = Sequence!(execute) -> (then_0, then_1, then_2:Exec, then_3:Exec)
   > Add(
    TargetArray:Array of MVVM Base Viewmodel Object References=Get MembersList() -> (MembersList:Array of MVVM Base Viewmodel Object References)
    NewItem:"MVVM Base Viewmodel Object Reference (by ref)"=Get FortTextChatGroupConversationVM() -> (FortTextChatGroupConversationVM:Fort Group Conversation Viewmodel Object Reference)
   ) {is=title K2Node_CallArrayFunction}
   Branch(
    execute=@n1.then_1
    Condition=Is Not Empty!(TargetArray:Array of Fort Social User Viewmodel Object References=Get JoinedMembers() -> (JoinedMembers:Array of Fort Social User Viewmodel Object References)) -> (ReturnValue:bool)
   )
   n2 = Get MembersCollapsibleHeader() -> (MembersCollapsibleHeader:BP Sidebar Custom Header Entry VM Object Reference)
   > Set with Broadcast ChildrenItems!(
    execute
    ChildrenItems:Array of MVVM Base Viewmodel Object References=Get JoinedMembers()
    self:BP Sidebar Custom Header Entry VM Object Reference=@n2
   ) -> (then, Output_Get:Array of MVVM Base Viewmodel Object References)
   > Add(TargetArray=Get MembersList(), NewItem=@n2)
   Branch(
    execute=@n1.then_2
    Condition=Is Not Empty(TargetArray=Get PendingMembers() -> (PendingMembers:Array of Fort Social User Viewmodel Object References))
   )
   n3 = Get PendingCollapsibleHeader() -> (PendingCollapsibleHeader:BP Sidebar Custom Header Entry VM Object Reference)
   > Set with Broadcast ChildrenItems(ChildrenItems=Get PendingMembers(), self=@n3)
   > Add(TargetArray=Get MembersList(), NewItem=@n3)
   Return Node!(execute=@n1.then_3, MembersListWithHeaders:Array of MVVM Base Viewmodel Object References=Get MembersList())
  ]}
 EventGraph {nodes [
   n1 = "Item Clicked (WBP_Sidebar_SocialTreeList)"!() -> (OutputDelegate, then, ItemVM:Fort Social Item VM Object Reference) {is=title K2Node_ComponentBoundEvent}
   > NavigateToPlayerProfile!(
    execute
    SelfReference:Object Reference=Self-Reference()
    SocialItemVM:Fort Social Item VM Object Reference=@n1.ItemVM
    SidebarNavigationVM:Fort Sidebar Navigation VM Object Reference=Get FortSidebarNavigationVM() -> (FortSidebarNavigationVM:Fort Sidebar Navigation VM Object Reference)
    Sidebar VM:Fort Sidebar VM Object Reference
   ) -> (then)
   n2 = OnConversationHasBeenRemovedChanged!() -> (OutputDelegate, then, bHasBeenRemoved:bool=~) {is=title K2Node_CustomEvent}
   > Branch(Condition=@n2.bHasBeenRemoved)
   > PopToNavigationTag!(
    execute
    self:Fort Sidebar Navigation VM Object Reference=Get FortSidebarNavigationVM()
    NavigationTag:Gameplay Tag Structure=(TagName="UI.StateTree.TransitionTag.Sidebar.Chat")
   ) -> (then, ReturnValue:bool)
   Event On Activated()
   > n3 = GetCurrentNavigationPoint!(execute, self:Fort Sidebar Navigation VM Object Reference=Get FortSidebarNavigationVM()) -> (then, ReturnValue:Fort Sidebar Navigation Point Object Reference)
   > n4 = Cast To BP_NavigationPoint!(execute, Object=@n3) -> (then, CastFailed, AsBP Navigation Point:BP Navigation Point Object Reference) {is=title K2Node_DynamicCast}
   > InitializeTextChatConversationVM!(
    execute
    self:Fort Text Chat Conversation VM Object Reference=Get FortTextChatGroupConversationVM()
    Player:Local Player Object Reference=GetOwningLocalPlayer!(self:Widget Object Reference) -> (ReturnValue:Local Player Object Reference)
    InConversationId:"Name (by ref)"=Get ConversationId!(self:BP Navigation Point Object Reference=@n4) -> (ConversationId:Name=None)
   ) -> (then)
   Event OnInitialized()
   > n5 = Construct Object from Class(Class=/Sidebar/Panels/Social/List/Data/BP_Sidebar_CustomHeaderEntryVM.BP_Sidebar_CustomHeaderEntryVM_C) -> (ReturnValue:BP Sidebar Custom Header Entry VM Object Reference)
   > Set with Broadcast DisplayName!(execute, DisplayName:Text="LOCTEXT(\"Members\")", self:BP Sidebar Custom Header Entry VM Object Reference=@n5) -> (then, Output_Get:Text)
   > Set MembersCollapsibleHeader(MembersCollapsibleHeader:BP Sidebar Custom Header Entry VM Object Reference=@n5) -> (Output_Get:BP Sidebar Custom Header Entry VM Object Reference)
   > n6 = Construct Object from Class(Class=&BP_Sidebar_CustomHeaderEntryVM_C)
   > Set with Broadcast DisplayName(DisplayName="LOCTEXT(\"Pending Invites\")", self=@n6)
   > Set PendingCollapsibleHeader(PendingCollapsibleHeader:BP Sidebar Custom Header Entry VM Object Reference=@n6) -> (Output_Get:BP Sidebar Custom Header Entry VM Object Reference)
  ]}
 SetGroupMembersList {nodes [
   n1 = SetGroupMembersList!() -> (then, MembersWithHeaders:Array of MVVM Base Viewmodel Object References) {is=title K2Node_FunctionEntry}
   n2 = Get TreeView(self=Get WBP_Sidebar_SocialTreeList())
   > Set List Items!(execute, self:List View Object Reference=@n2, InListItems:Array of Object References=@n1) -> (then) {is=title K2Node_CallFunction}
   > ExpandAll(self:Tree View Object Reference=@n2)
  ]}
]
variables [
 MembersCollapsibleHeader {type=BP_Sidebar_CustomHeaderEntryVM_C, default=null}
 PendingCollapsibleHeader {type=BP_Sidebar_CustomHeaderEntryVM_C, default=null}
]
widgets {
 info {
  inheritedWidgetCount=0
  namedSlotCount=1
  parentClass=&CommonActivatableWidget
  rootWidgetClass=/Sidebar/Shared/WBP_Sidebar_PanelLayout.WBP_Sidebar_PanelLayout_C
  widgetCount=3
 }
 widgets [
  WBP_Sidebar_PanelLayout {widgetClassPath=/Sidebar/Shared/WBP_Sidebar_PanelLayout, bIsVariable=true, widget=WBP_Sidebar_PanelLayout, properties {headingText=""}}
  AcceptGridRoot {
   widgetClassPath=&GridPanel
   namedSlotHost=WBP_Sidebar_PanelLayout
   widget=AcceptGridRoot
   properties {bOverride_Cursor=true, columnFill [1], rowFill [1]}
  }
  WBP_Sidebar_SocialTreeList {
   widgetClassPath=/Sidebar/Panels/Social/List/WBP_Sidebar_SocialTreeList
   bIsVariable=true
   parent=AcceptGridRoot
   widget=WBP_Sidebar_SocialTreeList
   slot {class=GridSlot}
  }
 ]
}
