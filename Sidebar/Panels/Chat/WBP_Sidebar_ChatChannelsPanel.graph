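# Sidebar Chats panel, a CommonActivatableWidget listing group text chat conversations with new-chat, back-to-top, empty, error and restriction states.
# Viewmodel-facing functions set visibility, layout and spinner state, and append a load-older button entry to the paginated list.
# Clicks push tagged navigation points for messages, invite acceptance or new chats.
# Activation refreshes conversations and binds a message sound; focus falls back between button, list and restriction.
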WidgetBlueprint /Sidebar/Panels/Chat/WBP_Sidebar_ChatChannelsPanel
graphs [
 AdjustLayoutAndMessageWhenRestrictionVisible {nodes [
   AdjustLayoutAndMessageWhenRestrictionVisible!() -> (then, HasTextChatRestrictions:bool=~) {is=title K2Node_FunctionEntry}
   > "SetText (Text)"(
    self=Get NoChats_Title() -> (NoChats_Title:Common Text Object Reference)
    InText=Select(
     Option 0:Text="LOCTEXT(\"No chats yet\")"
     Option 1:Text="LOCTEXT(\"Text chat is off\")"
     Index=Get HasTextChatRestrictions() -> (HasTextChatRestrictions:bool)
    ) -> (ReturnValue:Text)
   )
   > SetVisibility(
    self=Get NoChats_Message() -> (NoChats_Message:Common Text Object Reference)
    InVisibility=Select(Option 0:ESlateVisibility Enum=SelfHitTestInvisible, Option 1:ESlateVisibility Enum=Collapsed, Index=Get HasTextChatRestrictions()) -> (ReturnValue:ESlateVisibility Enum)
   )
   > SetVisibility(
    self=Get Button_NewChat() -> (Button_NewChat:WBP Sidebar Button Generic Object Reference)
    InVisibility=Select(Option 0=SelfHitTestInvisible, Option 1=Collapsed, Index=Get HasTextChatRestrictions())
   )
   > RequestRefreshFocus(self:Common Activatable Widget Object Reference)
  ]}
 AsyncLoadActionFinished {nodes [
   n1 = AsyncLoadActionFinished!() -> (then, AsyncActionResult:ELastAsyncActionResult Enum=None) {is=title K2Node_FunctionEntry}
   > n2 = Switch on ELastAsyncActionResult!(execute, Selection:ELastAsyncActionResult Enum=@n1) -> (None:Exec, Failed:Exec, Success:Exec) {is=title K2Node_SwitchEnum}
   SetVisibility(execute=[@n2.None, @n2.Success], self=Get Overlay_ErrorResults() -> (Overlay_ErrorResults:Overlay Object Reference), InVisibility=Collapsed)
   SetVisibility(execute=@n2.Failed, self=Get Overlay_ErrorResults(), InVisibility=SelfHitTestInvisible)
  ]}
 BP_GetDesiredFocusTarget {nodes [
   Get Desired Focus Target()
   n1 = Get Button_NewChat()
   n2 = Get LastFocusedWidget() -> (LastFocusedWidget:User Widget Object Reference)
   n3 = NOT Boolean(A=IsValid(Object=@n2))
   n4 = "Equal (Object)"(A=@n2, B=Get Button_NewChat())
   > n5 = Branch(Condition=AND Boolean(A=OR Boolean(A=@n3, B=@n4), B=IsVisible(self=@n1)))
   > Return Node!(execute, ReturnValue:Widget Object Reference=Get Button_NewChat())
   n6 = Get ListView!(self:Fort Polymorphic List View Object Reference=Get ListView_Conversations() -> (ListView_Conversations:WBP Sidebar Text Chat Message List Object Reference)) -> (ListView:Common List View Object Reference)
   n7 = Branch(execute=@n5.else, Condition="integer > integer"(A=GetNumItems!(self:List View Object Reference=@n6) -> (ReturnValue:int=0)))
   n8 = Get ListView(self=Get ListView_Conversations())
   n9 = Get Selected Item!(self:List View Object Reference=@n8) -> (ReturnValue:Object Reference) {is=title K2Node_CallFunction}
   > n10 = Branch(Condition="Equal (Integer)"(A=GetIndexForItem!(self:List View Object Reference=@n8, Item:Object Reference=@n9) -> (ReturnValue:int=0), B=-1))
   > SetSelectedIndex!(execute, self:List View Object Reference=@n8, Index:int=0) -> (then)
   > n11 = ScrollToTop(self:List View Base Object Reference=@n8)
   Return Node(execute=[@n11.then, @n10.else], ReturnValue=Get ListView(self=Get ListView_Conversations()))
   n12 = Get WBP_Sidebar_TextChat_Restrictions() -> (WBP_Sidebar_TextChat_Restrictions:WBP Sidebar Text Chat Restrictions Object Reference)
   n13 = Get Button_Restriction!(self:WBP Sidebar Text Chat Restrictions Object Reference=@n12) -> (Button_Restriction:WBP Sidebar Button Generic Object Reference)
   n14 = Branch(execute=@n7.else, Condition=AND Boolean(A=IsVisible(self=@n12), B=IsVisible(self=@n13)))
   > Return Node(ReturnValue=Get Button_Restriction(self=Get WBP_Sidebar_TextChat_Restrictions()))
   Return Node(execute=@n14.else, ReturnValue=Self-Reference())
  ]}
 CF_AddPaginationButtonToList {nodes [
   n1 = CF_AddPaginationButtonToList!() -> (then, ConversationVMs:Array of Fort Text Chat Conversation VM Object References, HasOlderPagesOfConversations:bool=~) {is=title K2Node_FunctionEntry}
   > n2 = Set ConversationListWithPaginationButton(ConversationListWithPaginationButton:Array of MVVM Base Viewmodel Object References=@n1) -> (Output_Get:Array of MVVM Base Viewmodel Object References)
   > n3 = Branch(Condition=AND Boolean(
     A=Is Not Empty!(TargetArray:Array of MVVM Base Viewmodel Object References=@n2) -> (ReturnValue:bool)
     B=Get HasOlderPagesOfConversations() -> (HasOlderPagesOfConversations:bool)
    ))
   > n4 = Add(
    TargetArray:Array of MVVM Base Viewmodel Object References=Get ConversationListWithPaginationButton() -> (ConversationListWithPaginationButton:Array of MVVM Base Viewmodel Object References)
    NewItem:"MVVM Base Viewmodel Object Reference (by ref)"=Get LoadOlderConversationsButtonEntry() -> (LoadOlderConversationsButtonEntry:BP Sidebar Custom Button Entry VM Object Reference)
   ) {is=title K2Node_CallArrayFunction}
   Return Node!(execute=[@n4.then, @n3.else], ConversationList:Array of MVVM Base Viewmodel Object References=Get ConversationListWithPaginationButton())
  ]}
 CF_BackToTopVisibility {nodes [
   n1 = CF_BackToTopVisibility!() -> (then, IsInMostRecentPage:bool=~) {is=title K2Node_FunctionEntry}
   > Return Node!(execute, BackToTopPageButtonVisibility:ESlateVisibility Enum=Select(Option 0=SelfHitTestInvisible, Option 1=Collapsed, Index=@n1))
  ]}
 CF_ShouldShowSpinner {nodes [
   n1 = CF_ShouldShowSpinner!() -> (
    then
    LoadConversationsRequest_bIsInProgress:bool=~
    LoadConversationsRequest_LastResult:ELastAsyncActionResult Enum
    LoadConversationsRequest_LastResultMessage:Text
    bCurrentLoadWasUserRequested:bool=~
    Conversations:Array of Fort Text Chat Conversation VM Object References
   ) {is=title K2Node_FunctionEntry}
   n2 = Is Empty!(TargetArray:Array of Fort Text Chat Conversation VM Object References=@n1.Conversations) -> (ReturnValue:bool)
   > Return Node!(execute, bShouldShowSpinner:bool=AND Boolean(A=@n1, B=OR Boolean(A=@n1.bCurrentLoadWasUserRequested, B=@n2)))
  ]}
 EventGraph {nodes [
   n1 = "On Item Clicked (ListView_Conversations)"!() -> (OutputDelegate, then, Item:Object Reference) {is=title K2Node_ComponentBoundEvent}
   > n2 = Cast To FortTextChatConversationVM!(execute, Object=@n1.Item) -> (then, CastFailed, AsFort Text Chat Conversation VM:Fort Text Chat Conversation VM Object Reference) {is=title K2Node_DynamicCast}
   > n3 = Construct Object from Class!(
    execute
    Class=/Sidebar/Shared/Navigation/BP_NavigationPoint.BP_NavigationPoint_C
    self
    SocialUser:Fort Social User Object Reference
    ConversationId:Name=Get ConversationId!(self:Fort Text Chat Conversation VM Object Reference=@n2) -> (ConversationId:Name=None)
    ConfirmationVM:Sidebar Confirmation Panel VM Object Reference
    ReservedSpots:int=0
    SocialItemVM:Fort Social Item VM Object Reference
    NavigationTag:Gameplay Tag Structure=Select(
     Option 0:Gameplay Tag Structure=(TagName="UI.StateTree.TransitionTag.Sidebar.Chat.MessageList")
     Option 1:Gameplay Tag Structure=(TagName="UI.StateTree.TransitionTag.Sidebar.Chat.AcceptInvite")
     Index=Get bHasPendingInvite!(self:Fort Text Chat Conversation VM Object Reference=@n2) -> (bHasPendingInvite:bool)
    ) -> (ReturnValue:Gameplay Tag Structure)
   ) -> (then, ReturnValue:BP Navigation Point Object Reference)
   > PushNavigation!(
    execute
    self:Fort Sidebar Navigation VM Object Reference=Get FortSidebarNavigationVM() -> (FortSidebarNavigationVM:Fort Sidebar Navigation VM Object Reference)
    NavigationPoint:Fort Sidebar Navigation Point Object Reference=@n3
   ) -> (then)
   "On Clicked (Button_BackToTopMostRecentPage)"!() -> (OutputDelegate, then, Button:Common Button Base Object Reference) {is=title K2Node_ComponentBoundEvent}
   > LoadMostRecentConversations(self:Fort Text Chat Conversations List VM Object Reference=Get FortTextChatGroupConversationsListVM() -> (FortTextChatGroupConversationsListVM:Fort Text Chat Group Conversations List VM Object Reference))
   "On Clicked (Button_NewChat)"!() -> (OutputDelegate, then, Button:Common Button Base Object Reference) {is=title K2Node_ComponentBoundEvent}
   > n4 = Construct Object from Class(
    Class=&BP_NavigationPoint_C
    ConversationId=None
    ReservedSpots=0
    NavigationTag=(TagName="UI.StateTree.TransitionTag.Sidebar.Chat.Invite")
   )
   > PushNavigation(self=Get FortSidebarNavigationVM(), NavigationPoint=@n4)
   n5 = "On Custom List Button Clicked (ListView_Conversations)"!() -> (OutputDelegate, then, ButtonType:Name=None) {is=title K2Node_ComponentBoundEvent}
   > Branch(Condition="Equal (Name)"(A=@n5.ButtonType, B=Get LoadOlderConversationsButtonType() -> (LoadOlderConversationsButtonType:Name=None)))
   > LoadOlderConversations(self:Fort Text Chat Conversations List VM Object Reference=Get FortTextChatGroupConversationsListVM())
   n6 = Get ListView(self=Get ListView_Conversations())
   > SetSelectedIndex(self=@n6, Index=0)
   > ScrollToTop(self=@n6)
   n7 = OnHandleInputMethodChanged!() -> (OutputDelegate, then, bNewInputType:ECommonInputType Enum) {is=title K2Node_CustomEvent}
   > SetReturnFocusToSelection!(
    execute
    self:List View Object Reference=Get ListView(self=Get ListView_Conversations())
    bEnabled:bool=Select!(MouseAndKeyboard:bool=false, Gamepad:bool=true, Touch:bool=false, Count:bool=false, Index:ECommonInputType Enum=@n7.bNewInputType) -> (ReturnValue:bool)
   ) -> (then)
   OnMessageReceived() {is=title K2Node_CustomEvent}
   > PlayTextChatSound!(execute, self, Sound:Sound Base Object Reference=Get MessageReceivedSound() -> (MessageReceivedSound:Sound Base Object Reference)) -> (then)
   Event Destruct()
   n8 = Create Event()
   n9 = Get CommonInputSubsystem()
   > Unbind Event from On Input Method Changed(self=@n9, Delegate=@n8)
   Event Construct()
   > Bind Event to On Input Method Changed(self=@n9, Delegate=@n8)
   > OnHandleInputMethodChanged!(execute, self, bNewInputType:ECommonInputType Enum=GetCurrentInputType(self=@n9)) -> (then) {is=title K2Node_CallFunction}
   Event OnInitialized()
   > n10 = Construct Object from Class!(
    execute
    Class=/Sidebar/Panels/Social/List/Data/BP_Sidebar_CustomButtonEntryVM.BP_Sidebar_CustomButtonEntryVM_C
    self
    DisplayName:Text="LOCTEXT(\"Load more conversations\")"
    ButtonType:Name=Get LoadOlderConversationsButtonType() -> (LoadOlderConversationsButtonType=None)
    ButtonIcon:Texture 2D Object Reference
    UseIcon:bool=False
   ) -> (then, ReturnValue:BP Sidebar Custom Button Entry VM Object Reference)
   > Set LoadOlderConversationsButtonEntry(LoadOlderConversationsButtonEntry:BP Sidebar Custom Button Entry VM Object Reference=@n10) -> (Output_Get:BP Sidebar Custom Button Entry VM Object Reference)
   n11 = Event PreConstruct()
   > Sequence()
   > Branch(Condition=@n11.IsDesignTime)
   > SetGridAndRestrictionsLayout!(execute, self, HasConversations:bool=true) -> (then)
   > AdjustLayoutAndMessageWhenRestrictionVisible!(execute, self, HasTextChatRestrictions:bool=false) -> (then) {is=title K2Node_CallFunction}
   Event On Activated()
   n12 = Create Event()
   n13 = Get FortTextChatGroupConversationsListVM()
   > Bind Event to On Message Received(self:Fort Text Chat Conversations List VM Object Reference=@n13, Delegate=@n12) {is=title K2Node_AddDelegate}
   n14 = Get ListView_Conversations()
   > n15 = For Each Loop(Array:Array of Object References=GetListItems!(self:List View Object Reference=Get ListView(self=@n14)) -> (ReturnValue:Array of Object References)) -> (Array Element:Object Reference)
   > n16 = Cast To FortTextChatConversationVM(Object=@n15)
   > RefreshConversationMessages(self:Fort Text Chat Conversation VM Object Reference=@n16)
   SidebarFeatureFinishedLoading(execute=@n15.Completed, self:Fort Sidebar VM Object Reference=Get FortSidebarVM() -> (FortSidebarVM:Fort Sidebar VM Object Reference))
   Event On Deactivated()
   > Unbind Event from On Message Received(self:Fort Text Chat Conversations List VM Object Reference=@n13, Delegate=@n12) {is=title K2Node_RemoveDelegate}
  ]}
 GetChatListVisibility {nodes [
   n1 = GetChatListVisibility!() -> (then, AsyncActionResult:ELastAsyncActionResult Enum=None, ConversationsList:Array of Fort Text Chat Conversation VM Object References) {is=title K2Node_FunctionEntry}
   n2 = Is Not Empty(TargetArray:Array of Fort Text Chat Conversation VM Object References=@n1.ConversationsList)
   n3 = OR Boolean(A="Equal (Enum)"(A:ELastAsyncActionResult Enum=@n1, B:ELastAsyncActionResult Enum=Success), B="Equal (Enum)"(A=@n1, B=None))
   > Return Node!(
    execute
    Visibility:ESlateVisibility Enum=Select!(Option 0:ESlateVisibility Enum=Collapsed, Option 1:ESlateVisibility Enum=Visible, Index:bool=AND Boolean(A=@n3, B=@n2)) -> (ReturnValue:ESlateVisibility Enum)
   )
  ]}
 GetNoChatsMessageVisibility {nodes [
   n1 = GetNoChatsMessageVisibility!() -> (
    then
    ConversationsList:Array of Fort Text Chat Conversation VM Object References
    IsInMostRecentPage:bool=~
    HasOlderPagesOfConversations:bool=~
   ) {is=title K2Node_FunctionEntry}
   n2 = Is Empty(TargetArray=@n1)
   n3 = NOT Boolean(A=@n1.HasOlderPagesOfConversations)
   > Return Node(Visibility=Select(Option 0=Collapsed, Option 1=HitTestInvisible, Index=AND Boolean!(A=@n2, B=@n1.IsInMostRecentPage, C:bool=@n3) -> (ReturnValue)))
  ]}
 OnFocusEnterChanged {nodes [
   n1 = OnFocusEnterChanged!() -> (then, Type:EUINavigation Enum, OldFocusedWidget:Widget Object Reference, NewFocusedWidget:Widget Object Reference) {is=title K2Node_FunctionEntry}
   > Set LastFocusedWidget(LastFocusedWidget:User Widget Object Reference=GetOuterUserWidget!(Widget:Widget Object Reference=@n1.NewFocusedWidget) -> (ReturnValue:User Widget Object Reference)) -> (Output_Get:User Widget Object Reference)
  ]}
 PlayTextChatSound {nodes [
   n1 = PlayTextChatSound!() -> (then, Sound:Sound Base Object Reference) {is=title K2Node_FunctionEntry}
   > Branch(Condition=IsTextChatSoundEnabled!(self:Peristent Text Chat Base Viewmodel Object Reference=Get FortTextChatGroupConversationsListVM()) -> (ReturnValue:bool))
   > Is Valid(InputObject=@n1)
   > PlaySound2D(Sound=@n1)
  ]}
 PopulateConversationEntries {nodes [
   n1 = PopulateConversationEntries!() -> (then, ConversationVMs:Array of MVVM Base Viewmodel Object References) {is=title K2Node_FunctionEntry}
   > Set List Items!(execute, self:List View Object Reference=Get ListView(self=Get ListView_Conversations()), InListItems:Array of Object References=@n1) -> (then) {is=title K2Node_CallFunction}
   n2 = Is Not Empty(TargetArray:Array of MVVM Base Viewmodel Object References=@n1)
   > SetGridAndRestrictionsLayout(HasConversations=@n2)
   > Branch(Condition=@n2)
   > RequestRefreshFocus()
  ]}
 SetGridAndRestrictionsLayout {nodes [
   SetGridAndRestrictionsLayout!() -> (then, HasConversations:bool=~) {is=title K2Node_FunctionEntry}
   > SetVerticalAlignment(
    self:Vertical Box Slot Object Reference=SlotAsVerticalBoxSlot(Widget=Get ConversationsGrid() -> (ConversationsGrid:Grid Panel Object Reference))
    InVerticalAlignment=Select(
     Option 0:EVerticalAlignment Enum=VAlign_Center
     Option 1:EVerticalAlignment Enum=VAlign_Fill
     Index=Get HasConversations() -> (HasConversations:bool)
    ) -> (ReturnValue:EVerticalAlignment Enum)
   )
   > SetGridSlotColumnAndRow!(
    execute
    Widget:Widget Object Reference=Get WBP_Sidebar_TextChat_Restrictions()
    Row:int=SelectInt(A=0, B=2, bPickA=Get HasConversations())
    Row Span:int=0
    Column:int=0
    Column Span:int=0
   ) -> (then)
  ]}
]
variables [
 LastFocusedWidget {type=UserWidget, default=null}
 LoadOlderConversationsButtonEntry {type=BP_Sidebar_CustomButtonEntryVM_C, default=null}
 LoadOlderConversationsButtonType {type=string, default=LoadOlderMessaqes}
 MessageReceivedSound {type=SoundBase, default=/Game/Sounds/Fort_UI_Sounds/Chat_UI/MSS/MSS_UI_FrontEnd_Chat_MessageReceived}
]
widgets {
 info {
  inheritedWidgetCount=0
  namedSlotCount=1
  parentClass=&CommonActivatableWidget
  rootWidgetClass=/Sidebar/Shared/WBP_Sidebar_PanelLayout.WBP_Sidebar_PanelLayout_C
  widgetCount=15
 }
 widgets [
  PanelLayout {
   widgetClassPath=/Sidebar/Shared/WBP_Sidebar_PanelLayout
   bIsVariable=true
   widget=PanelLayout
   properties {headingText=Chats, showBackButton=false}
  }
  VerticalBox_WithNavComponent {
   widgetClassPath=&VerticalBox
   namedSlotHost=PanelLayout
   widget=VerticalBox_WithNavComponent
   properties {bOverride_Cursor=true}
   uIComponents [{
     component=&self:UIComponentWidgetBlueprintExtension_0.ComponentContainer.NavigationUIComponent_0
     componentClassPath=/Script/UMG.NavigationUIComponent
    }]
  }
  Button_NewChat {
   widgetClassPath=/Sidebar/Buttons/WBP_Sidebar_ButtonGeneric
   bIsVariable=true
   parent=VerticalBox_WithNavComponent
   widget=Button_NewChat
   properties {
    minHeightOverride=60
    text=New chat
    font {
     bForceMonospaced=false
     bMaterialIsStencil=false
     fontMaterial=/FortUILibrary/UIKit/Buttons/MaterialInstances/CTAButtons/MI_UIKit_TextQuiet_Prompt
     fontObject=&HeadingNow
     letterSpacing=10
     monospacedWidth=1
     size=15.75
     skewAmount=0
     typefaceFontName=74 Regular
     outlineSettings {
      bApplyOutlineToDropShadows=false
      bMiteredCorners=false
      bSeparateFillAlpha=false
      outlineMaterial=/FortUILibrary/Systems/ModularButton/MaterialInstances/MI_UIKit_TextOutline
      outlineSize=0
      outlineColor=<rgba 0 0 0 1>
     }
    }
   }
   slot {class=VerticalBoxSlot, properties {padding=<ltrb 18 12 18 24>}}
  }
  ConversationsGrid {
   widgetClassPath=&GridPanel
   bIsVariable=true
   parent=VerticalBox_WithNavComponent
   widget=ConversationsGrid
   properties {bOverride_Cursor=true, clipping=ClipToBounds, columnFill [1], rowFill [0,1,0]}
   slot {class=VerticalBoxSlot, properties {padding=<ltrb -18 0 -18 -18>, size {sizeRule=Fill, value=1}}}
  }
  Overlay_ErrorResults {
   widgetClassPath=&Overlay
   bIsVariable=true
   parent=ConversationsGrid
   widget=Overlay_ErrorResults
   properties {visibility=HitTestInvisible}
   slot {class=GridSlot, properties {row=1}}
  }
  ErrorMessage {
   widgetClassPath=&CommonTextBlock
   bIsVariable=true
   parent=Overlay_ErrorResults
   widget=ErrorMessage
   properties {
    autoWrapText=true
    bIsScrollingEnabled=false
    justification=Center
    mobileFontSizeMultiplier=1.33
    scrollStyle=/FortUILibrary/Styles/Typography/Foundation/TextScrollStyle-NoFade.TextScrollStyle-NoFade_C
    text=There was an error while loading chats.
    textOverflowPolicy=Ellipsis
    colorAndOpacity {colorUseRule=UseColor_Specified, specifiedColor=<rgba 1 0.1808 0.1689 1>}
    font {
     bForceMonospaced=false
     bMaterialIsStencil=false
     fontMaterial=null
     fontObject=&HeadingNow
     letterSpacing=0
     monospacedWidth=1
     size=13.5
     skewAmount=0
     typefaceFontName=74 Regular
     outlineSettings {
      bApplyOutlineToDropShadows=false
      bMiteredCorners=false
      bSeparateFillAlpha=false
      outlineMaterial=null
      outlineSize=0
      outlineColor=<rgba 0 0 0 1>
     }
    }
    shadowOffset=<0 0>
   }
   slot {class=OverlaySlot, properties {horizontalAlignment=HAlign_Fill, padding=<ltrb 0 20 0 8>}}
  }
  WBP_Sidebar_TextChat_Restrictions {
   widgetClassPath=&/WBP_Sidebar_TextChat_Restrictions
   bIsVariable=true
   parent=ConversationsGrid
   widget=WBP_Sidebar_TextChat_Restrictions
   properties {showBackplate=false}
   slot {class=GridSlot, properties {verticalAlignment=VAlign_Top, padding=<ltrb 18 0 18 0>}}
  }
  NoChats_VerticalBox {
   widgetClassPath=&VerticalBox
   bIsVariable=true
   parent=ConversationsGrid
   widget=NoChats_VerticalBox
   properties {bOverride_Cursor=true}
   slot {class=GridSlot, properties {row=1, verticalAlignment=VAlign_Center, padding=<ltrb 30 0 30 0>}}
  }
  Icon {
   widgetClassPath=&Image
   bIsVariable=true
   parent=NoChats_VerticalBox
   widget=Icon
   properties {
    visibility=SelfHitTestInvisible
    brush {
     bIsDynamicallyLoaded=false
     drawAs=Image
     imageType=FullColor
     mirroring=NoMirror
     resourceName=None
     resourceObject=/Sidebar/Materials/Icons/Textures/T_UI_Icon_Sidebar_NoChats
     tiling=NoTile
     imageSize=<104 96>
     margin=<ltrb 0 0 0 0>
     outlineSettings {
      bUseBrushTransparency=false
      roundingType=HalfHeightRadius
      width=0
      color {colorUseRule=UseColor_Specified, specifiedColor=<rgba 0 0 0 0>}
      cornerRadii=<0 0 0 0>
     }
     tintColor {colorUseRule=UseColor_Specified, specifiedColor=<rgba 0.402 0.402 0.4233 1>}
     uVRegion {bIsValid=false, max=<0 0>, min=<0 0>}
    }
   }
   slot {class=VerticalBoxSlot, properties {horizontalAlignment=HAlign_Center, verticalAlignment=VAlign_Center}}
  }
  NoChats_Title {
   widgetClassPath=&CommonTextBlock
   bIsVariable=true
   parent=NoChats_VerticalBox
   widget=NoChats_Title
   properties {
    bIsScrollingEnabled=false
    justification=Center
    minDesiredWidth=300
    mobileFontSizeMultiplier=1.19
    text=No chats yet
    font {
     bForceMonospaced=false
     bMaterialIsStencil=false
     fontMaterial=/Sidebar/Materials/Fonts/MI_Sidebar_FontColor_Primary
     fontObject=&HeadingNow
     letterSpacing=0
     monospacedWidth=1
     size=20.25
     skewAmount=0
     typefaceFontName=75 Medium
     outlineSettings {
      bApplyOutlineToDropShadows=false
      bMiteredCorners=false
      bSeparateFillAlpha=false
      outlineMaterial=null
      outlineSize=0
      outlineColor=<rgba 0 0 0 0.15>
     }
    }
    shadowOffset=<0 0>
   }
   slot {class=VerticalBoxSlot, properties {padding=<ltrb 0 12 0 12>}}
  }
  NoChats_Message {
   widgetClassPath=&CommonTextBlock
   bIsVariable=true
   parent=NoChats_VerticalBox
   widget=NoChats_Message
   properties {
    autoWrapText=true
    bIsScrollingEnabled=false
    justification=Center
    minDesiredWidth=300
    mobileFontSizeMultiplier=1.41
    text=Your chats will appear here. Keep the conversation going in-game or on the Epic Games mobile app or launcher.
    font {
     bForceMonospaced=false
     bMaterialIsStencil=false
     fontMaterial=/Sidebar/Materials/Fonts/MI_Sidebar_FontColor_Secondary
     fontObject=&HeadingNow
     letterSpacing=0
     monospacedWidth=1
     size=12.75
     skewAmount=0
     typefaceFontName=74 Regular
     outlineSettings {
      bApplyOutlineToDropShadows=false
      bMiteredCorners=false
      bSeparateFillAlpha=false
      outlineMaterial=null
      outlineSize=0
      outlineColor=<rgba 0 0 0 0.15>
     }
    }
    shadowOffset=<0 0>
   }
   slot {class=VerticalBoxSlot}
  }
  HasChats_VerticalBox {widgetClassPath=&VerticalBox, parent=ConversationsGrid, widget=HasChats_VerticalBox, slot {class=GridSlot, properties {row=1}}}
  Button_BackToTopMostRecentPage {
   widgetClassPath=&WBP_Sidebar_ButtonGeneric
   bIsVariable=true
   parent=HasChats_VerticalBox
   widget=Button_BackToTopMostRecentPage
   properties {
    configuration=Number
    contentHorizontalAlignment=HAlign_Fill
    minHeightOverride=56
    promptHorizontalAlignment=HAlign_Right
    statusIndicatorHorizontalAlignment=HAlign_Left
    statusIndicatorVerticalAlignment=VAlign_Center
    text=Back to top
    useStatusIndicator=true
    visibility=Collapsed
    buttonPadding=<ltrb 64 0 64 0>
    promptNudge=<0 0>
    statusIndicatorPadding=<ltrb 12 0 0 0>
    triggeringInputAction {dataTable=/Sidebar/Shared/DT_Sidebar_InputActions, rowName=BackToTop}
   }
   slot {class=VerticalBoxSlot, properties {padding=<ltrb 36 12 36 24>}}
  }
  Border_Clipping {
   widgetClassPath=&Border
   parent=HasChats_VerticalBox
   widget=Border_Clipping
   properties {
    clipping=ClipToBounds
    background {
     bIsDynamicallyLoaded=false
     drawAs=NoDrawType
     imageType=NoImage
     mirroring=NoMirror
     resourceName=None
     resourceObject=null
     tiling=NoTile
     imageSize=<32 32>
     margin=<ltrb 0 0 0 0>
     outlineSettings {
      bUseBrushTransparency=false
      roundingType=HalfHeightRadius
      width=0
      color {colorUseRule=UseColor_Specified, specifiedColor=<rgba 0 0 0 0>}
      cornerRadii=<0 0 0 0>
     }
     tintColor {colorUseRule=UseColor_Specified, specifiedColor=<rgba 1 1 1 1>}
     uVRegion {bIsValid=false, max=<0 0>, min=<0 0>}
    }
    padding=<ltrb 18 18 18 18>
   }
   slot {class=VerticalBoxSlot, properties {size {sizeRule=Fill, value=1}}}
  }
  ListView_Conversations {
   widgetClassPath=/Sidebar/Panels/Chat/List/WBP_Sidebar_TextChat_MessageList
   bIsVariable=true
   parent=Border_Clipping
   widget=ListView_Conversations
   properties {bOverride_Cursor=true, visibility=Visible}
   slot {class=BorderSlot, properties {padding=<ltrb 18 18 18 18>}}
  }
 ]
}
