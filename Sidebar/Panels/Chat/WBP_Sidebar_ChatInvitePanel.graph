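# Sidebar chat panel for selecting friends to start a new chat or invite to an existing group.
# On activation, the navigation point supplies the conversation ID and reserved spots, setting the mode and excluded friends.
# A searchable social-header tree provides checked users, whose count drives the member subheading and button state.
# Invite opens a DM, creates a conversation or sends group invites; result handlers change navigation, and a limit message disables the button.
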WidgetBlueprint /Sidebar/Panels/Chat/WBP_Sidebar_ChatInvitePanel
graphs [
 BP_GetDesiredFocusTarget {nodes [
   Get Desired Focus Target()
   n1 = Get WBP_Sidebar_InputField!(self:WBP Sidebar Filtered Input Field Object Reference=Get WBP_Sidebar_FilteredInputField() -> (WBP_Sidebar_FilteredInputField:WBP Sidebar Filtered Input Field Object Reference)) -> (WBP_Sidebar_InputField:WBP Sidebar Input Field Object Reference)
   > Return Node!(
    execute
    ReturnValue:Widget Object Reference=Get EditableText!(self:WBP UIKit Block Editable Text Object Reference=Get Block_EditableText!(self:WBP UIKit Input Field Base Object Reference=@n1) -> (Block_EditableText:WBP UIKit Block Editable Text Object Reference)) -> (EditableText:UIKit Editable Text Object Reference)
   )
  ]}
 ClearCheckedMarks {nodes [
   Inputs()
   > n1 = Get!(execute:Exec) -> (then:Exec, else:Exec, FortSocialHeaderListVM:Fort Social Header List VM Object Reference)
   > n2 = For Each Loop(Array:Array of Fort Social Header Item VM Object References=Get Headers!(self:Fort Social Header List VM Object Reference=@n1) -> (Headers:Array of Fort Social Header Item VM Object References)) -> (Array Element:Fort Social Header Item VM Object Reference)
   Outputs!(then:Exec=@n2.Completed)
   > n3 = For Each Loop(Array:Array of Fort Social Item VM Object References=Get UnfilteredSocialItems!(self:Fort Social Header Item VM Object Reference=@n2) -> (UnfilteredSocialItems:Array of Fort Social Item VM Object References)) -> (Array Element:Fort Social Item VM Object Reference)
   > Set bIsChecked!(execute, bIsChecked:bool=false, self:Fort Social Item VM Object Reference=@n3) -> (then, Output_Get:bool)
  ]}
 ClearInputField {nodes [
   Inputs()
   n1 = Get WBP_Sidebar_InputField(self=Get WBP_Sidebar_FilteredInputField())
   n2 = MakeLiteralText()
   > SetText!(execute, self:WBP UIKit Block Editable Text Object Reference=Get Block_EditableText(self=@n1), Text:Text=@n2) -> (then)
   > OnTextChanged_Event!(execute, self:WBP UIKit Input Field Base Object Reference=@n1, Text:"Text (by ref)"=@n2) -> (then)
   > Outputs()
  ]}
 EventGraph {nodes [
   "On Input Icon Updated (InputActionWidget)"() {is=title K2Node_ComponentBoundEvent}
   > n1 = GetIconDynamicMaterial!(execute, self:Common Action Widget Object Reference=Get InputActionWidget() -> (InputActionWidget:Common Action Widget Object Reference)) -> (then, ReturnValue:Material Instance Dynamic Object Reference)
   > SetScalarParameterValue(self=@n1, ParameterName=Use Dark, Value=1)
   "On Clicked (Button_Invite)"!() -> (OutputDelegate, then, Button:Common Button Base Object Reference) {is=title K2Node_ComponentBoundEvent}
   > n2 = GetAllCheckedSocialUsers!(execute, self) -> (then, CheckedUsers:Array of Fort Social User Object References)
   > n3 = Branch(Condition=Get IsNewChat() -> (IsNewChat:bool))
   > n4 = Branch(Condition="Equal (Integer)"(A=Length(TargetArray:Array of Fort Social User Object References=@n2), B=1))
   > TransitionToDMConversation!(
    execute
    self
    DMTarget:Fort Social User Object Reference="Get (a copy)"(Array:Array of Fort Social User Object References=@n2) -> (Output:Fort Social User Object Reference)
   ) -> (then)
   CreateConversation!(
    execute=@n4.else
    self:Fort Text Chat Conversations List VM Object Reference=Get FortTextChatConversationsListVM() -> (FortTextChatConversationsListVM:Fort Text Chat Conversations List VM Object Reference)
    UsersToInvite:Array of Social User Object References=@n2
   ) -> (then)
   RequestInviteSocialUsers!(
    execute=@n3.else
    self:Fort Group Conversation Viewmodel Object Reference=Get FortTextChatGroupConversationVM() -> (FortTextChatGroupConversationVM:Fort Group Conversation Viewmodel Object Reference)
    UsersToInvite:Array of Social User Object References=@n2
   ) -> (then)
   "On Clear Button Clicked (WBP_Sidebar_FilteredInputField)"() {is=title K2Node_ComponentBoundEvent}
   > RequestRefreshFocus(self:Common Activatable Widget Object Reference)
   "On Item Clicked (TreeList)"!() -> (OutputDelegate, then, Item:Object Reference) {is=title K2Node_ComponentBoundEvent}
   > UpdateDisplayWithPlayerCounts()
   > Set Invite Button Panel Visibility and Padding!(execute:Exec) -> (then:Exec) {is=title K2Node_MacroInstance}
   > SetVisibility(self=Get HB_Error() -> (HB_Error:Horizontal Box Object Reference), InVisibility=Collapsed)
   n5 = "On Text Changed (WBP_Sidebar_FilteredInputField)"!() -> (OutputDelegate, then, Text:Text) {is=title K2Node_ComponentBoundEvent}
   > n6 = Get()
   > SetFilterTerm!(execute, self:Fort Social Header List VM Object Reference=@n6, InFilterTerm:String="To String (Text)"(InText=@n5.Text)) -> (then)
   n7 = "On Text Committed (WBP_Sidebar_FilteredInputField)"!() -> (OutputDelegate, then, CommitMethod:ETextCommit Enum=Default) {is=title K2Node_ComponentBoundEvent}
   > n8 = Switch on ETextCommit!(execute, Selection:ETextCommit Enum=@n7.CommitMethod) -> (Default:Exec, OnEnter:Exec, OnUserMovedFocus:Exec, OnCleared:Exec) {is=title K2Node_SwitchEnum}
   RequestRefreshFocus(execute=@n8.OnEnter)
   n9 = OnConversationHasBeenRemovedChanged!() -> (OutputDelegate, then, bHasBeenRemoved:bool=~) {is=title K2Node_CustomEvent}
   > Branch(Condition=@n9.bHasBeenRemoved)
   > PopToNavigationTag!(
    execute
    self:Fort Sidebar Navigation VM Object Reference=Get FortSidebarNavigationVM() -> (FortSidebarNavigationVM:Fort Sidebar Navigation VM Object Reference)
    NavigationTag:Gameplay Tag Structure=(TagName="UI.StateTree.TransitionTag.Sidebar.Chat")
   ) -> (then, ReturnValue:bool)
   Event Construct()
   > SetAllowListAutoUpdate!(
    execute
    self:Fort Social Header List VM Object Reference=Get FortSocialHeaderListVM() -> (FortSocialHeaderListVM:Fort Social Header List VM Object Reference)
    bIsEnabled:bool=false
   ) -> (then)
   > PopulateSocialHeaders()
   > Set CachedLastResult(CachedLastResult:ELastAsyncActionResult Enum=None) -> (Output_Get:ELastAsyncActionResult Enum=None)
   Event PreConstruct()
   > SetMobileStyle() {is=title K2Node_CallFunction}
   > Initialize Invite Button Visuals!(execute:Exec) -> (then:Exec) {is=title K2Node_MacroInstance}
   Event On Activated()
   > UpdateNow(self:Fort Social Header List VM Object Reference=Get FortSocialHeaderListVM())
   > Get Data from Navigation Point!(execute:Exec) -> (then:Exec) {is=title K2Node_MacroInstance}
   > Clear Checked Marks!(execute:Exec) -> (then:Exec) {is=title K2Node_MacroInstance}
   > Clear Input Field!(execute:Exec) -> (then:Exec) {is=title K2Node_MacroInstance}
   > UpdateDisplayWithPlayerCounts()
   > Set Header!(execute:Exec) -> (then:Exec) {is=title K2Node_MacroInstance}
   > SetVisibility(self=[Get HB_Error(), Get ButtonInviteOverlay() -> (ButtonInviteOverlay:Overlay Object Reference)], InVisibility=Collapsed)
  ]}
 GetAllCheckedSocialUsers {nodes [
   GetAllCheckedSocialUsers() {is=title K2Node_FunctionEntry}
   > n1 = Get()
   > n2 = For Each Loop(Array:Array of Fort Social Header Item VM Object References=Get Headers(self=@n1)) -> (Array Element:Fort Social Header Item VM Object Reference)
   Return Node!(
    execute=@n2.Completed
    CheckedUsers:Array of Fort Social User Object References=Get CheckedSocialUsers() -> (CheckedSocialUsers:Array of Fort Social User Object References)
   )
   > n3 = For Each Loop(Array:Array of Fort Social Item VM Object References=Get UnfilteredSocialItems(self=@n2)) -> (Array Element:Fort Social Item VM Object Reference)
   > Branch(Condition=Get bIsChecked!(self:Fort Social Item VM Object Reference=@n3) -> (bIsChecked:bool))
   > Add(
    TargetArray:Array of Fort Social User Object References=Get CheckedSocialUsers()
    NewItem:"Fort Social User Object Reference (by ref)"=Get TargetUser!(self:Fort Social Item VM Object Reference=@n3) -> (TargetUser:Fort Social User Weak object)
   ) {is=title K2Node_CallArrayFunction}
  ]}
 GetDataFromNavigationPoint {nodes [
   Inputs()
   > n1 = GetCurrentNavigationPoint!(execute, self:Fort Sidebar Navigation VM Object Reference=Get FortSidebarNavigationVM()) -> (then, ReturnValue:Fort Sidebar Navigation Point Object Reference)
   > n2 = Cast To BP_NavigationPoint!(execute, Object=@n1) -> (then, CastFailed, AsBP Navigation Point:BP Navigation Point Object Reference) {is=title K2Node_DynamicCast}
   > Set ReservedMembers(ReservedMembers:int=Get ReservedSpots!(self:BP Navigation Point Object Reference=@n2) -> (ReservedSpots:int=0)) -> (Output_Get:int=0)
   > n3 = Set ConversationIDFromGroup(ConversationIDFromGroup:Name=Get ConversationId!(self:BP Navigation Point Object Reference=@n2) -> (ConversationId:Name=None)) -> (Output_Get:Name=None)
   > InitializeTextChatConversationVM!(
    execute
    self:Fort Text Chat Conversation VM Object Reference=Get FortTextChatGroupConversationVM()
    Player:Local Player Object Reference=GetOwningLocalPlayer!(self:Widget Object Reference) -> (ReturnValue:Local Player Object Reference)
    InConversationId:"Name (by ref)"=@n3
   ) -> (then)
   > Set IsNewChat(IsNewChat:bool="Equal (Name)"(A=Get ConversationIDFromGroup() -> (ConversationIDFromGroup:Name=None))) -> (Output_Get:bool)
   > n4 = For Each Loop(Array:Array of Fort Social Header Item VM Object References=Get Headers(self=Get FortSocialHeaderListVM())) -> (Array Element:Fort Social Header Item VM Object Reference)
   Outputs(then=@n4.Completed)
   > SetExcludedFriendsConversationId!(
    execute
    self:Fort Social Header Item VM Object Reference=@n4
    InExcludedFriendsConversationId:String="To String (Name)"(InName=Get ConversationIDFromGroup() -> (ConversationIDFromGroup=None))
   ) -> (then)
  ]}
 GetNumCheckedAndTotalPlayers {nodes [
   GetNumCheckedAndTotalPlayers() {is=title K2Node_FunctionEntry}
   > n1 = Get()
   > n2 = For Each Loop(Array=Get Headers(self=@n1))
   Return Node!(execute=@n2.Completed, NumCheckedPlayers:int=Get NumChecked() -> (NumChecked:int=0), NumTotalPlayers:int=Get NumTotal() -> (NumTotal:int=0))
   n3 = Get UnfilteredSocialItems(self=@n2)
   > Set NumTotal(NumTotal:int=int + int(A=Get NumTotal() -> (NumTotal=0), B=Length(TargetArray:Array of Fort Social Item VM Object References=@n3))) -> (Output_Get:int=0)
   > n4 = For Each Loop(Array:Array of Fort Social Item VM Object References=@n3) -> (Array Element:Fort Social Item VM Object Reference)
   > Branch(Condition=Get bIsChecked(self=@n4))
   > Increment Int(Value=Get NumChecked() -> (NumChecked=0))
  ]}
 HandleConversationRequestLastResult {nodes [
   n1 = HandleConversationRequestLastResult!() -> (then, Last Result:ELastAsyncActionResult Enum=None) {is=title K2Node_FunctionEntry}
   > n2 = Branch(Condition="Equal (Enum)"(
     A:ELastAsyncActionResult Enum=@n1
     B:ELastAsyncActionResult Enum=Get CachedLastResult() -> (CachedLastResult:ELastAsyncActionResult Enum=None)
    ))
   n3 = Set CachedLastResult(execute=@n2.else, CachedLastResult=@n1) -> (Output_Get=None)
   > n4 = Switch on ELastAsyncActionResult!(execute, Selection:ELastAsyncActionResult Enum=@n3) -> (None:Exec, Failed:Exec, Success:Exec) {is=title K2Node_SwitchEnum}
   n5 = Get CreateConversationRequest!(self:Fort Text Chat Conversations List VM Object Reference=Get FortTextChatConversationsListVM()) -> (CreateConversationRequest:Create Conversation Async Action Details Structure)
   TransitionToConversation!(
    execute=@n4.Success
    self
    ConversationId:Name=String To Name(InString=Break Create Conversation Async Action Details!(CreateConversationAsyncActionDetails:"Create Conversation Async Action Details Structure (by ref)"=@n5) -> (ResultConversationId:String, bIsInProgress:bool=~, LastResult:ELastAsyncActionResult Enum, LastResultMessage:Text) {is=title K2Node_BreakStruct})
   ) -> (then)
  ]}
 HandleInviteRequestLastResult {nodes [
   n1 = HandleInviteRequestLastResult!() -> (then, Last Result:ELastAsyncActionResult Enum=None) {is=title K2Node_FunctionEntry}
   > n2 = Branch(Condition="Equal (Enum)"(A=@n1, B=Get InviteLastResult() -> (InviteLastResult:ELastAsyncActionResult Enum=None)))
   n3 = Set InviteLastResult(execute=@n2.else, InviteLastResult:ELastAsyncActionResult Enum=@n1) -> (Output_Get:ELastAsyncActionResult Enum=None)
   > n4 = Switch on ELastAsyncActionResult(Selection=@n3)
   PopNavigation(execute=[@n4.Failed, @n4.Success], self:Fort Sidebar Navigation VM Object Reference=Get FortSidebarNavigationVM())
  ]}
 InitializeInviteButtonVisuals {nodes [
   Inputs()
   n1 = Get Slot_Main!(self:WBP UIKit Button Generic Base Object Reference=Get Button_Invite() -> (Button_Invite:WBP Sidebar Button Generic Object Reference)) -> (Slot_Main:Named Slot Object Reference)
   > SetGridSlotAlignment!(
    execute
    Widget:Widget Object Reference=@n1
    Horizontal:EHorizontalAlignment Enum=HAlign_Left
    Vertical:EVerticalAlignment Enum=VAlign_Center
   ) -> (then)
   > SetGridSlotPadding!(execute, Widget:Widget Object Reference=@n1, Padding_Left:float=18, Padding_Top:float=0, Padding_Right:float=0, Padding_Bottom:float=0) -> (then)
   > SetCornerRadius!(
    execute
    self:WBP Sidebar Block Backplate Object Reference=Get InviteButton_Backplate() -> (InviteButton_Backplate:WBP Sidebar Block Backplate Object Reference)
    Radius In Pixels:bool=true
    Corner Radius:double=30
   ) -> (then)
   > SetVisibility(self=Get ButtonInviteOverlay(), InVisibility=Collapsed)
   > Set!(
    execute
    Padding_Left:float=-18
    Padding_Top:float=-12
    Padding_Right:float=-18
    Padding_Bottom:float=-18
    self:Stack Box Slot Object Reference=SlotAsStackBoxSlot!(Widget:Widget Object Reference=Get GP_TreeList() -> (GP_TreeList:Grid Panel Object Reference)) -> (ReturnValue:Stack Box Slot Object Reference)
   ) -> (then, Output_Get:Margin Structure)
   > Outputs()
  ]}
 PopulateSocialHeaders {nodes [
   PopulateSocialHeaders() {is=title K2Node_FunctionEntry}
   n1 = Get FortSocialHeaderListVM()
   > ClearHeaders(self:Fort Social Header List VM Object Reference=@n1)
   > n2 = For Each Loop(Array:Array of ESocialHeaderItemType Enums=Get Social Headers() -> (Social Headers:Array of ESocialHeaderItemType Enums=Unknown)) -> (Array Element:ESocialHeaderItemType Enum)
   > n3 = AddHeader!(execute, self:Fort Social Header List VM Object Reference=@n1, HeaderType:ESocialHeaderItemType Enum=@n2) -> (then, ReturnValue:Fort Social Header Item VM Object Reference)
   > SetExcludedFriendsConversationId(self=@n3, InExcludedFriendsConversationId="To String (Name)"(InName=Get ConversationIDFromGroup() -> (ConversationIDFromGroup=None)))
  ]}
 SetConversationLimitExceededVisuals {nodes [
   n1 = SetConversationLimitExceededVisuals!() -> (then, Last Result Message:Text) {is=title K2Node_FunctionEntry}
   > Set Text(Text=@n1, self=Get ErrorMessage() -> (ErrorMessage:Common Text Object Reference))
   n2 = TextIsEmpty(InText=@n1)
   > SetVisibility(self=Get HB_Error(), InVisibility=Select(Option 0=SelfHitTestInvisible, Option 1=Collapsed, Index=@n2))
   > SetIsEnabled(self=Get Button_Invite(), bInIsEnabled=@n2)
  ]}
 SetHeader {nodes [
   Inputs()
   > Set with Broadcast HeadingText!(
    execute
    HeadingText:Text=Select(Option 0:Text="LOCTEXT(\"Invite to group\")", Option 1:Text="LOCTEXT(\"New chat\")", Index=Get IsNewChat()) -> (ReturnValue:Text)
    self:WBP Sidebar Panel Layout Object Reference=Get PanelLayout() -> (PanelLayout:WBP Sidebar Panel Layout Object Reference)
   ) -> (then, Output_Get:Text)
   > Outputs()
  ]}
 SetInviteButtonPanelVisibilityAndPadding {nodes [
   Inputs()
   > n1 = GetNumCheckedAndTotalPlayers!(execute, self) -> (then, NumCheckedPlayers:int=0, NumTotalPlayers:int=0) {is=title K2Node_CallFunction}
   n2 = "integer > integer"(A=@n1)
   > SetVisibility(
    self=Get ButtonInviteOverlay()
    InVisibility=Select(Option 0:ESlateVisibility Enum=Collapsed, Option 1:ESlateVisibility Enum=SelfHitTestInvisible, Index=@n2) -> (ReturnValue:ESlateVisibility Enum)
   )
   > Set(
    Padding_Left=-18
    Padding_Top=-12
    Padding_Right=-18
    Padding_Bottom=Select(Option 0:float=-18, Option 1:float=0, Index=@n2) -> (ReturnValue:float)
    self=SlotAsStackBoxSlot(Widget=Get GP_TreeList())
   )
   > Outputs()
  ]}
 SetMobileStyle {nodes [
   SetMobileStyle() {is=title K2Node_FunctionEntry}
   > Branch(Condition=IsUsingCompactUI())
   n1 = Get WBP_Sidebar_FilteredInputField()
   > SetSize!(
    execute
    self:WBP UIKit Input Field Base Object Reference=Get WBP_Sidebar_InputField(self=@n1)
    WIdth:S UI Bool Float Structure
    Height_Enable_3_#1:bool=true
    Height_Value_4_#2:double=80
   ) -> (then)
   n2 = SlotAsStackBoxSlot(Widget=@n1)
   n3 = Get!(self:Stack Box Slot Object Reference=@n2) -> (Padding_Left:float=~, Padding_Top:float=~, Padding_Right:float=~, Padding_Bottom:float=~)
   > Set(Padding_Left=22, Padding_Top=@n3.Padding_Top, Padding_Right=22, Padding_Bottom=@n3.Padding_Bottom, self=@n2)
   n4 = SlotAsStackBoxSlot(Widget=Get GP_TreeList())
   n5 = Get(self=@n4) -> (Padding_Left=~, Padding_Top=~, Padding_Right=~, Padding_Bottom=~)
   > Set(Padding_Left=@n5, Padding_Top=12, Padding_Right=@n5.Padding_Right, Padding_Bottom=@n5.Padding_Bottom, self=@n4)
  ]}
 SetTreeViewItems {nodes [
   n1 = SetTreeViewItems!() -> (then, OuterArray:Array of Fort Social Header Item VM Object References) {is=title K2Node_FunctionEntry}
   > SetTreeViewItems!(
    execute
    self:WBP Sidebar Chat Invite Tree List Object Reference=Get TreeList() -> (TreeList:WBP Sidebar Chat Invite Tree List Object Reference)
    OuterArray:Array of Fort Social Header Item VM Object References=@n1
   ) -> (then) {is=title K2Node_CallFunction}
   > UpdateDisplayWithPlayerCounts()
  ]}
 TransitionToConversation {nodes [
   n1 = TransitionToConversation!() -> (then, ConversationId:Name) {is=title K2Node_FunctionEntry}
   > n2 = Construct Object from Class!(
    execute
    Class=/Sidebar/Shared/Navigation/BP_NavigationPoint.BP_NavigationPoint_C
    self
    SocialUser:Fort Social User Object Reference
    ConversationId:Name=@n1
    ConfirmationVM:Sidebar Confirmation Panel VM Object Reference
    ReservedSpots:int=0
    SocialItemVM:Fort Social Item VM Object Reference
    NavigationTag:Gameplay Tag Structure=(TagName="UI.StateTree.TransitionTag.Sidebar.Chat.MessageList")
   ) -> (then, ReturnValue:BP Navigation Point Object Reference)
   > PushNavigation!(
    execute
    self:Fort Sidebar Navigation VM Object Reference=Get FortSidebarNavigationVM()
    NavigationPoint:Fort Sidebar Navigation Point Object Reference=@n2
   ) -> (then)
  ]}
 TransitionToDMConversation {nodes [
   n1 = TransitionToDMConversation!() -> (then, DMTarget:Fort Social User Object Reference) {is=title K2Node_FunctionEntry}
   > n2 = Construct Object from Class(
    Class=&BP_NavigationPoint_C
    SocialUser=@n1
    ConversationId=None
    ReservedSpots=0
    NavigationTag=(TagName="UI.StateTree.TransitionTag.Sidebar.Chat.MessageList")
   )
   > PushNavigation(self=Get FortSidebarNavigationVM(), NavigationPoint=@n2)
  ]}
 UpdateDisplayWithPlayerCounts {nodes [
   UpdateDisplayWithPlayerCounts() {is=title K2Node_FunctionEntry}
   > n1 = GetNumCheckedAndTotalPlayers() -> (NumCheckedPlayers=0, NumTotalPlayers=0)
   > Set CheckedPlayers(CheckedPlayers:int=@n1) -> (Output_Get:int=0)
   > n2 = Branch(Condition=Get IsNewChat())
   > Set with Broadcast SubheadingText!(
    execute
    SubheadingText:Text=Format Text!(
     Format="LOCTEXT(\"{0} / {1} members\")"
     0:int=Get CheckedPlayers() -> (CheckedPlayers:int=0)
     1:int=Get MaxConversationMemberSize() -> (MaxConversationMemberSize:int=0)
    ) -> (Result)
    self:WBP Sidebar Panel Layout Object Reference=Get PanelLayout()
   ) -> (then, Output_Get:Text)
   n3 = Get CheckedPlayers() -> (CheckedPlayers=0)
   n4 = "integer <= integer"(A=@n3, B=Get MaxConversationMemberSize() -> (MaxConversationMemberSize=0))
   n5 = Get Button_Invite()
   > SetIsEnabled(self=@n5, bInIsEnabled=AND Boolean(A="integer > integer"(A=@n3), B=@n4))
   n6 = AND Boolean(A="Equal (Integer)"(A=@n3, B=1), B=Get IsNewChat())
   > SetText(
    self:WBP UIKit Button Generic Base Object Reference=@n5
    Text=Select(
     Option 0:Text="LOCTEXT(\"Too many members selected\")"
     Option 1:Text=Select(Option 0="LOCTEXT(\"Invite to chat\")", Option 1="LOCTEXT(\"Start chat\")", Index=@n6)
     Index=@n4
    ) -> (ReturnValue:Text)
   )
   n7 = Set ReservedPlayers(
    execute=@n2.else
    ReservedPlayers:int=int + int(A=Get ReservedMembers() -> (ReservedMembers:int=0), B=Get CheckedPlayers() -> (CheckedPlayers=0))
   ) -> (Output_Get:int=0)
   n8 = Get MaxConversationMemberSize() -> (MaxConversationMemberSize=0)
   > Set with Broadcast SubheadingText(SubheadingText=Format Text(Format="LOCTEXT(\"{0} / {1} members\")", 0=@n7, 1=int + int(A=@n8)), self=Get PanelLayout())
   n9 = "integer <= integer"(A=Get ReservedPlayers() -> (ReservedPlayers:int=0), B=int + int(A=Get MaxConversationMemberSize() -> (MaxConversationMemberSize=0)))
   n10 = Get CheckedPlayers() -> (CheckedPlayers=0)
   n11 = Get Button_Invite()
   > SetIsEnabled(self=@n11, bInIsEnabled=AND Boolean(A="integer > integer"(A=@n10), B=@n9))
   > SetText(self=@n11, Text=Select(Option 0="LOCTEXT(\"Too many players selected\")", Option 1="LOCTEXT(\"Invite to group\")", Index=@n9))
  ]}
]
variables [
 CachedLastResult {type=ELastAsyncActionResult, default=None}
 ConversationIDFromGroup {type=string, default=None}
 InviteLastResult {type=ELastAsyncActionResult, default=None}
 IsNewChat {type=boolean}
 MaxConversationMemberSize {type=integer, default=15}
 ReservedMembers {type=integer}
 Social Headers {type="array<ESocialHeaderItemType>", default [ChatInvite]}
]
widgets {
 info {
  inheritedWidgetCount=0
  namedSlotCount=2
  parentClass=&CommonActivatableWidget
  rootWidgetClass=/Sidebar/Shared/WBP_Sidebar_PanelLayout.WBP_Sidebar_PanelLayout_C
  widgetCount=13
 }
 widgets [
  PanelLayout {
   widgetClassPath=/Sidebar/Shared/WBP_Sidebar_PanelLayout
   bIsVariable=true
   widget=PanelLayout
   properties {headingText=New chat, subheadingText=1 / 16 Members}
  }
  SB_Main {
   widgetClassPath=&StackBox
   namedSlotHost=PanelLayout
   widget=SB_Main
   properties {bOverride_Cursor=true, orientation=Orient_Vertical}
  }
  WBP_Sidebar_FilteredInputField {
   widgetClassPath=/Sidebar/Buttons/WBP_Sidebar_FilteredInputField
   bIsVariable=true
   parent=SB_Main
   widget=WBP_Sidebar_FilteredInputField
   properties {
    bIsFocusable=true
    hideValidationInfo=true
    hintText=Search...
    isSearchField=true
    useHintText=true
    clearButtonPadding=<ltrb 0 0 12 0>
    desiredImageSize=<28 28>
    icon Properties {
     material=/FortUILibrary/Systems/InputField/MaterialInstances/MI_UIKit_InputFieldIcon
     image Size=<32 32>
     placement {
      horizontal Alignment=HAlign_Left
      vertical Alignment=VAlign_Center
      grid {column=0, column Span=0, layer=0, row=1, row Span=0, nudge=<0 0>}
      padding=<ltrb 18 0 0 0>
     }
    }
    inputActionPromptPadding=<ltrb 0 0 12 0>
   }
   slot {class=StackBoxSlot, properties {verticalAlignment=VAlign_Center, padding=<ltrb 18 12 18 12>}}
  }
  GP_TreeList {
   widgetClassPath=&GridPanel
   bIsVariable=true
   parent=SB_Main
   widget=GP_TreeList
   properties {clipping=ClipToBounds, columnFill [1], rowFill [1]}
   slot {class=StackBoxSlot, properties {padding=<ltrb -18 -12 -18 -18>, size {sizeRule=Fill, value=1}}}
  }
  TreeList {
   widgetClassPath=/Sidebar/Panels/Chat/List/WBP_Sidebar_ChatInviteTreeList
   bIsVariable=true
   parent=GP_TreeList
   widget=TreeList
   properties {itemClassToEntryData {
     &Class'/Script/SidebarUI.FortSocialHeaderItemVM' {bIsEntrySelectable=true, desiredEntryClass=/Sidebar/Panels/Social/List/WBP_Sidebar_SocialHeader.WBP_Sidebar_SocialHeader_C}
     &Class'/Script/SidebarUI.FortSocialItemVM' {bIsEntrySelectable=true, desiredEntryClass=/Sidebar/Shared/WBP_Sidebar_PlayerSelectButton.WBP_Sidebar_PlayerSelectButton_C}
    }}
   slot {class=GridSlot, properties {padding=<ltrb 18 12 18 12>}}
  }
  ButtonInviteOverlay {
   widgetClassPath=&Overlay
   bIsVariable=true
   parent=SB_Main
   widget=ButtonInviteOverlay
   properties {visibility=Visible}
   slot {class=StackBoxSlot, properties {verticalAlignment=VAlign_Bottom}}
  }
  InviteButton_Backplate {
   widgetClassPath=/Sidebar/Buttons/Blocks/WBP_Sidebar_Block_Backplate
   bIsVariable=true
   parent=ButtonInviteOverlay
   widget=InviteButton_Backplate
   slot {class=OverlaySlot, properties {horizontalAlignment=HAlign_Fill, verticalAlignment=VAlign_Fill, padding=<ltrb -12 0 -12 -12>}}
  }
  VerticalBox_0 {
   widgetClassPath=&VerticalBox
   parent=ButtonInviteOverlay
   widget=VerticalBox_0
   slot {class=OverlaySlot, properties {horizontalAlignment=HAlign_Fill, padding=<ltrb 12 24 12 12>}}
  }
  HB_Error {
   widgetClassPath=&HorizontalBox
   bIsVariable=true
   parent=VerticalBox_0
   widget=HB_Error
   properties {visibility=Collapsed}
   slot {class=VerticalBoxSlot, properties {padding=<ltrb 0 0 0 24>}}
  }
  WarningIcon {
   widgetClassPath=&Image
   bIsVariable=true
   parent=HB_Error
   widget=WarningIcon
   properties {
    visibility=SelfHitTestInvisible
    brush {
     bIsDynamicallyLoaded=false
     drawAs=Image
     imageType=FullColor
     mirroring=NoMirror
     resourceName=None
     resourceObject=/FortUILibrary/Textures/MaterialInstances/MI_UIKit_Icon_AgeGatedAlert
     tiling=NoTile
     imageSize=<24 24>
     margin=<ltrb 0 0 0 0>
     outlineSettings {
      bUseBrushTransparency=false
      roundingType=HalfHeightRadius
      width=0
      color {colorUseRule=UseColor_Specified, specifiedColor=<rgba 0 0 0 0>}
      cornerRadii=<0 0 0 0>
     }
     tintColor {colorUseRule=UseColor_Specified, specifiedColor=<rgba 1 1 1 1>}
     uVRegion {bIsValid=false, max=<0 0>, min=<0 0>}
    }
    colorAndOpacity=<rgba 1 0.147 0.1329 1>
   }
   slot {class=HorizontalBoxSlot, properties {horizontalAlignment=HAlign_Left, verticalAlignment=VAlign_Center, padding=<ltrb 0 2 12 6>}}
  }
  ErrorMessage {
   widgetClassPath=&CommonTextBlock
   bIsVariable=true
   parent=HB_Error
   widget=ErrorMessage
   properties {
    autoWrapText=true
    bAutoCollapseWithEmptyText=true
    bIsScrollingEnabled=false
    mobileFontSizeMultiplier=1.2
    scrollStyle=/FortUILibrary/Styles/Typography/Foundation/TextScrollStyle-NoFade.TextScrollStyle-NoFade_C
    text=Max 200 groups reached. Leave one to create a new group.
    textOverflowPolicy=Ellipsis
    colorAndOpacity {colorUseRule=UseColor_Specified, specifiedColor=<rgba 1 0.147 0.1329 1>}
    font {
     bForceMonospaced=false
     bMaterialIsStencil=false
     fontMaterial=null
     fontObject=&HeadingNow
     letterSpacing=0
     monospacedWidth=1
     size=13.5
     skewAmount=0
     typefaceFontName=74 Regular
     outlineSettings {
      bApplyOutlineToDropShadows=false
      bMiteredCorners=false
      bSeparateFillAlpha=false
      outlineMaterial=null
      outlineSize=0
      outlineColor=<rgba 0 0 0 1>
     }
    }
    shadowOffset=<0 0>
   }
   slot {class=HorizontalBoxSlot, properties {size {sizeRule=Fill, value=1}}}
  }
  Button_Invite {
   widgetClassPath=/Sidebar/Buttons/WBP_Sidebar_ButtonGeneric
   bIsVariable=true
   parent=VerticalBox_0
   widget=Button_Invite
   properties {
    buttonType=Attention
    minHeightOverride=60
    text=Start chat
    font {
     bForceMonospaced=false
     bMaterialIsStencil=false
     fontMaterial=/Sidebar/Materials/Fonts/MI_Sidebar_FontColor_Button_Attention
     fontObject=&HeadingNow
     letterSpacing=10
     monospacedWidth=1
     size=15.75
     skewAmount=0
     typefaceFontName=74 Regular
     outlineSettings {
      bApplyOutlineToDropShadows=false
      bMiteredCorners=false
      bSeparateFillAlpha=false
      outlineMaterial=/FortUILibrary/Systems/ModularButton/MaterialInstances/MI_UIKit_TextOutline
      outlineSize=0
      outlineColor=<rgba 0 0 0 1>
     }
    }
    triggeringInputAction {dataTable=/Sidebar/Shared/DT_Sidebar_InputActions, rowName=Invite}
   }
   slot {class=VerticalBoxSlot}
  }
  InputActionWidget {
   widgetClassPath=&CommonActionWidget
   bIsVariable=true
   namedSlotHost=Button_Invite
   widget=InputActionWidget
   properties {inputActions [{dataTable=&DT_Sidebar_InputActions, rowName=Invite}]}
  }
 ]
}
