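# Sidebar panel for managing a group text chat: renaming it, viewing joined and invited members, inviting, leaving and reporting.
# On activation it initializes the group conversation viewmodel from the navigation point's conversation ID; Bind functions update name, reset and error states.
# Leave, remove-friend and block route through confirmation overlays, with leave-and-report deferring the leave until the report succeeds or reaches BlockPlayers.
# Custom focus functions skip empty lists, and compact UI adjusts sizing.
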WidgetBlueprint /Sidebar/Panels/Chat/WBP_Sidebar_ManageGroupPanel
graphs [
 BP_GetDesiredFocusTarget {nodes [
   Get Desired Focus Target()
   n1 = Get bIsReadOnly!(self:Fort Text Chat Conversation VM Object Reference=Get FortTextChatGroupConversationVM() -> (FortTextChatGroupConversationVM:Fort Group Conversation Viewmodel Object Reference)) -> (bIsReadOnly:bool)
   n2 = Get NotifyIsRestrictionWidgetVisible!(self:WBP Sidebar Text Chat Restrictions Object Reference=Get WBP_Sidebar_TextChat_Restrictions() -> (WBP_Sidebar_TextChat_Restrictions:WBP Sidebar Text Chat Restrictions Object Reference)) -> (NotifyIsRestrictionWidgetVisible:bool)
   n3 = Get InputField_GroupName() -> (InputField_GroupName:WBP Sidebar Input Field Object Reference)
   > n4 = Branch(Condition=AND Boolean(A=IsVisible(self=@n3), B=GetInputFieldEnabled!(self, IsChannelReadOnly:bool=@n1, IsRestrictionActive:bool=@n2) -> (InputFieldEnabled:bool)))
   n5 = Get InputField_GroupName()
   > Return Node!(
    execute
    ReturnValue:Widget Object Reference=Get EditableText!(self:WBP UIKit Block Editable Text Object Reference=Get Block_EditableText!(self:WBP UIKit Input Field Base Object Reference=@n5) -> (Block_EditableText:WBP UIKit Block Editable Text Object Reference)) -> (EditableText:UIKit Editable Text Object Reference)
   )
   n6 = Get Button_AddMembers() -> (Button_AddMembers:WBP Sidebar Button Generic Object Reference)
   n7 = Branch(execute=@n4.else, Condition=AND Boolean(A=IsVisible(self=@n6), B=GetIsEnabled!(self:Widget Object Reference=@n6) -> (ReturnValue:bool)))
   > Return Node(ReturnValue=Get Button_AddMembers())
   n8 = Get TreeView!(self:Fort Polymorphic Tree View Object Reference=Get Joined_TreeList() -> (Joined_TreeList:WBP Sidebar Social Tree List Object Reference)) -> (TreeView:Tree View Object Reference)
   n9 = Branch(execute=@n7.else, Condition="integer > integer"(A=GetNumItems!(self:List View Object Reference=@n8) -> (ReturnValue:int=0)))
   > Return Node(ReturnValue=Get TreeView(self=Get Joined_TreeList()))
   n10 = Get TreeView(self=Get Invited_TreeList() -> (Invited_TreeList:WBP Sidebar Social Tree List Object Reference))
   n11 = Branch(execute=@n9.else, Condition="integer > integer"(A=GetNumItems(self=@n10) -> (ReturnValue=0)))
   > Return Node(ReturnValue=Get TreeView(self=Get Invited_TreeList()))
   n12 = GetBottomButtonFocusTarget!(execute=@n11.else, self) -> (then, Bottom Focus Button:Widget Object Reference)
   > Return Node(ReturnValue=@n12)
  ]}
 BindModifyGroupNameErrors {nodes [
   n1 = BindModifyGroupNameErrors!() -> (then, AsyncActionDetails:Async Action Details Structure) {is=title K2Node_FunctionEntry}
   n2 = Break Async Action Details!(AsyncActionDetails:"Async Action Details Structure (by ref)"=@n1) -> (bIsInProgress:bool=~, LastResult:ELastAsyncActionResult Enum, LastResultMessage:Text) {is=title K2Node_BreakStruct}
   > n3 = Branch(Condition=@n2)
   n4 = "Equal (Enum)"(A:ELastAsyncActionResult Enum=@n2.LastResult, B:ELastAsyncActionResult Enum=Failed)
   SetVisibility(
    execute=@n3.else
    self=Get StackBox_Error() -> (StackBox_Error:Stack Box Object Reference)
    InVisibility=Select(Option 0=Collapsed, Option 1=SelfHitTestInvisible, Index=@n4)
   )
   > "SetText (Text)"(
    self=Get Text_Error() -> (Text_Error:Common Text Object Reference)
    InText=Select(Option 0:Text, Option 1:Text=@n2.LastResultMessage, Index=@n4) -> (ReturnValue:Text)
   )
   > Branch(Condition="Equal (Enum)"(A=@n2.LastResult, B=Success))
   > SetVisibility(self=Get Button_SaveName() -> (Button_SaveName:WBP Sidebar Button Generic Object Reference), InVisibility=Collapsed)
  ]}
 BindPrecheckGroupNameErrors {nodes [
   n1 = BindPrecheckGroupNameErrors!() -> (then, AsyncActionDetails:Async Action Details Structure) {is=title K2Node_FunctionEntry}
   n2 = Break Async Action Details(AsyncActionDetails=@n1) -> (bIsInProgress=~)
   > n3 = Branch(Condition=@n2)
   n4 = "Equal (Enum)"(A=@n2.LastResult, B=Failed)
   SetVisibility(
    execute=@n3.else
    self=Get StackBox_Error()
    InVisibility=Select(Option 0:ESlateVisibility Enum=Collapsed, Option 1:ESlateVisibility Enum=SelfHitTestInvisible, Index=@n4) -> (ReturnValue:ESlateVisibility Enum)
   )
   > "SetText (Text)"(self=Get Text_Error(), InText=Select(Option 0:Text, Option 1:Text=@n2.LastResultMessage, Index=@n4) -> (ReturnValue:Text))
  ]}
 BindSetGroupName {nodes [
   n1 = BindSetGroupName!() -> (then, GroupName:"Text (by ref)") {is=title K2Node_FunctionEntry}
   n2 = Get FortTextChatGroupConversationVM()
   n3 = "Equal Exactly (Text)"!(
    A:"Text (by ref)"=Get ConversationName!(self:Fort Text Chat Conversation VM Object Reference=@n2) -> (ConversationName:Text)
    B:"Text (by ref)"=Get DefaultGeneratedConversationName!(self:Fort Group Conversation Viewmodel Object Reference=@n2) -> (DefaultGeneratedConversationName:Text)
   ) -> (ReturnValue:bool)
   > SetText!(execute, self:WBP UIKit Input Field Base Object Reference=Get InputField_GroupName(), Text:Text=SelectText(B=@n1, bPickA=@n3)) -> (then)
   > SetVisibility(
    self=Get Button_Reset() -> (Button_Reset:WBP Sidebar Button Generic Object Reference)
    InVisibility=Select(Option 0:ESlateVisibility Enum=SelfHitTestInvisible, Option 1:ESlateVisibility Enum=Collapsed, Index=@n3) -> (ReturnValue:ESlateVisibility Enum)
   )
  ]}
 CF_ShowSoloOrGroupAvatar {nodes [
   n1 = CF_ShowSoloOrGroupAvatar!() -> (then, Conversation Type:EFortTextChatPersistentConversationType Enum=EpicParty) {is=title K2Node_FunctionEntry}
   > Return Node!(
    execute
    Show Group Avatar:bool="Not Equal (Enum)"!(A:EFortTextChatPersistentConversationType Enum=@n1, B:EFortTextChatPersistentConversationType Enum=Epic121) -> (ReturnValue:bool)
   )
  ]}
 CanGroupInviteMoreMembers {nodes [
   n1 = CanGroupInviteMoreMembers!() -> (then, JoinedMembers:Array of Fort Social User Viewmodel Object References, Invited:Array of Fort Social User Viewmodel Object References) {is=title K2Node_FunctionEntry}
   n2 = Length(TargetArray:Array of Fort Social User Viewmodel Object References=@n1)
   n3 = Length(TargetArray=@n1.Invited)
   > Return Node!(execute, GroupIsNotFull:bool="integer < integer"(A=int + int(A=@n2, B=@n3), B=16))
  ]}
 CreateReportModal {nodes [
   CreateReportModal() {is=title K2Node_FunctionEntry}
   > n1 = Create Widget!(
    execute:Exec
    Class:User Widget Class Reference=/FortniteCoreUI/UI/Social/Chat/TextChat/Dialogs/WBP_Dialog_VM_TextChat_Report.WBP_Dialog_VM_TextChat_Report_C
    OwningPlayer:Player Controller Object Reference
   ) -> (then:Exec, ReturnValue:WBP Dialog VM Text Chat Report Object Reference)
   > SetReportedConversationVM!(
    execute
    self:WBP Dialog VM Text Chat Report Object Reference=@n1
    ReportedConversationVM:Fort Text Chat Conversation VM Object Reference=Get FortTextChatGroupConversationVM()
   ) -> (then)
   > Bind Event to On Report Closed(self:WBP Dialog VM Text Chat Report Object Reference=@n1, Delegate=Create Event()) {is=title K2Node_AddDelegate}
   > TryPushWidgetAsModalPanel!(execute, PushWidget:User Widget Object Reference=@n1) -> (then)
   > AddTagToInputMode!(
    execute
    self:Enhanced Input Subsystem Interface Interface=Get EnhancedInputLocalPlayerSubsystem() -> (ReturnValue:Enhanced Input Local Player Subsystem Object Reference) {is=title K2Node_GetSubsystem}
    TagToAdd:"Gameplay Tag Structure (by ref)"=MakeLiteralGameplayTag!(Value:Gameplay Tag Structure=(TagName="UI.TextChat.Disabled")) -> (ReturnValue:Gameplay Tag Structure)
    Options:"Modify Context Options Structure (by ref)"=(bIgnoreAllPressedKeysUntilRelease=True,bForceImmediately=False,bNotifyUserSettings=False)
   ) -> (then)
  ]}
 DesignTime {nodes [
   n1 = Inputs!() -> (execute, IsDesignTime:bool=~)
   > n2 = Branch(Condition=@n1)
   > n3 = SetText(
    self:WBP UIKit Block Text Object Reference=Get Block_Text!(self:WBP UIKit Button Generic Base Object Reference=Get Button_AddMembers()) -> (Block_Text:WBP UIKit Block Text Object Reference)
    Text=GetInviteButtonText!(self, JoinedMembers:Array of Fort Social User Viewmodel Object References, Invited:Array of Fort Social User Viewmodel Object References) -> (InviteButtonLabel:Text)
   )
   Outputs!(then:Exec=[@n3.then, @n2.else])
  ]}
 EventGraph {nodes [
   n1 = "Item Remove Friend (Joined_TreeList)"!() -> (OutputDelegate, then, SocialUser:Fort Social User Object Reference, Interaction:Fort Social Interaction VM Object Reference) {is=title K2Node_ComponentBoundEvent}
   > Set RemoveFriendInteraction(RemoveFriendInteraction:Fort Social Interaction VM Object Reference=@n1.Interaction) -> (Output_Get:Fort Social Interaction VM Object Reference)
   > Navigate to Remove Friend!(execute:Exec, SocialUser:Fort Social User Object Reference=@n1.SocialUser) -> (then:Exec) {is=title K2Node_MacroInstance}
   "On Clicked (Button_AddMembers)"!() -> (OutputDelegate, then, Button:Common Button Base Object Reference) {is=title K2Node_ComponentBoundEvent}
   > Navigate to Chat Invite!(execute:Exec) -> (then:Exec) {is=title K2Node_MacroInstance}
   n2 = "On Text Commited (InputField_GroupName)"!() -> (OutputDelegate, then, Text:Text, Commit method:ETextCommit Enum=Default) {is=title K2Node_ComponentBoundEvent}
   > n3 = Switch on ETextCommit!(execute, Selection:ETextCommit Enum=@n2.Commit method) -> (Default:Exec, OnEnter:Exec, OnUserMovedFocus:Exec, OnCleared:Exec) {is=title K2Node_SwitchEnum}
   n4 = Get FortTextChatGroupConversationVM()
   RequestModifyGroupName!(execute=@n3.OnEnter, self:Fort Group Conversation Viewmodel Object Reference=@n4, NewGroupName:Text=@n2.Text) -> (then)
   n5 = "Item Remove Friend (Invited_TreeList)"!() -> (OutputDelegate, then, SocialUser:Fort Social User Object Reference, Interaction:Fort Social Interaction VM Object Reference) {is=title K2Node_ComponentBoundEvent}
   > Set RemoveFriendInteraction(RemoveFriendInteraction=@n5.Interaction)
   > Navigate to Remove Friend(SocialUser=@n5.SocialUser)
   n6 = "Item Block Player (Invited_TreeList)"!() -> (OutputDelegate, then, SocialUser:Fort Social User Object Reference, Interaction:Fort Social Interaction VM Object Reference) {is=title K2Node_ComponentBoundEvent}
   > Set BlockPlayerInteraction(BlockPlayerInteraction:Fort Social Interaction VM Object Reference=@n6.Interaction) -> (Output_Get:Fort Social Interaction VM Object Reference)
   > Navigate to Block Player!(execute:Exec, SocialUser:Fort Social User Object Reference=@n6.SocialUser) -> (then:Exec) {is=title K2Node_MacroInstance}
   "On Clicked (Button_ReportGroup)"!() -> (OutputDelegate, then, Button:Common Button Base Object Reference) {is=title K2Node_ComponentBoundEvent}
   > CreateReportModal() {is=title K2Node_CallFunction}
   "On Clicked (Button_LeaveGroup)"!() -> (OutputDelegate, then, Button:Common Button Base Object Reference) {is=title K2Node_ComponentBoundEvent}
   n7 = Get FortTextChatGroupConversationVM()
   > NavigateToConfirmPanel!(
    execute
    self
    Class:Sidebar Confirmation Panel VM Class Reference=Select(
     Option 0:Sidebar Confirmation Panel VM Class Reference=/Sidebar/Panels/Overlay/Data/ConfirmationVM_LeaveGroup.ConfirmationVM_LeaveGroup_C
     Option 1:Sidebar Confirmation Panel VM Class Reference=/Sidebar/Panels/Overlay/Data/ConfirmationVM_LeaveGroupWithReport.ConfirmationVM_LeaveGroupWithReport_C
     Index=GetCanReportConversation!(self:Fort Text Chat Conversation VM Object Reference=@n7) -> (ReturnValue:bool)
    ) -> (ReturnValue:Sidebar Confirmation Panel VM Class Reference)
   ) -> (then)
   n8 = "Item Block Player (Joined_TreeList)"!() -> (OutputDelegate, then, SocialUser:Fort Social User Object Reference, Interaction:Fort Social Interaction VM Object Reference) {is=title K2Node_ComponentBoundEvent}
   > Set BlockPlayerInteraction(BlockPlayerInteraction=@n8.Interaction)
   > Navigate to Block Player(SocialUser=@n8.SocialUser)
   "On Clicked (Button_SaveName)"!() -> (OutputDelegate, then, Button:Common Button Base Object Reference) {is=title K2Node_ComponentBoundEvent}
   > RequestModifyGroupName(self=@n4, NewGroupName=Get EditableText(self:WBP UIKit Input Field Base Object Reference=Get InputField_GroupName()) -> (EditableText:Text))
   "On Clicked (Button_Reset)"!() -> (OutputDelegate, then, Button:Common Button Base Object Reference) {is=title K2Node_ComponentBoundEvent}
   > RequestModifyGroupName(self=Get FortTextChatGroupConversationVM())
   n9 = "Item Clicked (Joined_TreeList)"!() -> (OutputDelegate, then, ItemVM:Fort Social Item VM Object Reference) {is=title K2Node_ComponentBoundEvent}
   > OpenPlayerProfile!(execute, self, SocialItemVM:Fort Social Item VM Object Reference=@n9.ItemVM) -> (then)
   n10 = "Item Clicked (Invited_TreeList)"!() -> (OutputDelegate, then, ItemVM:Fort Social Item VM Object Reference) {is=title K2Node_ComponentBoundEvent}
   > OpenPlayerProfile(SocialItemVM=@n10.ItemVM)
   n11 = "On Text Changed (InputField_GroupName)"!() -> (OutputDelegate, then, Text:Text) {is=title K2Node_ComponentBoundEvent}
   > PreCheckGroupName!(
    execute
    self:Fort Group Conversation Viewmodel Object Reference=Get FortTextChatGroupConversationVM()
    NewGroupName:"Text (by ref)"=@n11.Text
   ) -> (then)
   n12 = Get Button_SaveName()
   > n13 = Branch(Condition=IsVisible(self=@n12))
   SetVisibility(execute=@n13.else, self=@n12, InVisibility=SelfHitTestInvisible)
   n14 = Get Button_Reset()
   > Branch(Condition=IsVisible(self=@n14))
   > SetVisibility(self=@n14, InVisibility=Collapsed)
   n15 = OnConversationHasBeenRemovedChanged!() -> (OutputDelegate, then, bHasBeenRemoved:bool=~) {is=title K2Node_CustomEvent}
   > Branch(Condition=@n15.bHasBeenRemoved)
   > PopToNavigationTag!(
    execute
    self:Fort Sidebar Navigation VM Object Reference=Get FortSidebarNavigationVM() -> (FortSidebarNavigationVM:Fort Sidebar Navigation VM Object Reference)
    NavigationTag:Gameplay Tag Structure=(TagName="UI.StateTree.TransitionTag.Sidebar.Chat")
   ) -> (then, ReturnValue:bool)
   Event On Activated()
   > n16 = GetCurrentNavigationPoint!(execute, self:Fort Sidebar Navigation VM Object Reference=Get FortSidebarNavigationVM()) -> (then, ReturnValue:Fort Sidebar Navigation Point Object Reference)
   > n17 = Cast To BP_NavigationPoint!(execute, Object=@n16) -> (then, CastFailed, AsBP Navigation Point:BP Navigation Point Object Reference) {is=title K2Node_DynamicCast}
   n18 = Get FortTextChatGroupConversationVM()
   > InitializeTextChatConversationVM!(
    execute
    self:Fort Text Chat Conversation VM Object Reference=@n18
    Player:Local Player Object Reference=GetOwningLocalPlayer!(self:Widget Object Reference) -> (ReturnValue:Local Player Object Reference)
    InConversationId:"Name (by ref)"=Get ConversationId!(self:BP Navigation Point Object Reference=@n17) -> (ConversationId:Name=None)
   ) -> (then)
   > LoadCachedMessages!(execute, self:Fort Text Chat Conversation VM Object Reference=@n18, PaginationOption:EPersistentTextChatPaginationOptions Enum=MostRecent) -> (then)
   > Update Group Name Values!(execute:Exec) -> (then:Exec) {is=title K2Node_MacroInstance}
   n19 = Event PreConstruct()
   > Design Time!(execute:Exec, IsDesignTime:bool=@n19.IsDesignTime) -> (then:Exec) {is=title K2Node_MacroInstance}
   > SetCompactUIStyle() {is=title K2Node_CallFunction}
  ]}
 GetBottomButtonFocusTarget {nodes [
   GetBottomButtonFocusTarget() {is=title K2Node_FunctionEntry}
   > n1 = Branch(Condition=IsVisible(self=Get Button_ReportGroup() -> (Button_ReportGroup:WBP Sidebar Button Generic Object Reference)))
   > Return Node!(execute, Bottom Focus Button:Widget Object Reference=Get Button_ReportGroup())
   Return Node(execute=@n1.else, Bottom Focus Button=Get Button_LeaveGroup() -> (Button_LeaveGroup:WBP Sidebar Button Generic Object Reference))
  ]}
 GetInputFieldEnabled {nodes [
   n1 = GetInputFieldEnabled!() -> (then, IsChannelReadOnly:bool=~, IsRestrictionActive:bool=~) {is=title K2Node_FunctionEntry}
   > Return Node!(execute, InputFieldEnabled:bool=NOT Boolean(A=OR Boolean(A=@n1, B=@n1.IsRestrictionActive)))
  ]}
 GetInviteButtonEnabled {nodes [
   n1 = GetInviteButtonEnabled!() -> (
    then
    IsRestrictionActive:bool=~
    JoinedMembers:Array of Fort Social User Viewmodel Object References
    Invited:Array of Fort Social User Viewmodel Object References
   ) {is=title K2Node_FunctionEntry}
   > Return Node!(
    execute
    InviteButtonEnabled:bool=AND Boolean(
     A=NOT Boolean(A=@n1)
     B=CanGroupInviteMoreMembers!(
      self
      JoinedMembers:Array of Fort Social User Viewmodel Object References=@n1.JoinedMembers
      Invited:Array of Fort Social User Viewmodel Object References=@n1.Invited
     ) -> (GroupIsNotFull:bool) {is=title K2Node_CallFunction}
    )
   )
  ]}
 GetInviteButtonText {nodes [
   n1 = GetInviteButtonText!() -> (then, JoinedMembers:Array of Fort Social User Viewmodel Object References, Invited:Array of Fort Social User Viewmodel Object References) {is=title K2Node_FunctionEntry}
   > Return Node!(
    execute
    InviteButtonLabel:Text=SelectText(A="LOCTEXT(\"Invite to group\")", B="LOCTEXT(\"Group is full\")", bPickA=CanGroupInviteMoreMembers(JoinedMembers=@n1, Invited=@n1.Invited))
   )
  ]}
 HandleConfirmationOverlayActions {nodes [
   n1 = HandleConfirmationOverlayActions!() -> (then, ActionID:Name) {is=title K2Node_FunctionEntry}
   > Allow Handle Overlay Confirmation!(Exec:Exec, Action ID:Name=@n1) -> (then:Exec) {is=title K2Node_MacroInstance}
   > n2 = Branch(Condition="Equal (Name)"(A=@n1, B=Leave))
   > RequestLeave(self:Fort Group Conversation Viewmodel Object Reference=Get FortTextChatGroupConversationVM())
   n3 = Branch(execute=@n2.else, Condition="Equal (Name)"(A=@n1, B=LeaveAndReport))
   > Set LeaveGroupAfterReport(LeaveGroupAfterReport:bool=true) -> (Output_Get:bool)
   > CreateReportModal()
   n4 = Branch(execute=@n3.else, Condition="Equal (Name)"(A=@n1, B=RemoveFriend))
   > ExecuteInteraction(self:Fort Social Interaction VM Object Reference=Get RemoveFriendInteraction() -> (RemoveFriendInteraction:Fort Social Interaction VM Object Reference))
   Branch(execute=@n4.else, Condition="Equal (Name)"(A=@n1, B=Block))
   > ExecuteInteraction(self=Get BlockPlayerInteraction() -> (BlockPlayerInteraction:Fort Social Interaction VM Object Reference))
  ]}
 NavigateToBlockPlayer {nodes [
   n1 = Inputs!() -> (execute, SocialUser:Fort Social User Object Reference)
   > n2 = Construct Object from Class(Class=/Sidebar/Panels/Overlay/Data/ConfirmationVM_BlockPlayer.ConfirmationVM_BlockPlayer_C) -> (ReturnValue:Confirmation VM Block Player Object Reference)
   > Set Description!(
    execute
    Description:Text=Format Text!(
     Format="LOCTEXT(\"Would you like to block {PlayerName}?\")"
     PlayerName:String=GetNicknameOrDisplayName!(self:Fort Social User Object Reference=@n1) -> (ReturnValue:String)
    ) -> (Result)
    self:Sidebar Confirmation Panel VM Object Reference=@n2
   ) -> (then, Output_Get:Text)
   > NavigateToConfirmPanel!(
    execute
    SelfReference:Object Reference=Self-Reference()
    SidebarConfirmationVM:Sidebar Confirmation Panel VM Object Reference=@n2
    SidebarNavigationVM:Fort Sidebar Navigation VM Object Reference=Get FortSidebarNavigationVM()
   ) -> (then)
   > Outputs()
  ]}
 NavigateToChatInvite {nodes [
   Inputs!() -> (execute)
   n1 = Get FortTextChatGroupConversationVM()
   n2 = Get InvitedMembers!(self:Fort Text Chat Conversation VM Object Reference=@n1) -> (InvitedMembers:Array of Fort Social User Viewmodel Object References)
   n3 = Get JoinedMembers!(self:Fort Text Chat Conversation VM Object Reference=@n1) -> (JoinedMembers:Array of Fort Social User Viewmodel Object References)
   > n4 = Construct Object from Class!(
    execute
    Class=/Sidebar/Shared/Navigation/BP_NavigationPoint.BP_NavigationPoint_C
    self
    SocialUser:Fort Social User Object Reference
    ConversationId:Name=Get ConversationId(self:Fort Text Chat Conversation VM Object Reference=@n1) -> (ConversationId=None)
    ConfirmationVM:Sidebar Confirmation Panel VM Object Reference
    ReservedSpots:int=int + int(A=Length(TargetArray=@n3), B=Length(TargetArray=@n2))
    SocialItemVM:Fort Social Item VM Object Reference
    NavigationTag:Gameplay Tag Structure=(TagName="UI.StateTree.TransitionTag.Sidebar.Chat.Invite")
   ) -> (then, ReturnValue:BP Navigation Point Object Reference)
   > PushNavigation!(
    execute
    self:Fort Sidebar Navigation VM Object Reference=Get FortSidebarNavigationVM()
    NavigationPoint:Fort Sidebar Navigation Point Object Reference=@n4
   ) -> (then)
   > Outputs()
  ]}
 NavigateToConfirmPanel {nodes [
   n1 = NavigateToConfirmPanel!() -> (then, Class:Sidebar Confirmation Panel VM Class Reference) {is=title K2Node_FunctionEntry}
   > n2 = Construct Object from Class!(execute, Class=@n1, self) -> (then, ReturnValue:Sidebar Confirmation Panel VM Object Reference)
   > NavigateToConfirmPanel(SelfReference=Self-Reference(), SidebarConfirmationVM=@n2, SidebarNavigationVM=Get FortSidebarNavigationVM()) {is=title K2Node_CallFunction}
  ]}
 NavigateToRemoveFriend {nodes [
   n1 = Inputs!() -> (execute, SocialUser:Fort Social User Object Reference)
   > n2 = Construct Object from Class(Class=/Sidebar/Panels/Overlay/Data/ConfirmationVM_RemoveFriend.ConfirmationVM_RemoveFriend_C) -> (ReturnValue:Confirmation VM Remove Friend Object Reference)
   > Set Description(
    Description=Format Text!(Format="LOCTEXT(\"Are you sure you’d like to remove {FriendName} as a friend?\")", FriendName:String=GetNicknameOrDisplayName(self=@n1)) -> (Result)
    self=@n2
   )
   > NavigateToConfirmPanel(SelfReference=Self-Reference(), SidebarConfirmationVM=@n2, SidebarNavigationVM=Get FortSidebarNavigationVM())
   > Outputs()
  ]}
 OnNavDownFromInviteButton {nodes [
   OnNavDownFromInviteButton!() -> (then, Navigation:EUINavigation Enum) {is=title K2Node_FunctionEntry}
   n1 = Get TreeView(self=Get Joined_TreeList())
   > n2 = Branch(Condition="integer > integer"(A=GetNumItems(self=@n1) -> (ReturnValue=0)))
   > Return Node!(execute, ReturnValue:Widget Object Reference=Get TreeView(self=Get Joined_TreeList()))
   n3 = Get TreeView(self=Get Invited_TreeList())
   n4 = Branch(execute=@n2.else, Condition="integer > integer"(A=GetNumItems(self=@n3) -> (ReturnValue=0)))
   > Return Node(ReturnValue=Get TreeView(self=Get Invited_TreeList()))
   n5 = GetBottomButtonFocusTarget(execute=@n4.else) {is=title K2Node_CallFunction}
   > Return Node(ReturnValue=@n5)
  ]}
 OnNavDownFromJoinedList {nodes [
   OnNavDownFromJoinedList!() -> (then, Navigation:EUINavigation Enum) {is=title K2Node_FunctionEntry}
   n1 = Get TreeView(self=Get Invited_TreeList())
   > n2 = Branch(Condition="integer > integer"(A=GetNumItems(self=@n1) -> (ReturnValue=0)))
   > Return Node(ReturnValue=Get TreeView(self=Get Invited_TreeList()))
   n3 = GetBottomButtonFocusTarget(execute=@n2.else)
   > Return Node(ReturnValue=@n3)
  ]}
 OnNavDownFromPendingList {nodes [
   OnNavDownFromPendingList!() -> (then, Navigation:EUINavigation Enum) {is=title K2Node_FunctionEntry}
   > n1 = GetBottomButtonFocusTarget()
   > Return Node(ReturnValue=@n1)
  ]}
 OnNavUpFromPendingList {nodes [
   OnNavUpFromPendingList!() -> (then, Navigation:EUINavigation Enum) {is=title K2Node_FunctionEntry}
   n1 = Get TreeView(self=Get Joined_TreeList())
   > n2 = Branch(Condition="integer > integer"(A=GetNumItems(self=@n1) -> (ReturnValue=0)))
   > Return Node(ReturnValue=Get TreeView(self=Get Joined_TreeList()))
   Return Node(execute=@n2.else, ReturnValue=Get Button_AddMembers())
  ]}
 OnReportClosed {nodes [
   OnReportClosed!() -> (then, ReportState:ETextReportState Enum=BeginReport) {is=title K2Node_FunctionEntry}
   > n1 = Sequence()
   n2 = Get ReportState() -> (ReportState:ETextReportState Enum=BeginReport)
   n3 = "Equal (Enum)"(A:ETextReportState Enum=@n2, B:ETextReportState Enum=ReportSuccessful)
   n4 = "Equal (Enum)"(A=@n2, B=BlockPlayers)
   > Branch(Condition=AND Boolean(A=OR Boolean(A=@n3, B=@n4), B=Get LeaveGroupAfterReport() -> (LeaveGroupAfterReport:bool)))
   > RequestLeave(self=Get FortTextChatGroupConversationVM())
   Set LeaveGroupAfterReport(execute=@n1.then_1, LeaveGroupAfterReport=false)
  ]}
 OpenPlayerProfile {nodes [
   n1 = OpenPlayerProfile!() -> (then, SocialItemVM:Fort Social Item VM Object Reference) {is=title K2Node_FunctionEntry}
   > NavigateToPlayerProfile!(
    execute
    SelfReference:Object Reference=Self-Reference()
    SocialItemVM:Fort Social Item VM Object Reference=@n1
    SidebarNavigationVM:Fort Sidebar Navigation VM Object Reference=Get FortSidebarNavigationVM()
    Sidebar VM:Fort Sidebar VM Object Reference
   ) -> (then)
  ]}
 SetCompactUIStyle {nodes [
   SetCompactUIStyle() {is=title K2Node_FunctionEntry}
   > Branch(Condition=IsUsingCompactUI())
   > n1 = Sequence!(execute) -> (then_0, then_1, then_2:Exec)
   n2 = Get Button_AddMembers()
   SetIconSize!(execute=@n1.then_1, self:WBP UIKit Button Generic Base Object Reference=@n2, Icon:Vector 2D Structure=(X=48,Y=48)) -> (then)
   n3 = Get!(self:WBP UIKit Button Generic Base Object Reference=@n2) -> (IconPadding_Left:float=~, IconPadding_Top:float=~, IconPadding_Right:float=~, IconPadding_Bottom:float=~)
   > SetIconPadding!(
    execute
    self:WBP UIKit Button Generic Base Object Reference=@n2
    Padding_Left:float=@n3
    Padding_Top:float=@n3.IconPadding_Top
    Padding_Right:float=16
    Padding_Bottom:float=@n3.IconPadding_Bottom
   ) -> (then)
   > Set!(
    execute
    Padding_Left:float=22
    Padding_Top:float=12
    Padding_Right:float=22
    Padding_Bottom:float=12
    self:Stack Box Slot Object Reference=SlotAsStackBoxSlot!(Widget:Widget Object Reference=@n2) -> (ReturnValue:Stack Box Slot Object Reference)
   ) -> (then, Output_Get:Margin Structure)
   n4 = Get Button_ReportGroup()
   n5 = Get Button_LeaveGroup()
   > SetIconSize(self=[@n4, @n5], Icon=(X=36,Y=36))
   n6 = Get(self=@n4) -> (IconPadding_Left=~, IconPadding_Top=~, IconPadding_Right=~, IconPadding_Bottom=~)
   > SetIconPadding(self=@n4, Padding_Left=@n6, Padding_Top=@n6.IconPadding_Top, Padding_Right=12, Padding_Bottom=@n6.IconPadding_Bottom)
   n7 = Get(self=@n5) -> (IconPadding_Left=~, IconPadding_Top=~, IconPadding_Right=~, IconPadding_Bottom=~)
   > SetIconPadding(self=@n5, Padding_Left=@n7, Padding_Top=@n7.IconPadding_Top, Padding_Right=12, Padding_Bottom=@n7.IconPadding_Bottom)
   Set VerticalEntrySpacing!(execute=@n1.then_2, VerticalEntrySpacing:float=6, self:List View Object Reference=Get TreeView(self=Get Joined_TreeList())) -> (then, Output_Get:float)
   > Set VerticalEntrySpacing(VerticalEntrySpacing=6, self=Get TreeView(self=Get Invited_TreeList()))
   n8 = SlotAsStackBoxSlot(Widget=Get Text_InvitedLabel() -> (Text_InvitedLabel:Common Text Object Reference))
   n9 = Get!(self:Stack Box Slot Object Reference=@n8) -> (Padding_Left:float=~, Padding_Top:float=~, Padding_Right:float=~, Padding_Bottom:float=~)
   > Set(Padding_Left=@n9, Padding_Top=20, Padding_Right=@n9.Padding_Right, Padding_Bottom=@n9.Padding_Bottom, self=@n8)
   n10 = SlotAsStackBoxSlot(Widget=Get Text_MembersLabel() -> (Text_MembersLabel:Common Text Object Reference))
   n11 = Get(self=@n10) -> (Padding_Left=~, Padding_Top=~, Padding_Right=~, Padding_Bottom=~)
   > Set(Padding_Left=22, Padding_Top=@n11.Padding_Top, Padding_Right=@n11.Padding_Right, Padding_Bottom=@n11.Padding_Bottom, self=@n10)
   n12 = Get Button_Reset()
   SetIconSize(execute=@n1.then_0, self=@n12, Icon=(X=48,Y=48))
   n13 = Get!(self:WBP UIKit Button Generic Base Object Reference=@n12) -> (MobileButtonPadding_Left:float=~, MobileButtonPadding_Top:float=~, MobileButtonPadding_Right:float=~, MobileButtonPadding_Bottom:float=~)
   > SetButtonPadding!(
    execute
    self:WBP UIKit Button Generic Base Object Reference=@n12
    Padding_Left:float=16
    Padding_Top:float=@n13.MobileButtonPadding_Top
    Padding_Right:float=16
    Padding_Bottom:float=@n13.MobileButtonPadding_Bottom
   ) -> (then)
   n14 = Get Icon_Error() -> (Icon_Error:Image Object Reference)
   > SetDesiredSizeOverride!(execute, self:Image Object Reference=@n14, DesiredSize:Vector 2D Structure=(X=24,Y=24)) -> (then)
   n15 = SlotAsStackBoxSlot(Widget=@n14)
   n16 = Get!(self:Stack Box Slot Object Reference=@n15) -> (Padding_Left:float=~, Padding_Top:float=~, Padding_Right:float=~, Padding_Bottom:float=~)
   > Set(Padding_Left=4, Padding_Top=4, Padding_Right=@n16.Padding_Right, Padding_Bottom=@n16.Padding_Bottom, self=@n15)
  ]}
 SetInputFieldEnabled {nodes [
   n1 = SetInputFieldEnabled!() -> (then, IsEnabled:bool=~) {is=title K2Node_FunctionEntry}
   > SetVisibility(
    self=Get InputField_GroupName()
    InVisibility=Select(Option 0:ESlateVisibility Enum=HitTestInvisible, Option 1:ESlateVisibility Enum=SelfHitTestInvisible, Index=@n1) -> (ReturnValue:ESlateVisibility Enum)
   )
   > SetIsEnabled(self=[Get Button_SaveName(), Get Button_Reset()], bInIsEnabled=Get IsEnabled() -> (IsEnabled:bool))
   > Set RenderOpacity!(execute, RenderOpacity:float=SelectFloat(A=1, B=0.2, bPickA=Get IsEnabled()), self:Widget Object Reference=Get InputField_GroupName()) -> (then, Output_Get:float)
  ]}
 SetInvitedMembers {nodes [
   n1 = SetInvitedMembers!() -> (then, InvitedMembers:Array of Fort Social User Viewmodel Object References) {is=title K2Node_FunctionEntry}
   > Set List Items!(execute, self:List View Object Reference=Get TreeView(self=Get Invited_TreeList()), InListItems:Array of Object References=@n1) -> (then) {is=title K2Node_CallFunction}
   n2 = Length(TargetArray=@n1)
   > SetVisibility(self=Get Text_InvitedLabel(), InVisibility=Select(Option 0=Collapsed, Option 1=SelfHitTestInvisible, Index="integer > integer"(A=@n2)))
  ]}
 SetJoinedMembers {nodes [
   n1 = SetJoinedMembers!() -> (then, JoinedMembers:Array of Fort Social User Viewmodel Object References) {is=title K2Node_FunctionEntry}
   > Set List Items(self=Get TreeView(self=Get Joined_TreeList()), InListItems=@n1)
  ]}
 SetPlayerAvatars {nodes [
   n1 = SetPlayerAvatars!() -> (then, SocialUserVMs:Array of Fort Social User Viewmodel Object References) {is=title K2Node_FunctionEntry}
   > n2 = ConvertSocialUserArrayWithoutSelf!(execute, Social User VM:Array of Fort Social User Viewmodel Object References=@n1) -> (then, Social Users Without Self:Array of Fort Social User Object References)
   > InitializeWithSocialUsers!(
    execute
    self:WBP Sidebar Player Avatar Object Reference=Get GroupAvatar() -> (GroupAvatar:WBP Sidebar Player Avatar Object Reference)
    SocialUsers:Array of Fort Social User Object References=@n2
   ) -> (then)
  ]}
 UpdateGroupNameValues {nodes [
   Inputs!() -> (execute)
   n1 = Get FortTextChatGroupConversationVM()
   n2 = Get ConversationName(self=@n1)
   n3 = Get DefaultGeneratedConversationName(self=@n1)
   > SetText(
    self:WBP UIKit Input Field Base Object Reference=Get InputField_GroupName()
    Text=SelectText(B=@n2, bPickA="Equal Exactly (Text)"(A=@n2, B=@n3))
   )
   > SetVisibility(self=Get StackBox_Error(), InVisibility=Collapsed)
   > "SetText (Text)"(self=Get Text_Error())
   n4 = Get FortTextChatGroupConversationVM()
   n5 = Get ConversationName(self=@n4)
   n6 = Get DefaultGeneratedConversationName(self=@n4)
   > SetVisibility(self=Get Button_Reset(), InVisibility=Select(Option 0=SelfHitTestInvisible, Option 1=Collapsed, Index="Equal Exactly (Text)"(A=@n5, B=@n6)))
   > SetVisibility(self=Get Button_SaveName(), InVisibility=Collapsed)
   > Outputs()
  ]}
]
variables [
 BlockPlayerInteraction {type=FortSocialInteractionVM, default=null}
 LeaveGroupAfterReport {type=boolean}
 RemoveFriendInteraction {type=FortSocialInteractionVM, default=null}
]
widgets {
 info {
  inheritedWidgetCount=0
  namedSlotCount=1
  parentClass=&CommonActivatableWidget
  rootWidgetClass=/Sidebar/Shared/WBP_Sidebar_PanelLayout.WBP_Sidebar_PanelLayout_C
  widgetCount=26
 }
 widgets [
  PanelLayout {widgetClassPath=/Sidebar/Shared/WBP_Sidebar_PanelLayout, bIsVariable=true, widget=PanelLayout, properties {headingText=Manage group chat}}
  Root_Grid {
   widgetClassPath=&GridPanel
   namedSlotHost=PanelLayout
   widget=Root_Grid
   properties {bOverride_Cursor=true, columnFill [1], rowFill [1,0]}
  }
  ScrollBox {
   widgetClassPath=&ScrollBox
   bIsVariable=true
   parent=Root_Grid
   widget=ScrollBox
   properties {
    navigationScrollPadding=80
    scrollBarVisibility=Collapsed
    scrollWhenFocusChanges=AnimatedScroll
    scrollbarPadding=<ltrb 0 0 0 0>
    scrollbarThickness=<0 0>
    widgetBarStyle {
     thickness=16
     draggedThumbImage {
      bIsDynamicallyLoaded=false
      drawAs=NoDrawType
      imageType=FullColor
      mirroring=NoMirror
      resourceName=../../../Engine/Content/Slate/Common/Scrollbar_Thumb.png
      resourceObject=null
      tiling=NoTile
      imageSize=<0 0>
      margin=<ltrb 0.25 0.25 0.25 0.25>
      outlineSettings {
       bUseBrushTransparency=false
       roundingType=HalfHeightRadius
       width=0
       color {colorUseRule=UseColor_Specified, specifiedColor=<rgba 0 0 0 0>}
       cornerRadii=<0 0 0 0>
      }
      tintColor {colorUseRule=UseColor_Specified, specifiedColor=<rgba 1 1 1 1>}
      uVRegion {bIsValid=false, max=<0 0>, min=<0 0>}
     }
     horizontalBackgroundImage {
      bIsDynamicallyLoaded=false
      drawAs=NoDrawType
      imageType=NoImage
      mirroring=NoMirror
      resourceName=None
      resourceObject=null
      tiling=NoTile
      imageSize=<0 0>
      margin=<ltrb 0 0 0 0>
      outlineSettings {
       bUseBrushTransparency=false
       roundingType=HalfHeightRadius
       width=0
       color {colorUseRule=UseColor_Specified, specifiedColor=<rgba 0 0 0 0>}
       cornerRadii=<0 0 0 0>
      }
      tintColor {colorUseRule=UseColor_Specified, specifiedColor=<rgba 1 1 1 1>}
      uVRegion {bIsValid=false, max=<0 0>, min=<0 0>}
     }
     horizontalBottomSlotImage {
      bIsDynamicallyLoaded=false
      drawAs=NoDrawType
      imageType=FullColor
      mirroring=NoMirror
      resourceName=../../../Engine/Content/Slate/Common/Scrollbar_Background_Horizontal.png
      resourceObject=null
      tiling=NoTile
      imageSize=<8 8>
      margin=<ltrb 0 0 0 0>
      outlineSettings {
       bUseBrushTransparency=false
       roundingType=HalfHeightRadius
       width=0
       color {colorUseRule=UseColor_Specified, specifiedColor=<rgba 0 0 0 0>}
       cornerRadii=<0 0 0 0>
      }
      tintColor {colorUseRule=UseColor_Specified, specifiedColor=<rgba 1 1 1 1>}
      uVRegion {bIsValid=false, max=<0 0>, min=<0 0>}
     }
     horizontalTopSlotImage {
      bIsDynamicallyLoaded=false
      drawAs=NoDrawType
      imageType=FullColor
      mirroring=NoMirror
      resourceName=../../../Engine/Content/Slate/Common/Scrollbar_Background_Horizontal.png
      resourceObject=null
      tiling=NoTile
      imageSize=<8 8>
      margin=<ltrb 0 0 0 0>
      outlineSettings {
       bUseBrushTransparency=false
       roundingType=HalfHeightRadius
       width=0
       color {colorUseRule=UseColor_Specified, specifiedColor=<rgba 0 0 0 0>}
       cornerRadii=<0 0 0 0>
      }
      tintColor {colorUseRule=UseColor_Specified, specifiedColor=<rgba 1 1 1 1>}
      uVRegion {bIsValid=false, max=<0 0>, min=<0 0>}
     }
     hoveredThumbImage {
      bIsDynamicallyLoaded=false
      drawAs=NoDrawType
      imageType=FullColor
      mirroring=NoMirror
      resourceName=../../../Engine/Content/Slate/Common/Scrollbar_Thumb.png
      resourceObject=null
      tiling=NoTile
      imageSize=<0 0>
      margin=<ltrb 0.25 0.25 0.25 0.25>
      outlineSettings {
       bUseBrushTransparency=false
       roundingType=HalfHeightRadius
       width=0
       color {colorUseRule=UseColor_Specified, specifiedColor=<rgba 0 0 0 0>}
       cornerRadii=<0 0 0 0>
      }
      tintColor {colorUseRule=UseColor_Specified, specifiedColor=<rgba 1 1 1 1>}
      uVRegion {bIsValid=false, max=<0 0>, min=<0 0>}
     }
     normalThumbImage {
      bIsDynamicallyLoaded=false
      drawAs=NoDrawType
      imageType=FullColor
      mirroring=NoMirror
      resourceName=../../../Engine/Content/Slate/Common/Scrollbar_Thumb.png
      resourceObject=null
      tiling=NoTile
      imageSize=<0 0>
      margin=<ltrb 0.25 0.25 0.25 0.25>
      outlineSettings {
       bUseBrushTransparency=false
       roundingType=HalfHeightRadius
       width=0
       color {colorUseRule=UseColor_Specified, specifiedColor=<rgba 0 0 0 0>}
       cornerRadii=<0 0 0 0>
      }
      tintColor {colorUseRule=UseColor_Specified, specifiedColor=<rgba 1 1 1 1>}
      uVRegion {bIsValid=false, max=<0 0>, min=<0 0>}
     }
     verticalBackgroundImage {
      bIsDynamicallyLoaded=false
      drawAs=NoDrawType
      imageType=NoImage
      mirroring=NoMirror
      resourceName=None
      resourceObject=null
      tiling=NoTile
      imageSize=<0 0>
      margin=<ltrb 0 0 0 0>
      outlineSettings {
       bUseBrushTransparency=false
       roundingType=HalfHeightRadius
       width=0
       color {colorUseRule=UseColor_Specified, specifiedColor=<rgba 0 0 0 0>}
       cornerRadii=<0 0 0 0>
      }
      tintColor {colorUseRule=UseColor_Specified, specifiedColor=<rgba 1 1 1 1>}
      uVRegion {bIsValid=false, max=<0 0>, min=<0 0>}
     }
     verticalBottomSlotImage {
      bIsDynamicallyLoaded=false
      drawAs=NoDrawType
      imageType=FullColor
      mirroring=NoMirror
      resourceName=../../../Engine/Content/Slate/Common/Scrollbar_Background_Vertical.png
      resourceObject=null
      tiling=NoTile
      imageSize=<8 8>
      margin=<ltrb 0 0 0 0>
      outlineSettings {
       bUseBrushTransparency=false
       roundingType=HalfHeightRadius
       width=0
       color {colorUseRule=UseColor_Specified, specifiedColor=<rgba 0 0 0 0>}
       cornerRadii=<0 0 0 0>
      }
      tintColor {colorUseRule=UseColor_Specified, specifiedColor=<rgba 1 1 1 1>}
      uVRegion {bIsValid=false, max=<0 0>, min=<0 0>}
     }
     verticalTopSlotImage {
      bIsDynamicallyLoaded=false
      drawAs=NoDrawType
      imageType=FullColor
      mirroring=NoMirror
      resourceName=../../../Engine/Content/Slate/Common/Scrollbar_Background_Vertical.png
      resourceObject=null
      tiling=NoTile
      imageSize=<8 8>
      margin=<ltrb 0 0 0 0>
      outlineSettings {
       bUseBrushTransparency=false
       roundingType=HalfHeightRadius
       width=0
       color {colorUseRule=UseColor_Specified, specifiedColor=<rgba 0 0 0 0>}
       cornerRadii=<0 0 0 0>
      }
      tintColor {colorUseRule=UseColor_Specified, specifiedColor=<rgba 1 1 1 1>}
      uVRegion {bIsValid=false, max=<0 0>, min=<0 0>}
     }
    }
    widgetStyle {
     barThickness=9
     bottomShadowBrush {
      bIsDynamicallyLoaded=false
      drawAs=NoDrawType
      imageType=FullColor
      mirroring=NoMirror
      resourceName=../../../Engine/Content/Slate/Common/ScrollBoxShadowBottom.png
      resourceObject=null
      tiling=NoTile
      imageSize=<16 8>
      margin=<ltrb 0.5 0 0.5 1>
      outlineSettings {
       bUseBrushTransparency=false
       roundingType=HalfHeightRadius
       width=0
       color {colorUseRule=UseColor_Specified, specifiedColor=<rgba 0 0 0 0>}
       cornerRadii=<0 0 0 0>
      }
      tintColor {colorUseRule=UseColor_Specified, specifiedColor=<rgba 1 1 1 1>}
      uVRegion {bIsValid=false, max=<0 0>, min=<0 0>}
     }
     horizontalScrolledContentPadding=<ltrb 0 0 1 0>
     leftShadowBrush {
      bIsDynamicallyLoaded=false
      drawAs=NoDrawType
      imageType=FullColor
      mirroring=NoMirror
      resourceName=../../../Engine/Content/Slate/Common/ScrollBoxShadowLeft.png
      resourceObject=null
      tiling=NoTile
      imageSize=<8 16>
      margin=<ltrb 1 0.5 0 0.5>
      outlineSettings {
       bUseBrushTransparency=false
       roundingType=HalfHeightRadius
       width=0
       color {colorUseRule=UseColor_Specified, specifiedColor=<rgba 0 0 0 0>}
       cornerRadii=<0 0 0 0>
      }
      tintColor {colorUseRule=UseColor_Specified, specifiedColor=<rgba 1 1 1 1>}
      uVRegion {bIsValid=false, max=<0 0>, min=<0 0>}
     }
     rightShadowBrush {
      bIsDynamicallyLoaded=false
      drawAs=NoDrawType
      imageType=FullColor
      mirroring=NoMirror
      resourceName=../../../Engine/Content/Slate/Common/ScrollBoxShadowRight.png
      resourceObject=null
      tiling=NoTile
      imageSize=<8 16>
      margin=<ltrb 0 0.5 1 0.5>
      outlineSettings {
       bUseBrushTransparency=false
       roundingType=HalfHeightRadius
       width=0
       color {colorUseRule=UseColor_Specified, specifiedColor=<rgba 0 0 0 0>}
       cornerRadii=<0 0 0 0>
      }
      tintColor {colorUseRule=UseColor_Specified, specifiedColor=<rgba 1 1 1 1>}
      uVRegion {bIsValid=false, max=<0 0>, min=<0 0>}
     }
     topShadowBrush {
      bIsDynamicallyLoaded=false
      drawAs=NoDrawType
      imageType=FullColor
      mirroring=NoMirror
      resourceName=../../../Engine/Content/Slate/Common/ScrollBoxShadowTop.png
      resourceObject=null
      tiling=NoTile
      imageSize=<16 8>
      margin=<ltrb 0.5 1 0.5 0>
      outlineSettings {
       bUseBrushTransparency=false
       roundingType=HalfHeightRadius
       width=0
       color {colorUseRule=UseColor_Specified, specifiedColor=<rgba 0 0 0 0>}
       cornerRadii=<0 0 0 0>
      }
      tintColor {colorUseRule=UseColor_Specified, specifiedColor=<rgba 1 1 1 1>}
      uVRegion {bIsValid=false, max=<0 0>, min=<0 0>}
     }
     verticalScrolledContentPadding=<ltrb 0 0 0 1>
    }
   }
   slot {class=GridSlot, properties {padding=<ltrb -18 0 -18 0>}}
  }
  Content {
   widgetClassPath=&StackBox
   parent=ScrollBox
   widget=Content
   properties {bOverride_Cursor=true, orientation=Orient_Vertical}
   slot {class=ScrollBoxSlot, properties {padding=<ltrb 18 0 18 18>}}
  }
  WBP_Sidebar_TextChat_Restrictions {
   widgetClassPath=&/WBP_Sidebar_TextChat_Restrictions
   bIsVariable=true
   parent=Content
   widget=WBP_Sidebar_TextChat_Restrictions
   properties {restrictionWidgetPadding=<ltrb 0 0 0 12>}
   slot {class=StackBoxSlot}
  }
  GroupAvatarAndName {
   widgetClassPath=&StackBox
   parent=Content
   widget=GroupAvatarAndName
   slot {class=StackBoxSlot, properties {verticalAlignment=VAlign_Top, padding=<ltrb 18 0 18 24>}}
  }
  GroupAvatar {
   widgetClassPath=/Sidebar/Panels/WBP_Sidebar_PlayerAvatar
   bIsVariable=true
   parent=GroupAvatarAndName
   widget=GroupAvatar
   properties {avatarSize=<88 88>}
   slot {class=StackBoxSlot, properties {horizontalAlignment=HAlign_Left, verticalAlignment=VAlign_Center, padding=<ltrb 0 0 20 0>}}
  }
  StackBox_2 {
   widgetClassPath=&StackBox
   parent=GroupAvatarAndName
   widget=StackBox_2
   properties {orientation=Orient_Vertical}
   slot {class=StackBoxSlot, properties {size {sizeRule=Fill, value=1}}}
  }
  Text_GroupLabel {
   widgetClassPath=&CommonTextBlock
   parent=StackBox_2
   widget=Text_GroupLabel
   properties {
    mobileFontSizeMultiplier=1.33
    text=Group name
    font {
     bForceMonospaced=false
     bMaterialIsStencil=false
     fontMaterial=/Sidebar/Materials/Fonts/MI_Sidebar_FontColor_Primary
     fontObject=&HeadingNow
     letterSpacing=0
     monospacedWidth=1
     size=15.75
     skewAmount=0
     typefaceFontName=75 Medium
     outlineSettings {
      bApplyOutlineToDropShadows=false
      bMiteredCorners=false
      bSeparateFillAlpha=false
      outlineMaterial=null
      outlineSize=0
      outlineColor=<rgba 0 0 0 1>
     }
    }
    shadowOffset=<0 0>
   }
   slot {class=StackBoxSlot, properties {padding=<ltrb 4 0 0 12>}}
  }
  StackBox_1 {
   widgetClassPath=&StackBox
   parent=StackBox_2
   widget=StackBox_1
   slot {class=StackBoxSlot, properties {size {sizeRule=Fill, value=1}}}
  }
  InputField_GroupName {
   widgetClassPath=/Sidebar/Buttons/WBP_Sidebar_InputField
   bIsVariable=true
   parent=StackBox_1
   widget=InputField_GroupName
   properties {
    bDisplayInputActionWhenNotInteractable=false
    bHideInputActionWithKeyboard=false
    bShouldUseFallbackDefaultInputAction=false
    hint Text=Name your group...
    holdData=&CommonUI_HoldData_Base_C
    icon=/FortUILibrary/Textures/Icons/T_UI_Icon_UIKit_EditCourse_MSDF
    minHeight=60
    mobile Text Size Multiplier=1.55
    iconProperties {
     material=/Sidebar/Materials/Icons/MI_Sidebar_Icon_Base
     image Size=<28 28>
     placement {
      horizontal Alignment=HAlign_Right
      vertical Alignment=VAlign_Center
      grid {column=2, column Span=0, layer=0, row=1, row Span=0, nudge=<0 0>}
      padding=<ltrb 0 0 15 0>
     }
    }
    textPadding=<ltrb 20 4 16 4>
   }
   slot {class=StackBoxSlot, properties {size {sizeRule=Fill, value=1}}}
  }
  Button_SaveName {
   widgetClassPath=/Sidebar/Buttons/WBP_Sidebar_ButtonGeneric
   bIsVariable=true
   parent=StackBox_1
   widget=Button_SaveName
   properties {
    buttonType=Attention
    icon=/FortUILibrary/Textures/Icons/T_UI_Icon_UIKit_Close_MSDF
    maxWidth=116
    minHeightOverride=60
    minWidth=60
    minWidthOverride=100
    mobileMinWidth=100
    text=Save
    visibility=Collapsed
    font {
     bForceMonospaced=false
     bMaterialIsStencil=false
     fontMaterial=/Sidebar/Materials/Fonts/MI_Sidebar_FontColor_Button_Attention
     fontObject=&HeadingNow
     letterSpacing=10
     monospacedWidth=1
     size=15.75
     skewAmount=0
     typefaceFontName=74 Regular
     outlineSettings {
      bApplyOutlineToDropShadows=false
      bMiteredCorners=false
      bSeparateFillAlpha=false
      outlineMaterial=/FortUILibrary/Systems/ModularButton/MaterialInstances/MI_UIKit_TextOutline
      outlineSize=0
      outlineColor=<rgba 0 0 0 1>
     }
    }
   }
   slot {class=StackBoxSlot, properties {verticalAlignment=VAlign_Center, padding=<ltrb 12 0 0 0>}}
  }
  Button_Reset {
   widgetClassPath=&WBP_Sidebar_ButtonGeneric
   bIsVariable=true
   parent=StackBox_1
   widget=Button_Reset
   properties {
    icon=&T_UI_Icon_UIKit_Close_MSDF
    maxWidth=105
    minHeightOverride=60
    minWidthOverride=60
    text=""
    useIconBlock=true
    useTextBlock=false
    visibility=Collapsed
    buttonPadding=<ltrb 0 0 0 0>
   }
   slot {class=StackBoxSlot, properties {verticalAlignment=VAlign_Center, padding=<ltrb 12 0 0 0>}}
  }
  StackBox_Error {
   widgetClassPath=&StackBox
   bIsVariable=true
   parent=StackBox_2
   widget=StackBox_Error
   properties {visibility=Collapsed}
   slot {class=StackBoxSlot, properties {padding=<ltrb 0 20 0 0>}}
  }
  Icon_Error {
   widgetClassPath=&Image
   bIsVariable=true
   parent=StackBox_Error
   widget=Icon_Error
   properties {
    brush {
     bIsDynamicallyLoaded=false
     drawAs=Image
     imageType=FullColor
     mirroring=NoMirror
     resourceName=None
     resourceObject=/FortUILibrary/Textures/MaterialInstances/MI_UIKit_Icon_Alert
     tiling=NoTile
     imageSize=<21 21>
     margin=<ltrb 0 0 0 0>
     outlineSettings {
      bUseBrushTransparency=false
      roundingType=HalfHeightRadius
      width=0
      color {colorUseRule=UseColor_Specified, specifiedColor=<rgba 0 0 0 0>}
      cornerRadii=<0 0 0 0>
     }
     tintColor {colorUseRule=UseColor_Specified, specifiedColor=<rgba 1 1 1 1>}
     uVRegion {bIsValid=false, max=<0 0>, min=<0 0>}
    }
    colorAndOpacity=<rgba 0.9823 0.1119 0.1119 1>
   }
   slot {class=StackBoxSlot, properties {verticalAlignment=VAlign_Top, padding=<ltrb 0 0 9 0>}}
  }
  Text_Error {
   widgetClassPath=&CommonTextBlock
   bIsVariable=true
   parent=StackBox_Error
   widget=Text_Error
   properties {
    autoWrapText=true
    mobileFontSizeMultiplier=1.33
    text=Error
    colorAndOpacity {colorUseRule=UseColor_Specified, specifiedColor=<rgba 0.9823 0.1119 0.1119 1>}
    font {
     bForceMonospaced=false
     bMaterialIsStencil=false
     fontMaterial=/FortUILibrary/Systems/InputField/MaterialInstances/MI_UIKit_InputFieldText
     fontObject=&HeadingNow
     letterSpacing=10
     monospacedWidth=1
     size=13.5
     skewAmount=0
     typefaceFontName=74 Regular
     outlineSettings {
      bApplyOutlineToDropShadows=false
      bMiteredCorners=false
      bSeparateFillAlpha=false
      outlineMaterial=null
      outlineSize=0
      outlineColor=<rgba 0 0 0 1>
     }
    }
    shadowOffset=<0 0>
   }
   slot {class=StackBoxSlot, properties {size {sizeRule=Fill, value=1}}}
  }
  Text_MembersLabel {
   widgetClassPath=&CommonTextBlock
   bIsVariable=true
   parent=Content
   widget=Text_MembersLabel
   properties {
    mobileFontSizeMultiplier=1.33
    text=Members
    font {
     bForceMonospaced=false
     bMaterialIsStencil=false
     fontMaterial=&MI_Sidebar_FontColor_Primary
     fontObject=&HeadingNow
     letterSpacing=0
     monospacedWidth=1
     size=15.75
     skewAmount=0
     typefaceFontName=75 Medium
     outlineSettings {
      bApplyOutlineToDropShadows=false
      bMiteredCorners=false
      bSeparateFillAlpha=false
      outlineMaterial=null
      outlineSize=0
      outlineColor=<rgba 0 0 0 1>
     }
    }
    shadowOffset=<0 0>
   }
   slot {class=StackBoxSlot, properties {padding=<ltrb 18 0 18 24>}}
  }
  Button_AddMembers {
   widgetClassPath=&WBP_Sidebar_ButtonGeneric
   bIsVariable=true
   parent=Content
   widget=Button_AddMembers
   properties {
    icon=/Sidebar/Materials/Icons/Textures/T_UI_Icon_Sidebar_AddFriend_MSDF
    iconMaterial=/Sidebar/Materials/Icons/MI_Sidebar_Icon_PartyOptions
    minHeight=0
    minHeightOverride=60
    navigation=&self:WidgetTree.Button_AddMembers.WidgetNavigation_0
    text=Invite to group
    useIconBlock=true
    font {
     bForceMonospaced=false
     bMaterialIsStencil=false
     fontMaterial=/Sidebar/Materials/Fonts/MI_Sidebar_FontColor_Secondary
     fontObject=&HeadingNow
     letterSpacing=10
     monospacedWidth=1
     size=15
     skewAmount=0
     typefaceFontName=74 Regular
     outlineSettings {
      bApplyOutlineToDropShadows=false
      bMiteredCorners=false
      bSeparateFillAlpha=false
      outlineMaterial=&MI_UIKit_TextOutline
      outlineSize=0
      outlineColor=<rgba 0 0 0 1>
     }
    }
    iconPadding=<ltrb 0 0 8 0>
    iconSize=<36 36>
   }
   slot {class=StackBoxSlot, properties {padding=<ltrb 18 0 18 16>}}
  }
  Joined_TreeList {
   widgetClassPath=/Sidebar/Panels/Social/List/WBP_Sidebar_SocialTreeList
   bIsVariable=true
   parent=Content
   widget=Joined_TreeList
   properties {navigation=&self:WidgetTree.Joined_TreeList.WidgetNavigation_0}
   slot {class=StackBoxSlot}
  }
  Text_InvitedLabel {
   widgetClassPath=&CommonTextBlock
   bIsVariable=true
   parent=Content
   widget=Text_InvitedLabel
   properties {
    mobileFontSizeMultiplier=1.33
    text=Pending invites
    font {
     bForceMonospaced=false
     bMaterialIsStencil=false
     fontMaterial=&MI_Sidebar_FontColor_Primary
     fontObject=&HeadingNow
     letterSpacing=0
     monospacedWidth=1
     size=15.75
     skewAmount=0
     typefaceFontName=75 Medium
     outlineSettings {
      bApplyOutlineToDropShadows=false
      bMiteredCorners=false
      bSeparateFillAlpha=false
      outlineMaterial=null
      outlineSize=0
      outlineColor=<rgba 0 0 0 1>
     }
    }
    shadowOffset=<0 0>
    strikeBrush {
     bIsDynamicallyLoaded=false
     drawAs=Image
     imageType=FullColor
     mirroring=NoMirror
     resourceName=None
     resourceObject=null
     tiling=NoTile
     imageSize=<32 32>
     margin=<ltrb 0 0 0 0>
     outlineSettings {
      bUseBrushTransparency=false
      roundingType=HalfHeightRadius
      width=0
      color {colorUseRule=UseColor_Specified, specifiedColor=<rgba 0 0 0 0>}
      cornerRadii=<0 0 0 0>
     }
     tintColor {colorUseRule=UseColor_Specified, specifiedColor=<rgba 1 1 1 1>}
     uVRegion {bIsValid=false, max=<0 0>, min=<0 0>}
    }
   }
   slot {class=StackBoxSlot, properties {padding=<ltrb 18 24 18 12>}}
  }
  Invited_TreeList {
   widgetClassPath=&WBP_Sidebar_SocialTreeList
   bIsVariable=true
   parent=Content
   widget=Invited_TreeList
   properties {navigation=&self:WidgetTree.Invited_TreeList.WidgetNavigation_0}
   slot {class=StackBoxSlot}
  }
  VerticalBox_0 {
   widgetClassPath=&VerticalBox
   parent=Root_Grid
   widget=VerticalBox_0
   slot {class=GridSlot, properties {row=1, padding=<ltrb 18 0 18 18>}}
  }
  Divider {
   widgetClassPath=&Image
   bIsVariable=true
   parent=VerticalBox_0
   widget=Divider
   properties {brush {
     bIsDynamicallyLoaded=false
     drawAs=Image
     imageType=FullColor
     mirroring=NoMirror
     resourceName=None
     resourceObject=/Sidebar/Materials/MI_Sidebar_Divider
     tiling=NoTile
     imageSize=<2 2>
     margin=<ltrb 0 0 0 0>
     outlineSettings {
      bUseBrushTransparency=false
      roundingType=HalfHeightRadius
      width=0
      color {colorUseRule=UseColor_Specified, specifiedColor=<rgba 0 0 0 0>}
      cornerRadii=<0 0 0 0>
     }
     tintColor {colorUseRule=UseColor_Specified, specifiedColor=<rgba 1 1 1 1>}
     uVRegion {bIsValid=false, max=<0 0>, min=<0 0>}
    }}
   slot {class=VerticalBoxSlot, properties {padding=<ltrb -36 0 -36 18>}}
  }
  HB_ButtonsAnchored {widgetClassPath=&HorizontalBox, parent=VerticalBox_0, widget=HB_ButtonsAnchored, slot {class=VerticalBoxSlot}}
  Button_ReportGroup {
   widgetClassPath=&WBP_Sidebar_ButtonGeneric
   bIsVariable=true
   parent=HB_ButtonsAnchored
   widget=Button_ReportGroup
   properties {
    icon=/Sidebar/Materials/Icons/Textures/T_UI_Icon_Sidebar_Report_MSDF
    iconMaterial=&MI_Sidebar_Icon_PartyOptions
    minHeight=0
    minHeightOverride=60
    text=Report group
    useIconBlock=true
    font {
     bForceMonospaced=false
     bMaterialIsStencil=false
     fontMaterial=&MI_Sidebar_FontColor_Secondary
     fontObject=&HeadingNow
     letterSpacing=10
     monospacedWidth=1
     size=15
     skewAmount=0
     typefaceFontName=74 Regular
     outlineSettings {
      bApplyOutlineToDropShadows=false
      bMiteredCorners=false
      bSeparateFillAlpha=false
      outlineMaterial=&MI_UIKit_TextOutline
      outlineSize=0
      outlineColor=<rgba 0 0 0 1>
     }
    }
    iconPadding=<ltrb 0 0 8 0>
    secondFont {
     bForceMonospaced=false
     bMaterialIsStencil=false
     fontMaterial=&MI_Sidebar_FontColor_Secondary
     fontObject=&Burbank
     letterSpacing=48
     monospacedWidth=1
     size=21
     skewAmount=0
     typefaceFontName=BurbankRegular-Black
     outlineSettings {
      bApplyOutlineToDropShadows=false
      bMiteredCorners=false
      bSeparateFillAlpha=false
      outlineMaterial=null
      outlineSize=0
      outlineColor=<rgba 0 0 0 1>
     }
    }
   }
   slot {class=HorizontalBoxSlot, properties {padding=<ltrb 0 0 6 0>, size {sizeRule=Fill, value=1}}}
  }
  Button_LeaveGroup {
   widgetClassPath=&WBP_Sidebar_ButtonGeneric
   bIsVariable=true
   parent=HB_ButtonsAnchored
   widget=Button_LeaveGroup
   properties {
    icon=/Sidebar/Materials/Icons/Textures/T_UI_Icon_Sidebar_LeaveParty_MSDF
    iconMaterial=&MI_Sidebar_Icon_PartyOptions
    minHeight=0
    minHeightOverride=60
    text=Leave group
    font {
     bForceMonospaced=false
     bMaterialIsStencil=false
     fontMaterial=/Sidebar/Materials/Fonts/MI_Sidebar_FontColor_Button_Alert
     fontObject=&HeadingNow
     letterSpacing=10
     monospacedWidth=1
     size=15
     skewAmount=0
     typefaceFontName=74 Regular
     outlineSettings {
      bApplyOutlineToDropShadows=false
      bMiteredCorners=false
      bSeparateFillAlpha=false
      outlineMaterial=&MI_UIKit_TextOutline
      outlineSize=0
      outlineColor=<rgba 0 0 0 1>
     }
    }
    iconPadding=<ltrb 0 0 12 0>
    iconSize=<36 36>
    secondFont {
     bForceMonospaced=false
     bMaterialIsStencil=false
     fontMaterial=&MI_Sidebar_FontColor_Secondary
     fontObject=&Burbank
     letterSpacing=48
     monospacedWidth=1
     size=21
     skewAmount=0
     typefaceFontName=BurbankRegular-Black
     outlineSettings {
      bApplyOutlineToDropShadows=false
      bMiteredCorners=false
      bSeparateFillAlpha=false
      outlineMaterial=null
      outlineSize=0
      outlineColor=<rgba 0 0 0 1>
     }
    }
   }
   slot {class=HorizontalBoxSlot, properties {padding=<ltrb 6 0 0 0>, size {sizeRule=Fill, value=1}}}
  }
 ]
}
