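# Sidebar text chat panel widget for group and DM conversations, driven by a group conversation view model with cached-message paging.
# On activation it initializes the conversation from the current navigation point, while conversion functions build the message list, subheading, spinner and inactive-message states.
# Inputs handle sending, speech-to-text dictation with a timer, read-aloud, and starting or joining parties.
# Context menus route profile, report, block, mute and manage-group actions to navigation, modals or the view model.
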WidgetBlueprint /Sidebar/Panels/Chat/WBP_Sidebar_TextChatPanel
graphs [
 AddRemoveManageGroupOrProfileContextMenuOption {nodes [
   AddRemoveManageGroupOrProfileContextMenuOption!() -> (then, PersistentMessageType:EFortTextChatPersistentConversationType Enum=EpicParty) {is=title K2Node_FunctionEntry}
   > RemoveInteraction!(
    execute
    self:UIKit Context Menu Action Context Object Reference=Get OptionsContextMenuActionContext() -> (OptionsContextMenuActionContext:UIKit Context Menu Action Context Object Reference)
    InteractionNameID:Name=ViewProfile
   ) -> (then)
   > RemoveInteraction(self=Get OptionsContextMenuActionContext(), InteractionNameID=ManageGroupChat)
   > n1 = Branch(Condition="Equal (Enum)"(
     A:EFortTextChatPersistentConversationType Enum=Get PersistentMessageType() -> (PersistentMessageType:EFortTextChatPersistentConversationType Enum=EpicParty)
     B:EFortTextChatPersistentConversationType Enum=Epic121
    ))
   > AddInteraction!(
    execute
    self:UIKit Context Menu Action Context Object Reference=Get OptionsContextMenuActionContext()
    InActionInfo_DisplayName:Text="LOCTEXT(\"View profile\")"
    InActionInfo_IconTexture:Texture 2D Soft Object Reference
    InActionInfo_IconMaterial:Material Instance Soft Object Reference
    InActionInfo_SortPriority:int=1
    InActionInfo_bIsAvailable:bool=True
    InActionInfo_bIsEnabled:bool=True
    InActionInfo_bIsSelected:bool=False
    InActionInfo_bAddDivider:bool=False
    InActionInfo_bPositionDividerAfter:bool=False
    InActionInfo_bCloseMenuOnSelect:bool=False
    InActionInfo_InteractionNameID:Name=ViewProfile
    InActionInfo_Interaction:Data Table Row Handle Structure=(DataTable=None,RowName="")
    InActionInfo_PressedDelegate:Delegate="(null).None"
    InActionInfo_ReleasedDelegate:Delegate="(null).None"
    InInteractionDelegate:"Delegate (by ref)"=Create Event()
   ) -> (then)
   n2 = Get FortTextChatGroupConversationVM() -> (FortTextChatGroupConversationVM:Fort Group Conversation Viewmodel Object Reference)
   AddInteraction(
    execute=@n1.else
    self=Get OptionsContextMenuActionContext()
    InActionInfo_DisplayName="LOCTEXT(\"Manage group chat\")"
    InActionInfo_SortPriority=1
    InActionInfo_bIsAvailable="Not Equal (Enum)"!(
     A:EFortTextChatPersistentConversationType Enum=Get ConversationType!(self:Fort Text Chat Conversation VM Object Reference=@n2) -> (ConversationType:EFortTextChatPersistentConversationType Enum=EpicParty)
     B:EFortTextChatPersistentConversationType Enum=EpicParty
    ) -> (ReturnValue:bool)
    InActionInfo_bIsEnabled=True
    InActionInfo_bIsSelected=False
    InActionInfo_bAddDivider=False
    InActionInfo_bPositionDividerAfter=False
    InActionInfo_bCloseMenuOnSelect=False
    InActionInfo_InteractionNameID=ManageGroupChat
    InActionInfo_Interaction=(DataTable=None,RowName="")
    InActionInfo_PressedDelegate="(null).None"
    InActionInfo_ReleasedDelegate="(null).None"
    InInteractionDelegate=Create Event()
   )
  ]}
 AddRemoveReportGroupContextMenuOption {nodes [
   AddRemoveReportGroupContextMenuOption!() -> (then, CanReport:bool=~) {is=title K2Node_FunctionEntry}
   > RemoveInteraction(self=Get OptionsContextMenuActionContext(), InteractionNameID=Report)
   n1 = Get FortTextChatGroupConversationVM()
   > AddInteraction(
    self=Get OptionsContextMenuActionContext()
    InActionInfo_DisplayName="LOCTEXT(\"Report chat\")"
    InActionInfo_SortPriority=3
    InActionInfo_bIsAvailable=Contains Item(
     TargetArray:Array of EFortTextChatConversationContextMenuAction Enums=GetConversationContextMenuActions!(self:Fort Text Chat Conversation VM Object Reference=@n1) -> (ReturnValue:Array of EFortTextChatConversationContextMenuAction Enums=Manage)
     ItemToFind:"EFortTextChatConversationContextMenuAction Enum (by ref)"=Report
    )
    InActionInfo_bIsEnabled=True
    InActionInfo_bIsSelected=False
    InActionInfo_bAddDivider=False
    InActionInfo_bPositionDividerAfter=False
    InActionInfo_bCloseMenuOnSelect=False
    InActionInfo_InteractionNameID=Report
    InActionInfo_Interaction=(DataTable=None,RowName="")
    InActionInfo_PressedDelegate="(null).None"
    InActionInfo_ReleasedDelegate="(null).None"
    InInteractionDelegate=Create Event()
   )
  ]}
 BP_GetDesiredFocusTarget {nodes [
   Get Desired Focus Target()
   > n1 = Branch(Condition=Get bIsEnabled!(self:Widget Object Reference=Get WBP_Sidebar_InputField() -> (WBP_Sidebar_InputField:WBP Sidebar Input Field Object Reference)) -> (bIsEnabled:bool))
   n2 = isInputFieldNotEmpty!(self) -> (NotEmpty:bool)
   n3 = Get is SpeechToText Enabled() -> (is SpeechToText Enabled:bool)
   n4 = Get Multiline InputField Enabled!(self) -> (isEnabled:bool) {is=title K2Node_CallFunction}
   n5 = Get EditableText!(self:WBP UIKit Block Editable Text Object Reference=Get Block_EditableText!(self:WBP UIKit Input Field Base Object Reference=Get WBP_Sidebar_InputField()) -> (Block_EditableText:WBP UIKit Block Editable Text Object Reference)) -> (EditableText:UIKit Editable Text Object Reference)
   n6 = Get EditableText(self:WBP Text Chat Block Multiline Editable Text Object Reference=Get Block_EditableText(self:WBP Text Chat Multiline Input Field Object Reference=Get WBP_Sidebar_InputField_Multiline() -> (WBP_Sidebar_InputField_Multiline:WBP Sidebar Input Field Multiline Object Reference)) -> (Block_EditableText:WBP Text Chat Block Multiline Editable Text Object Reference)) -> (EditableText:"Text Box (Multi- Line) Object Reference")
   > Return Node!(
    execute
    ReturnValue:Widget Object Reference=Select(
     Option 0:Widget Object Reference=Select(Option 0=@n5, Option 1=@n6, Index=@n4)
     Option 1:Widget Object Reference=Get Button_Send() -> (Button_Send:WBP Sidebar Button Generic Object Reference)
     Index=AND Boolean(A=@n2, B=@n3)
    ) -> (ReturnValue:Widget Object Reference)
   )
   Return Node(execute=@n1.else, ReturnValue=Get Button_Options() -> (Button_Options:WBP Sidebar Button Generic Object Reference))
  ]}
 Button Dictate Button Visuals {nodes [
   Button Dictate Button Visuals!() -> (then, DF Texture:Texture 2D Object Reference, Material Instance:Material Instance Object Reference) {is=title K2Node_FunctionEntry}
   n1 = Get Button_Dictate() -> (Button_Dictate:WBP Sidebar Button Generic Object Reference)
   > SetIconSize!(
    execute
    self:WBP UIKit Button Generic Base Object Reference=@n1
    Icon:Vector 2D Structure=Get DefaultIconSize() -> (DefaultIconSize:Vector 2D Structure)
   ) -> (then)
   > SetIconTexture!(
    execute
    self:WBP UIKit Button Generic Base Object Reference=@n1
    Icon:Texture 2D Object Reference=Get DF Texture() -> (DF Texture:Texture 2D Object Reference)
   ) -> (then)
   > SetIconMaterial!(
    execute
    self:WBP UIKit Button Generic Base Object Reference=@n1
    Material Instance:Material Instance Object Reference=Get Material Instance() -> (Material Instance:Material Instance Object Reference)
   ) -> (then)
  ]}
 CF_InactiveMessageVisible {nodes [
   CF_InactiveMessageVisible!() -> (then, IsChannelReadOnly:bool=~, IsRestrictionActive:bool=~) {is=title K2Node_FunctionEntry}
   n1 = NOT Boolean(A=Get IsRestrictionActive() -> (IsRestrictionActive:bool))
   n2 = Get IsChannelReadOnly() -> (IsChannelReadOnly:bool)
   > Return Node!(
    execute
    InactiveMessageVisibility:ESlateVisibility Enum=Select(Option 0:ESlateVisibility Enum=Collapsed, Option 1:ESlateVisibility Enum=SelfHitTestInvisible, Index=AND Boolean(A=@n2, B=@n1)) -> (ReturnValue:ESlateVisibility Enum)
   )
  ]}
 CF_InactiveText {nodes [
   n1 = CF_InactiveText!() -> (then, ConversationType:EFortTextChatPersistentConversationType Enum=EpicParty) {is=title K2Node_FunctionEntry}
   > Return Node!(
    execute
    InactiveMessageText:Text=CF_InactiveMessageText!(
     self:WBP Text Chat Inactive Message Object Reference=Get WBP_TextChat_InactiveMessage() -> (WBP_TextChat_InactiveMessage:WBP Text Chat Inactive Message Object Reference)
     ConversationType:EFortTextChatPersistentConversationType Enum=@n1
    ) -> (InactiveMessageText:Text)
   )
  ]}
 CF_ShouldShowSpinner {nodes [
   n1 = CF_ShouldShowSpinner!() -> (then, IsInitialized:bool=~, LoadingCachedMessages:bool=~) {is=title K2Node_FunctionEntry}
   > Return Node!(execute, ShouldShowSpinner:bool=OR Boolean(A=NOT Boolean(A=@n1), B=@n1.LoadingCachedMessages))
  ]}
 CF_SubheadingText {nodes [
   n1 = CF_SubheadingText!() -> (
    then
    Joined Members:Array of Fort Social User Viewmodel Object References
    Pending Invites:Array of Fort Social User Viewmodel Object References
    Conversation Type:EFortTextChatPersistentConversationType Enum=EpicParty
   ) {is=title K2Node_FunctionEntry}
   > n2 = Branch(Condition=NOT Boolean(A="Equal (Enum)"(A=@n1.Conversation Type, B=Epic121)))
   > Set NumJoinedMembers(NumJoinedMembers:int=Length(TargetArray:Array of Fort Social User Viewmodel Object References=Get Joined Members() -> (Joined Members:Array of Fort Social User Viewmodel Object References))) -> (Output_Get:int=0)
   n3 = Get NumJoinedMembers() -> (NumJoinedMembers:int=0)
   n4 = Get NumJoinedMembers() -> (NumJoinedMembers=0)
   n5 = Get NumJoinedMembers() -> (NumJoinedMembers=0)
   > Set JoinedMembersText(JoinedMembersText:Text=Select(
     Option 0:Text=Format Text!(Format="LOCTEXT(\"{NumJoinedMembers} member\")", NumJoinedMembers:int=@n3) -> (Result)
     Option 1:Text=Format Text(Format="LOCTEXT(\"{NumJoinedMembers} members\")", NumJoinedMembers=@n4)
     Index="integer > integer"(A=@n5, B=1)
    ) -> (ReturnValue:Text)) -> (Output_Get:Text)
   > n6 = Set NumPendingInvites(NumPendingInvites:int=Length(TargetArray=Get Pending Invites() -> (Pending Invites:Array of Fort Social User Viewmodel Object References))) -> (Output_Get:int=0)
   > n7 = Branch(Condition="integer > integer"(A=@n6))
   n8 = Get NumPendingInvites() -> (NumPendingInvites:int=0)
   n9 = Get JoinedMembersText() -> (JoinedMembersText:Text)
   n10 = Get NumPendingInvites() -> (NumPendingInvites=0)
   n11 = Get JoinedMembersText()
   n12 = Get NumPendingInvites() -> (NumPendingInvites=0)
   > n13 = Set JoinedMembersAndPendingInvitesText(JoinedMembersAndPendingInvitesText:Text=Select(
     Option 0=Format Text!(Format="LOCTEXT(\"{JoinedMembersText} · {NumPendingInvites} pending invite\")", JoinedMembersText:Text=@n11, NumPendingInvites:int=@n10) -> (Result)
     Option 1=Format Text(Format="LOCTEXT(\"{JoinedMembersText} · {NumPendingInvites} pending invites\")", JoinedMembersText=@n9, NumPendingInvites=@n8)
     Index="integer > integer"(A=@n12, B=1)
    )) -> (Output_Get:Text)
   > Return Node!(execute, Subheading:Text=@n13)
   Return Node(execute=@n7.else, Subheading=Get JoinedMembersText())
   Return Node(execute=@n2.else)
  ]}
 ConvertMessagesAndPaginationButtonIntoSingleList {nodes [
   n1 = ConvertMessagesAndPaginationButtonIntoSingleList!() -> (then, MessageVMs:Array of Text Chat Message VM Object References, HasMorePages:bool=~) {is=title K2Node_FunctionEntry}
   > n2 = InsertDividersIntoMessageList_Base!(execute, MessageVMs:Array of Text Chat Message VM Object References=@n1, OwningWidget:Object Reference=Self-Reference()) -> (then, MessageList:Array of MVVM Base Viewmodel Object References)
   > Set MessageListWithPaginationButton(MessageListWithPaginationButton:Array of MVVM Base Viewmodel Object References=@n2) -> (Output_Get:Array of MVVM Base Viewmodel Object References)
   > n3 = Branch(Condition=Get HasMorePages() -> (HasMorePages:bool))
   > n4 = Insert!(
    execute
    TargetArray:Array of MVVM Base Viewmodel Object References=Get MessageListWithPaginationButton() -> (MessageListWithPaginationButton:Array of MVVM Base Viewmodel Object References)
    NewItem:"MVVM Base Viewmodel Object Reference (by ref)"=Get LoadOlderMessagesButtonEntry() -> (LoadOlderMessagesButtonEntry:BP Sidebar Custom Button Entry VM Object Reference)
    Index:int=0
   ) -> (then) {is=title K2Node_CallArrayFunction}
   Return Node!(execute=[@n4.then, @n5.then], MessageList:Array of MVVM Base Viewmodel Object References=Get MessageListWithPaginationButton())
   n5 = Insert(
    execute=@n3.else
    TargetArray=Get MessageListWithPaginationButton()
    NewItem=Get WelcomeMessageEntry() -> (WelcomeMessageEntry:BP VM Text Chat Welcome Entry Object Reference)
    Index=0
   )
  ]}
 ConvertSpeakParametersToVisibility {nodes [
   n1 = ConvertSpeakParametersToVisibility!() -> (
    then
    Text to Speech State:EFortTextToSpeechState Enum=None
    Has Unspoken New Messages:bool=~
    Speech to Text State:EFortSpeechToTextState Enum=Idle
   ) {is=title K2Node_FunctionEntry}
   > n2 = Branch(Condition=GetIsSpeaking!(self:Fort Text Chat Conversation VM Object Reference=Get FortTextChatGroupConversationVM()) -> (ReturnValue:bool))
   > Return Node!(execute, Visibility:ESlateVisibility Enum=Visible)
   n3 = OR Boolean(
    A="Equal (Enum)"(A:EFortSpeechToTextState Enum=@n1.Speech to Text State, B:EFortSpeechToTextState Enum=Idle)
    B="Equal (Enum)"(A=@n1.Speech to Text State, B=Unavailable)
   )
   n4 = Get is TextToSpeech Enabled() -> (is TextToSpeech Enabled:bool)
   Return Node(
    execute=@n2.else
    Visibility=Select(
     Option 0:ESlateVisibility Enum=Collapsed
     Option 1:ESlateVisibility Enum=Visible
     Index=AND Boolean!(A=@n1.Has Unspoken New Messages, B=@n4, C:bool=@n3) -> (ReturnValue)
    ) -> (ReturnValue:ESlateVisibility Enum)
   )
  ]}
 ConvertSpeakingStateToText {nodes [
   n1 = ConvertSpeakingStateToText!() -> (then, Speaking State:EFortTextToSpeechState Enum=None) {is=title K2Node_FunctionEntry}
   > Return Node!(
    execute
    Text:Text=Select!(
     None:Text="LOCTEXT(\"Read new messages aloud\")"
     Preparing:Text="LOCTEXT(\"Please wait...\")"
     Playing:Text="LOCTEXT(\"Stop reading aloud\")"
     Failed:Text="LOCTEXT(\"Please wait...\")"
     Index:EFortTextToSpeechState Enum=@n1
    ) -> (ReturnValue:Text)
   )
  ]}
 CreateReportModal {nodes [
   n1 = CreateReportModal!() -> (then, Is Reporting Message:bool=~) {is=title K2Node_FunctionEntry}
   > n2 = Create Widget!(
    execute:Exec
    Class:User Widget Class Reference=/FortniteCoreUI/UI/Social/Chat/TextChat/Dialogs/WBP_Dialog_VM_TextChat_Report.WBP_Dialog_VM_TextChat_Report_C
    OwningPlayer:Player Controller Object Reference
   ) -> (then:Exec, ReturnValue:WBP Dialog VM Text Chat Report Object Reference)
   > SetReportedConversationVM!(
    execute
    self:WBP Dialog VM Text Chat Report Object Reference=@n2
    ReportedConversationVM:Fort Text Chat Conversation VM Object Reference=Get FortTextChatGroupConversationVM()
   ) -> (then)
   > n3 = Branch(Condition=@n1)
   > n4 = SetReportedMessageId!(execute, self:WBP Dialog VM Text Chat Report Object Reference=@n2, MessageId:Name=Get Message ID() -> (Message ID:Name=None)) -> (then)
   TryPushWidgetAsModalPanel!(execute=[@n4.then, @n3.else], PushWidget:User Widget Object Reference=@n2) -> (then)
  ]}
 DeInitVirtualKeyboard {nodes [
   DeInitVirtualKeyboard() {is=title K2Node_FunctionEntry}
   n1 = Get FortPlatformEventVM() -> (FortPlatformEventVM:Fort Platform Event VM Object Reference)
   > Unbind Event from On Virtual Keyboard Shown(self:Fort Platform Event VM Object Reference=@n1, Delegate=Create Event()) {is=title K2Node_RemoveDelegate}
   > Unbind Event from On Virtual Keyboard Hidden(self:Fort Platform Event VM Object Reference=@n1, Delegate=Create Event()) {is=title K2Node_RemoveDelegate}
  ]}
 EventGraph {nodes [
   "On Back Button Clicked (PanelLayout)"() {is=title K2Node_ComponentBoundEvent}
   > PopToNavigationTag!(
    execute
    self:Fort Sidebar Navigation VM Object Reference=Get FortSidebarNavigationVM() -> (FortSidebarNavigationVM:Fort Sidebar Navigation VM Object Reference)
    NavigationTag:Gameplay Tag Structure=(TagName="UI.StateTree.TransitionTag.Sidebar.Chat")
   ) -> (then, ReturnValue:bool)
   n1 = "On List View Initialized (MessageList)"!() -> (OutputDelegate, then, Item:Object Reference, Widget:User Widget Object Reference) {is=title K2Node_ComponentBoundEvent}
   > n2 = Cast To WBP_TextChat_Entry_UserMessage!(execute, Object=@n1.Widget) -> (then, CastFailed, AsWBP Text Chat Entry User Message:WBP Text Chat Entry User Message Object Reference) {is=title K2Node_DynamicCast}
   > Bind Event to On Message Clicked(self:WBP Text Chat Entry User Message Object Reference=@n2, Delegate=Create Event()) {is=title K2Node_AddDelegate}
   n3 = Cast To WBP_TextChat_MessageSystem!(execute=@n2.CastFailed, Object=@n1.Widget) -> (then, CastFailed, AsWBP Text Chat Message System:WBP Text Chat Message System Object Reference) {is=title K2Node_DynamicCast}
   > Bind Event to On Message Clicked(self:WBP Text Chat Message System Object Reference=@n3, Delegate=Create Event())
   n4 = "On Loading Switcher Active Widget Changed (PanelLayout)"!() -> (OutputDelegate, then, IsAsyncActionInProgress:bool) {is=title K2Node_ComponentBoundEvent}
   > n5 = Branch(Condition=@n4.IsAsyncActionInProgress)
   RequestRefreshFocus(execute=@n5.else, self:Common Activatable Widget Object Reference)
   "On Text Changed (WBP_Sidebar_InputField)"!() -> (OutputDelegate, then, Text:Text) {is=title K2Node_ComponentBoundEvent}
   > SetSendButtonEnabled()
   > PlayTextChatSound!(execute, self, Sound:Sound Base Object Reference=Get TextInputSound() -> (TextInputSound:Sound Base Object Reference)) -> (then)
   n6 = "On Text Commited (WBP_Sidebar_InputField)"!() -> (OutputDelegate, then, Text:Text, Commit method:ETextCommit Enum=Default) {is=title K2Node_ComponentBoundEvent}
   > n7 = Switch on ETextCommit!(execute, Selection:ETextCommit Enum=@n6.Commit method) -> (Default:Exec, OnEnter:Exec, OnUserMovedFocus:Exec, OnCleared:Exec) {is=title K2Node_SwitchEnum}
   SendTextMessage!(execute=@n7.OnEnter, self, Message:Text=@n6.Text) -> (then)
   > RequestRefreshFocus()
   n8 = "On Clicked (Button_StopTTS_Decoy)"!() -> (OutputDelegate, then, Button:Common Button Base Object Reference) {is=title K2Node_ComponentBoundEvent}
   n9 = Branch(execute=[@n8.then, @n20.then], Condition=GetIsSpeaking(self=Get FortTextChatGroupConversationVM()))
   n10 = Get FortTextChatGroupConversationVM()
   > StopSpeakingMessages(self:Fort Text Chat Conversation VM Object Reference=@n10)
   > RequestRefreshFocus()
   SpeakAllNewUnspokenMessages(execute=@n9.else, self:Fort Text Chat Conversation VM Object Reference=@n10)
   "On Clicked (Button_Dictate)"!() -> (OutputDelegate, then, Button:Common Button Base Object Reference) {is=title K2Node_ComponentBoundEvent}
   > n11 = Switch on EFortSpeechToTextState!(
    execute
    Selection:EFortSpeechToTextState Enum=Get CurrentSpeechToTextState!(self:Fort Text Chat Conversation VM Object Reference=Get FortTextChatGroupConversationVM()) -> (CurrentSpeechToTextState:EFortSpeechToTextState Enum=Idle)
   ) -> (Idle:Exec, Initializing:Exec, Recording:Exec, Transcribing:Exec, Deinitializing:Exec, Unavailable:Exec) {is=title K2Node_SwitchEnum}
   n12 = Get FortTextChatGroupConversationVM()
   > StopSpeakingMessages(self=@n12)
   > StartOrStopSpeechToTextRecording!(execute, self:Fort Text Chat Conversation VM Object Reference=@n12, bWantsToStart:bool=true) -> (then)
   SetText!(
    execute=[@n36.then, @n32.else, @n11.Deinitializing, @n11.Initializing, @n11.Recording, @n11.Transcribing]
    self:WBP UIKit Button Generic Base Object Reference=Get Button_Dictate()
    Text:Text
   ) -> (then)
   > StartOrStopSpeechToTextRecording(self=Get FortTextChatGroupConversationVM(), bWantsToStart=false)
   > Clear and Invalidate Timer by Handle(Handle=Get Timer Handle() -> (Timer Handle:Timer Handle Structure))
   > Set Record Time(Record Time:double=Get SpeechToTextRecordTime!(self:Fort Text Chat Conversation VM Object Reference=Get FortTextChatGroupConversationVM()) -> (SpeechToTextRecordTime:float)) -> (Output_Get:double)
   "On Text Changed (WBP_Sidebar_InputField_Multiline)"!() -> (OutputDelegate, then, Text:Text) {is=title K2Node_ComponentBoundEvent}
   > SetSendButtonEnabled()
   > PlayTextChatSound(Sound=Get TextInputSound())
   n13 = "On Text Commited (WBP_Sidebar_InputField_Multiline)"!() -> (OutputDelegate, then, Text:Text, Commit method:ETextCommit Enum=Default) {is=title K2Node_ComponentBoundEvent}
   > n14 = Switch on ETextCommit(Selection=@n13.Commit method)
   SendTextMessage(execute=@n14.OnEnter, Message=@n13.Text)
   > RequestRefreshFocus()
   "On Clicked (Button_StartParty)"!() -> (OutputDelegate, then, Button:Common Button Base Object Reference) {is=title K2Node_ComponentBoundEvent}
   > n15 = Switch on EFortConversationPartyState!(
    execute
    Selection:EFortConversationPartyState Enum=Get ConversationPartyState!(self:Fort Group Conversation Viewmodel Object Reference=Get FortTextChatGroupConversationVM()) -> (ConversationPartyState:EFortConversationPartyState Enum=None)
   ) -> (None:Exec, Creating:Exec, Active:Exec, Joining:Exec, Joined:Exec) {is=title K2Node_SwitchEnum}
   n16 = Get FortTextChatGroupConversationVM()
   n17 = Branch(
    execute=@n15.Active
    Condition=Get bShouldConfirmPartyAction!(self:Fort Group Conversation Viewmodel Object Reference=@n16) -> (bShouldConfirmPartyAction:bool)
   )
   > NavigateToConfirmPanel!(
    execute
    self
    Class:Sidebar Confirmation Panel VM Class Reference=/Sidebar/Panels/Overlay/Data/ConfirmationVM_JoinParty.ConfirmationVM_JoinParty_C
   ) -> (then)
   RequestJoinParty(execute=@n17.else, self:Fort Group Conversation Viewmodel Object Reference=@n16)
   RequestDeepLinkGeneric!(
    execute=@n15.Joined
    self:Fort Sidebar VM Object Reference=Get FortSidebarVM() -> (FortSidebarVM:Fort Sidebar VM Object Reference)
    FeatureId:Name=Social
    NavigationTag:Gameplay Tag Structure
    ConversationId:String
    TargetSocialUser:Fort Social User Object Reference
    TargetWidget:Name=None
   ) -> (then)
   n18 = Get FortTextChatGroupConversationVM()
   n19 = Branch(execute=@n15.None, Condition=Get bShouldConfirmPartyAction(self=@n18))
   > NavigateToConfirmPanel(Class=/Sidebar/Panels/Overlay/Data/ConfirmationVM_StartNewParty.ConfirmationVM_StartNewParty_C)
   RequestStartParty(execute=@n19.else, self:Fort Group Conversation Viewmodel Object Reference=@n18)
   n20 = "On Clicked (Button_ReadAloud)"!() -> (OutputDelegate, then, Button:Common Button Base Object Reference) {is=title K2Node_ComponentBoundEvent}
   "On Clicked (Button_Options)"!() -> (OutputDelegate, then, Button:Common Button Base Object Reference) {is=title K2Node_ComponentBoundEvent}
   > Open!(
    execute
    self:Menu Anchor Object Reference=Get Anchor!(self:UIKit Context Menu Anchor Object Reference=Get Options_ContextMenu() -> (Options_ContextMenu:WBP UIKit Context Menu Anchor Button Object Reference)) -> (Anchor:Menu Anchor Object Reference)
    bFocusMenu:bool=true
   ) -> (then)
   "On Context Menu Closed (Message_ContextMenu)"() {is=title K2Node_ComponentBoundEvent}
   > n21 = Get!(execute:Exec) -> (then:Exec, else:Exec, WBP Text Chat Entry User Message:WBP Text Chat Entry User Message Object Reference)
   n22 = Get Message_ContextMenu() -> (Message_ContextMenu:WBP UIKit Context Menu Anchor Button Object Reference)
   > SetBackgroundVisibility!(
    execute
    self:WBP Text Chat Entry User Message Object Reference=@n21
    ShouldBeVisible:bool=IsOpen!(self:Menu Anchor Object Reference=Get Anchor(self=@n22)) -> (ReturnValue:bool)
   ) -> (then)
   > n23 = Set bIsContextIconVisible!(execute, bIsContextIconVisible:bool=false, self:WBP Text Chat Entry User Message Object Reference=@n21) -> (then, Output_Get:bool)
   > n24 = SetContextMenuIconVisibility!(execute, self:WBP Text Chat Entry User Message Object Reference=@n21, IsContextMenuVisible:bool=@n23) -> (then)
   n25 = Get!(execute=[@n24.then, @n21.else]) -> (then, else, WBP Text Chat Message System:WBP Text Chat Message System Object Reference)
   > Set bIsContextIconVisible(bIsContextIconVisible=false, self:WBP Text Chat Message System Object Reference=@n25)
   > SetContextMenuIconVisibility(self:WBP Text Chat Message System Object Reference=@n25, IsContextMenuVisible=false)
   n26 = "On Item Clicked (MessageList)"!() -> (OutputDelegate, then, Item:Object Reference) {is=title K2Node_ComponentBoundEvent}
   > Branch(Condition=Get is TextToSpeech Enabled())
   n27 = Get MessageList() -> (MessageList:WBP Sidebar Text Chat Message List Object Reference)
   > Set Context Menu Item Index(Context Menu Item Index:int=GetIndexForItem!(
     self:List View Object Reference=Get ListView!(self:Fort Polymorphic List View Object Reference=@n27) -> (ListView:Common List View Object Reference)
     Item:Object Reference=@n26.Item
    ) -> (ReturnValue:int=0)) -> (Output_Get:int=0)
   "On Clicked (Button_Send)"!() -> (OutputDelegate, then, Button:Common Button Base Object Reference) {is=title K2Node_ComponentBoundEvent}
   n28 = Get EditableText(self:WBP UIKit Block Editable Text Object Reference=Get Block_EditableText(self:WBP UIKit Input Field Base Object Reference=Get WBP_Sidebar_InputField()) -> (Block_EditableText:WBP UIKit Block Editable Text Object Reference)) -> (EditableText:UIKit Editable Text Object Reference)
   n29 = Get EditableText(self:WBP Text Chat Block Multiline Editable Text Object Reference=Get Block_EditableText(self:WBP Text Chat Multiline Input Field Object Reference=Get WBP_Sidebar_InputField_Multiline()) -> (Block_EditableText:WBP Text Chat Block Multiline Editable Text Object Reference)) -> (EditableText:"Text Box (Multi- Line) Object Reference")
   > SendTextMessage(Message=SelectText(
     A=Get Text(self:"Text Box (Multi- Line) Object Reference"=@n29)
     B="GetText (Editable Text)"!(self:Editable Text Object Reference=@n28) -> (ReturnValue:Text)
     bPickA=Get Multiline InputField Enabled()
    ))
   > RequestRefreshFocus()
   n30 = "On Custom List Button Clicked (MessageList)"!() -> (OutputDelegate, then, ButtonType:Name=None) {is=title K2Node_ComponentBoundEvent}
   > Branch(Condition="Equal (Name)"(A=@n30.ButtonType, B=Get LoadOlderMessagesButtonType() -> (LoadOlderMessagesButtonType:Name=None)))
   > LoadCachedMessages!(
    execute
    self:Fort Text Chat Conversation VM Object Reference=Get FortTextChatGroupConversationVM()
    PaginationOption:EPersistentTextChatPaginationOptions Enum=OlderThanCurrent
   ) -> (then)
   "On Clicked (Button_JumpToMostRecent)"!() -> (OutputDelegate, then, Button:Common Button Base Object Reference) {is=title K2Node_ComponentBoundEvent}
   > LoadCachedMessages(self=Get FortTextChatGroupConversationVM(), PaginationOption=MostRecent)
   OnViewDMProfile!() -> (OutputDelegate, then, InteractionNameID:Name) {is=title K2Node_CustomEvent}
   > OpenProfileForDMMember()
   Report!() -> (OutputDelegate, then, InteractionNameID:Name) {is=title K2Node_CustomEvent}
   > CreateReportModal!(execute, self, Is Reporting Message:bool=false) -> (then) {is=title K2Node_CallFunction}
   Dictation Timer() {is=title K2Node_CustomEvent}
   n31 = Get Record Time() -> (Record Time:double)
   > n32 = Branch(Condition="Not Equal (Float)"(A=@n31))
   > n33 = Set Record Time(Record Time=float - float(A=@n31, B=1))
   n34 = "float >= float"(A=@n33, B=10)
   > SetText(
    self=Get Button_Dictate()
    Text=Format Text!(Format="LOCTEXT(\"0:{zero}{n}\")", zero:Text=SelectText(B="LOCTEXT(\"0\")", bPickA=@n34), n:int=Truncate(A=@n33)) -> (Result)
   )
   Cancel Speech To Text Recording() {is=title K2Node_CustomEvent}
   n35 = Get CurrentSpeechToTextState(self=Get FortTextChatGroupConversationVM()) -> (CurrentSpeechToTextState=Idle)
   > n36 = Branch(Condition=AND Boolean!(A="Not Equal (Enum)"(A:EFortSpeechToTextState Enum=@n35, B:EFortSpeechToTextState Enum=Idle), B=Get is SpeechToText Enabled()) -> (ReturnValue))
   On STT Tooltip Open Settings() {is=title K2Node_CustomEvent}
   > RequestCloseSidebar(self:Fort Sidebar VM Object Reference=Get FortSidebarVM())
   > Delay(Duration=0.2)
   > RequestOpenSettings!(
    execute
    self:Fort Settings Context Object Reference=Get FortSettingsContext!() -> (ReturnValue:Fort Settings Context Object Reference) {is=title K2Node_GetBlueprintContextEx}
    ContextWidget:Common User Widget Object Reference=Self-Reference()
    SettingsTabId:Name=Select!(Option 0:Name=CommunicationCollection, Option 1:Name=Communication, Index:bool=IsSaveTheWorldUI!() -> (IsSaveTheWorld:bool)) -> (ReturnValue:Name)
    SettingToSelect:String=ShowTextToSpeech
   ) -> (then)
   OnMessageSent() {is=title K2Node_CustomEvent}
   > PlayTextChatSound(Sound=Get MessageSentSound() -> (MessageSentSound:Sound Base Object Reference))
   OnMessageReceived() {is=title K2Node_CustomEvent}
   > PlayTextChatSound(Sound=Get MessageReceivedSound() -> (MessageReceivedSound:Sound Base Object Reference))
   n37 = OnSystemMessageClicked!() -> (OutputDelegate, then, Message ID:Name, Button:Common Button Base Object Reference) {is=title K2Node_CustomEvent}
   > Set Message ID(Message ID:Name=@n37.Message ID) -> (Output_Get:Name=None)
   > n38 = SetMessageContextMenuActions!(execute, self, Local user message:bool=false) -> (then, Has Interactions:bool)
   > Branch(Condition=@n38)
   > n39 = Cast To WBP_TextChat_MessageSystem(Object=@n37.Button)
   > n40 = Set WBP Text Chat Message System(WBP Text Chat Message System:WBP Text Chat Message System Object Reference=@n39) -> (Output_Get:WBP Text Chat Message System Object Reference)
   n41 = Get Message_ContextMenu!(self:WBP Text Chat Message System Object Reference=@n40) -> (Message_ContextMenu)
   > SetContext!(
    execute
    self:WBP UIKit Context Menu Anchor Button Base Object Reference=@n41
    InteractionContext_Data:UIKit Context Menu Action Context Object Reference=Get Message_ContextMenuActionContext() -> (Message_ContextMenuActionContext:UIKit Context Menu Action Context Object Reference)
   ) -> (then)
   > Open(self=Get Anchor(self=@n41), bFocusMenu=true)
   Manage Group!() -> (OutputDelegate, then, InteractionNameID:Name) {is=title K2Node_CustomEvent}
   > n42 = Construct Object from Class!(
    execute
    Class=/Sidebar/Shared/Navigation/BP_NavigationPoint.BP_NavigationPoint_C
    self
    SocialUser:Fort Social User Object Reference
    ConversationId:Name=Get ConversationId!(self:Fort Text Chat Conversation VM Object Reference=Get FortTextChatGroupConversationVM()) -> (ConversationId:Name=None)
    ConfirmationVM:Sidebar Confirmation Panel VM Object Reference
    ReservedSpots:int=0
    SocialItemVM:Fort Social Item VM Object Reference
    NavigationTag:Gameplay Tag Structure=(TagName="UI.StateTree.TransitionTag.Sidebar.Chat.ManageGroup")
   ) -> (then, ReturnValue:BP Navigation Point Object Reference)
   > PushNavigation!(
    execute
    self:Fort Sidebar Navigation VM Object Reference=Get FortSidebarNavigationVM()
    NavigationPoint:Fort Sidebar Navigation Point Object Reference=@n42
   ) -> (then)
   Communicate Settings!() -> (OutputDelegate, then, InteractionNameID:Name) {is=title K2Node_CustomEvent}
   > RequestCloseSidebar(self=Get FortSidebarVM())
   > OpenSettingsToTextChat!(execute, OwningWidget:Common User Widget Object Reference=Self-Reference()) -> (then)
   n43 = OnMessageClicked!() -> (OutputDelegate, then, MessageID:Name, Username:String, Button:Common Button Base Object Reference) {is=title K2Node_CustomEvent}
   > Set Message ID(Message ID=@n43.MessageID) -> (Output_Get=None)
   > Set Username(Username:String=@n43.Username) -> (Output_Get:String)
   > n44 = Cast To WBP_TextChat_Entry_UserMessage(Object=@n43.Button)
   > n45 = Set WBP Text Chat Entry User Message(WBP Text Chat Entry User Message:WBP Text Chat Entry User Message Object Reference=@n44) -> (Output_Get:WBP Text Chat Entry User Message Object Reference)
   n46 = Get Message_ContextMenu!() -> (Message_ContextMenu)
   > SetBackgroundVisibility(self=@n45, ShouldBeVisible=IsOpen(self=Get Anchor(self=@n46)))
   n47 = Get FortTextChatMessageVM!(self:WBP Text Chat Entry User Message Object Reference=@n44) -> (FortTextChatMessageVM:Text Chat Message VM Object Reference)
   > Set Message Social User(Message Social User:Fort Social User Object Reference=GetSocialUser!(self:Fort Social User Viewmodel Object Reference=Get PlayerAvatar!(self:Text Chat Message VM Object Reference=@n47) -> (PlayerAvatar:Fort Player Avatar Viewmodel Object Reference)) -> (ReturnValue:Fort Social User Object Reference)) -> (Output_Get:Fort Social User Object Reference)
   > n48 = SetMessageContextMenuActions(Local user message="Equal (Enum)"(
     A:EFortTextChatMessageType Enum=Get MessageType!(self:Text Chat Message VM Object Reference=@n47) -> (MessageType:EFortTextChatMessageType Enum=Sent)
     B:EFortTextChatMessageType Enum=Sent
    ))
   > Branch(Condition=@n48)
   n49 = Get Message_ContextMenu!(self:WBP Text Chat Entry User Message Object Reference=Get WBP Text Chat Entry User Message() -> (WBP Text Chat Entry User Message:WBP Text Chat Entry User Message Object Reference)) -> (Message_ContextMenu)
   > SetContext(self=@n49, InteractionContext_Data=Get Message_ContextMenuActionContext())
   > Open(self=Get Anchor(self=@n49), bFocusMenu=true)
   n50 = ContextMenuAction!() -> (OutputDelegate, then, InteractionNameID:Name) {is=title K2Node_CustomEvent}
   > n51 = "For Each Loop (Map)"!(
    execute:Exec
    MapPin:Map of EFortTextChatMessageContextMenuAction Enums to Names=Get Chat Menu Actions() -> (Chat Menu Actions:Map of EFortTextChatMessageContextMenuAction Enums to Names=Mention)
    BreakPin:Exec
   ) -> (then:Exec, KeyPin:EFortTextChatMessageContextMenuAction Enum, ValuePin:Name, CompletedPin:Exec)
   > Branch(Condition="Equal (Name)"(A=@n50.InteractionNameID, B=@n51.ValuePin))
   > n52 = Switch on EFortTextChatMessageContextMenuAction!(execute, Selection:EFortTextChatMessageContextMenuAction Enum=@n51) -> (
    Mention:Exec
    ViewProfile:Exec
    Report:Exec
    Block:Exec
    Unblock:Exec
    Mute:Exec
    Unmute:Exec
    SpeakMessage:Exec
    SpeakMessagesFromHere:Exec
    MAX_NUM:Exec
   ) {is=title K2Node_SwitchEnum}
   n53 = Create Widget!(
    execute=@n52.Block
    Class=/FortniteCoreUI/UI/Social/Chat/TextChat/Dialogs/WBP_DialogVM_TextChat_Block.WBP_DialogVM_TextChat_Block_C
    OwningPlayer=GetOwningPlayer()
    UIKitDialogViewModel:UI Kit Dialog Viewmodel Object Reference
   ) -> (then, ReturnValue:WBP Dialog VM Text Chat Block Object Reference)
   > n54 = SetDescriptionText!(
    execute
    self:WBP Dialog VM Text Chat Block Object Reference=@n53
    Username:Text="To Text (String)"(InString=Get Username() -> (Username:String))
   ) -> (then)
   n55 = PerformBlockUser!() -> (OutputDelegate, then, Button:Common Button Base Object Reference, DialogRef:Object Reference) {is=title K2Node_CustomEvent}
   Bind Event to On Clicked Button Confirm(execute=@n54.then, self:WBP Dialog VM Text Chat Block Object Reference=@n53, Delegate=@n55) {is=title K2Node_AddDelegate}
   > TryPushWidgetAsModalPanel(PushWidget=@n53)
   PerformMessageContextMenuAction!(
    execute=[@n56.then, @n52.Mute, @n52.Unblock, @n52.Unmute]
    self:Fort Text Chat Conversation VM Object Reference=Get FortTextChatGroupConversationVM()
    InMessageId:"Name (by ref)"=Get Message ID() -> (Message ID=None)
    MenuAction:EFortTextChatMessageContextMenuAction Enum=@n51
   ) -> (then)
   > NavigateToIndex!(
    execute
    self:List View Object Reference=Get ListView(self=Get MessageList())
    Index:int=Get Context Menu Item Index() -> (Context Menu Item Index:int=0)
   ) -> (then)
   CreateReportModal(execute=@n52.Report, Is Reporting Message=true)
   n56 = Cancel Speech To Text Recording(execute=[@n52.SpeakMessage, @n52.SpeakMessagesFromHere]) {is=title K2Node_CallFunction}
   n57 = Construct Object from Class(
    execute=@n52.ViewProfile
    Class=&BP_NavigationPoint_C
    SocialUser=Get Message Social User() -> (Message Social User:Fort Social User Object Reference)
    ConversationId=None
    ReservedSpots=0
    NavigationTag=(TagName="UI.StateTree.TransitionTag.Sidebar.Social.OtherPlayer")
   )
   > PushNavigation(self=Get FortSidebarNavigationVM(), NavigationPoint=@n57)
   n58 = OnConversationHasBeenRemovedChanged!() -> (OutputDelegate, then, bHasBeenRemoved:bool=~) {is=title K2Node_CustomEvent}
   > Branch(Condition=@n58.bHasBeenRemoved)
   > PopToNavigationTag(self=Get FortSidebarNavigationVM(), NavigationTag=(TagName="UI.StateTree.TransitionTag.Sidebar.Chat"))
   n59 = OnInputMethodChanged!() -> (OutputDelegate, then, InputType:ECommonInputType Enum) {is=title K2Node_CustomEvent}
   > n60 = Set Current Input Type(Current Input Type:ECommonInputType Enum=@n59.InputType) -> (Output_Get:ECommonInputType Enum)
   > InputTypeReadAloudVisuals()
   > n61 = Branch(Condition=Get is SpeechToText Enabled())
   > SetSendButtonEnabled()
   SetVisibility(
    execute=@n61.else
    self=Get Button_Send()
    InVisibility=Select!(
     MouseAndKeyboard:ESlateVisibility Enum=SelfHitTestInvisible
     Gamepad:ESlateVisibility Enum=Collapsed
     Touch:ESlateVisibility Enum=SelfHitTestInvisible
     Count:ESlateVisibility Enum=Collapsed
     Index:ECommonInputType Enum=@n60
    ) -> (ReturnValue:ESlateVisibility Enum)
   )
   UpdateLoadingState!() -> (OutputDelegate, then, BoolParam:bool=~) {is=title K2Node_CustomEvent}
   n62 = Get FortTextChatGroupConversationVM()
   n63 = Get bIsInitialized!(self:Fort Text Chat Conversation VM Object Reference=@n62) -> (bIsInitialized:bool)
   > Set with Broadcast ShowAsyncActionSpinner!(
    execute
    ShowAsyncActionSpinner:bool=OR Boolean(
     A=NOT Boolean(A=@n63)
     B=Get!(self:Fort Text Chat Conversation VM Object Reference=@n62) -> (
      LoadCachedMessagesRequest_bIsInProgress:bool=~
      LoadCachedMessagesRequest_LastResult:ELastAsyncActionResult Enum
      LoadCachedMessagesRequest_LastResultMessage:Text
     )
    )
    self:WBP Sidebar Panel Layout Object Reference=Get PanelLayout() -> (PanelLayout:WBP Sidebar Panel Layout Object Reference)
   ) -> (then, Output_Get:bool)
   Event On Activated()
   > Get TTS and STT Settings() {is=title K2Node_CallFunction}
   > Set with Broadcast ShowAsyncActionSpinner(ShowAsyncActionSpinner=true, self=Get PanelLayout())
   > n64 = GetCurrentNavigationPoint!(execute, self:Fort Sidebar Navigation VM Object Reference=Get FortSidebarNavigationVM()) -> (then, ReturnValue:Fort Sidebar Navigation Point Object Reference)
   > n65 = Cast To BP_NavigationPoint!(execute, Object=@n64) -> (then, CastFailed, AsBP Navigation Point:BP Navigation Point Object Reference) {is=title K2Node_DynamicCast}
   n66 = Get SocialUser!(self:BP Navigation Point Object Reference=@n65) -> (SocialUser:Fort Social User Object Reference)
   > n67 = Is Valid(InputObject=@n66)
   n68 = Get FortTextChatGroupConversationVM()
   InitializeTextChatConversationVM!(
    execute=@n67.Is Not Valid
    self:Fort Text Chat Conversation VM Object Reference=@n68
    Player:Local Player Object Reference=GetOwningLocalPlayer!(self:Widget Object Reference) -> (ReturnValue:Local Player Object Reference)
    InConversationId:"Name (by ref)"=Get ConversationId(self:BP Navigation Point Object Reference=@n65) -> (ConversationId=None)
   ) -> (then)
   > LoadCachedMessages(self=@n68, PaginationOption=MostRecent)
   > n69 = RequestUpdatedEpicPartyForConversation(self:Fort Group Conversation Viewmodel Object Reference=@n68)
   Set Conversation Header Font Size!(execute:Exec=[@n75.then, @n69.then]) -> (then:Exec) {is=title K2Node_MacroInstance}
   > UpdateStartPartyButtonVisibility!(execute, self, IsEpicParties:bool=Get IsUsingEpicParties() -> (IsUsingEpicParties:bool)) -> (then)
   n70 = Create Event()
   n71 = Get FortTextChatGroupConversationVM()
   > Bind Event to On Message Sent(self:Fort Text Chat Conversation VM Object Reference=@n71, Delegate=@n70) {is=title K2Node_AddDelegate}
   n72 = Create Event()
   > Bind Event to On Message Received(self:Fort Text Chat Conversation VM Object Reference=@n71, Delegate=@n72) {is=title K2Node_AddDelegate}
   n73 = Create Event()
   n74 = Get WBP_TextChat_STT_CoachMark() -> (WBP_TextChat_STT_CoachMark:WBP Text Chat STT Coach Mark Object Reference)
   > Bind Event to On Other Action Button Clicked(self:WBP Text Chat STT Coach Mark Object Reference=@n74, Delegate=@n73) {is=title K2Node_AddDelegate}
   > SetVisibility(
    self=Get WBP_Sidebar_InputField_Multiline()
    InVisibility=Select!(Option 0:ESlateVisibility Enum=Collapsed, Option 1:ESlateVisibility Enum=SelfHitTestInvisible, Index:bool=Get Multiline InputField Enabled()) -> (ReturnValue)
   )
   > SetVisibility(
    self=Get WBP_Sidebar_InputField()
    InVisibility=Select(Option 0=SelfHitTestInvisible, Option 1=Collapsed, Index=Get Multiline InputField Enabled())
   )
   > InputTypeReadAloudVisuals()
   InitializeTextChatConversationVMWithDM!(
    execute=@n67.Is Valid
    self:Fort Text Chat Conversation VM Object Reference=@n71
    Player:Local Player Object Reference=GetOwningLocalPlayer()
    TargetDMSocialUser:Fort Social User Object Reference=@n66
   ) -> (then)
   > n75 = LoadCachedMessages(self=@n71, PaginationOption=MostRecent)
   Event On Deactivated()
   n76 = Get FortTextChatGroupConversationVM()
   > UnloadMessages(self:Fort Text Chat Conversation VM Object Reference=@n76)
   > Unbind Event from On Message Sent(self:Fort Text Chat Conversation VM Object Reference=@n76, Delegate=@n70) {is=title K2Node_RemoveDelegate}
   > Unbind Event from On Message Received(self:Fort Text Chat Conversation VM Object Reference=@n76, Delegate=@n72) {is=title K2Node_RemoveDelegate}
   > Unbind Event from On Other Action Button Clicked(self:WBP Text Chat STT Coach Mark Object Reference=@n74, Delegate=@n73) {is=title K2Node_RemoveDelegate}
   > Cancel Speech To Text Recording()
   Event OnInitialized()
   > InitializeContextMenus()
   > Bind Event to On Entry Generated(self:List View Base Object Reference=Get ListView(self=Get MessageList()), Delegate=Create Event()) {is=title K2Node_AddDelegate}
   > n77 = Construct Object from Class!(
    execute
    Class=/Sidebar/Panels/Social/List/Data/BP_Sidebar_CustomButtonEntryVM.BP_Sidebar_CustomButtonEntryVM_C
    self
    DisplayName:Text="LOCTEXT(\"Load older messages\")"
    ButtonType:Name=Get LoadOlderMessagesButtonType() -> (LoadOlderMessagesButtonType=None)
    ButtonIcon:Texture 2D Object Reference
    UseIcon:bool=False
   ) -> (then, ReturnValue:BP Sidebar Custom Button Entry VM Object Reference)
   > Set LoadOlderMessagesButtonEntry(LoadOlderMessagesButtonEntry:BP Sidebar Custom Button Entry VM Object Reference=@n77) -> (Output_Get:BP Sidebar Custom Button Entry VM Object Reference)
   > n78 = Construct Object from Class!(
    execute
    Class=/FortniteCoreUI/UI/Social/Chat/TextChat/Data/BP_VM_TextChat_Welcome_Entry.BP_VM_TextChat_Welcome_Entry_C
    self
    IsCompactChat:bool=False
   ) -> (then, ReturnValue:BP VM Text Chat Welcome Entry Object Reference)
   > Set WelcomeMessageEntry(WelcomeMessageEntry:BP VM Text Chat Welcome Entry Object Reference=@n78) -> (Output_Get:BP VM Text Chat Welcome Entry Object Reference)
   > SetCompactUIStyle() {is=title K2Node_CallFunction}
   > Initialize Visuals!(execute:Exec) -> (then:Exec) {is=title K2Node_MacroInstance}
   Event Construct()
   n79 = Create Event()
   n80 = Get CommonInputSubsystem()
   > Bind Event to On Input Method Changed(self=@n80, Delegate=@n79)
   > OnInputMethodChanged!(execute, self, InputType:ECommonInputType Enum=GetCurrentInputType(self=@n80)) -> (then) {is=title K2Node_CallFunction}
   > n81 = GetConsoleVariableBoolValue!(execute, VariableName:String=fort.UseEpicParties) -> (then, ReturnValue:bool)
   > n82 = Set IsUsingEpicParties(IsUsingEpicParties:bool=@n81) -> (Output_Get:bool)
   > UpdateStartPartyButtonVisibility(IsEpicParties=@n82)
   > InitVirtualKeyboard()
   > SetOptionsContextMenuStaticActions()
   n83 = Get WBP_Sidebar_InputField_Multiline()
   > Set WrapTextAt!(execute, WrapTextAt:float=390, self:Text Layout Widget Object Reference=Get EditableText(self=Get Block_EditableText(self=@n83))) -> (then, Output_Get:float)
   Event Destruct()
   > Unbind Event from On Input Method Changed(self=@n80, Delegate=@n79)
   > DeInitVirtualKeyboard() {is=title K2Node_CallFunction}
   n84 = Cast To WBP_DialogVM_TextChat_Block!(execute=@n85.then, Object=@n55.DialogRef) -> (then, CastFailed, AsWBP Dialog VM Text Chat Block:WBP Dialog VM Text Chat Block Object Reference) {is=title K2Node_DynamicCast}
   > DeactivateWidget(self=@n84)
   n85 = PerformMessageContextMenuAction(execute=@n55.then, self=Get FortTextChatGroupConversationVM(), InMessageId=Get Message ID() -> (Message ID=None), MenuAction=Block)
  ]}
 ExecuteConfirmationAction {nodes [
   n1 = ExecuteConfirmationAction!() -> (then, Action Name:Name) {is=title K2Node_FunctionEntry}
   > n2 = Branch(Condition="Equal (Name)"(A=@n1, B=StartParty))
   n3 = Get FortTextChatGroupConversationVM()
   > RequestStartParty(self=@n3)
   Branch(execute=@n2.else, Condition="Equal (Name)"(A=@n1, B=JoinParty))
   > RequestJoinParty(self=@n3)
  ]}
 Get Multiline InputField Enabled {nodes [
   Get Multiline InputField Enabled() {is=title K2Node_FunctionEntry}
   > n1 = GetConsoleVariableBoolValue(VariableName=TextChat.EnableMultilineInputField)
   > Return Node!(execute, isEnabled:bool=@n1)
  ]}
 Get TTS and STT Settings {nodes [
   Get TTS and STT Settings() {is=title K2Node_FunctionEntry}
   > SetSettingId!(
    execute
    self:Fort Setting Value VM Object Reference=Get FortSocialSettingDiscreteVM() -> (FortSocialSettingDiscreteVM:Fort Social Setting Discrete VM Object Reference)
    InSettingId:String=ShowTextToSpeech
   ) -> (then)
   n1 = Get FortSocialSettingDiscreteVM()
   > Set is TextToSpeech Enabled(is TextToSpeech Enabled:bool=Select(
     Option 0:bool=false
     Option 1:bool=true
     Index:int=GetSettingValueIndex!(self:Fort Social Setting Discrete VM Object Reference=@n1) -> (ReturnValue:int=0)
    ) -> (ReturnValue:bool)) -> (Output_Get:bool)
   > SetSettingId(self=Get FortSocialSettingDiscreteVM(), InSettingId=ShowSpeechToText)
   n2 = Get FortSocialSettingDiscreteVM()
   > Set is SpeechToText Enabled(is SpeechToText Enabled:bool=Select(Option 0=false, Option 1=true, Index=GetSettingValueIndex(self=@n2) -> (ReturnValue=0))) -> (Output_Get:bool)
   n3 = "Equal (Enum)"(A:ECommonInputType Enum=Get Current Input Type() -> (Current Input Type:ECommonInputType Enum), B:ECommonInputType Enum=Gamepad)
   n4 = Get is SpeechToText Enabled()
   > SetVisibility(
    self=Get Button_Send()
    InVisibility=Select(Option 0:ESlateVisibility Enum=Visible, Option 1:ESlateVisibility Enum=Collapsed, Index:bool=OR Boolean(A=@n3, B=@n4)) -> (ReturnValue:ESlateVisibility Enum)
   )
   > SetVisibility(self=Get Button_Dictate(), InVisibility=Select(Option 0=Collapsed, Option 1=Visible, Index=@n4))
  ]}
 GetInputFieldEnabled {nodes [
   n1 = GetInputFieldEnabled!() -> (then, IsChannelReadOnly:bool=~, IsRestrictionActive:bool=~) {is=title K2Node_FunctionEntry}
   > Return Node!(execute, InputFieldEnabled:bool=NOT Boolean(A=OR Boolean(A=@n1, B=@n1.IsRestrictionActive)))
  ]}
 GetInputFieldTextHint {nodes [
   n1 = GetInputFieldTextHint!() -> (
    then
    IsRecordingCurrentChannel:bool=~
    IsCurrentChannelReadOnly:bool=~
    CurrentChannelName:String
    CurrentChannelType:EFortTextChatChannelType Enum=All
   ) {is=title K2Node_FunctionEntry}
   > n2 = GetInputFieldTextHint_Base!(
    execute
    IsRecordingCurrentChannel:bool=@n1
    IsCurrentChannelReadOnly:bool=@n1.IsCurrentChannelReadOnly
    CurrentChannelName:String=@n1.CurrentChannelName
    CurrentChannelType:EFortTextChatChannelType Enum=@n1.CurrentChannelType
   ) -> (then, HintText:Text)
   > n3 = Get!(execute:Exec) -> (then:Exec, else:Exec, FortTextChatGroupConversationVM:Fort Group Conversation Viewmodel Object Reference)
   n4 = Get CurrentSpeechToTextState(self=@n3) -> (CurrentSpeechToTextState=Idle)
   > n5 = Branch(Condition=AND Boolean(A="Not Equal (Enum)"(A=@n4, B=Idle), B="Not Equal (Enum)"(A=@n4, B=Unavailable)))
   > Return Node!(execute, HintText:Text=Get Hint Text!(self:WBP UIKit Input Field Base Object Reference=Get WBP_Sidebar_InputField()) -> (Hint Text:Text))
   Return Node(execute=[@n5.else, @n3.else], HintText=@n2)
  ]}
 GetJumpToRecentButtonVisibility {nodes [
   n1 = GetJumpToRecentButtonVisibility!() -> (then, MessagesVM:Array of Text Chat Message VM Object References, IsInMostRecentPage:bool=~) {is=title K2Node_FunctionEntry}
   n2 = Is Empty!(TargetArray:Array of Text Chat Message VM Object References=@n1) -> (ReturnValue:bool)
   > Return Node!(
    execute
    JumpButtonVisibility:ESlateVisibility Enum=Select(Option 0=SelfHitTestInvisible, Option 1=Collapsed, Index=OR Boolean(A=@n2, B=@n1.IsInMostRecentPage))
   )
  ]}
 HandleOnVirtualKeyboardHidden {nodes [
   HandleOnVirtualKeyboardHidden() {is=title K2Node_FunctionEntry}
   > Branch(Condition=GetConfigDataBoolWithDefault!(EntryName:String=EnableSidebarChatResizeAboveKeyboard, bDefaultValue:bool=false) -> (ReturnValue:bool))
   > SetVisibility(self=Get Grid_Controls() -> (Grid_Controls:Grid Panel Object Reference), InVisibility=SelfHitTestInvisible)
   > ResetPanelLayoutHeader()
   > ScrollToBottom(self:List View Base Object Reference=Get ListView(self=Get MessageList()))
  ]}
 HandleOnVirtualKeyboardShown {nodes [
   HandleOnVirtualKeyboardShown!() -> (then, RectTop:int, RectLeft:int, RectBottom:int, RectRight:int, VirtualKeyboardInputHeight:int) {is=title K2Node_FunctionEntry}
   > Branch(Condition=GetConfigDataBoolWithDefault(EntryName=EnableSidebarChatResizeAboveKeyboard, bDefaultValue=false))
   n1 = Get FortPlatformEventVM()
   n2 = Get WBP_Sidebar_InputField()
   > n3 = Branch(Condition=AND Boolean(
     A=IsWidgetAllowedToEnableIntegratedKeyboard!(self:Fort Platform Event VM Object Reference=@n1) -> (ReturnValue:bool)
     B=Get MobileEnableIntegratedKeyboard!(self:WBP UIKit Input Field Base Object Reference=@n2) -> (MobileEnableIntegratedKeyboard:bool)
    ))
   HidePanelLayoutHeader(execute=[@n4.then, @n3.then])
   > ScrollToBottom(self=Get ListView(self=Get MessageList()))
   n4 = SetVisibility(execute=@n3.else, self=Get Grid_Controls(), InVisibility=Collapsed)
  ]}
 HandleTooltipOrientation {nodes [
   HandleTooltipOrientation() {is=title K2Node_FunctionEntry}
   n1 = IsUsingCompactUI()
   > SetCoachmarkArrowPosition!(
    execute
    self:WBP Text Chat STT Coach Mark Object Reference=Get WBP_TextChat_STT_CoachMark()
    ArrowPosition:E_UI_OutBoardPosition Enum=Select(Option 0:E_UI_OutBoardPosition Enum=NewEnumerator12, Option 1:E_UI_OutBoardPosition Enum=NewEnumerator4, Index=@n1) -> (ReturnValue:E_UI_OutBoardPosition Enum)
   ) -> (then)
   n2 = Get VisAttach_CoachmarkSTT() -> (VisAttach_CoachmarkSTT:Common Visual Attachment Object Reference)
   > Set ContentAnchor!(
    execute
    ContentAnchor:Vector 2D Structure=MakeVector2D(X=SelectFloat(A=-0.03, B=1.025, bPickA=@n1), Y=1)
    self:Common Visual Attachment Object Reference=@n2
   ) -> (then, Output_Get:Vector 2D Structure)
   > SetPadding!(
    execute
    self:Overlay Slot Object Reference=SlotAsOverlaySlot(Widget=@n2)
    InPadding_Left:float=0
    InPadding_Top:float=0
    InPadding_Right:float=0
    InPadding_Bottom:float=SelectFloat(A=112, B=0, bPickA=@n1)
   ) -> (then)
  ]}
 HidePanelLayoutHeader {nodes [
   HidePanelLayoutHeader() {is=title K2Node_FunctionEntry}
   > Branch(Condition=GetConfigDataBoolWithDefault(EntryName=EnableHidePanelHeaderWhenVirtualKeyboardOpen, bDefaultValue=false))
   n1 = Get PanelLayout()
   > SetVisibility(self=Get GP_Header!(self:WBP Sidebar Panel Layout Object Reference=@n1) -> (GP_Header:Grid Panel Object Reference), InVisibility=Collapsed)
   > SetVisibility(self=Get Divider!(self:WBP Sidebar Panel Layout Object Reference=@n1) -> (Divider:Image Object Reference), InVisibility=Collapsed)
  ]}
 InitVirtualKeyboard {nodes [
   InitVirtualKeyboard() {is=title K2Node_FunctionEntry}
   n1 = Get FortPlatformEventVM()
   > Bind Event to On Virtual Keyboard Shown(self:Fort Platform Event VM Object Reference=@n1, Delegate=Create Event()) {is=title K2Node_AddDelegate}
   > Bind Event to On Virtual Keyboard Hidden(self:Fort Platform Event VM Object Reference=@n1, Delegate=Create Event()) {is=title K2Node_AddDelegate}
  ]}
 InitializeContextMenus {nodes [
   InitializeContextMenus() {is=title K2Node_FunctionEntry}
   > n1 = Construct Object from Class!(execute, Class=/Script/FortniteUI.UIKitContextMenuActionContext, self) -> (then, ReturnValue:UIKit Context Menu Action Context Object Reference)
   > n2 = Set Message_ContextMenuActionContext(Message_ContextMenuActionContext:UIKit Context Menu Action Context Object Reference=@n1) -> (Output_Get:UIKit Context Menu Action Context Object Reference)
   > SetContext(self=Get Message_ContextMenu!() -> (Message_ContextMenu), InteractionContext_Data=@n2)
   > n3 = Construct Object from Class(Class=&UIKitContextMenuActionContext)
   > n4 = Set OptionsContextMenuActionContext(OptionsContextMenuActionContext:UIKit Context Menu Action Context Object Reference=@n3) -> (Output_Get:UIKit Context Menu Action Context Object Reference)
   > SetContext(self=Get Options_ContextMenu(), InteractionContext_Data=@n4)
  ]}
 InitializeVisuals {nodes [
   Inputs()
   n1 = Get Slot_Main!(self:WBP UIKit Button Generic Base Object Reference=Get Button_JumpToMostRecent() -> (Button_JumpToMostRecent:WBP Sidebar Button Generic Object Reference)) -> (Slot_Main:Named Slot Object Reference)
   > SetGridSlotAlignment!(
    execute
    Widget:Widget Object Reference=@n1
    Horizontal:EHorizontalAlignment Enum=HAlign_Left
    Vertical:EVerticalAlignment Enum=VAlign_Center
   ) -> (then)
   > SetGridSlotPadding!(execute, Widget:Widget Object Reference=@n1, Padding_Left:float=18, Padding_Top:float=0, Padding_Right:float=0, Padding_Bottom:float=0) -> (then)
   > Outputs!(then:Exec)
  ]}
 InputTypeReadAloudVisuals {nodes [
   InputTypeReadAloudVisuals() {is=title K2Node_FunctionEntry}
   > Branch(Condition=Get is TextToSpeech Enabled())
   n1 = IsUsingInputType!(InputType:ECommonInputType Enum=Touch, PlayerController:Player Controller Object Reference=GetOwningPlayer()) -> (ReturnValue:bool)
   > Set PlayingIcon(PlayingIcon:Object Reference=Select(Option 0:Object Reference, Option 1:Object Reference=/Sidebar/Materials/Icons/MI_Sidebar_Recording_Icon, Index=@n1) -> (ReturnValue:Object Reference)) -> (Output_Get:Object Reference)
   > Set!(
    execute
    PaddingWithButtonPrompt_Left:float=SelectFloat(A=12, B=36, bPickA=@n1)
    PaddingWithButtonPrompt_Top:float=0
    PaddingWithButtonPrompt_Right:float=0
    PaddingWithButtonPrompt_Bottom:float=0
   ) -> (then, Output_Get:Margin Structure)
   > Set ReadAloudHasPrompt(ReadAloudHasPrompt:bool=NOT Boolean(A=@n1)) -> (Output_Get:bool)
   > Set Read Button Visuals!(
    execute
    self
    Current Speaking State:EFortTextToSpeechState Enum=Get CurrentSpeakingState!(self:Fort Text Chat Conversation VM Object Reference=Get FortTextChatGroupConversationVM()) -> (CurrentSpeakingState:EFortTextToSpeechState Enum=None)
   ) -> (then) {is=title K2Node_CallFunction}
  ]}
 NavigateToConfirmPanel {nodes [
   n1 = NavigateToConfirmPanel!() -> (then, Class:Sidebar Confirmation Panel VM Class Reference) {is=title K2Node_FunctionEntry}
   > n2 = Construct Object from Class(Class=@n1) -> (ReturnValue:Sidebar Confirmation Panel VM Object Reference)
   > NavigateToConfirmPanel!(
    execute
    SelfReference:Object Reference=Self-Reference()
    SidebarConfirmationVM:Sidebar Confirmation Panel VM Object Reference=@n2
    SidebarNavigationVM:Fort Sidebar Navigation VM Object Reference=Get FortSidebarNavigationVM()
   ) -> (then) {is=title K2Node_CallFunction}
  ]}
 NavigateToLastMessage {nodes [
   NavigateToLastMessage!() -> (then, Navigation:EUINavigation Enum) {is=title K2Node_FunctionEntry}
   > n1 = Branch(Condition=GetIsSpeaking(self=Get FortTextChatGroupConversationVM()))
   > Return Node!(execute, ReturnValue:Widget Object Reference=Get Button_ReadAloud() -> (Button_ReadAloud:WBP Sidebar Button Generic Object Reference))
   n2 = Get ListView(self=Get MessageList())
   n3 = Get ListView(self=Get MessageList())
   Navigate To Item!(
    execute=@n1.else
    self:List View Object Reference=@n3
    Item:Object Reference=GetItemAt!(self:List View Object Reference=@n2, Index:int=GetNumItems!(self:List View Object Reference=@n2) -> (ReturnValue:int=0)) -> (ReturnValue:Object Reference)
   ) -> (then)
   > Return Node(ReturnValue=@n3)
  ]}
 NavigatetoStartParty {nodes [
   NavigatetoStartParty!() -> (then, Navigation:EUINavigation Enum) {is=title K2Node_FunctionEntry}
   > Branch(Condition=IsVisible(self=Get Button_StartParty() -> (Button_StartParty:WBP Sidebar Party Button Object Reference)))
   > Return Node(ReturnValue=Get Button_StartParty())
  ]}
 OnListViewEntryGenerated {nodes [
   n1 = OnListViewEntryGenerated!() -> (then, Widget:User Widget Object Reference) {is=title K2Node_FunctionEntry}
   > n2 = Cast To WBP_TextChat_Entry_MessageBase!(execute, Object=@n1) -> (then, CastFailed, AsWBP Text Chat Entry Message Base:WBP Text Chat Entry Message Base Object Reference) {is=title K2Node_DynamicCast}
   n3 = GetNumItems(self=Get ListView(self=Get MessageList())) -> (ReturnValue=0)
   n4 = Get IndexOfItem!(self:WBP Text Chat Entry Message Base Object Reference=@n2) -> (IndexOfItem:int=0)
   > Branch(Condition="Equal (Integer)"(A=@n4, B=int - int(A=@n3)))
   n5 = Get CommonInputSubsystem()
   > n6 = Branch(Condition="Not Equal (Enum)"(A:ECommonInputType Enum=GetCurrentInputType(self=@n5), B:ECommonInputType Enum=Gamepad))
   n7 = Get ListView(self=Get MessageList())
   ScrollToBottom(execute=[@n8.then, @n6.then], self=@n7)
   n8 = SetSelectedIndex!(execute=@n6.else, self:List View Object Reference=@n7, Index:int=@n4) -> (then)
  ]}
 OnSpeechToTextMicStatusChanged {nodes [
   n1 = OnSpeechToTextMicStatusChanged!() -> (then, Mic Status:EFortSpeechToTextMicrophoneStatus Enum=Available) {is=title K2Node_FunctionEntry}
   > SetIsLocked!(
    execute
    self:Common Button Base Object Reference=Get Button_Dictate()
    bInIsLocked:bool="Not Equal (Enum)"(A:EFortSpeechToTextMicrophoneStatus Enum=@n1, B:EFortSpeechToTextMicrophoneStatus Enum=Available)
   ) -> (then)
   > UpdateButtonType!(
    execute
    self:WBP Sidebar Button Generic Object Reference=Get Button_Dictate()
    ButtonType:E_UI_Sidebar_Button_Type Enum=Select!(
     Available:E_UI_Sidebar_Button_Type Enum=NewEnumerator0
     PermissionDenied:E_UI_Sidebar_Button_Type Enum=NewEnumerator3
     PlatformChat:E_UI_Sidebar_Button_Type Enum=NewEnumerator3
     Index:EFortSpeechToTextMicrophoneStatus Enum=@n1
    ) -> (ReturnValue:E_UI_Sidebar_Button_Type Enum)
   ) -> (then)
  ]}
 OpenProfileForDMMember {nodes [
   OpenProfileForDMMember() {is=title K2Node_FunctionEntry}
   > n1 = For Each Loop with Break!(
    Exec:Exec
    Array:Array of Fort Social User Viewmodel Object References=Get Members!(self:Fort Text Chat Conversation VM Object Reference=Get FortTextChatGroupConversationVM()) -> (Members:Array of Fort Social User Viewmodel Object References)
    Break:Exec=@n4.then
   ) -> (LoopBody:Exec, Array Element:Fort Social User Viewmodel Object Reference, Array Index:int, Completed:Exec)
   n2 = GetSocialUser(self=@n1)
   > Branch(Condition=NOT Boolean(A=IsLocalUserForBlueprint!(self:Fort Social User Object Reference=@n2) -> (ReturnValue:bool)))
   > n3 = Construct Object from Class!(
    execute
    Class=&BP_NavigationPoint_C
    self
    SocialUser:Fort Social User Object Reference=@n2
    ConversationId:Name=None
    ConfirmationVM:Sidebar Confirmation Panel VM Object Reference
    ReservedSpots:int=0
    SocialItemVM:Fort Social Item VM Object Reference
    NavigationTag:Gameplay Tag Structure=(TagName="UI.StateTree.TransitionTag.Sidebar.Social.OtherPlayer")
   ) -> (then, ReturnValue:BP Navigation Point Object Reference)
   > n4 = PushNavigation(self=Get FortSidebarNavigationVM(), NavigationPoint=@n3)
  ]}
 PlayDictationSound {nodes [
   n1 = PlayDictationSound!() -> (then, Curret Speech To Text State:EFortSpeechToTextState Enum=Idle) {is=title K2Node_FunctionEntry}
   > n2 = Switch on EFortSpeechToTextState(Selection=@n1)
   PlaySound2D(execute=@n2.Recording, Sound=/FortniteCoreUI/UI/Social/Chat/TextChat/Audio/UI_TTS_Activate_A)
   PlaySound2D(execute=@n2.Transcribing, Sound=/FortniteCoreUI/UI/Social/Chat/TextChat/Audio/UI_TTS_Deactivate_A)
  ]}
 PlayTextChatSound {nodes [
   n1 = PlayTextChatSound!() -> (then, Sound:Sound Base Object Reference) {is=title K2Node_FunctionEntry}
   > Branch(Condition=IsTextChatSoundEnabled!(self:Peristent Text Chat Base Viewmodel Object Reference=Get FortTextChatGroupConversationVM()) -> (ReturnValue:bool))
   > Is Valid(InputObject=@n1)
   > PlaySound2D(Sound=@n1)
  ]}
 PopulateMessages {nodes [
   n1 = PopulateMessages!() -> (then, MessageVMs:Array of MVVM Base Viewmodel Object References) {is=title K2Node_FunctionEntry}
   n2 = Get ListView(self=Get MessageList())
   > Set List Items!(execute, self:List View Object Reference=@n2, InListItems:Array of Object References=@n1) -> (then) {is=title K2Node_CallFunction}
   > ScrollToBottom(self=@n2)
   n3 = Get CommonInputSubsystem()
   > Branch(Condition="Not Equal (Enum)"(A:ECommonInputType Enum=GetCurrentInputType(self=@n3), B:ECommonInputType Enum=Touch))
   > SetSelectedIndex(self=@n2, Index=int - int(A=GetNumItems(self=@n2) -> (ReturnValue=0)))
  ]}
 ResetPanelLayoutHeader {nodes [
   ResetPanelLayoutHeader() {is=title K2Node_FunctionEntry}
   > Branch(Condition=GetConfigDataBoolWithDefault(EntryName=EnableHidePanelHeaderWhenVirtualKeyboardOpen, bDefaultValue=false))
   n1 = Get PanelLayout()
   > SetVisibility(self=Get GP_Header(self=@n1), InVisibility=SelfHitTestInvisible)
   > SetVisibility(self=Get Divider(self=@n1), InVisibility=SelfHitTestInvisible)
  ]}
 SendTextMessage {nodes [
   n1 = SendTextMessage!() -> (then, Message:Text) {is=title K2Node_FunctionEntry}
   > Branch(Condition=NOT Boolean(A=TextIsEmpty(InText=@n1)))
   > n2 = Get()
   > Set MessageToSend!(
    execute
    MessageToSend:Text=Get Message() -> (Message:Text)
    self:Fort Text Chat Sending Message VM Object Reference=Get MessageToSend!(self:Fort Text Chat Conversation VM Object Reference=@n2) -> (MessageToSend:Fort Text Chat Sending Message VM Object Reference)
   ) -> (then, Output_Get:Text)
   > Send(self:Fort Text Chat Conversation VM Object Reference=@n2)
   > SetText(self:WBP UIKit Input Field Base Object Reference=Get WBP_Sidebar_InputField())
   > SetText(self:WBP Text Chat Multiline Input Field Object Reference=Get WBP_Sidebar_InputField_Multiline())
   > SetSendButtonEnabled()
  ]}
 Set Dictation Visuals {nodes [
   Set Dictation Visuals!() -> (then, Current Speech to Text State:EFortSpeechToTextState Enum=Idle) {is=title K2Node_FunctionEntry}
   > n1 = Branch(Condition=Get Multiline InputField Enabled() {is=title K2Node_CallFunction})
   n2 = Select!(
    Idle:ESlateVisibility Enum=SelfHitTestInvisible
    Initializing:ESlateVisibility Enum=HitTestInvisible
    Recording:ESlateVisibility Enum=HitTestInvisible
    Transcribing:ESlateVisibility Enum=HitTestInvisible
    Deinitializing:ESlateVisibility Enum=HitTestInvisible
    Unavailable:ESlateVisibility Enum=SelfHitTestInvisible
    Index:EFortSpeechToTextState Enum=Get Current Speech to Text State() -> (Current Speech to Text State:EFortSpeechToTextState Enum=Idle)
   ) -> (ReturnValue:ESlateVisibility Enum)
   > n3 = SetVisibility(self=Get WBP_Sidebar_InputField_Multiline(), InVisibility=@n2)
   n4 = Switch on EFortSpeechToTextState(execute=[@n3.then, @n15.then], Selection=Get Current Speech to Text State() -> (Current Speech to Text State=Idle))
   n5 = Get Timer Handle()
   n6 = Branch(
    execute=[@n4.Deinitializing, @n4.Transcribing]
    Condition=Is Timer Active by Handle!(Handle:Timer Handle Structure=@n5) -> (ReturnValue:bool)
   )
   > Clear and Invalidate Timer by Handle(Handle=@n5)
   > n7 = Set Record Time(Record Time=Get SpeechToTextRecordTime(self=Get FortTextChatGroupConversationVM()))
   SetHintText!(
    execute=[@n7.then, @n6.else]
    self:WBP UIKit Input Field Base Object Reference=Get WBP_Sidebar_InputField()
    Hint Text:Text=Get Hint Text Please Wait() -> (Hint Text Please Wait:Text)
   ) -> (then)
   > SetHintText(self:WBP Text Chat Multiline Input Field Object Reference=Get WBP_Sidebar_InputField_Multiline(), Hint Text=Get Hint Text Please Wait())
   > Button Dictate Button Visuals!(
    execute
    self
    DF Texture:Texture 2D Object Reference=/Sidebar/Materials/Icons/Textures/T_UI_Icon_Sidebar_Mic_MSDF
    Material Instance:Material Instance Object Reference=/FortUILibrary/UIKit/Throbber/MI_UIKit_Monochrome_Throbber
   ) -> (then) {is=title K2Node_CallFunction}
   > SetText(self:WBP UIKit Button Generic Base Object Reference=Get Button_Dictate())
   n8 = Get WBP_Sidebar_InputField_Multiline()
   > Set WrapTextAt(WrapTextAt=390, self=Get EditableText(self=Get Block_EditableText(self=@n8)))
   RequestRefreshFocus(execute=@n4.Idle)
   > SetSendButtonEnabled()
   > n9 = Branch(Condition=TextIsEmpty(InText=Get Current Hint() -> (Current Hint:Text)))
   Button Dictate Button Visuals(
    execute=[@n10.then, @n9.then]
    DF Texture=&T_UI_Icon_Sidebar_Mic_MSDF
    Material Instance=/Sidebar/Materials/Icons/MI_Sidebar_Icon_Base
   )
   SetHintText(execute=@n9.else, self:WBP UIKit Input Field Base Object Reference=Get WBP_Sidebar_InputField(), Hint Text=Get Current Hint())
   > n10 = SetHintText(self:WBP Text Chat Multiline Input Field Object Reference=Get WBP_Sidebar_InputField_Multiline(), Hint Text=Get Current Hint())
   Set Current Hint(execute=@n4.Initializing, Current Hint:Text=Get Hint Text(self=Get WBP_Sidebar_InputField())) -> (Output_Get:Text)
   > SetHintText(self:WBP UIKit Input Field Base Object Reference=Get WBP_Sidebar_InputField(), Hint Text=Get Hint Text Please Wait())
   > SetHintText(self:WBP Text Chat Multiline Input Field Object Reference=Get WBP_Sidebar_InputField_Multiline(), Hint Text=Get Hint Text Please Wait())
   > Button Dictate Button Visuals(Material Instance=&MI_UIKit_Monochrome_Throbber)
   n11 = Set Timer by Event(execute=@n4.Recording, Delegate=Create Event(), Time=1, bLooping=true)
   > Set Timer Handle(Timer Handle:Timer Handle Structure=@n11) -> (Output_Get:Timer Handle Structure)
   n12 = Get WBP_Sidebar_InputField_Multiline()
   > Set WrapTextAt(WrapTextAt=370, self=Get EditableText(self=Get Block_EditableText(self=@n12)))
   > SetHintText(
    self:WBP UIKit Input Field Base Object Reference=Get WBP_Sidebar_InputField()
    Hint Text=Get Hint Text Listening() -> (Hint Text Listening:Text)
   )
   > SetHintText(self:WBP Text Chat Multiline Input Field Object Reference=Get WBP_Sidebar_InputField_Multiline(), Hint Text=Get Hint Text Listening())
   > Button Dictate Button Visuals(Material Instance=&MI_Sidebar_Recording_Icon)
   n13 = Get SpeechToTextRecordTime(self=Get FortTextChatGroupConversationVM())
   n14 = "float >= float"(A=@n13, B=10)
   > SetText(
    self=Get Button_Dictate()
    Text=Format Text(Format="LOCTEXT(\"0:{zero}{n}\")", zero=SelectText(B="LOCTEXT(\"0\")", bPickA=@n14), n=Truncate(A=@n13))
   )
   n15 = SetVisibility(execute=@n1.else, self=Get WBP_Sidebar_InputField(), InVisibility=@n2)
  ]}
 Set Read Button Visuals {nodes [
   Set Read Button Visuals!() -> (then, Current Speaking State:EFortTextToSpeechState Enum=None) {is=title K2Node_FunctionEntry}
   n1 = Get Button_ReadAloud()
   > SetIcon!(
    execute
    self:WBP UIKit Button Generic Base Object Reference=@n1
    Icon:Object Reference=Select!(
     None:Object Reference
     Preparing:Object Reference=&MI_UIKit_Monochrome_Throbber
     Playing:Object Reference=Get PlayingIcon() -> (PlayingIcon:Object Reference)
     Failed:Object Reference=&MI_UIKit_Monochrome_Throbber
     Index:EFortTextToSpeechState Enum=Get Current Speaking State() -> (Current Speaking State:EFortTextToSpeechState Enum=None)
    ) -> (ReturnValue:Object Reference)
   ) -> (then)
   n2 = Get ReadAloudPreparingIconSize() -> (ReadAloudPreparingIconSize:Vector 2D Structure)
   n3 = Get DefaultIconSize()
   > SetIconSize(
    self=@n1
    Icon=Select(
     None:Vector 2D Structure=@n3
     Preparing:Vector 2D Structure=@n2
     Playing:Vector 2D Structure=@n3
     Failed:Vector 2D Structure=@n2
     Index=Get Current Speaking State() -> (Current Speaking State=None)
    ) -> (ReturnValue:Vector 2D Structure)
   )
   > SetIconPadding!(
    execute
    self:WBP UIKit Button Generic Base Object Reference=@n1
    Padding:Margin Structure=Select!(
     None:Margin Structure
     Preparing_Left:float=0
     Preparing_Top:float=0
     Preparing_Right:float=8
     Preparing_Bottom:float=0
     Playing:Margin Structure
     Failed_Left:float=0
     Failed_Top:float=0
     Failed_Right:float=8
     Failed_Bottom:float=0
     Index=Get Current Speaking State() -> (Current Speaking State=None)
    ) -> (ReturnValue:Margin Structure)
   ) -> (then)
   n4 = Get PaddingWithButtonPrompt() -> (PaddingWithButtonPrompt:Margin Structure)
   n5 = Get PaddingWithoutButtonPrompt() -> (PaddingWithoutButtonPrompt:Margin Structure)
   n6 = Get ReadAloudHasPrompt() -> (ReadAloudHasPrompt:bool)
   > SetTextPadding!(
    execute
    self:WBP UIKit Button Generic Base Object Reference=@n1
    TextPadding:Margin Structure=Select!(
     None=Select!(Option 0:Margin Structure=@n5, Option 1:Margin Structure=@n4, Index:bool=@n6) -> (ReturnValue)
     Preparing:Margin Structure=@n5
     Playing=@n4
     Failed:Margin Structure=@n5
     Index=Get Current Speaking State() -> (Current Speaking State=None)
    ) -> (ReturnValue)
   ) -> (then)
   > SetHideInputAction!(
    execute
    self:Common Button Base Object Reference=@n1
    bInHideInputAction:bool=Select!(
     None:bool=false
     Preparing:bool=true
     Playing:bool=false
     Failed:bool=true
     Index:EFortTextToSpeechState Enum=Get Current Speaking State() -> (Current Speaking State=None)
    ) -> (ReturnValue:bool)
   ) -> (then)
  ]}
 Set Transcribed Text {nodes [
   n1 = Set Transcribed Text!() -> (then, Text:Text) {is=title K2Node_FunctionEntry}
   > n2 = Branch(Condition=TextIsEmpty(InText=@n1))
   SetText(execute=@n2.else, self:WBP UIKit Input Field Base Object Reference=Get WBP_Sidebar_InputField(), Text=Get Text!() -> (Text))
   n3 = Get WBP_Sidebar_InputField_Multiline()
   > "SetText (Multi-Line Text Box)"!(execute, self:"Text Box (Multi- Line) Object Reference"=Get EditableText(self=Get Block_EditableText(self=@n3)), InText:Text=Get Text()) -> (then)
  ]}
 SetCompactUIStyle {nodes [
   SetCompactUIStyle() {is=title K2Node_FunctionEntry}
   > Branch(Condition=IsUsingCompactUI())
   > SetPlacement!(
    execute
    self:Menu Anchor Object Reference=Get Anchor(self=Get Options_ContextMenu())
    InPlacement:EMenuPlacement Enum=MenuPlacement_AboveRightAnchor
   ) -> (then)
   n1 = Get Button_Dictate()
   n2 = Get Button_Options()
   n3 = Get Button_Send()
   > SetIconSize(self=[Get Button_ReadAloud(), @n1, @n2, @n3], Icon=(X=48,Y=48))
   n4 = Make Margin(Left=16, Top=0, Right=16, Bottom=0)
   > Set MobileButtonPadding!(execute, MobileButtonPadding:Margin Structure=@n4, self:WBP UIKit Button Generic Base Object Reference=@n1) -> (then, Output_Get:Margin Structure)
   > Set MobileButtonPadding(MobileButtonPadding=@n4, self=@n2)
   > Set MobileButtonPadding(MobileButtonPadding=@n4, self=@n3)
   > SetMobileFontSizeMultiplier!(
    execute
    self:Common Text Object Reference=Get Text_Heading!(self:WBP Sidebar Panel Layout Object Reference=Get PanelLayout()) -> (Text_Heading:Common Text Object Reference)
    InMobileFontSizeMultiplier:float=1
   ) -> (then)
   > Set SlateFontSize_DM(SlateFontSize_DM:float=24) -> (Output_Get:float)
   > Set SlateFontSize_Group(SlateFontSize_Group:float=21) -> (Output_Get:float)
   > Set ReadAloudPlayingIconSize(ReadAloudPlayingIconSize:Vector 2D Structure=(X=24,Y=24)) -> (Output_Get:Vector 2D Structure)
   > Set DefaultIconSize(DefaultIconSize:Vector 2D Structure=(X=48,Y=48)) -> (Output_Get:Vector 2D Structure)
   > Set ReadAloudPreparingIconSize(ReadAloudPreparingIconSize:Vector 2D Structure=(X=38,Y=38)) -> (Output_Get:Vector 2D Structure)
  ]}
 SetConversationHeaderFontSize {nodes [
   Inputs()
   n1 = Get Text_Heading(self=Get PanelLayout())
   n2 = Break Slate Font Info!(SlateFontInfo:"Slate Font Info Structure (by ref)"=Get Font!(self:Text Object Reference=@n1) -> (Font:Slate Font Info Structure)) -> (
    FontObject:Object Reference
    FontMaterial:Object Reference
    OutlineSettings:Font Outline Settings Structure
    TypefaceFontName:Name
    Size:float=~
    LetterSpacing:int
    SkewAmount:float=~
    bForceMonospaced:bool=~
    bMaterialIsStencil:bool=~
    MonospacedWidth:float=~
   ) {is=title K2Node_BreakStruct}
   n3 = Get ConversationType(self=Get FortTextChatGroupConversationVM()) -> (ConversationType=EpicParty)
   > SetFont!(
    execute
    self=@n1
    InFontInfo_FontObject:Object Reference=@n2
    InFontInfo_FontMaterial:Object Reference=@n2.FontMaterial
    InFontInfo_OutlineSettings:Font Outline Settings Structure=@n2.OutlineSettings
    InFontInfo_TypefaceFontName:Name=@n2.TypefaceFontName
    InFontInfo_Size:float=Select!(
     Option 0:float=Get SlateFontSize_Group() -> (SlateFontSize_Group:float)
     Option 1:float=Get SlateFontSize_DM() -> (SlateFontSize_DM:float)
     Index:bool="Equal (Enum)"(A:EFortTextChatPersistentConversationType Enum=@n3, B:EFortTextChatPersistentConversationType Enum=Epic121)
    ) -> (ReturnValue:float)
    InFontInfo_LetterSpacing:int=@n2.LetterSpacing
    InFontInfo_SkewAmount:float=@n2.SkewAmount
    InFontInfo_bForceMonospaced:bool=@n2.bForceMonospaced
    InFontInfo_bMaterialIsStencil:bool=@n2.bMaterialIsStencil
    InFontInfo_MonospacedWidth:float=@n2.MonospacedWidth
   ) -> (then)
   > Outputs()
  ]}
 SetInputFieldEnabled {nodes [
   n1 = SetInputFieldEnabled!() -> (then, IsEnabled:bool=~) {is=title K2Node_FunctionEntry}
   > SetIsEnabled(self=[Get WBP_Sidebar_InputField(), Get WBP_Sidebar_InputField_Multiline()], bInIsEnabled=@n1)
   > Set RenderOpacity!(
    execute
    RenderOpacity:float=SelectFloat(A=1, B=0.4, bPickA=Get IsEnabled() -> (IsEnabled:bool))
    self:Widget Object Reference=Get WBP_Sidebar_InputField()
   ) -> (then, Output_Get:float)
   > Set RenderOpacity(RenderOpacity=SelectFloat(A=1, B=0.4, bPickA=Get IsEnabled()), self=Get WBP_Sidebar_InputField_Multiline())
   > SetSendButtonEnabled()
   > RequestRefreshFocus()
  ]}
 SetMemberAvatars {nodes [
   n1 = SetMemberAvatars!() -> (then, Members List:Array of Fort Social User Viewmodel Object References) {is=title K2Node_FunctionEntry}
   > n2 = ConvertSocialUserArrayWithoutSelf!(execute, Social User VM:Array of Fort Social User Viewmodel Object References=@n1) -> (then, Social Users Without Self:Array of Fort Social User Object References)
   > InitializeWithSocialUsers!(
    execute
    self:WBP Sidebar Player Avatar Object Reference=Get WBP_Sidebar_PlayerAvatar() -> (WBP_Sidebar_PlayerAvatar:WBP Sidebar Player Avatar Object Reference)
    SocialUsers:Array of Fort Social User Object References=@n2
   ) -> (then)
  ]}
 SetMessageContextMenuActions {nodes [
   SetMessageContextMenuActions!() -> (then, Local user message:bool=~) {is=title K2Node_FunctionEntry}
   > RemoveAllInteractions(self:UIKit Context Menu Action Context Object Reference=Get Message_ContextMenuActionContext())
   > n1 = Set Selected Message Actions(Selected Message Actions:Array of EFortTextChatMessageContextMenuAction Enums=GetMessageContextMenuActions!(
     self:Fort Text Chat Conversation VM Object Reference=Get FortTextChatGroupConversationVM()
     InMessageId:"Name (by ref)"=Get Message ID() -> (Message ID=None)
    ) -> (ReturnValue:Array of EFortTextChatMessageContextMenuAction Enums=Mention)) -> (Output_Get:Array of EFortTextChatMessageContextMenuAction Enums=Mention)
   > n2 = For Each Loop(Array:Array of EFortTextChatMessageContextMenuAction Enums=@n1) -> (Array Element:EFortTextChatMessageContextMenuAction Enum)
   AddInteraction(
    execute=@n2.Completed
    self=Get Message_ContextMenuActionContext()
    InActionInfo_DisplayName=Find(
     TargetMap:Map of EFortTextChatMessageContextMenuAction Enums to Texts=Get Chat Menu Text() -> (Chat Menu Text:Map of EFortTextChatMessageContextMenuAction Enums to Texts=Mention)
     Key:"EFortTextChatMessageContextMenuAction Enum (by ref)"=SpeakMessage
    ) -> (Value:Text)
    InActionInfo_IconTexture=/FortniteCoreUI/UI/Icons/Icon-Settings/T-Icon-Audio-64
    InActionInfo_SortPriority=1
    InActionInfo_bIsAvailable=Get is TextToSpeech Enabled()
    InActionInfo_bIsEnabled=True
    InActionInfo_bIsSelected=False
    InActionInfo_bAddDivider=False
    InActionInfo_bPositionDividerAfter=False
    InActionInfo_bCloseMenuOnSelect=False
    InActionInfo_InteractionNameID=Find(
     TargetMap:Map of EFortTextChatMessageContextMenuAction Enums to Names=Get Chat Menu Actions() -> (Chat Menu Actions=Mention)
     Key=SpeakMessage
    ) -> (Value:Name)
    InActionInfo_Interaction=(DataTable=None,RowName="")
    InActionInfo_PressedDelegate="(null).None"
    InActionInfo_ReleasedDelegate="(null).None"
    InInteractionDelegate=Create Event()
   )
   > AddInteraction(
    self=Get Message_ContextMenuActionContext()
    InActionInfo_DisplayName=Find(
     TargetMap:Map of EFortTextChatMessageContextMenuAction Enums to Texts=Get Chat Menu Text() -> (Chat Menu Text=Mention)
     Key=SpeakMessagesFromHere
    ) -> (Value:Text)
    InActionInfo_IconTexture=&T-Icon-Audio-64
    InActionInfo_SortPriority=2
    InActionInfo_bIsAvailable=Get is TextToSpeech Enabled()
    InActionInfo_bIsEnabled=True
    InActionInfo_bIsSelected=False
    InActionInfo_bAddDivider=true
    InActionInfo_bPositionDividerAfter=true
    InActionInfo_bCloseMenuOnSelect=False
    InActionInfo_InteractionNameID=Find(
     TargetMap:Map of EFortTextChatMessageContextMenuAction Enums to Names=Get Chat Menu Actions() -> (Chat Menu Actions=Mention)
     Key=SpeakMessagesFromHere
    ) -> (Value:Name)
    InActionInfo_Interaction=(DataTable=None,RowName="")
    InActionInfo_PressedDelegate="(null).None"
    InActionInfo_ReleasedDelegate="(null).None"
    InInteractionDelegate=Create Event()
   )
   > Return Node!(
    execute
    Has Interactions:bool=HasAvailableInteractions!(self:UIKit Context Menu Action Context Object Reference=Get Message_ContextMenuActionContext()) -> (ReturnValue:bool)
   )
   n3 = Get Local user message() -> (Local user message:bool)
   Branch(
    execute=@n2.LoopBody
    Condition=AND Boolean!(
     A=NOT Boolean(A=@n3)
     B="Not Equal (Enum)"(A:EFortTextChatMessageContextMenuAction Enum=@n2, B:EFortTextChatMessageContextMenuAction Enum=SpeakMessage)
     C:bool="Not Equal (Enum)"(A=@n2, B=SpeakMessagesFromHere)
    ) -> (ReturnValue)
   )
   n4 = Find(TargetMap=Get Chat Menu Actions() -> (Chat Menu Actions=Mention), Key=@n2)
   > AddInteraction(
    self=Get Message_ContextMenuActionContext()
    InActionInfo_DisplayName=Find(TargetMap:Map of EFortTextChatMessageContextMenuAction Enums to Texts=Get Chat Menu Text() -> (Chat Menu Text=Mention), Key=@n2) -> (Value:Text)
    InActionInfo_SortPriority=3
    InActionInfo_bIsAvailable=@n4.ReturnValue
    InActionInfo_bIsEnabled=True
    InActionInfo_bIsSelected=False
    InActionInfo_bAddDivider=False
    InActionInfo_bPositionDividerAfter=False
    InActionInfo_bCloseMenuOnSelect=False
    InActionInfo_InteractionNameID=@n4
    InActionInfo_Interaction=(DataTable=None,RowName="")
    InActionInfo_PressedDelegate="(null).None"
    InActionInfo_ReleasedDelegate="(null).None"
    InInteractionDelegate=Create Event()
   )
  ]}
 SetOptionsContextMenuStaticActions {nodes [
   SetOptionsContextMenuStaticActions() {is=title K2Node_FunctionEntry}
   > RemoveInteraction(self=Get OptionsContextMenuActionContext(), InteractionNameID=CommunicateSettings)
   > AddInteraction(
    self=Get OptionsContextMenuActionContext()
    InActionInfo_DisplayName="LOCTEXT(\"Communication settings\")"
    InActionInfo_SortPriority=2
    InActionInfo_bIsAvailable=True
    InActionInfo_bIsEnabled=True
    InActionInfo_bIsSelected=False
    InActionInfo_bAddDivider=False
    InActionInfo_bPositionDividerAfter=False
    InActionInfo_bCloseMenuOnSelect=False
    InActionInfo_InteractionNameID=CommunicateSettings
    InActionInfo_Interaction=(DataTable=None,RowName="")
    InActionInfo_PressedDelegate="(null).None"
    InActionInfo_ReleasedDelegate="(null).None"
    InInteractionDelegate=Create Event()
   )
   > RemoveInteraction(self=Get OptionsContextMenuActionContext(), InteractionNameID=Report)
   n1 = Get FortTextChatGroupConversationVM()
   > AddInteraction(
    self=Get OptionsContextMenuActionContext()
    InActionInfo_DisplayName="LOCTEXT(\"Report\")"
    InActionInfo_SortPriority=3
    InActionInfo_bIsAvailable=Contains Item(TargetArray=GetConversationContextMenuActions(self=@n1) -> (ReturnValue=Manage), ItemToFind=Report)
    InActionInfo_bIsEnabled=True
    InActionInfo_bIsSelected=False
    InActionInfo_bAddDivider=False
    InActionInfo_bPositionDividerAfter=False
    InActionInfo_bCloseMenuOnSelect=False
    InActionInfo_InteractionNameID=Report
    InActionInfo_Interaction=(DataTable=None,RowName="")
    InActionInfo_PressedDelegate="(null).None"
    InActionInfo_ReleasedDelegate="(null).None"
    InInteractionDelegate=Create Event()
   )
  ]}
 SetPartyMemberAvatarsVisibility {nodes [
   n1 = SetPartyMemberAvatarsVisibility!() -> (then, Party state:EFortConversationPartyState Enum=None, Party Members:Array of Fort Social User Info View Model Object References) {is=title K2Node_FunctionEntry}
   n2 = Is Not Empty!(TargetArray:Array of Fort Social User Info View Model Object References=@n1.Party Members) -> (ReturnValue:bool)
   n3 = Get Party state() -> (Party state:EFortConversationPartyState Enum=None)
   n4 = "Equal (Enum)"(A:EFortConversationPartyState Enum=@n3, B:EFortConversationPartyState Enum=Joining)
   n5 = "Equal (Enum)"(A=@n3, B=Joined)
   n6 = "Equal (Enum)"(A=@n3, B=Active)
   > Return Node!(
    execute
    Visibility:ESlateVisibility Enum=Select(
     Option 0:ESlateVisibility Enum=Collapsed
     Option 1:ESlateVisibility Enum=SelfHitTestInvisible
     Index=OR Boolean!(A=@n6, B=@n5, C:bool=@n4, D:bool=@n2) -> (ReturnValue)
    ) -> (ReturnValue:ESlateVisibility Enum)
   )
  ]}
 SetSendButtonEnabled {nodes [
   SetSendButtonEnabled() {is=title K2Node_FunctionEntry}
   n1 = Get Block_EditableText(self=Get WBP_Sidebar_InputField_Multiline())
   n2 = TextIsEmpty(InText=Get Text!(self:"Text Box (Multi- Line) Object Reference"=Get EditableText(self=@n1)) -> (Text))
   n3 = Get WBP_Sidebar_InputField_Multiline()
   n4 = Get Block_EditableText(self:WBP UIKit Input Field Base Object Reference=Get WBP_Sidebar_InputField()) -> (Block_EditableText:WBP UIKit Block Editable Text Object Reference)
   n5 = TextIsEmpty(InText="GetText (Editable Text)"(self=Get EditableText(self:WBP UIKit Block Editable Text Object Reference=@n4) -> (EditableText:UIKit Editable Text Object Reference)))
   n6 = Get WBP_Sidebar_InputField()
   n7 = Select(
    Option 0:bool=AND Boolean!(A=NOT Boolean(A=@n5), B=Get bIsEnabled(self=@n6)) -> (ReturnValue)
    Option 1:bool=AND Boolean(A=NOT Boolean(A=@n2), B=Get bIsEnabled(self=@n3))
    Index=Get Multiline InputField Enabled()
   ) -> (ReturnValue:bool)
   > SetIconMaterial(
    self=Get Button_Send()
    Material Instance=Select(
     Option 0:Material Instance Object Reference=&MI_Sidebar_Icon_Base
     Option 1:Material Instance Object Reference=/Sidebar/Materials/Icons/MI_Sidebar_Icon_Inverse
     Index=@n7
    ) -> (ReturnValue:Material Instance Object Reference)
   )
   > SetIsEnabled(self=Get Button_Send(), bInIsEnabled=@n7)
   > Branch(Condition=Get is SpeechToText Enabled())
   > SetVisibility(
    self=Get Button_Send()
    InVisibility=Select(Option 0:ESlateVisibility Enum=Collapsed, Option 1:ESlateVisibility Enum=Visible, Index=@n7) -> (ReturnValue:ESlateVisibility Enum)
   )
   > SetVisibility(self=Get Button_Dictate(), InVisibility=Select(Option 0=Visible, Option 1=Collapsed, Index=@n7))
  ]}
 Should Show STT Onboarding Tooltip {nodes [
   n1 = Should Show STT Onboarding Tooltip!() -> (then, Loading:bool=~) {is=title K2Node_FunctionEntry}
   n2 = Get WBP_TextChat_STT_CoachMark()
   n3 = Get FortTextChatGroupConversationVM()
   > n4 = Branch(Condition=AND Boolean!(
     A=NOT Boolean(A=@n1)
     B=Get bIsInitialized(self=@n3)
     C:bool=ShouldShowOnboarding!(self:Fort Sidebar Onboard Tooltip Widget Object Reference=@n2) -> (ReturnValue:bool)
    ) -> (ReturnValue))
   > HandleTooltipOrientation() {is=title K2Node_CallFunction}
   > ActivateWidget(self=Get WBP_TextChat_STT_CoachMark())
   Branch(execute=@n4.else, Condition=IsVisible(self=Get WBP_TextChat_STT_CoachMark()))
   > SetVisibility(self=Get WBP_TextChat_STT_CoachMark(), InVisibility=Collapsed)
  ]}
 Speak All Messages {nodes [
   Speak All Messages!() -> (then, InteractionNameID:Name) {is=title K2Node_FunctionEntry}
   > Cancel Speech To Text Recording()
   > SpeakMessagesInclusiveAfter!(
    execute
    self:Fort Text Chat Conversation VM Object Reference=Get FortTextChatGroupConversationVM()
    MessageId:"Name (by ref)"=Get Message ID() -> (Message ID=None)
   ) -> (then)
  ]}
 Speak Message {nodes [
   Speak Message!() -> (then, InteractionNameID:Name) {is=title K2Node_FunctionEntry}
   > Cancel Speech To Text Recording()
   > SpeakMessage!(
    execute
    self:Fort Text Chat Conversation VM Object Reference=Get FortTextChatGroupConversationVM()
    MessageId:"Name (by ref)"=Get Message ID() -> (Message ID=None)
   ) -> (then)
   > NavigateToIndex(self=Get ListView(self=Get MessageList()), Index=Get Context Menu Item Index() -> (Context Menu Item Index=0))
  ]}
 UpdateStartPartyButtonVisibility {nodes [
   n1 = UpdateStartPartyButtonVisibility!() -> (then, IsEpicParties:bool=~) {is=title K2Node_FunctionEntry}
   n2 = SupportsStartPartyFromConversation!(self:Fort Text Chat Conversation VM Object Reference=Get FortTextChatGroupConversationVM()) -> (ReturnValue:bool)
   > SetVisibility(
    self=Get Button_StartParty()
    InVisibility=Select(Option 0=Collapsed, Option 1=SelfHitTestInvisible, Index=AND Boolean!(A=@n1, B=@n2) -> (ReturnValue))
   )
  ]}
 isInputFieldNotEmpty {nodes [
   isInputFieldNotEmpty() {is=title K2Node_FunctionEntry}
   n1 = Get Block_EditableText(self:WBP Text Chat Multiline Input Field Object Reference=Get WBP_Sidebar_InputField_Multiline()) -> (Block_EditableText:WBP Text Chat Block Multiline Editable Text Object Reference)
   n2 = TextIsEmpty(InText="GetText (Multi-Line Text Box)"!(self:"Text Box (Multi- Line) Object Reference"=Get EditableText(self:WBP Text Chat Block Multiline Editable Text Object Reference=@n1) -> (EditableText:"Text Box (Multi- Line) Object Reference")) -> (ReturnValue:Text))
   n3 = Get Block_EditableText(self:WBP UIKit Input Field Base Object Reference=Get WBP_Sidebar_InputField()) -> (Block_EditableText:WBP UIKit Block Editable Text Object Reference)
   n4 = TextIsEmpty(InText="GetText (Editable Text)"(self=Get EditableText(self:WBP UIKit Block Editable Text Object Reference=@n3) -> (EditableText:UIKit Editable Text Object Reference)))
   > Return Node!(execute, NotEmpty:bool=OR Boolean!(A=NOT Boolean(A=@n4), B=NOT Boolean(A=@n2)) -> (ReturnValue))
  ]}
]
variables [
 Chat Menu Actions {default {
   block=Block
   report=Report chat
   speakMessage=Read aloud
   speakMessagesFromHere=Read aloud from here
   unblock=Unblock
   viewProfile=View profile
  }}
 Chat Menu Text {default {
   block=Block
   report=Report chat
   speakMessage=Read aloud
   speakMessagesFromHere=Read aloud from here
   unblock=Unblock
   viewProfile=View profile
  }}
 Context Menu Item Index {type=integer}
 Current Hint {type=string, default=""}
 Current Input Type {type=ECommonInputType, default=MouseAndKeyboard}
 DefaultIconSize {type=Vector2D, default=<32 32>}
 Hint Text Listening {type=string, default=Listening...}
 Hint Text Please Wait {type=string, default=Please wait...}
 is SpeechToText Enabled {type=boolean}
 is TextToSpeech Enabled {type=boolean}
 IsUsingEpicParties {type=boolean}
 LoadOlderMessagesButtonEntry {type=BP_Sidebar_CustomButtonEntryVM_C, default=null}
 LoadOlderMessagesButtonType {type=string, default=LoadOlderMessaqes}
 Message ID {type=string, default=None}
 Message Social User {type=FortSocialUser, default=null}
 Message_ContextMenuActionContext {type=UIKitContextMenuActionContext, default=null}
 MessageReceivedSound {type=SoundBase, default=/Game/Sounds/Fort_UI_Sounds/Chat_UI/MSS/MSS_UI_FrontEnd_Chat_MessageReceived}
 MessageSentSound {type=SoundBase, default=/Game/Sounds/Fort_UI_Sounds/Chat_UI/MSS/MSS_UI_FrontEnd_Chat_MessageSent}
 OptionsContextMenuActionContext {type=UIKitContextMenuActionContext, default=null}
 PaddingWithButtonPrompt {type=Margin, default=<ltrb 36 0 0 0>}
 PaddingWithoutButtonPrompt {type=Margin, default=<ltrb 0 0 0 0>}
 PlayingIcon {type=Object, default=null}
 ReadAloudHasPrompt {type=boolean, default=true}
 ReadAloudPlayingIconSize {type=Vector2D, default=<32 32>}
 ReadAloudPreparingIconSize {type=Vector2D, default=<28 28>}
 Record Time {type=number, default=10}
 Selected Message Actions {type="array<EFortTextChatMessageContextMenuAction>", default []}
 SlateFontSize_DM {type=number, default=18}
 SlateFontSize_Group {type=number, default=15.75}
 TextInputSound {type=SoundBase, default=null}
 Timer Handle {type=TimerHandle}
 Username {type=string, default=""}
 WBP Text Chat Entry User Message {type=WBP_TextChat_Entry_UserMessage_C, default=null}
 WBP Text Chat Message System {type=WBP_TextChat_MessageSystem_C, default=null}
 WelcomeMessageEntry {type=BP_VM_TextChat_Welcome_Entry_C, default=null}
]
widgets {
 info {
  inheritedWidgetCount=0
  namedSlotCount=5
  parentClass=&CommonActivatableWidget
  rootWidgetClass=&Overlay
  widgetCount=28
 }
 widgets [
  Overlay_1 {widgetClassPath=&Overlay, widget=Overlay_1}
  PanelLayout {
   widgetClassPath=/Sidebar/Shared/WBP_Sidebar_PanelLayout
   bIsVariable=true
   parent=Overlay_1
   widget=PanelLayout
   properties {autoPopNavigationOnBack=false, headingText=Chat, is Panel Name Scrolling Enabled=true, isChatView=true, subheadingText=Activity Status}
   slot {class=OverlaySlot, properties {horizontalAlignment=HAlign_Fill, verticalAlignment=VAlign_Fill}}
  }
  VB_TextChat {
   widgetClassPath=&VerticalBox
   namedSlotHost=PanelLayout
   widget=VB_TextChat
   properties {navigation=&self:WidgetTree.VB_TextChat.WidgetNavigation_0}
  }
  Overlay_Clipping {
   widgetClassPath=&Overlay
   parent=VB_TextChat
   widget=Overlay_Clipping
   properties {clipping=ClipToBounds}
   slot {class=VerticalBoxSlot, properties {padding=<ltrb -18 0 -18 0>, size {sizeRule=Fill, value=1}}}
  }
  Overlay_Messages {
   widgetClassPath=&Overlay
   parent=Overlay_Clipping
   widget=Overlay_Messages
   properties {navigation=&self:WidgetTree.Overlay_Messages.WidgetNavigation_0}
   slot {class=OverlaySlot, properties {horizontalAlignment=HAlign_Fill, verticalAlignment=VAlign_Bottom, padding=<ltrb 18 0 18 18>}}
  }
  MessageList {
   widgetClassPath=/Sidebar/Panels/Chat/List/WBP_Sidebar_TextChat_MessageList
   bIsVariable=true
   parent=Overlay_Messages
   widget=MessageList
   properties {bIsFocusable=true, bOverride_Cursor=true}
   slot {class=OverlaySlot, properties {horizontalAlignment=HAlign_Fill, verticalAlignment=VAlign_Bottom, padding=<ltrb 0 12 0 0>}}
  }
  Message_ContextMenu {
   widgetClassPath=/FortUILibrary/UIKit/ContextMenu/WBP_UIKit_ContextMenuAnchorButton
   bIsVariable=true
   parent=Overlay_Clipping
   widget=Message_ContextMenu
   properties {
    button Background Material=null
    button Outline Material=null
    buttonClass=/Game/UI/Frontend/Lobby/Nameplates/WBP_EmptyButton_ContextMenu.WBP_EmptyButton_ContextMenu_C
    buttonIcon=null
    closeMenuOnContextClicked=true
    menuPlacement=MenuPlacement_AboveRightAnchor
    min Height=1
    button Padding=<ltrb 0 0 0 0>
   }
   slot {class=OverlaySlot, properties {horizontalAlignment=HAlign_Right, verticalAlignment=VAlign_Bottom}}
  }
  Button_JumpToMostRecent {
   widgetClassPath=/Sidebar/Buttons/WBP_Sidebar_ButtonGeneric
   bIsVariable=true
   parent=VB_TextChat
   widget=Button_JumpToMostRecent
   properties {
    configuration=Number
    contentHorizontalAlignment=HAlign_Fill
    minHeightOverride=56
    promptHorizontalAlignment=HAlign_Right
    statusIndicatorHorizontalAlignment=HAlign_Left
    statusIndicatorVerticalAlignment=VAlign_Center
    text=Jump to present
    useStatusIndicator=true
    visibility=Collapsed
    buttonPadding=<ltrb 64 0 64 0>
    promptNudge=<0 0>
    statusIndicatorPadding=<ltrb 12 0 0 0>
    triggeringInputAction {dataTable=/Sidebar/Shared/DT_Sidebar_InputActions, rowName=JumpToRecent}
   }
   slot {class=VerticalBoxSlot, properties {padding=<ltrb 0 18 0 18>}}
  }
  InputActionWidget {
   widgetClassPath=&CommonActionWidget
   bIsVariable=true
   namedSlotHost=Button_JumpToMostRecent
   widget=InputActionWidget
   properties {inputActions [{dataTable=&DT_Sidebar_InputActions, rowName=JumpToRecent}]}
  }
  WBP_Sidebar_TextChat_Restrictions {
   widgetClassPath=&/WBP_Sidebar_TextChat_Restrictions
   bIsVariable=true
   parent=VB_TextChat
   widget=WBP_Sidebar_TextChat_Restrictions
   properties {restrictionWidgetPadding=<ltrb 0 8 0 0>}
   slot {class=VerticalBoxSlot}
  }
  WBP_TextChat_InactiveMessage {
   widgetClassPath=/FortniteCoreUI/UI/Social/Chat/TextChat/WBP_TextChat_InactiveMessage
   bIsVariable=true
   parent=VB_TextChat
   widget=WBP_TextChat_InactiveMessage
   properties {visibility=Collapsed, restrictionWidgetInternalPadding=<ltrb 0 8 0 0>}
   slot {class=VerticalBoxSlot}
  }
  Button_ReadAloud {
   widgetClassPath=&WBP_Sidebar_ButtonGeneric
   bIsVariable=true
   parent=VB_TextChat
   widget=Button_ReadAloud
   properties {
    bDisplayInputActionWhenNotInteractable=false
    bHideInputAction=false
    bShouldSelectUponReceivingFocus=false
    bShouldUseFallbackDefaultInputAction=false
    backgroundClass=null
    backgroundMaterial=null
    cornerRadius=0.5
    enableFocusedSelectedState=false
    iconMaterial=null
    minHeight=44
    minHeightOverride=56
    radiusInPixels=false
    shouldPromptDirectionFollowCulture=false
    text=Read Aloud
    textJustification=Left
    useIconBlock=true
    buttonPadding=<ltrb 16 0 18 0>
    font {
     bForceMonospaced=false
     bMaterialIsStencil=false
     fontMaterial=null
     fontObject=&HeadingNow
     letterSpacing=10
     monospacedWidth=1
     size=15
     skewAmount=0
     typefaceFontName=74 Regular
     outlineSettings {
      bApplyOutlineToDropShadows=false
      bMiteredCorners=false
      bSeparateFillAlpha=false
      outlineMaterial=/FortUILibrary/Systems/ModularButton/MaterialInstances/MI_UIKit_TextOutline
      outlineSize=0
      outlineColor=<rgba 0 0 0 1>
     }
    }
    promptNudge=<16 0>
    textPadding=<ltrb 36 0 0 0>
    triggeringInputAction {dataTable=&InputActionDataTable, rowName=TextToSpeechPrimaryAction}
   }
   slot {class=VerticalBoxSlot, properties {horizontalAlignment=HAlign_Center, verticalAlignment=VAlign_Center, padding=<ltrb 0 8 0 0>}}
  }
  Button_StopTTS_Decoy {
   widgetClassPath=/Game/UI/Frontend/BottomBar/WBP_BottomBarDecoyButton
   bIsVariable=true
   namedSlotHost=Button_ReadAloud
   widget=Button_StopTTS_Decoy
   properties {bDisplayInActionBar=false, triggeringInputAction {dataTable=&InputActionDataTable, rowName=TextToSpeechSecondaryAction}}
  }
  Grid_Controls {
   widgetClassPath=&GridPanel
   bIsVariable=true
   parent=VB_TextChat
   widget=Grid_Controls
   properties {columnFill [1], rowFill [1,0]}
   slot {class=VerticalBoxSlot, properties {padding=<ltrb 0 21 0 0>}}
  }
  HB_InputText_NavBoundary {
   widgetClassPath=&HorizontalBox
   parent=Grid_Controls
   widget=HB_InputText_NavBoundary
   properties {bOverride_Cursor=true, navigation=&self:WidgetTree.HB_InputText_NavBoundary.WidgetNavigation_0}
   slot {class=GridSlot}
  }
  InputField {
   widgetClassPath=&VerticalBox
   parent=HB_InputText_NavBoundary
   widget=InputField
   slot {class=HorizontalBoxSlot, properties {padding=<ltrb 0 0 4 0>, size {sizeRule=Fill, value=1}}}
  }
  WBP_Sidebar_InputField {
   widgetClassPath=/Sidebar/Buttons/WBP_Sidebar_InputField
   bIsVariable=true
   parent=InputField
   widget=WBP_Sidebar_InputField
   properties {holdData=&CommonUI_HoldData_Base_C, icon=null, minHeight=56, textPadding=<ltrb 16 4 20 4>}
   slot {class=VerticalBoxSlot}
  }
  WBP_Sidebar_InputField_Multiline {
   widgetClassPath=/Sidebar/Buttons/WBP_Sidebar_InputField_Multiline
   bIsVariable=true
   parent=InputField
   widget=WBP_Sidebar_InputField_Multiline
   properties {maxHeight=86, minHeight=56, useIcon=false, visibility=Collapsed}
   slot {class=VerticalBoxSlot}
  }
  Button_Send {
   widgetClassPath=&WBP_Sidebar_ButtonGeneric
   bIsVariable=true
   parent=HB_InputText_NavBoundary
   widget=Button_Send
   properties {
    buttonType=Attention
    icon=/FortUILibrary/Textures/Icons/T_UI_Icon_UIKit_SendChat_MSDF
    minHeight=0
    minHeightOverride=56
    minWidth=56
    useIconBlock=true
    useTextBlock=false
    buttonPadding=<ltrb 12 12 12 12>
    font {
     bForceMonospaced=false
     bMaterialIsStencil=false
     fontMaterial=null
     fontObject=&HeadingNow
     letterSpacing=10
     monospacedWidth=1
     size=15
     skewAmount=0
     typefaceFontName=74 Regular
     outlineSettings {
      bApplyOutlineToDropShadows=false
      bMiteredCorners=false
      bSeparateFillAlpha=false
      outlineMaterial=&MI_UIKit_TextOutline
      outlineSize=0
      outlineColor=<rgba 0 0 0 1>
     }
    }
   }
   slot {class=HorizontalBoxSlot, properties {verticalAlignment=VAlign_Bottom, padding=<ltrb 8 0 0 0>}}
  }
  Button_Dictate {
   widgetClassPath=&WBP_Sidebar_ButtonGeneric
   bIsVariable=true
   parent=HB_InputText_NavBoundary
   widget=Button_Dictate
   properties {
    contentHorizontalAlignment=HAlign_Fill
    contentVerticalAlignment=VAlign_Fill
    icon=&T_UI_Icon_Sidebar_Mic_MSDF
    minHeight=0
    minHeightOverride=56
    minWidth=52
    text=""
    textJustification=Left
    useDisabledVisualsOnLocked=true
    useIconBlock=true
    visibility=Visible
    buttonPadding=<ltrb 12 12 12 12>
    font {
     bForceMonospaced=false
     bMaterialIsStencil=false
     fontMaterial=/Sidebar/Materials/Fonts/MI_Sidebar_FontColor_Secondary
     fontObject=&HeadingNow
     letterSpacing=0
     monospacedWidth=1
     size=13.5
     skewAmount=0
     typefaceFontName=74 Regular
     outlineSettings {
      bApplyOutlineToDropShadows=false
      bMiteredCorners=false
      bSeparateFillAlpha=false
      outlineMaterial=&MI_UIKit_TextOutline
      outlineSize=0
      outlineColor=<rgba 0 0 0 1>
     }
    }
    textPadding=<ltrb 2 0 0 0>
   }
   slot {class=HorizontalBoxSlot, properties {verticalAlignment=VAlign_Bottom, padding=<ltrb 8 0 0 0>}}
  }
  Button_Options {
   widgetClassPath=&WBP_Sidebar_ButtonGeneric
   bIsVariable=true
   parent=HB_InputText_NavBoundary
   widget=Button_Options
   properties {
    icon=/Sidebar/Materials/Icons/Textures/T_UI_Icon_Sidebar_ContextMenu_MSDF
    minHeightOverride=56
    useIconBlock=true
    useTextBlock=false
    buttonPadding=<ltrb 12 12 12 12>
    font {
     bForceMonospaced=false
     bMaterialIsStencil=false
     fontMaterial=null
     fontObject=&HeadingNow
     letterSpacing=10
     monospacedWidth=1
     size=15
     skewAmount=0
     typefaceFontName=74 Regular
     outlineSettings {
      bApplyOutlineToDropShadows=false
      bMiteredCorners=false
      bSeparateFillAlpha=false
      outlineMaterial=&MI_UIKit_TextOutline
      outlineSize=0
      outlineColor=<rgba 0 0 0 1>
     }
    }
   }
   slot {class=HorizontalBoxSlot, properties {horizontalAlignment=HAlign_Left, verticalAlignment=VAlign_Bottom, padding=<ltrb 8 0 0 0>}}
  }
  Overlay_0 {widgetClassPath=&Overlay, namedSlotHost=Button_Options, widget=Overlay_0}
  Options_ContextMenu {
   widgetClassPath=&WBP_UIKit_ContextMenuAnchorButton
   bIsVariable=true
   parent=Overlay_0
   widget=Options_ContextMenu
   properties {
    button Background Material=null
    button Outline Material=null
    buttonClass=&WBP_EmptyButton_ContextMenu_C
    buttonText=""
    closeMenuOnContextClicked=true
    isMenuModal=false
    menuPlacement=MenuPlacement_BelowAnchor
   }
   slot {class=OverlaySlot, properties {horizontalAlignment=HAlign_Fill, verticalAlignment=VAlign_Fill}}
  }
  Button_StartParty {
   widgetClassPath=&/WBP_Sidebar_PartyButton
   bIsVariable=true
   parent=Grid_Controls
   widget=Button_StartParty
   properties {holdData=&CommonUI_HoldData_Base_C, minHeight=56}
   slot {class=GridSlot, properties {horizontalAlignment=HAlign_Left, row=1, verticalAlignment=VAlign_Top, padding=<ltrb 0 16 0 0>}}
  }
  Grid_OptionsAndSend {
   widgetClassPath=&GridPanel
   bIsVariable=true
   parent=Grid_Controls
   widget=Grid_OptionsAndSend
   properties {bOverride_Cursor=true, columnFill [1,0], rowFill [0]}
   slot {class=GridSlot, properties {column=2, verticalAlignment=VAlign_Bottom}}
  }
  WBP_Sidebar_PlayerAvatar {
   widgetClassPath=/Sidebar/Panels/WBP_Sidebar_PlayerAvatar
   bIsVariable=true
   namedSlotHost=PanelLayout
   widget=WBP_Sidebar_PlayerAvatar
   properties {avatarSize=<72 72>}
  }
  VisAttach_CoachmarkSTT {
   widgetClassPath=&CommonVisualAttachment
   bIsVariable=true
   parent=Overlay_1
   widget=VisAttach_CoachmarkSTT
   properties {contentAnchor=<1.025 1>}
   slot {class=OverlaySlot, properties {verticalAlignment=VAlign_Bottom}}
  }
  WBP_TextChat_STT_CoachMark {
   widgetClassPath=/FortniteCoreUI/UI/Social/Chat/TextChat/Coachmarks/WBP_TextChat_STT_CoachMark
   bIsVariable=true
   parent=VisAttach_CoachmarkSTT
   widget=WBP_TextChat_STT_CoachMark
   slot {class=SizeBoxSlot}
  }
 ]
}
