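# UI StateTree for the sidebar chat, choosing which chat panel appears in the sidebar sub-panel.
# Each panel state (conversation list, messages, invite, manage group, accept invite) slots a cached, player-specific widget into UI.Sidebar.SubPanel.
# ChatRoot event transitions reach every panel except ConversationList, the first child.
# MessageList, ManageGroup and AcceptChatInvite also transition on events to LocalPlayer or OtherPlayer linked-asset states.
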StateTree /Sidebar/StateTree/ST_SidebarChat
states [
 ChatRoot {
  id=#1
  type=STATE
  children [
   #2
   #3
   #4
   #5
   #6
   #7
   #8
  ]
  tasks []
  transitions [
   {id=#9, trigger=ON_EVENT, state {id=#3, name=MessageList}}
   {id=#10, trigger=ON_EVENT, state {id=#4, name=Invite}}
   {id=#11, trigger=ON_EVENT, state {id=#5, name=ManageGroup}}
   {id=#12, trigger=ON_EVENT, state {id=#6, name=AcceptChatInvite}}
  ]
 }
 ConversationList {
  id=#2
  type=STATE
  tasks [{
    id=#13
    description=UIState Tree Widget Slotting Task
    properties {
     iD=#14
     instance {
      _structType=/Script/UIStateTree.UIStateTreeWidgetSlottingTaskInstanceData
      bIsTaskActive=false
      widgets [{
        bAutoActivateWidget=true
        bCacheInstance=true
        bFill=true
        bIsPlayerSpecific=true
        bIsWidgetEnabled=true
        bKeepConstructed=true
        cacheEntryName=ChatPanel
        memoryTag=None
        widget=/Sidebar/Panels/Chat/WBP_Sidebar_ChatChannelsPanel.WBP_Sidebar_ChatChannelsPanel_C
        hostTag {tagName=UI.Sidebar.SubPanel}
        padding=<ltrb 0 0 0 0>
       }]
     }
     node {
      name=None
      _structType=/Script/UIStateTree.UIStateTreeWidgetSlottingTask
      bCanEditConsideredForCompletion=true
      bConsideredForCompletion=false
      bHasTaskCompletionDelegateDispatcher=false
      bTaskEnabled=true
      transitionHandlingPriority=Normal
      bindingsBatch {value=65535}
      executionRuntimeTemplateIndex {value=65535}
      instanceDataHandle {index=65535, source=None, stateHandle {index=65535}}
      instanceTemplateIndex {value=65535}
      outputBindingsBatch {value=65535}
     }
    }
   }]
  transitions []
 }
 MessageList {
  id=#3
  type=STATE
  tasks [{
    id=#15
    description=UIState Tree Widget Slotting Task
    properties {
     iD=#16
     instance {
      _structType=&UIStateTreeWidgetSlottingTaskInstanceData
      bIsTaskActive=false
      widgets [{
        bAutoActivateWidget=true
        bCacheInstance=true
        bFill=true
        bIsPlayerSpecific=true
        bIsWidgetEnabled=true
        bKeepConstructed=true
        cacheEntryName=MessageListPanel
        memoryTag=None
        widget=/Sidebar/Panels/Chat/WBP_Sidebar_TextChatPanel.WBP_Sidebar_TextChatPanel_C
        hostTag {tagName=UI.Sidebar.SubPanel}
        padding=<ltrb 0 0 0 0>
       }]
     }
     node {
      name=None
      _structType=&UIStateTreeWidgetSlottingTask
      bCanEditConsideredForCompletion=true
      bConsideredForCompletion=false
      bHasTaskCompletionDelegateDispatcher=false
      bTaskEnabled=true
      transitionHandlingPriority=Normal
      bindingsBatch {value=65535}
      executionRuntimeTemplateIndex {value=65535}
      instanceDataHandle {index=65535, source=None, stateHandle {index=65535}}
      instanceTemplateIndex {value=65535}
      outputBindingsBatch {value=65535}
     }
    }
   }]
  transitions [
   {id=#17, trigger=ON_EVENT, state {id=#8, name=OtherPlayer}}
   {id=#18, trigger=ON_EVENT, state {id=#7, name=LocalPlayer}}
  ]
 }
 Invite {
  id=#4
  type=STATE
  tasks [{
    id=#19
    description=UIState Tree Widget Slotting Task
    properties {
     iD=#20
     instance {
      _structType=&UIStateTreeWidgetSlottingTaskInstanceData
      bIsTaskActive=false
      widgets [{
        bAutoActivateWidget=true
        bCacheInstance=true
        bFill=true
        bIsPlayerSpecific=true
        bIsWidgetEnabled=true
        bKeepConstructed=true
        cacheEntryName=InvitePanel
        memoryTag=None
        widget=/Sidebar/Panels/Chat/WBP_Sidebar_ChatInvitePanel.WBP_Sidebar_ChatInvitePanel_C
        hostTag {tagName=UI.Sidebar.SubPanel}
        padding=<ltrb 0 0 0 0>
       }]
     }
     node {
      name=None
      _structType=&UIStateTreeWidgetSlottingTask
      bCanEditConsideredForCompletion=true
      bConsideredForCompletion=false
      bHasTaskCompletionDelegateDispatcher=false
      bTaskEnabled=true
      transitionHandlingPriority=Normal
      bindingsBatch {value=65535}
      executionRuntimeTemplateIndex {value=65535}
      instanceDataHandle {index=65535, source=None, stateHandle {index=65535}}
      instanceTemplateIndex {value=65535}
      outputBindingsBatch {value=65535}
     }
    }
   }]
  transitions []
 }
 ManageGroup {
  id=#5
  type=STATE
  tasks [{
    id=#21
    description=UIState Tree Widget Slotting Task
    properties {
     iD=#22
     instance {
      _structType=&UIStateTreeWidgetSlottingTaskInstanceData
      bIsTaskActive=false
      widgets [{
        bAutoActivateWidget=true
        bCacheInstance=true
        bFill=true
        bIsPlayerSpecific=true
        bIsWidgetEnabled=true
        bKeepConstructed=true
        cacheEntryName=ManageGroupPanel
        memoryTag=None
        widget=/Sidebar/Panels/Chat/WBP_Sidebar_ManageGroupPanel.WBP_Sidebar_ManageGroupPanel_C
        hostTag {tagName=UI.Sidebar.SubPanel}
        padding=<ltrb 0 0 0 0>
       }]
     }
     node {
      name=None
      _structType=&UIStateTreeWidgetSlottingTask
      bCanEditConsideredForCompletion=true
      bConsideredForCompletion=false
      bHasTaskCompletionDelegateDispatcher=false
      bTaskEnabled=true
      transitionHandlingPriority=Normal
      bindingsBatch {value=65535}
      executionRuntimeTemplateIndex {value=65535}
      instanceDataHandle {index=65535, source=None, stateHandle {index=65535}}
      instanceTemplateIndex {value=65535}
      outputBindingsBatch {value=65535}
     }
    }
   }]
  transitions [
   {id=#23, trigger=ON_EVENT, state {id=#8, name=OtherPlayer}}
   {id=#24, trigger=ON_EVENT, state {id=#7, name=LocalPlayer}}
  ]
 }
 AcceptChatInvite {
  id=#6
  type=STATE
  tasks [
   {id=#25, description=None, properties {iD=#26}}
   {
    id=#27
    description=UIState Tree Widget Slotting Task
    properties {
     iD=#28
     instance {
      _structType=&UIStateTreeWidgetSlottingTaskInstanceData
      bIsTaskActive=false
      widgets [{
        bAutoActivateWidget=true
        bCacheInstance=true
        bFill=true
        bIsPlayerSpecific=true
        bIsWidgetEnabled=true
        bKeepConstructed=true
        cacheEntryName=ChatAcceptPanel
        memoryTag=None
        widget=/Sidebar/Panels/Chat/WBP_Sidebar_ChatAcceptPanel.WBP_Sidebar_ChatAcceptPanel_C
        hostTag {tagName=UI.Sidebar.SubPanel}
        padding=<ltrb 0 0 0 0>
       }]
     }
     node {
      name=None
      _structType=&UIStateTreeWidgetSlottingTask
      bCanEditConsideredForCompletion=true
      bConsideredForCompletion=false
      bHasTaskCompletionDelegateDispatcher=false
      bTaskEnabled=true
      transitionHandlingPriority=Normal
      bindingsBatch {value=65535}
      executionRuntimeTemplateIndex {value=65535}
      instanceDataHandle {index=65535, source=None, stateHandle {index=65535}}
      instanceTemplateIndex {value=65535}
      outputBindingsBatch {value=65535}
     }
    }
   }
  ]
  transitions [
   {id=#29, trigger=ON_EVENT, state {id=#8, name=OtherPlayer}}
   {id=#30, trigger=ON_EVENT, state {id=#7, name=LocalPlayer}}
  ]
 }
 LocalPlayer {id=#7, type=LINKED_ASSET, transitions []}
 OtherPlayer {id=#8, type=LINKED_ASSET, transitions []}
]
