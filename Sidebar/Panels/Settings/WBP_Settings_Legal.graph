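# Legal settings panel for the Sidebar, an activatable widget with Legal information and Credits buttons, focusing Legal information by default.
# Each click starts async widget creation with input suspended and requests the sidebar close via FortSidebarVM.
# On completion, Legal info is pushed as a modal, while Credits becomes front-end content without header or chat, also hiding the footer in compact UI.
# On Construct, the Credits button collapses when in zone.
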WidgetBlueprint /Sidebar/Panels/Settings/WBP_Settings_Legal
graphs [
 BP_GetDesiredFocusTarget {nodes [
   Get Desired Focus Target()
   > Return Node!(
    execute
    ReturnValue:Widget Object Reference=Get Button_LegalInformation() -> (Button_LegalInformation:WBP Sidebar Button Generic Object Reference)
   )
  ]}
 EventGraph {nodes [
   "On Clicked (Button_Credits)"!() -> (OutputDelegate, then, Button:Common Button Base Object Reference) {is=title K2Node_ComponentBoundEvent}
   > n1 = CreateWidgetAsync!(
    execute:Exec
    UserWidgetSoftClass:User Widget Soft Class Reference=Get CreditsScreenClass() -> (CreditsScreenClass:User Widget Soft Class Reference)
    OwningPlayer:Player Controller Object Reference=GetOwningPlayer()
    bSuspendInputUntilComplete:bool=true
   ) -> (then:Exec, OnComplete:Exec, OnFailure:Exec, UserWidget:User Widget Object Reference) {is=title K2Node_AsyncAction}
   > RequestCloseSidebar(self:Fort Sidebar VM Object Reference=Get FortSidebarVM() -> (FortSidebarVM:Fort Sidebar VM Object Reference))
   n2 = Branch(execute=@n1.OnComplete, Condition=IsUsingCompactUI())
   n3 = Get FortFrontEndContext!() -> (ReturnValue:Fort Front End Context Object Reference) {is=title K2Node_GetBlueprintContextEx}
   > PushContentWidget_Adv!(
    execute
    self:Fort Front End Context Object Reference=@n3
    Widget:Widget Object Reference=@n1
    bHideHeader:bool=true
    bHideFooter:bool=true
    bHideChatWidget:bool=true
   ) -> (then)
   PushContentWidget_Adv(execute=@n2.else, self=@n3, Widget=@n1, bHideHeader=true, bHideFooter=false, bHideChatWidget=true)
   "On Clicked (Button_LegalInformation)"!() -> (OutputDelegate, then, Button:Common Button Base Object Reference) {is=title K2Node_ComponentBoundEvent}
   > n4 = CreateWidgetAsync(
    UserWidgetSoftClass=Get LegalInfoClass() -> (LegalInfoClass:User Widget Soft Class Reference)
    OwningPlayer=GetOwningPlayer()
    bSuspendInputUntilComplete=true
   )
   > RequestCloseSidebar(self=Get FortSidebarVM())
   TryPushWidgetAsModalPanel!(execute=@n4.OnComplete, PushWidget:User Widget Object Reference=@n4) -> (then)
   Event Construct()
   n5 = Get FortGlobalUIContext() -> (ReturnValue:Fort Global UIContext Object Reference) {is=title K2Node_GetSubsystem}
   > SetVisibility(
    self=Get Button_Credits() -> (Button_Credits:WBP Sidebar Button Generic Object Reference)
    InVisibility=Select(Option 0=Visible, Option 1=Collapsed, Index=IsInZone!(self:Fort Global UIContext Object Reference=@n5) -> (ReturnValue:bool))
   )
  ]}
]
variables [
 CreditsScreenClass {type=Class, default=/FortniteCoreUI/Frontend/Credits/Credits.Credits_C}
 LegalInfoClass {type=Class, default=/Game/UI/Foundation/Legal/LegalInfo.LegalInfo_C}
]
widgets {
 info {
  inheritedWidgetCount=0
  namedSlotCount=1
  parentClass=&CommonActivatableWidget
  rootWidgetClass=/Sidebar/Shared/WBP_Sidebar_PanelLayout.WBP_Sidebar_PanelLayout_C
  widgetCount=4
 }
 widgets [
  PanelLayout {widgetClassPath=/Sidebar/Shared/WBP_Sidebar_PanelLayout, bIsVariable=true, widget=PanelLayout, properties {headingText=Legal}}
  VBox_General {widgetClassPath=&VerticalBox, bIsVariable=true, namedSlotHost=PanelLayout, widget=VBox_General}
  Button_LegalInformation {
   widgetClassPath=/Sidebar/Buttons/WBP_Sidebar_ButtonGeneric
   bIsVariable=true
   parent=VBox_General
   widget=Button_LegalInformation
   properties {minHeightOverride=60, text=Legal information}
   slot {class=VerticalBoxSlot, properties {padding=<ltrb 18 18 18 12>}}
  }
  Button_Credits {
   widgetClassPath=&WBP_Sidebar_ButtonGeneric
   bIsVariable=true
   parent=VBox_General
   widget=Button_Credits
   properties {minHeightOverride=60, text=Credits}
   slot {class=VerticalBoxSlot, properties {padding=<ltrb 18 0 18 0>}}
  }
 ]
}
