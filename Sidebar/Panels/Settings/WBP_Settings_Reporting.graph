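# Activatable sidebar settings panel titled "Report", with buttons to submit a player report and open Community Rules.
# Submit a report opens the player feedback modal; Community Rules executes the CodeOfConduct action on the sidebar settings view model.
# Both buttons then request sidebar close through the sidebar view model, and the report button is the desired focus target.
# Credits and legal info class variables are declared but unused by the graphs shown.
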WidgetBlueprint /Sidebar/Panels/Settings/WBP_Settings_Reporting
graphs [
 BP_GetDesiredFocusTarget {nodes [
   Get Desired Focus Target()
   > Return Node!(execute, ReturnValue:Widget Object Reference=Get Button_Report() -> (Button_Report:WBP Sidebar Button Generic Object Reference))
  ]}
 EventGraph {nodes [
   "On Clicked (Button_Report)"!() -> (OutputDelegate, then, Button:Common Button Base Object Reference) {is=title K2Node_ComponentBoundEvent}
   > n1 = RequestToOpenPlayerFeedbackModal!(execute, Context:User Widget Object Reference=Self-Reference(), UserToReport:Social User Object Reference) -> (then)
   RequestCloseSidebar(execute=[@n2.then, @n1.then], self:Fort Sidebar VM Object Reference=Get FortSidebarVM() -> (FortSidebarVM:Fort Sidebar VM Object Reference))
   "On Clicked (Button_rules)"!() -> (OutputDelegate, then, Button:Common Button Base Object Reference) {is=title K2Node_ComponentBoundEvent}
   > n2 = ExecuteAction!(
    execute
    self:Fort Sidebar Settings VM Object Reference=Get FortSidebarSettingsVM() -> (FortSidebarSettingsVM:Fort Sidebar Settings VM Object Reference)
    SidebarAction:ESidebarAction Enum=CodeOfConduct
   ) -> (then)
  ]}
]
variables [
 CreditsScreenClass {type=Class, default=/FortniteCoreUI/Frontend/Credits/Credits.Credits_C}
 LegalInfoClass {type=Class, default=/Game/UI/Foundation/Legal/LegalInfo.LegalInfo_C}
]
widgets {
 info {
  inheritedWidgetCount=0
  namedSlotCount=1
  parentClass=&CommonActivatableWidget
  rootWidgetClass=/Sidebar/Shared/WBP_Sidebar_PanelLayout.WBP_Sidebar_PanelLayout_C
  widgetCount=4
 }
 widgets [
  PanelLayout {widgetClassPath=/Sidebar/Shared/WBP_Sidebar_PanelLayout, bIsVariable=true, widget=PanelLayout, properties {headingText=Report}}
  VBox_General {widgetClassPath=&VerticalBox, bIsVariable=true, namedSlotHost=PanelLayout, widget=VBox_General}
  Button_Report {
   widgetClassPath=/Sidebar/Buttons/WBP_Sidebar_ButtonGeneric
   bIsVariable=true
   parent=VBox_General
   widget=Button_Report
   properties {
    minHeightOverride=60
    text=Submit a report
    font {
     bForceMonospaced=false
     bMaterialIsStencil=false
     fontMaterial=/FortUILibrary/UIKit/Buttons/MaterialInstances/CTAButtons/MI_UIKit_TextQuiet_Prompt
     fontObject=&HeadingNow
     letterSpacing=10
     monospacedWidth=1
     size=15.75
     skewAmount=0
     typefaceFontName=74 Regular
     outlineSettings {
      bApplyOutlineToDropShadows=false
      bMiteredCorners=false
      bSeparateFillAlpha=false
      outlineMaterial=/FortUILibrary/Systems/ModularButton/MaterialInstances/MI_UIKit_TextOutline
      outlineSize=0
      outlineColor=<rgba 0 0 0 1>
     }
    }
   }
   slot {class=VerticalBoxSlot, properties {padding=<ltrb 18 18 18 12>}}
  }
  Button_rules {
   widgetClassPath=&WBP_Sidebar_ButtonGeneric
   bIsVariable=true
   parent=VBox_General
   widget=Button_rules
   properties {
    icon=/FortUILibrary/Textures/Icons/T_UI_Icon_UIKit_ExternalLink_MSDF
    iconHorizontalAlignment=HAlign_Left
    minHeightOverride=60
    text=Community Rules
    textBlockAlignment=Left
    useIconBlock=true
    font {
     bForceMonospaced=false
     bMaterialIsStencil=false
     fontMaterial=&MI_UIKit_TextQuiet_Prompt
     fontObject=&HeadingNow
     letterSpacing=10
     monospacedWidth=1
     size=15.75
     skewAmount=0
     typefaceFontName=74 Regular
     outlineSettings {
      bApplyOutlineToDropShadows=false
      bMiteredCorners=false
      bSeparateFillAlpha=false
      outlineMaterial=&MI_UIKit_TextOutline
      outlineSize=0
      outlineColor=<rgba 0 0 0 1>
     }
    }
    iconPadding=<ltrb 4 0 0 0>
   }
   slot {class=VerticalBoxSlot, properties {padding=<ltrb 18 0 18 6>}}
  }
 ]
}
