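# Sidebar settings panel for "Support and refunds", an activatable widget with Support and "Return or cancel purchase" buttons inside the shared panel layout.
# Each click executes the Support or PurchaseHistory action on the sidebar settings view model, then requests that the sidebar close.
# The refunds button collapses when the global UI context reports being in zone, and the Support button takes default focus.
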WidgetBlueprint /Sidebar/Panels/Settings/WBP_Settings_SupportAndRefunds
graphs [
 BP_GetDesiredFocusTarget {nodes [
   Get Desired Focus Target()
   > Return Node!(execute, ReturnValue:Widget Object Reference=Get Button_support() -> (Button_support:WBP Sidebar Button Generic Object Reference))
  ]}
 EventGraph {nodes [
   "On Clicked (Button_Refunds)"!() -> (OutputDelegate, then, Button:Common Button Base Object Reference) {is=title K2Node_ComponentBoundEvent}
   > n1 = ExecuteAction!(
    execute
    self:Fort Sidebar Settings VM Object Reference=Get FortSidebarSettingsVM() -> (FortSidebarSettingsVM:Fort Sidebar Settings VM Object Reference)
    SidebarAction:ESidebarAction Enum=PurchaseHistory
   ) -> (then)
   RequestCloseSidebar(execute=[@n2.then, @n1.then], self:Fort Sidebar VM Object Reference=Get FortSidebarVM() -> (FortSidebarVM:Fort Sidebar VM Object Reference))
   "On Clicked (Button_support)"!() -> (OutputDelegate, then, Button:Common Button Base Object Reference) {is=title K2Node_ComponentBoundEvent}
   > n2 = ExecuteAction(self=Get FortSidebarSettingsVM(), SidebarAction=Support)
   Event Construct()
   n3 = Get FortGlobalUIContext() -> (ReturnValue:Fort Global UIContext Object Reference) {is=title K2Node_GetSubsystem}
   > SetVisibility(
    self=Get Button_Refunds() -> (Button_Refunds:WBP Sidebar Button Generic Object Reference)
    InVisibility=Select(Option 0=Visible, Option 1=Collapsed, Index=IsInZone!(self:Fort Global UIContext Object Reference=@n3) -> (ReturnValue:bool))
   )
  ]}
]
variables [
 CreditsScreenClass {type=Class, default=/FortniteCoreUI/Frontend/Credits/Credits.Credits_C}
 LegalInfoClass {type=Class, default=/Game/UI/Foundation/Legal/LegalInfo.LegalInfo_C}
]
widgets {
 info {
  inheritedWidgetCount=0
  namedSlotCount=1
  parentClass=&CommonActivatableWidget
  rootWidgetClass=/Sidebar/Shared/WBP_Sidebar_PanelLayout.WBP_Sidebar_PanelLayout_C
  widgetCount=4
 }
 widgets [
  PanelLayout {widgetClassPath=/Sidebar/Shared/WBP_Sidebar_PanelLayout, bIsVariable=true, widget=PanelLayout, properties {headingText=Support and refunds}}
  VBox_General {widgetClassPath=&VerticalBox, bIsVariable=true, namedSlotHost=PanelLayout, widget=VBox_General}
  Button_support {
   widgetClassPath=/Sidebar/Buttons/WBP_Sidebar_ButtonGeneric
   bIsVariable=true
   parent=VBox_General
   widget=Button_support
   properties {
    icon=/FortUILibrary/Textures/Icons/T_UI_Icon_UIKit_ExternalLink_MSDF
    minHeightOverride=60
    text=Support
    textBlockAlignment=Left
    useIconBlock=true
    iconPadding=<ltrb 4 0 0 0>
   }
   slot {class=VerticalBoxSlot, properties {padding=<ltrb 18 18 18 12>}}
  }
  Button_Refunds {
   widgetClassPath=&WBP_Sidebar_ButtonGeneric
   bIsVariable=true
   parent=VBox_General
   widget=Button_Refunds
   properties {minHeightOverride=60, text=Return or cancel purchase}
   slot {class=VerticalBoxSlot, properties {padding=<ltrb 18 0 18 0>}}
  }
 ]
}
