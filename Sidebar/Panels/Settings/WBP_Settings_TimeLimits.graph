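# Sidebar settings panel showing a player's parental time limits and reports for Fortnite and UEFN.
# Update functions format PlayerPlaytimeVM time left, time used, windows and report status, with switchers choosing combined or separate play/create views.
# Ask and add buttons call the view model for requests or PIN grants, while request status swaps sending and sent/failed pages.
# The Thyme runtime option selects alternate copy and buttons, including a request dialog.
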WidgetBlueprint /Sidebar/Panels/Settings/WBP_Settings_TimeLimits
graphs [
 BP_GetDesiredFocusTarget {nodes [
   Get Desired Focus Target()
   > n1 = Branch(Condition=Get bIsEnable Thyme() -> (bIsEnable Thyme:bool))
   > n2 = Branch(Condition=Get bIsEnabled!(self:Widget Object Reference=Get Button_Thyme_AskMoreTime() -> (Button_Thyme_AskMoreTime:WBP Sidebar Button Generic Object Reference)) -> (bIsEnabled:bool))
   > Return Node!(execute, ReturnValue:Widget Object Reference=Get Button_Thyme_AskMoreTime())
   Branch(
    execute=@n2.else
    Condition=IsVisible(self=Get Button_Thyme_AddMoreTime() -> (Button_Thyme_AddMoreTime:WBP Sidebar Button Generic Object Reference))
   )
   > Return Node(ReturnValue=Get Button_Thyme_AddMoreTime())
   n3 = Branch(
    execute=@n1.else
    Condition=Get bIsEnabled(self=Get Button_AskMoreTime() -> (Button_AskMoreTime:WBP Sidebar Button Generic Object Reference))
   )
   > Return Node(ReturnValue=Get Button_AskMoreTime())
   Branch(execute=@n3.else, Condition=IsVisible(self=Get Button_AddMoreTime() -> (Button_AddMoreTime:WBP Sidebar Button Generic Object Reference)))
   > Return Node(ReturnValue=Get Button_AddMoreTime())
  ]}
 ConvertTimeToTimeLeftText {nodes [
   n1 = ConvertTimeToTimeLeftText!() -> (then, TimeSpan:Timespan Structure) {is=title K2Node_FunctionEntry}
   n2 = Format Text!(
    Format="LOCTEXT(\"{Time} left\")"
    Time:Text=TextToLower!(InText:"Text (by ref)"=GetTimespanAsText!(
      Timespan:"Timespan Structure (by ref)"=@n1
      bShowHoursWhenZero:bool=false
      bShowSeconds:bool=false
      FormatType:ETimespanAsTextFormat Enum=DaysHoursMinutesSeconds
      UnitDisplayType:ETimespanUnitDisplayFormat Enum=Full
      bUseSentenceCase:bool=false
     ) -> (ReturnValue:Text)) -> (ReturnValue:Text)
   ) -> (Result)
   n3 = "Timespan < Timespan"(
    A:Timespan Structure=Get TimeSpan() -> (TimeSpan:Timespan Structure)
    B:Timespan Structure=MakeTimespan!(Days:int=1, Hours:int=0, Minutes:int=0, Seconds:int=0, Milliseconds:int=0) -> (ReturnValue:Timespan Structure)
   ) -> (ReturnValue:bool=~) {is=title K2Node_PromotableOperator}
   n4 = MakeTimespan(Days=0, Hours=0, Minutes=0, Seconds=0, Milliseconds=0)
   n5 = Get TimeSpan()
   > Return Node(ReturnValue:Text=SelectText(
     A="LOCTEXT(\"No time left\")"
     B=SelectText(A=@n2, B="LOCTEXT(\"Time left pending\")", bPickA=@n3)
     bPickA="Timespan <= Timespan"(A:Timespan Structure=@n5, B:Timespan Structure=@n4) -> (ReturnValue:bool=~) {is=title K2Node_PromotableOperator}
    ))
  ]}
 EventGraph {nodes [
   "On Clicked (Button_Thyme_AskMoreTime)"!() -> (OutputDelegate, then, Button:Common Button Base Object Reference) {is=title K2Node_ComponentBoundEvent}
   > n1 = Branch(Condition=IsInSubGame!(SubGame:ESubGame Enum=Campaign) -> (ReturnValue:bool))
   > n2 = DeactivateWidget()
   OpenRequestMoreTimeDialog!(
    execute=[@n2.then, @n1.else]
    self:Player Playtime View Model Object Reference=Get PlayerPlaytimeVM() -> (PlayerPlaytimeVM:Player Playtime View Model Object Reference)
    ContextWidget:User Widget Object Reference=Self-Reference()
   ) -> (then)
   > Fire Event Element Interacted(Element=Get Button_Thyme_AskMoreTime())
   n3 = "On Clicked (Button_Thyme_AddMoreTime)"!() -> (OutputDelegate, then, Button:Common Button Base Object Reference) {is=title K2Node_ComponentBoundEvent}
   n4 = Branch(execute=[@n3.then, @n6.then], Condition=IsInSubGame(SubGame=Campaign))
   > n5 = DeactivateWidget()
   OpenPinTimeGrant!(
    execute=[@n5.then, @n4.else]
    self:Player Playtime View Model Object Reference=Get PlayerPlaytimeVM()
    Context:User Widget Object Reference=Self-Reference()
   ) -> (then)
   > Fire Event Element Interacted(Element=Select(
     Option 0:Widget Object Reference=Get Button_AddMoreTime()
     Option 1:Widget Object Reference=Get Button_Thyme_AddMoreTime()
     Index=Get bIsEnable Thyme()
    ) -> (ReturnValue:Widget Object Reference))
   "On Clicked (Button_AskMoreTime)"!() -> (OutputDelegate, then, Button:Common Button Base Object Reference) {is=title K2Node_ComponentBoundEvent}
   > RequestAdditionalPlaytime!(execute, self:Player Playtime View Model Object Reference=Get PlayerPlaytimeVM(), ActivityType:EActivityLimitType Enum=CombinedTime) -> (then)
   > SetIsEnabled(self=Get Button_AskMoreTime(), bInIsEnabled=false)
   > Fire Event Element Interacted(Element=Get Button_AskMoreTime())
   n6 = "On Clicked (Button_AddMoreTime)"!() -> (OutputDelegate, then, Button:Common Button Base Object Reference) {is=title K2Node_ComponentBoundEvent}
   "On Clicked (Button_Done)"!() -> (OutputDelegate, then, Button:Common Button Base Object Reference) {is=title K2Node_ComponentBoundEvent}
   > SetActiveState!(execute, self, State:Widget Object Reference=Get Box_ContentFirstPage() -> (Box_ContentFirstPage:Vertical Box Object Reference)) -> (then)
   n7 = On Request Additional Time!() -> (OutputDelegate, then, RequestStatus:EPlaytimeRequestState Enum=None) {is=title K2Node_CustomEvent}
   > n8 = Switch on EPlaytimeRequestState!(execute, Selection:EPlaytimeRequestState Enum=@n7.RequestStatus) -> (None:Exec, Waiting:Exec, Success:Exec, Failure:Exec) {is=title K2Node_SwitchEnum}
   SetActiveState(execute=@n8.Failure, State=Get Box_ContentEmailSent() -> (Box_ContentEmailSent:Vertical Box Object Reference))
   n9 = Get Image_Tick() -> (Image_Tick:Image Object Reference)
   > SetVisibility(self=@n9, InVisibility=Collapsed)
   n10 = Get Text_EmailDescription() -> (Text_EmailDescription:Common Text Object Reference)
   > SetVisibility(self=@n10, InVisibility=Collapsed)
   n11 = Get Text_EmailSent() -> (Text_EmailSent:Common Text Object Reference)
   > Set Text(Text="LOCTEXT(\"Email failed\")", self=@n11)
   n12 = GetOwningPlayer()
   > n13 = Cast To FortPlayerController(Object=@n12)
   > "Fire Event Sub View Opened (Non-Localized)"!(
    execute
    FortPC:Fort Player Controller Object Reference=@n13
    ViewName:String=PlaytimeRequestEmailFailed
    AdditionalInfo:Flow Analytics Additional Info Structure
    bIsOverlay:bool=false
   ) -> (then)
   SetActiveState(execute=@n8.Success, State=Get Box_ContentEmailSent())
   > SetVisibility(self=@n9, InVisibility=SelfHitTestInvisible)
   > SetVisibility(self=@n10, InVisibility=SelfHitTestInvisible)
   > Set Text(Text="LOCTEXT(\"Email sent\")", self=@n11)
   > n14 = Cast To FortPlayerController(Object=@n12)
   > "Fire Event Sub View Opened (Non-Localized)"(FortPC=@n14, ViewName=PlaytimeRequestEmailSent, bIsOverlay=false)
   SetActiveState(execute=@n8.Waiting, State=Get Box_EmailSending() -> (Box_EmailSending:Vertical Box Object Reference))
   Event Construct()
   > SetScrollWidget!(execute, ToScroll:Widget Object Reference=Get SB_TimeInfo() -> (SB_TimeInfo:Scroll Box Object Reference)) -> (then)
   n15 = Create Event()
   n16 = Get PlayerPlaytimeVM()
   > Assign On Request Additional Playtime Status Update!(execute:Exec, self:Player Playtime View Model Object Reference=@n16, Delegate:"Delegate (by ref)"=@n15) -> (then:Exec) {is=title K2Node_AssignDelegate}
   Event OnInitialized()
   > SetActiveState(State=Get Box_ContentFirstPage())
   Event Destruct()
   > Unbind Event from On Request Additional Playtime Status Update(self:Player Playtime View Model Object Reference=@n16, Delegate=@n15) {is=title K2Node_RemoveDelegate}
   Event On Activated()
   > Init(self:Player Playtime View Model Object Reference=Get PlayerPlaytimeVM())
   > SetListenToRefresh!(execute, self:Player Playtime View Model Object Reference=Get PlayerPlaytimeVM(), bListenToRefresh:bool=true) -> (then)
   > SetInitialTimeInfoSwitcherState()
   Event On Deactivated()
   > SetListenToRefresh(self=Get PlayerPlaytimeVM(), bListenToRefresh=false)
   Event PreConstruct()
   > Set IsUsingCompactUI(IsUsingCompactUI:bool=IsUsingCompactUI()) -> (Output_Get:bool)
   > Set bIsEnable Thyme(bIsEnable Thyme:bool=Get bEnableThyme!(self:Fort Runtime Options Object Reference=GetRuntimeOptions()) -> (bEnableThyme:bool)) -> (Output_Get:bool)
   > UpdateVisual()
  ]}
 GetCombinedTime {nodes [
   n1 = GetCombinedTime!() -> (then, PlayTime:Timespan Structure, CreateTime:Timespan Structure) {is=title K2Node_FunctionEntry}
   > Return Node!(
    execute
    CombinedTime:Timespan Structure=Timespan + Timespan(A:Timespan Structure=@n1, B:Timespan Structure=@n1.CreateTime) -> (ReturnValue:Timespan Structure) {is=title K2Node_PromotableOperator}
   )
  ]}
 GetCombinedTimeWindowsVisibility {nodes [
   n1 = GetCombinedTimeWindowsVisibility!() -> (then, HasWindowBounds:bool=~) {is=title K2Node_FunctionEntry}
   n2 = Get PlayerPlaytimeVM()
   n3 = Select(Option 0:ESlateVisibility Enum=Collapsed, Option 1:ESlateVisibility Enum=SelfHitTestInvisible, Index=@n1) -> (ReturnValue:ESlateVisibility Enum)
   > Return Node!(
    execute
    Visibility:ESlateVisibility Enum=Select!(
     None:ESlateVisibility Enum=Collapsed
     AllowedTimeOnly:ESlateVisibility Enum=SelfHitTestInvisible
     WindowsOnly:ESlateVisibility Enum=@n3
     All:ESlateVisibility Enum=@n3
     Index:EPlaytimeActiveScheduling Enum=GetPlaytimeScheduling!(self:Player Playtime View Model Object Reference=@n2, ActivityType:EActivityLimitType Enum=CombinedTime) -> (ReturnValue:EPlaytimeActiveScheduling Enum=None)
    ) -> (ReturnValue)
   )
  ]}
 GetCreateTimeWindowsVisibility {nodes [
   n1 = GetCreateTimeWindowsVisibility!() -> (then, HasCreateWindowBounds:bool=~) {is=title K2Node_FunctionEntry}
   n2 = Get PlayerPlaytimeVM()
   n3 = Select!(Option 0:ESlateVisibility Enum=Collapsed, Option 1:ESlateVisibility Enum=SelfHitTestInvisible, Index:bool=@n1) -> (ReturnValue)
   > Return Node(Visibility=Select!(
     None:ESlateVisibility Enum=Collapsed
     AllowedTimeOnly:ESlateVisibility Enum=SelfHitTestInvisible
     WindowsOnly:ESlateVisibility Enum=@n3
     All:ESlateVisibility Enum=@n3
     Index:EPlaytimeActiveScheduling Enum=GetPlaytimeScheduling(self=@n2, ActivityType=ActivityCreate) -> (ReturnValue=None)
    ) -> (ReturnValue))
  ]}
 GetDisclaimerText {nodes [
   n1 = GetDisclaimerText!() -> (then, RemainingPlayTime:Timespan Structure, RemainingCreateTime:Timespan Structure) {is=title K2Node_FunctionEntry}
   > Set L_RemainingPlayTimeMinutes(L_RemainingPlayTimeMinutes:double=Get Total Minutes!(A:Timespan Structure=@n1) -> (ReturnValue:double) {is=title K2Node_CallFunction}) -> (Output_Get:double)
   > Set L_RemainingCreateTimeMinutes(L_RemainingCreateTimeMinutes:double=Get Total Minutes(A=@n1.RemainingCreateTime)) -> (Output_Get:double)
   n2 = "float > float"(A=Get L_RemainingPlayTimeMinutes() -> (L_RemainingPlayTimeMinutes:double), B=30)
   n3 = "float > float"(A=Get L_RemainingCreateTimeMinutes() -> (L_RemainingCreateTimeMinutes:double), B=30)
   n4 = AND Boolean(A="float > float"(A=Get L_RemainingPlayTimeMinutes(), B=30), B="float < float"(A=Get L_RemainingCreateTimeMinutes(), B=30))
   n5 = Get L_RemainingPlayTimeMinutes()
   n6 = Get L_RemainingCreateTimeMinutes()
   n7 = Select!(
    Option 0:Text
    Option 1:Text="LOCTEXT(\"You can ask your parent or guardian for more create time when you have less than 30 minutes left.\")"
    Index:bool=AND Boolean(A="float < float"(A=@n5, B=30), B="float > float"(A=@n6, B=30))
   ) -> (ReturnValue:Text)
   > Return Node!(
    execute
    Text:Text=Select(
     Option 0=Select(
      Option 0=@n7
      Option 1="LOCTEXT(\"You can ask your parent or guardian for more play time when you have less than 30 minutes left.\")"
      Index=@n4
     )
     Option 1="LOCTEXT(\"You can ask your parent or guardian for more time when you have less than 30 minutes left.\")"
     Index=AND Boolean(A=@n2, B=@n3)
    )
   )
  ]}
 GetDisclaimerVisibility {nodes [
   n1 = GetDisclaimerVisibility!() -> (then, RemainingPlayTime:Timespan Structure, RemainingCreateTime:Timespan Structure) {is=title K2Node_FunctionEntry}
   n2 = "float < float"(A=Get Total Minutes(A=@n1.RemainingCreateTime), B=30)
   n3 = "float < float"(A=Get Total Minutes(A=@n1), B=30)
   > Return Node!(
    execute
    Visibility:ESlateVisibility Enum="To Visibility (Boolean)"!(
     bIsVisible:bool=AND Boolean(A=@n3, B=@n2)
     TrueVisibility:ESlateVisibility Enum=Collapsed
     FalseVisibility:ESlateVisibility Enum=SelfHitTestInvisible
    ) -> (ReturnValue:ESlateVisibility Enum=Visible)
   )
  ]}
 GetPlayTimeWindowsVisibility {nodes [
   n1 = GetPlayTimeWindowsVisibility!() -> (then, HasPlayWindowBounds:bool=~) {is=title K2Node_FunctionEntry}
   n2 = Get PlayerPlaytimeVM()
   n3 = Select(Option 0:ESlateVisibility Enum=Collapsed, Option 1:ESlateVisibility Enum=SelfHitTestInvisible, Index=@n1) -> (ReturnValue:ESlateVisibility Enum)
   > Return Node(Visibility=Select!(
     None:ESlateVisibility Enum=Collapsed
     AllowedTimeOnly:ESlateVisibility Enum=SelfHitTestInvisible
     WindowsOnly:ESlateVisibility Enum=@n3
     All:ESlateVisibility Enum=@n3
     Index:EPlaytimeActiveScheduling Enum=GetPlaytimeScheduling(self=@n2, ActivityType=ActivityPlay) -> (ReturnValue=None)
    ) -> (ReturnValue))
  ]}
 GetThymeButtonsVisibility {nodes [
   n1 = GetThymeButtonsVisibility!() -> (then, RemainingPlayTime:Timespan Structure, RemainingCreateTime:Timespan Structure) {is=title K2Node_FunctionEntry}
   n2 = "float > float"(A=Get Total Minutes(A=@n1), B=30)
   n3 = "float > float"(A=Get Total Minutes(A=@n1.RemainingCreateTime), B=30)
   > Return Node(Visibility="To Visibility (Boolean)"(bIsVisible=AND Boolean(A=@n2, B=@n3), TrueVisibility=Collapsed, FalseVisibility=SelfHitTestInvisible) -> (ReturnValue=Visible))
  ]}
 OnCombitedTimeWindowBoundsUpdated {nodes [
   n1 = OnCombitedTimeWindowBoundsUpdated!() -> (then, PlaytimeWindow:Playtime Window Structure) {is=title K2Node_FunctionEntry}
   > SetTimeWindowBoundsUpdated!(
    execute
    self
    PlaytimeWindow:Playtime Window Structure=@n1
    CommonTextObject:Common Text Object Reference=Get Text_WindowBounds() -> (Text_WindowBounds:Common Text Object Reference)
   ) -> (then)
  ]}
 OnCreateTimePlayedUpdated {nodes [
   n1 = OnCreateTimePlayedUpdated!() -> (then, TimeSpan:Timespan Structure) {is=title K2Node_FunctionEntry}
   > SetTimePlayedText!(
    execute
    self
    TimeSpan:Timespan Structure=@n1
    CommonTextObject:Common Text Object Reference=Get Text_CreatetimePlayed() -> (Text_CreatetimePlayed:Common Text Object Reference)
   ) -> (then)
  ]}
 OnCreateTimeRemainingUpdated {nodes [
   OnCreateTimeRemainingUpdated!() -> (then, TimeSpan:Timespan Structure) {is=title K2Node_FunctionEntry}
   n1 = Get PlayerPlaytimeVM()
   > n2 = Branch(Condition=AND Boolean(
     A=IsPlaytimeLimitsEnabled!(self:Player Playtime View Model Object Reference=@n1) -> (ReturnValue:bool)
     B=IsActivitySpecificTimeEnabled!(self:Player Playtime View Model Object Reference=@n1) -> (ReturnValue:bool)
    ))
   > "SetText (Text)"(
    self=Get Text_CreatetimeLeft() -> (Text_CreatetimeLeft:Common Text Object Reference)
    InText=ConvertTimeToTimeLeftText!(TimeSpan:Timespan Structure=Get TimeSpan()) -> (ReturnValue:Text) {is=title K2Node_CallFunction}
   )
   > SetActiveWidget(
    self:Widget Switcher Object Reference=Get CreateTimeLimitSwitcher() -> (CreateTimeLimitSwitcher:Common Activatable Widget Switcher Object Reference)
    Widget=Get Box_CreateTimeLimitOn() -> (Box_CreateTimeLimitOn:Vertical Box Object Reference)
   )
   SetActiveWidget(
    execute=@n2.else
    self=Get CreateTimeLimitSwitcher()
    Widget=Get Box_CreateTimeLimitOff() -> (Box_CreateTimeLimitOff:Vertical Box Object Reference)
   )
  ]}
 OnCreateTimeWindowBoundsUpdated {nodes [
   n1 = OnCreateTimeWindowBoundsUpdated!() -> (then, PlaytimeWindow:Playtime Window Structure) {is=title K2Node_FunctionEntry}
   > SetTimeWindowBoundsUpdated(PlaytimeWindow=@n1, CommonTextObject=Get Text_CreateTime_WindowBounds() -> (Text_CreateTime_WindowBounds:Common Text Object Reference))
  ]}
 OnPlayTimePlayedUpdated {nodes [
   n1 = OnPlayTimePlayedUpdated!() -> (then, TimeSpan:Timespan Structure) {is=title K2Node_FunctionEntry}
   > SetTimePlayedText(TimeSpan=@n1, CommonTextObject=Get Text_PlaytimeLimitsPlayed() -> (Text_PlaytimeLimitsPlayed:Common Text Object Reference))
  ]}
 OnPlayTimeRemainingUpdated {nodes [
   OnPlayTimeRemainingUpdated!() -> (then, TimeSpan:Timespan Structure) {is=title K2Node_FunctionEntry}
   n1 = Get PlayerPlaytimeVM()
   > n2 = Branch(Condition=AND Boolean(A=IsPlaytimeLimitsEnabled(self=@n1), B=IsActivitySpecificTimeEnabled(self=@n1)))
   > "SetText (Text)"(
    self=Get Text_PlaytimeLimitsLeft() -> (Text_PlaytimeLimitsLeft:Common Text Object Reference)
    InText=ConvertTimeToTimeLeftText(TimeSpan=Get TimeSpan())
   )
   > SetActiveWidget(
    self=Get PlayTimeLimitSwitcher() -> (PlayTimeLimitSwitcher:Common Activatable Widget Switcher Object Reference)
    Widget=Get Box_PlayTimeLimitOn() -> (Box_PlayTimeLimitOn:Vertical Box Object Reference)
   )
   SetActiveWidget(
    execute=@n2.else
    self=Get PlayTimeLimitSwitcher()
    Widget=Get Box_PlayTimeLimitOff() -> (Box_PlayTimeLimitOff:Vertical Box Object Reference)
   )
  ]}
 OnPlayTimeWindowBoundsUpdated {nodes [
   n1 = OnPlayTimeWindowBoundsUpdated!() -> (then, PlaytimeWindow:Playtime Window Structure) {is=title K2Node_FunctionEntry}
   > SetTimeWindowBoundsUpdated(
    PlaytimeWindow=@n1
    CommonTextObject=Get Text_PlayTimeLimits_WindowBounds() -> (Text_PlayTimeLimits_WindowBounds:Common Text Object Reference)
   )
  ]}
 OnReportStatusUpdated {nodes [
   n1 = OnReportStatusUpdated!() -> (then, NewStatus:bool=~) {is=title K2Node_FunctionEntry}
   n2 = MakeLiteralText(Value="LOCTEXT(\"off\")")
   n3 = MakeLiteralText(Value="LOCTEXT(\"on\")")
   > "SetText (Text)"(
    self=Get Text_ReportStatus() -> (Text_ReportStatus:Common Text Object Reference)
    InText=Format Text!(Format="LOCTEXT(\"Time Reports {Value}\")", Value:Text=SelectText(A=@n3, B=@n2, bPickA=@n1)) -> (Result)
   )
   > SetVisibility(
    self=Get Text_ReportDescription() -> (Text_ReportDescription:Common Text Object Reference)
    InVisibility=Select!(Option 0:ESlateVisibility Enum=Hidden, Option 1:ESlateVisibility Enum=HitTestInvisible, Index:bool=@n1) -> (ReturnValue)
   )
  ]}
 OnTimePlayedUpdated {nodes [
   n1 = OnTimePlayedUpdated!() -> (then, TimeSpan:Timespan Structure) {is=title K2Node_FunctionEntry}
   > SetTimePlayedText(TimeSpan=@n1, CommonTextObject=Get Text_PlaytimePlayed() -> (Text_PlaytimePlayed:Common Text Object Reference))
  ]}
 OnTimeRemainingUpdated {nodes [
   n1 = OnTimeRemainingUpdated!() -> (then, TimeSpan:Timespan Structure) {is=title K2Node_FunctionEntry}
   > n2 = Branch(Condition=IsPlaytimeLimitsEnabled(self=Get PlayerPlaytimeVM()))
   n3 = TextToLower(InText=GetTimespanAsText(Timespan=@n1, bShowHoursWhenZero=false, bShowSeconds=false, FormatType=DaysHoursMinutesSeconds, UnitDisplayType=Full, bUseSentenceCase=false))
   n4 = MakeTimespan(Days=1, Hours=0, Minutes=0, Seconds=0, Milliseconds=0)
   > "SetText (Text)"(
    self=Get Text_PlaytimeLeft() -> (Text_PlaytimeLeft:Common Text Object Reference)
    InText=SelectText(
     A=Format Text!(Format="LOCTEXT(\"{Time} left\")", Time:Text=@n3) -> (Result)
     B="LOCTEXT(\"Time left pending\")"
     bPickA="Timespan < Timespan"(A=@n1, B=@n4) -> (ReturnValue=~)
    )
   )
   > SetActiveWidget(
    self=Get TimeLimitSwitcher() -> (TimeLimitSwitcher:Common Activatable Widget Switcher Object Reference)
    Widget=Get Box_TimeLimitOn() -> (Box_TimeLimitOn:Vertical Box Object Reference)
   )
   > SetVisibility(self=Get Buttons_Switcher() -> (Buttons_Switcher:Common Activatable Widget Switcher Object Reference), InVisibility=Visible)
   SetActiveWidget(execute=@n2.else, self=Get TimeLimitSwitcher(), Widget=Get Box_TimeLimitOff() -> (Box_TimeLimitOff:Vertical Box Object Reference))
   > SetVisibility(self=Get Buttons_Switcher(), InVisibility=Collapsed)
  ]}
 Set Window Bounds {nodes [
   n1 = Set Window Bounds!() -> (then, Bounds_StartTime:Date Time Structure, Bounds_EndTime:Date Time Structure, TextObject:Common Text Object Reference) {is=title K2Node_FunctionEntry}
   > Set BeginTime(BeginTime:Text=ConvertServiceTimeToLocalTimeText!(self:Player Playtime View Model Object Reference=Get PlayerPlaytimeVM(), ServiceTime:"Date Time Structure (by ref)"=@n1) -> (ReturnValue:Text)) -> (Output_Get:Text)
   > Set EndTime(EndTime:Text=ConvertServiceTimeToLocalTimeText(self=Get PlayerPlaytimeVM(), ServiceTime=@n1.Bounds_EndTime)) -> (Output_Get:Text)
   n2 = "Equal Exactly (Text)"!(A:"Text (by ref)"=Get BeginTime() -> (BeginTime:Text), B:"Text (by ref)"=Get EndTime() -> (EndTime:Text)) -> (ReturnValue:bool)
   n3 = Format Text!(Format="LOCTEXT(\"Windows: {Begin} - {End}\")", Begin:Text=Get BeginTime(), End:Text=Get EndTime()) -> (Result)
   n4 = Get BeginTime()
   n5 = Get EndTime()
   > "SetText (Text)"(
    self=Get TextObject() -> (TextObject:Common Text Object Reference)
    InText=SelectText(
     A=SelectText(B=@n3, bPickA=@n2)
     B=Format Text(Format="LOCTEXT(\"Time Window: {Begin} - {End}\")", Begin=@n4, End=@n5)
     bPickA=Get bIsEnable Thyme()
    )
   )
  ]}
 SetActiveState {nodes [
   n1 = SetActiveState!() -> (then, State:Widget Object Reference) {is=title K2Node_FunctionEntry}
   > n2 = For Each Loop(Array:Array of Widget Object References=GetAllChildren(self=Get StateContentsVbox() -> (StateContentsVbox:Vertical Box Object Reference))) -> (Array Element:Widget Object Reference)
   > SetVisibility(self=@n2, InVisibility=Select(Option 0=Collapsed, Option 1=Visible, Index="Equal (Object)"(A=@n2, B=@n1)))
  ]}
 SetInitialTimeInfoSwitcherState {nodes [
   SetInitialTimeInfoSwitcherState() {is=title K2Node_FunctionEntry}
   n1 = Get PlayerPlaytimeVM()
   > SetActiveWidget(
    self=Get PlayTimeInfo_Switcher() -> (PlayTimeInfo_Switcher:Common Activatable Widget Switcher Object Reference)
    Widget=Select(
     Option 0:Widget Object Reference=Get CombinedTimeInfo() -> (CombinedTimeInfo:Vertical Box Object Reference)
     Option 1:Widget Object Reference=Get SeparatedTimeInfo() -> (SeparatedTimeInfo:Vertical Box Object Reference)
     Index=IsActivitySpecificTimeEnabled(self=@n1)
    ) -> (ReturnValue:Widget Object Reference)
   )
  ]}
 SetTimePlayedText {nodes [
   SetTimePlayedText!() -> (then, TimeSpan:Timespan Structure, CommonTextObject:Common Text Object Reference) {is=title K2Node_FunctionEntry}
   n1 = "Timespan > Timespan"(A:Timespan Structure=Get TimeSpan(), B:Timespan Structure=MakeTimespan(Days=0, Hours=0, Minutes=0, Seconds=0, Milliseconds=0)) -> (ReturnValue:bool=~) {is=title K2Node_PromotableOperator}
   n2 = "Timespan <= Timespan"(A=Get TimeSpan(), B=MakeTimespan(Days=1, Hours=0, Minutes=0, Seconds=0, Milliseconds=0)) -> (ReturnValue=~)
   n3 = TextToLower(InText=GetTimespanAsText(
     Timespan=Get TimeSpan()
     bShowHoursWhenZero=false
     bShowSeconds=false
     FormatType=DaysHoursMinutesSeconds
     UnitDisplayType=Full
     bUseSentenceCase=false
    ))
   > "SetText (Text)"(
    self=Get CommonTextObject() -> (CommonTextObject:Common Text Object Reference)
    InText=SelectText(
     A=Format Text!(Format="LOCTEXT(\"{Time} used\")", Time:Text=@n3) -> (Result)
     B="LOCTEXT(\"Time used pending\")"
     bPickA=AND Boolean(A=@n1, B=@n2)
    )
   )
  ]}
 SetTimeWindowBoundsUpdated {nodes [
   n1 = SetTimeWindowBoundsUpdated!() -> (then, PlaytimeWindow:Playtime Window Structure, CommonTextObject:Common Text Object Reference) {is=title K2Node_FunctionEntry}
   > Set Window Bounds!(execute, self, Bounds:Playtime Window Structure=@n1, TextObject:Common Text Object Reference=Get CommonTextObject()) -> (then) {is=title K2Node_CallFunction}
  ]}
 UpdateAddMoreTimeButton {nodes [
   n1 = UpdateAddMoreTimeButton!() -> (then, ActivityToShow:EActivityToShow Enum=None) {is=title K2Node_FunctionEntry}
   > SetText!(
    execute
    self:WBP UIKit Button Generic Base Object Reference=Get Button_AddMoreTime()
    Text:Text=Select!(
     None:Text="LOCTABLE(\"/Cilantro/RichTextStyles/StringTable_TimeLimitTexts\", \"AddMoreTime\")"
     CombinedTime:Text="LOCTABLE(\"&StringTable_TimeLimitTexts\", \"AddMoreTime\")"
     ActivityPlay:Text="LOCTABLE(\"&StringTable_TimeLimitTexts\", \"AddPlayTime\")"
     ActivityCreate:Text="LOCTABLE(\"&StringTable_TimeLimitTexts\", \"AddMoreCreateTime\")"
     UserSelection:Text="LOCTABLE(\"&StringTable_TimeLimitTexts\", \"AddMoreTime\")"
     Index:EActivityToShow Enum=@n1
    ) -> (ReturnValue:Text)
   ) -> (then)
   > SetActiveWidget(
    self=Get Buttons_Switcher()
    Widget=Select!(
     Option 0:Widget Object Reference=Get Container_ButtonsUnavailable() -> (Container_ButtonsUnavailable:Horizontal Box Object Reference)
     Option 1:Widget Object Reference=Get Container_Buttons() -> (Container_Buttons:Vertical Box Object Reference)
     Index:bool="Not Equal (Enum)"!(A:EActivityToShow Enum=@n1, B:EActivityToShow Enum=None) -> (ReturnValue:bool)
    ) -> (ReturnValue:Widget Object Reference)
   )
  ]}
 UpdateAddMoreTimeButton_Thyme {nodes [
   n1 = UpdateAddMoreTimeButton_Thyme!() -> (then, ActivityToShow:EActivityToShow Enum=None) {is=title K2Node_FunctionEntry}
   > SetText(
    self=Get Button_Thyme_AddMoreTime()
    Text=Select!(
     None:Text="LOCTABLE(\"&StringTable_TimeLimitTexts\", \"AddMoreTime\")"
     CombinedTime:Text="LOCTABLE(\"&StringTable_TimeLimitTexts\", \"AddMoreTime\")"
     ActivityPlay:Text="LOCTABLE(\"&StringTable_TimeLimitTexts\", \"AddPlayTime\")"
     ActivityCreate:Text="LOCTABLE(\"&StringTable_TimeLimitTexts\", \"AddMoreCreateTime\")"
     UserSelection:Text="LOCTABLE(\"&StringTable_TimeLimitTexts\", \"AddMoreTime\")"
     Index:EActivityToShow Enum=@n1
    ) -> (ReturnValue:Text)
   )
   n2 = "Not Equal (Enum)"(A=@n1, B=None)
   n3 = Get bIsEnable Thyme()
   > SetSize!(
    execute
    self:Vertical Box Slot Object Reference=SlotAsVerticalBoxSlot(Widget=Get VerticalBox_0() -> (VerticalBox_0:Vertical Box Object Reference))
    InSize_Value:float=1
    InSize_SizeRule:ESlateSizeRule Enum=Select!(Option 0:ESlateSizeRule Enum=Fill, Option 1:ESlateSizeRule Enum=Automatic, Index:bool=AND Boolean(A=@n3, B=@n2)) -> (ReturnValue:ESlateSizeRule Enum)
   ) -> (then)
  ]}
 UpdateAskMoreTimeButton {nodes [
   n1 = UpdateAskMoreTimeButton!() -> (then, ActivityToShow:EActivityToShow Enum=None) {is=title K2Node_FunctionEntry}
   > SetText(
    self=Get Button_AskMoreTime()
    Text=Select!(
     None:Text="LOCTABLE(\"&StringTable_TimeLimitTexts\", \"AskMoreTime\")"
     CombinedTime:Text="LOCTABLE(\"&StringTable_TimeLimitTexts\", \"AskMoreTime\")"
     ActivityPlay:Text="LOCTABLE(\"&StringTable_TimeLimitTexts\", \"AskPlayTime\")"
     ActivityCreate:Text="LOCTABLE(\"&StringTable_TimeLimitTexts\", \"AskCreateTime\")"
     UserSelection:Text="LOCTABLE(\"&StringTable_TimeLimitTexts\", \"AskMoreTime\")"
     Index:EActivityToShow Enum=@n1
    ) -> (ReturnValue:Text)
   )
   > n2 = Switch on EActivityToShow!(execute, Selection:EActivityToShow Enum=@n1) -> (None:Exec, CombinedTime:Exec, ActivityPlay:Exec, ActivityCreate:Exec, UserSelection:Exec) {is=title K2Node_SwitchEnum}
   n3 = CheckBlockedStateForUserSelection!(self:Player Playtime View Model Object Reference=Get PlayerPlaytimeVM()) -> (ReturnValue:EActivityToShow Enum=None)
   n4 = Get Button_AskMoreTime()
   SetText(
    execute=@n2.UserSelection
    self=@n4
    Text=Select(
     None="LOCTABLE(\"&StringTable_TimeLimitTexts\", \"AskMoreTime\")"
     CombinedTime="LOCTABLE(\"&StringTable_TimeLimitTexts\", \"AskMoreTime\")"
     ActivityPlay="LOCTABLE(\"&StringTable_TimeLimitTexts\", \"AskPlayTime\")"
     ActivityCreate="LOCTABLE(\"&StringTable_TimeLimitTexts\", \"AskCreateTime\")"
     UserSelection="LOCTABLE(\"&StringTable_TimeLimitTexts\", \"AskMoreTime\")"
     Index=@n3
    )
   )
   > Set bIsEnabled!(execute, bIsEnabled:bool="Not Equal (Enum)"(A=@n3, B=None), self:Widget Object Reference=@n4) -> (then, Output_Get:bool)
   n5 = Get PlayerPlaytimeVM()
   Set bIsEnabled(
    execute=[@n2.ActivityPlay, @n2.CombinedTime]
    bIsEnabled=NOT Boolean(A=IsAskForMorePlayTimeBlocked!(self:Player Playtime View Model Object Reference=@n5) -> (ReturnValue:bool))
    self=Get Button_AskMoreTime()
   )
   Set bIsEnabled(
    execute=@n2.ActivityCreate
    bIsEnabled=NOT Boolean(A=IsAskForMoreCreateTimeBlocked!(self:Player Playtime View Model Object Reference=@n5) -> (ReturnValue:bool))
    self=Get Button_AskMoreTime()
   )
   Set bIsEnabled(execute=@n2.None, bIsEnabled=false, self=Get Button_AskMoreTime())
  ]}
 UpdateAskMoreTimeButton_Thyme {nodes [
   n1 = UpdateAskMoreTimeButton_Thyme!() -> (then, ActivityToShow:EActivityToShow Enum=None) {is=title K2Node_FunctionEntry}
   > SetText(
    self=Get Button_Thyme_AskMoreTime()
    Text=Select(
     None="LOCTABLE(\"&StringTable_TimeLimitTexts\", \"AskMoreTime\")"
     CombinedTime="LOCTABLE(\"&StringTable_TimeLimitTexts\", \"AskMoreTime\")"
     ActivityPlay="LOCTABLE(\"&StringTable_TimeLimitTexts\", \"AskPlayTime\")"
     ActivityCreate="LOCTABLE(\"&StringTable_TimeLimitTexts\", \"AskCreateTime\")"
     UserSelection="LOCTABLE(\"&StringTable_TimeLimitTexts\", \"AskMoreTime\")"
     Index=@n1
    )
   )
   > n2 = Switch on EActivityToShow(Selection=@n1)
   n3 = CheckBlockedStateForUserSelection(self=Get PlayerPlaytimeVM()) -> (ReturnValue=None)
   n4 = Get Button_Thyme_AskMoreTime()
   SetText(
    execute=@n2.UserSelection
    self=@n4
    Text=Select(
     None="LOCTABLE(\"&StringTable_TimeLimitTexts\", \"AskMoreTime\")"
     CombinedTime="LOCTABLE(\"&StringTable_TimeLimitTexts\", \"AskMoreTime\")"
     ActivityPlay="LOCTABLE(\"&StringTable_TimeLimitTexts\", \"AskPlayTime\")"
     ActivityCreate="LOCTABLE(\"&StringTable_TimeLimitTexts\", \"AskCreateTime\")"
     UserSelection="LOCTABLE(\"&StringTable_TimeLimitTexts\", \"AskMoreTime\")"
     Index=@n3
    )
   )
   > Set bIsEnabled(bIsEnabled="Not Equal (Enum)"(A=@n3, B=None), self=@n4)
   n5 = Get PlayerPlaytimeVM()
   Set bIsEnabled(
    execute=[@n2.ActivityPlay, @n2.CombinedTime]
    bIsEnabled=NOT Boolean(A=IsAskForMorePlayTimeBlocked(self=@n5))
    self=Get Button_Thyme_AskMoreTime()
   )
   Set bIsEnabled(execute=@n2.ActivityCreate, bIsEnabled=NOT Boolean(A=IsAskForMoreCreateTimeBlocked(self=@n5)), self=Get Button_Thyme_AskMoreTime())
   Set bIsEnabled(execute=@n2.None, bIsEnabled=false, self=Get Button_Thyme_AskMoreTime())
  ]}
 UpdateVisual {nodes [
   UpdateVisual() {is=title K2Node_FunctionEntry}
   > "SetText (Text)"(
    self=Get Text_Message() -> (Text_Message:Common Text Object Reference)
    InText=SelectText(
     A="LOCTEXT(\"Time limits and reports for Fortnite \\r\\nand Unreal Editor for Fortnite (UEFN)\")"
     B="LOCTEXT(\"Your parent or guardian has enabled time limits and time reports for your account in Fortnite and Unreal Editor for Fortnite (UEFN).\")"
     bPickA=Get bIsEnable Thyme()
    )
   )
   > Set Justification!(
    execute
    Justification:ETextJustify Enum=Select!(Option 0:ETextJustify Enum=Left, Option 1:ETextJustify Enum=Center, Index:bool=Get bIsEnable Thyme()) -> (ReturnValue:ETextJustify Enum)
    self:Text Layout Widget Object Reference=Get Text_Message()
   ) -> (then, Output_Get:ETextJustify Enum=Left)
   > "SetText (Text)"(
    self=Get Text_ReportDescription()
    InText=SelectText(
     A="LOCTEXT(\"Your parent gets a weekly summary of your play \\r\\nand create time\")"
     B="LOCTEXT(\"Your parent or guardian can see a summary of the time you spend in Fortnite and UEFN.\")"
     bPickA=Get bIsEnable Thyme()
    )
   )
   > Set Justification(Justification=Select(Option 0=Left, Option 1=Center, Index=Get bIsEnable Thyme()), self=Get Text_ReportDescription()) -> (Output_Get=Left)
   > SetActiveWidget(
    self=Get S_Buttons() -> (S_Buttons:Common Activatable Widget Switcher Object Reference)
    Widget=Select(
     Option 0:Widget Object Reference=Get Buttons_Switcher()
     Option 1:Widget Object Reference=Get VB_ThymeInfo() -> (VB_ThymeInfo:Vertical Box Object Reference)
     Index=Get bIsEnable Thyme()
    ) -> (ReturnValue:Widget Object Reference)
   )
  ]}
]
variables [
 bIsEnable Thyme {type=boolean}
 CreditsScreenClass {type=Class, default=/FortniteCoreUI/Frontend/Credits/Credits.Credits_C}
 IsUsingCompactUI {type=boolean}
 LegalInfoClass {type=Class, default=/Game/UI/Foundation/Legal/LegalInfo.LegalInfo_C}
]
widgets {
 info {
  inheritedWidgetCount=0
  namedSlotCount=1
  parentClass=&CommonActivatableWidget
  rootWidgetClass=/Sidebar/Shared/WBP_Sidebar_PanelLayout.WBP_Sidebar_PanelLayout_C
  widgetCount=66
 }
 widgets [
  PanelLayout {
   widgetClassPath=/Sidebar/Shared/WBP_Sidebar_PanelLayout
   bIsVariable=true
   widget=PanelLayout
   properties {headingText="Time limits && reports"}
  }
  StateContentsVbox {widgetClassPath=&VerticalBox, bIsVariable=true, namedSlotHost=PanelLayout, widget=StateContentsVbox}
  Box_ContentFirstPage {
   widgetClassPath=&VerticalBox
   bIsVariable=true
   parent=StateContentsVbox
   widget=Box_ContentFirstPage
   properties {bOverride_Cursor=true}
   slot {class=VerticalBoxSlot, properties {size {sizeRule=Fill, value=1}}}
  }
  SB_TimeInfo {
   widgetClassPath=&ScrollBox
   bIsVariable=true
   parent=Box_ContentFirstPage
   widget=SB_TimeInfo
   slot {class=VerticalBoxSlot, properties {padding=<ltrb 30 36 30 0>, size {sizeRule=Fill, value=1}}}
  }
  VerticalBox_0 {
   widgetClassPath=&VerticalBox
   bIsVariable=true
   parent=SB_TimeInfo
   widget=VerticalBox_0
   slot {class=ScrollBoxSlot, properties {size {sizeRule=Fill, value=1}}}
  }
  Text_Message {
   widgetClassPath=&CommonTextBlock
   bIsVariable=true
   parent=VerticalBox_0
   widget=Text_Message
   properties {
    autoWrapText=true
    mobileFontSizeMultiplier=1.41
    text="Your parent or guardian has enabled time limits and time reports for your account in Fortnite and Unreal Editor for Fortnite (UEFN)."
    font {
     bForceMonospaced=false
     bMaterialIsStencil=false
     fontMaterial=/Sidebar/Materials/Fonts/MI_Sidebar_FontColor_Secondary
     fontObject=&HeadingNow
     letterSpacing=0
     monospacedWidth=1
     size=12.75
     skewAmount=0
     typefaceFontName=74 Regular
     outlineSettings {
      bApplyOutlineToDropShadows=false
      bMiteredCorners=false
      bSeparateFillAlpha=false
      outlineMaterial=null
      outlineSize=0
      outlineColor=<rgba 0 0 0 0.15>
     }
    }
    shadowOffset=<0 0>
   }
   slot {class=VerticalBoxSlot, properties {padding=<ltrb 0 0 0 24>}}
  }
  PlayTimeInfo_Switcher {
   widgetClassPath=&CommonActivatableWidgetSwitcher
   bIsVariable=true
   parent=VerticalBox_0
   widget=PlayTimeInfo_Switcher
   slot {class=VerticalBoxSlot, properties {padding=<ltrb 0 0 0 24>}}
  }
  CombinedTimeInfo {
   widgetClassPath=&VerticalBox
   bIsVariable=true
   parent=PlayTimeInfo_Switcher
   widget=CombinedTimeInfo
   slot {class=WidgetSwitcherSlot}
  }
  Text_PlayTimeHeader {
   widgetClassPath=&CommonTextBlock
   parent=CombinedTimeInfo
   widget=Text_PlayTimeHeader
   properties {
    justification=Center
    mobileFontSizeMultiplier=1.17
    text=Time limits
    font {
     bForceMonospaced=false
     bMaterialIsStencil=false
     fontMaterial=null
     fontObject=&HeadingNow
     letterSpacing=0
     monospacedWidth=1
     size=20.25
     skewAmount=0
     typefaceFontName=75 Medium
     outlineSettings {
      bApplyOutlineToDropShadows=false
      bMiteredCorners=false
      bSeparateFillAlpha=false
      outlineMaterial=null
      outlineSize=0
      outlineColor=<rgba 0 0 0 0.15>
     }
    }
    shadowOffset=<0 0>
   }
   slot {class=VerticalBoxSlot, properties {padding=<ltrb 0 0 0 8>}}
  }
  TimeLimitSwitcher {
   widgetClassPath=&CommonActivatableWidgetSwitcher
   bIsVariable=true
   parent=CombinedTimeInfo
   widget=TimeLimitSwitcher
   slot {class=VerticalBoxSlot}
  }
  Box_TimeLimitOn {widgetClassPath=&VerticalBox, bIsVariable=true, parent=TimeLimitSwitcher, widget=Box_TimeLimitOn, slot {class=WidgetSwitcherSlot}}
  Text_PlaytimeLeft {
   widgetClassPath=&CommonTextBlock
   bIsVariable=true
   parent=Box_TimeLimitOn
   widget=Text_PlaytimeLeft
   properties {
    justification=Center
    mobileFontSizeMultiplier=1.41
    text=Time left
    font {
     bForceMonospaced=false
     bMaterialIsStencil=false
     fontMaterial=&MI_Sidebar_FontColor_Secondary
     fontObject=&HeadingNow
     letterSpacing=0
     monospacedWidth=1
     size=12.75
     skewAmount=0
     typefaceFontName=74 Regular
     outlineSettings {
      bApplyOutlineToDropShadows=false
      bMiteredCorners=false
      bSeparateFillAlpha=false
      outlineMaterial=null
      outlineSize=0
      outlineColor=<rgba 0 0 0 0.15>
     }
    }
    shadowOffset=<0 0>
   }
   slot {class=VerticalBoxSlot, properties {padding=<ltrb 0 0 0 4>}}
  }
  Text_PlaytimePlayed {
   widgetClassPath=&CommonTextBlock
   bIsVariable=true
   parent=Box_TimeLimitOn
   widget=Text_PlaytimePlayed
   properties {
    justification=Center
    mobileFontSizeMultiplier=1.41
    text=TimePlayed
    font {
     bForceMonospaced=false
     bMaterialIsStencil=false
     fontMaterial=&MI_Sidebar_FontColor_Secondary
     fontObject=&HeadingNow
     letterSpacing=0
     monospacedWidth=1
     size=12.75
     skewAmount=0
     typefaceFontName=74 Regular
     outlineSettings {
      bApplyOutlineToDropShadows=false
      bMiteredCorners=false
      bSeparateFillAlpha=false
      outlineMaterial=null
      outlineSize=0
      outlineColor=<rgba 0 0 0 0.15>
     }
    }
    shadowOffset=<0 0>
   }
   slot {class=VerticalBoxSlot, properties {padding=<ltrb 0 0 0 4>}}
  }
  Text_WindowBounds {
   widgetClassPath=&CommonTextBlock
   bIsVariable=true
   parent=Box_TimeLimitOn
   widget=Text_WindowBounds
   properties {
    bAutoCollapseWithEmptyText=true
    justification=Center
    mobileFontSizeMultiplier=1.41
    text=WindowBounds
    font {
     bForceMonospaced=false
     bMaterialIsStencil=false
     fontMaterial=&MI_Sidebar_FontColor_Secondary
     fontObject=&HeadingNow
     letterSpacing=0
     monospacedWidth=1
     size=12.75
     skewAmount=0
     typefaceFontName=74 Regular
     outlineSettings {
      bApplyOutlineToDropShadows=false
      bMiteredCorners=false
      bSeparateFillAlpha=false
      outlineMaterial=null
      outlineSize=0
      outlineColor=<rgba 0 0 0 0.15>
     }
    }
    shadowOffset=<0 0>
   }
   slot {class=VerticalBoxSlot, properties {padding=<ltrb 0 0 0 4>}}
  }
  Box_TimeLimitOff {
   widgetClassPath=&VerticalBox
   bIsVariable=true
   parent=TimeLimitSwitcher
   widget=Box_TimeLimitOff
   slot {class=WidgetSwitcherSlot}
  }
  Text_LimitOff {
   widgetClassPath=&CommonTextBlock
   bIsVariable=true
   parent=Box_TimeLimitOff
   widget=Text_LimitOff
   properties {
    justification=Center
    mobileFontSizeMultiplier=1.41
    text=Time limits off
    font {
     bForceMonospaced=false
     bMaterialIsStencil=false
     fontMaterial=&MI_Sidebar_FontColor_Secondary
     fontObject=&HeadingNow
     letterSpacing=0
     monospacedWidth=1
     size=12.75
     skewAmount=0
     typefaceFontName=74 Regular
     outlineSettings {
      bApplyOutlineToDropShadows=false
      bMiteredCorners=false
      bSeparateFillAlpha=false
      outlineMaterial=null
      outlineSize=0
      outlineColor=<rgba 0 0 0 1>
     }
    }
    shadowOffset=<0 0>
   }
   slot {class=VerticalBoxSlot, properties {padding=<ltrb 0 0 0 4>}}
  }
  SeparatedTimeInfo {
   widgetClassPath=&VerticalBox
   bIsVariable=true
   parent=PlayTimeInfo_Switcher
   widget=SeparatedTimeInfo
   slot {class=WidgetSwitcherSlot, properties {padding=<ltrb 0 32 0 8>}}
  }
  PlayTimeLimits {widgetClassPath=&VerticalBox, parent=SeparatedTimeInfo, widget=PlayTimeLimits, slot {class=VerticalBoxSlot}}
  Text_PlayTimeLimitsHeader {
   widgetClassPath=&CommonTextBlock
   parent=PlayTimeLimits
   widget=Text_PlayTimeLimitsHeader
   properties {
    justification=Center
    mobileFontSizeMultiplier=1.17
    text=Play time
    font {
     bForceMonospaced=false
     bMaterialIsStencil=false
     fontMaterial=null
     fontObject=&HeadingNow
     letterSpacing=0
     monospacedWidth=1
     size=20.25
     skewAmount=0
     typefaceFontName=75 Medium
     outlineSettings {
      bApplyOutlineToDropShadows=false
      bMiteredCorners=false
      bSeparateFillAlpha=false
      outlineMaterial=null
      outlineSize=0
      outlineColor=<rgba 0 0 0 0.15>
     }
    }
    shadowOffset=<0 0>
   }
   slot {class=VerticalBoxSlot, properties {padding=<ltrb 0 0 0 8>}}
  }
  PlayTimeLimitSwitcher {
   widgetClassPath=&CommonActivatableWidgetSwitcher
   bIsVariable=true
   parent=PlayTimeLimits
   widget=PlayTimeLimitSwitcher
   slot {class=VerticalBoxSlot}
  }
  Box_PlayTimeLimitOn {
   widgetClassPath=&VerticalBox
   bIsVariable=true
   parent=PlayTimeLimitSwitcher
   widget=Box_PlayTimeLimitOn
   slot {class=WidgetSwitcherSlot}
  }
  Text_PlaytimeLimitsLeft {
   widgetClassPath=&CommonTextBlock
   bIsVariable=true
   parent=Box_PlayTimeLimitOn
   widget=Text_PlaytimeLimitsLeft
   properties {
    justification=Center
    mobileFontSizeMultiplier=1.41
    text=Time left
    font {
     bForceMonospaced=false
     bMaterialIsStencil=false
     fontMaterial=&MI_Sidebar_FontColor_Secondary
     fontObject=&HeadingNow
     letterSpacing=0
     monospacedWidth=1
     size=12.75
     skewAmount=0
     typefaceFontName=74 Regular
     outlineSettings {
      bApplyOutlineToDropShadows=false
      bMiteredCorners=false
      bSeparateFillAlpha=false
      outlineMaterial=null
      outlineSize=0
      outlineColor=<rgba 0 0 0 0.15>
     }
    }
    shadowOffset=<0 0>
   }
   slot {class=VerticalBoxSlot, properties {padding=<ltrb 0 0 0 4>}}
  }
  Text_PlaytimeLimitsPlayed {
   widgetClassPath=&CommonTextBlock
   bIsVariable=true
   parent=Box_PlayTimeLimitOn
   widget=Text_PlaytimeLimitsPlayed
   properties {
    justification=Center
    mobileFontSizeMultiplier=1.41
    text=TimePlayed
    font {
     bForceMonospaced=false
     bMaterialIsStencil=false
     fontMaterial=&MI_Sidebar_FontColor_Secondary
     fontObject=&HeadingNow
     letterSpacing=0
     monospacedWidth=1
     size=12.75
     skewAmount=0
     typefaceFontName=74 Regular
     outlineSettings {
      bApplyOutlineToDropShadows=false
      bMiteredCorners=false
      bSeparateFillAlpha=false
      outlineMaterial=null
      outlineSize=0
      outlineColor=<rgba 0 0 0 0.15>
     }
    }
    shadowOffset=<0 0>
   }
   slot {class=VerticalBoxSlot, properties {padding=<ltrb 0 0 0 4>}}
  }
  Text_PlayTimeLimits_WindowBounds {
   widgetClassPath=&CommonTextBlock
   bIsVariable=true
   parent=Box_PlayTimeLimitOn
   widget=Text_PlayTimeLimits_WindowBounds
   properties {
    bAutoCollapseWithEmptyText=true
    justification=Center
    mobileFontSizeMultiplier=1.41
    text=WindowBounds
    font {
     bForceMonospaced=false
     bMaterialIsStencil=false
     fontMaterial=&MI_Sidebar_FontColor_Secondary
     fontObject=&HeadingNow
     letterSpacing=0
     monospacedWidth=1
     size=12.75
     skewAmount=0
     typefaceFontName=74 Regular
     outlineSettings {
      bApplyOutlineToDropShadows=false
      bMiteredCorners=false
      bSeparateFillAlpha=false
      outlineMaterial=null
      outlineSize=0
      outlineColor=<rgba 0 0 0 0.15>
     }
    }
    shadowOffset=<0 0>
   }
   slot {class=VerticalBoxSlot, properties {padding=<ltrb 0 0 0 4>}}
  }
  Box_PlayTimeLimitOff {
   widgetClassPath=&VerticalBox
   bIsVariable=true
   parent=PlayTimeLimitSwitcher
   widget=Box_PlayTimeLimitOff
   slot {class=WidgetSwitcherSlot}
  }
  Text_LimitOff_1 {
   widgetClassPath=&CommonTextBlock
   bIsVariable=true
   parent=Box_PlayTimeLimitOff
   widget=Text_LimitOff_1
   properties {
    justification=Center
    mobileFontSizeMultiplier=1.41
    text=Time limits off
    font {
     bForceMonospaced=false
     bMaterialIsStencil=false
     fontMaterial=&MI_Sidebar_FontColor_Secondary
     fontObject=&HeadingNow
     letterSpacing=0
     monospacedWidth=1
     size=12.75
     skewAmount=0
     typefaceFontName=74 Regular
     outlineSettings {
      bApplyOutlineToDropShadows=false
      bMiteredCorners=false
      bSeparateFillAlpha=false
      outlineMaterial=null
      outlineSize=0
      outlineColor=<rgba 0 0 0 1>
     }
    }
    shadowOffset=<0 0>
   }
   slot {class=VerticalBoxSlot, properties {padding=<ltrb 0 0 0 4>}}
  }
  CreateTimeLimits {
   widgetClassPath=&VerticalBox
   parent=SeparatedTimeInfo
   widget=CreateTimeLimits
   slot {class=VerticalBoxSlot, properties {padding=<ltrb 0 32 0 0>}}
  }
  Text_CreateTimeHeader {
   widgetClassPath=&CommonTextBlock
   parent=CreateTimeLimits
   widget=Text_CreateTimeHeader
   properties {
    justification=Center
    mobileFontSizeMultiplier=1.17
    text=Create time
    font {
     bForceMonospaced=false
     bMaterialIsStencil=false
     fontMaterial=null
     fontObject=&HeadingNow
     letterSpacing=0
     monospacedWidth=1
     size=20.25
     skewAmount=0
     typefaceFontName=75 Medium
     outlineSettings {
      bApplyOutlineToDropShadows=false
      bMiteredCorners=false
      bSeparateFillAlpha=false
      outlineMaterial=null
      outlineSize=0
      outlineColor=<rgba 0 0 0 0.15>
     }
    }
    shadowOffset=<0 0>
   }
   slot {class=VerticalBoxSlot, properties {padding=<ltrb 0 0 0 8>}}
  }
  CreateTimeLimitSwitcher {
   widgetClassPath=&CommonActivatableWidgetSwitcher
   bIsVariable=true
   parent=CreateTimeLimits
   widget=CreateTimeLimitSwitcher
   slot {class=VerticalBoxSlot}
  }
  Box_CreateTimeLimitOn {
   widgetClassPath=&VerticalBox
   bIsVariable=true
   parent=CreateTimeLimitSwitcher
   widget=Box_CreateTimeLimitOn
   slot {class=WidgetSwitcherSlot}
  }
  Text_CreatetimeLeft {
   widgetClassPath=&CommonTextBlock
   bIsVariable=true
   parent=Box_CreateTimeLimitOn
   widget=Text_CreatetimeLeft
   properties {
    justification=Center
    mobileFontSizeMultiplier=1.41
    text=Time left
    font {
     bForceMonospaced=false
     bMaterialIsStencil=false
     fontMaterial=&MI_Sidebar_FontColor_Secondary
     fontObject=&HeadingNow
     letterSpacing=0
     monospacedWidth=1
     size=12.75
     skewAmount=0
     typefaceFontName=74 Regular
     outlineSettings {
      bApplyOutlineToDropShadows=false
      bMiteredCorners=false
      bSeparateFillAlpha=false
      outlineMaterial=null
      outlineSize=0
      outlineColor=<rgba 0 0 0 0.15>
     }
    }
    shadowOffset=<0 0>
   }
   slot {class=VerticalBoxSlot, properties {padding=<ltrb 0 0 0 4>}}
  }
  Text_CreatetimePlayed {
   widgetClassPath=&CommonTextBlock
   bIsVariable=true
   parent=Box_CreateTimeLimitOn
   widget=Text_CreatetimePlayed
   properties {
    justification=Center
    mobileFontSizeMultiplier=1.41
    text=TimePlayed
    font {
     bForceMonospaced=false
     bMaterialIsStencil=false
     fontMaterial=&MI_Sidebar_FontColor_Secondary
     fontObject=&HeadingNow
     letterSpacing=0
     monospacedWidth=1
     size=12.75
     skewAmount=0
     typefaceFontName=74 Regular
     outlineSettings {
      bApplyOutlineToDropShadows=false
      bMiteredCorners=false
      bSeparateFillAlpha=false
      outlineMaterial=null
      outlineSize=0
      outlineColor=<rgba 0 0 0 0.15>
     }
    }
    shadowOffset=<0 0>
   }
   slot {class=VerticalBoxSlot, properties {padding=<ltrb 0 0 0 4>}}
  }
  Text_CreateTime_WindowBounds {
   widgetClassPath=&CommonTextBlock
   bIsVariable=true
   parent=Box_CreateTimeLimitOn
   widget=Text_CreateTime_WindowBounds
   properties {
    bAutoCollapseWithEmptyText=true
    justification=Center
    mobileFontSizeMultiplier=1.41
    text=WindowBounds
    font {
     bForceMonospaced=false
     bMaterialIsStencil=false
     fontMaterial=&MI_Sidebar_FontColor_Secondary
     fontObject=&HeadingNow
     letterSpacing=0
     monospacedWidth=1
     size=12.75
     skewAmount=0
     typefaceFontName=74 Regular
     outlineSettings {
      bApplyOutlineToDropShadows=false
      bMiteredCorners=false
      bSeparateFillAlpha=false
      outlineMaterial=null
      outlineSize=0
      outlineColor=<rgba 0 0 0 0.15>
     }
    }
    shadowOffset=<0 0>
   }
   slot {class=VerticalBoxSlot, properties {padding=<ltrb 0 0 0 4>}}
  }
  Box_CreateTimeLimitOff {
   widgetClassPath=&VerticalBox
   bIsVariable=true
   parent=CreateTimeLimitSwitcher
   widget=Box_CreateTimeLimitOff
   slot {class=WidgetSwitcherSlot}
  }
  Text_LimitOff_2 {
   widgetClassPath=&CommonTextBlock
   bIsVariable=true
   parent=Box_CreateTimeLimitOff
   widget=Text_LimitOff_2
   properties {
    justification=Center
    mobileFontSizeMultiplier=1.41
    text=Time limits off
    font {
     bForceMonospaced=false
     bMaterialIsStencil=false
     fontMaterial=&MI_Sidebar_FontColor_Secondary
     fontObject=&HeadingNow
     letterSpacing=0
     monospacedWidth=1
     size=12.75
     skewAmount=0
     typefaceFontName=74 Regular
     outlineSettings {
      bApplyOutlineToDropShadows=false
      bMiteredCorners=false
      bSeparateFillAlpha=false
      outlineMaterial=null
      outlineSize=0
      outlineColor=<rgba 0 0 0 1>
     }
    }
    shadowOffset=<0 0>
   }
   slot {class=VerticalBoxSlot, properties {padding=<ltrb 0 0 0 4>}}
  }
  PlayReportInfo {widgetClassPath=&VerticalBox, parent=VerticalBox_0, widget=PlayReportInfo, slot {class=VerticalBoxSlot}}
  Text_ReportInfoHeader {
   widgetClassPath=&CommonTextBlock
   parent=PlayReportInfo
   widget=Text_ReportInfoHeader
   properties {
    justification=Center
    mobileFontSizeMultiplier=1.17
    text=Time reports
    font {
     bForceMonospaced=false
     bMaterialIsStencil=false
     fontMaterial=null
     fontObject=&HeadingNow
     letterSpacing=0
     monospacedWidth=1
     size=20.25
     skewAmount=0
     typefaceFontName=75 Medium
     outlineSettings {
      bApplyOutlineToDropShadows=false
      bMiteredCorners=false
      bSeparateFillAlpha=false
      outlineMaterial=null
      outlineSize=0
      outlineColor=<rgba 0 0 0 0.15>
     }
    }
    shadowOffset=<0 0>
   }
   slot {class=VerticalBoxSlot, properties {padding=<ltrb 0 0 0 8>}}
  }
  Text_ReportStatus {
   widgetClassPath=&CommonTextBlock
   bIsVariable=true
   parent=PlayReportInfo
   widget=Text_ReportStatus
   properties {
    justification=Center
    mobileFontSizeMultiplier=1.41
    text=Time reports on
    font {
     bForceMonospaced=false
     bMaterialIsStencil=false
     fontMaterial=&MI_Sidebar_FontColor_Secondary
     fontObject=&HeadingNow
     letterSpacing=0
     monospacedWidth=1
     size=12.75
     skewAmount=0
     typefaceFontName=74 Regular
     outlineSettings {
      bApplyOutlineToDropShadows=false
      bMiteredCorners=false
      bSeparateFillAlpha=false
      outlineMaterial=null
      outlineSize=0
      outlineColor=<rgba 0 0 0 0.15>
     }
    }
    shadowOffset=<0 0>
   }
   slot {class=VerticalBoxSlot, properties {padding=<ltrb 0 0 0 24>}}
  }
  Text_ReportDescription {
   widgetClassPath=&CommonTextBlock
   bIsVariable=true
   parent=PlayReportInfo
   widget=Text_ReportDescription
   properties {
    autoWrapText=true
    mobileFontSizeMultiplier=1.41
    text=Your parent or guardian can see a summary of the time you spend in Fortnite and UEFN.
    font {
     bForceMonospaced=false
     bMaterialIsStencil=false
     fontMaterial=&MI_Sidebar_FontColor_Secondary
     fontObject=&HeadingNow
     letterSpacing=0
     monospacedWidth=1
     size=12.75
     skewAmount=0
     typefaceFontName=74 Regular
     outlineSettings {
      bApplyOutlineToDropShadows=false
      bMiteredCorners=false
      bSeparateFillAlpha=false
      outlineMaterial=null
      outlineSize=0
      outlineColor=<rgba 0 0 0 0.15>
     }
    }
    shadowOffset=<0 0>
   }
   slot {class=VerticalBoxSlot}
  }
  S_Buttons {
   widgetClassPath=&CommonActivatableWidgetSwitcher
   bIsVariable=true
   parent=Box_ContentFirstPage
   widget=S_Buttons
   slot {class=VerticalBoxSlot, properties {verticalAlignment=VAlign_Top, padding=<ltrb 0 24 0 0>}}
  }
  Buttons_Switcher {
   widgetClassPath=&CommonActivatableWidgetSwitcher
   bIsVariable=true
   parent=S_Buttons
   widget=Buttons_Switcher
   properties {activeWidgetIndex=1}
   slot {class=WidgetSwitcherSlot}
  }
  Container_ButtonsUnavailable {
   widgetClassPath=&HorizontalBox
   bIsVariable=true
   parent=Buttons_Switcher
   widget=Container_ButtonsUnavailable
   slot {class=WidgetSwitcherSlot, properties {padding=<ltrb 30 0 30 0>}}
  }
  Image_Info {
   widgetClassPath=&Image
   parent=Container_ButtonsUnavailable
   widget=Image_Info
   properties {brush {
     bIsDynamicallyLoaded=false
     drawAs=Image
     imageType=FullColor
     mirroring=NoMirror
     resourceName=None
     resourceObject=/FortUILibrary/Textures/MaterialInstances/MI_UIKit_Icon_Information
     tiling=NoTile
     imageSize=<32 32>
     margin=<ltrb 0 0 0 0>
     outlineSettings {
      bUseBrushTransparency=false
      roundingType=HalfHeightRadius
      width=0
      color {colorUseRule=UseColor_Specified, specifiedColor=<rgba 0 0 0 0>}
      cornerRadii=<0 0 0 0>
     }
     tintColor {colorUseRule=UseColor_Specified, specifiedColor=<rgba 0.5365 0.5365 0.5365 1>}
     uVRegion {bIsValid=false, max=<0 0>, min=<0 0>}
    }}
   slot {class=HorizontalBoxSlot, properties {horizontalAlignment=HAlign_Left, verticalAlignment=VAlign_Center, size {sizeRule=Automatic, value=0.1}}}
  }
  CommonTextBlock {
   widgetClassPath=&CommonTextBlock
   parent=Container_ButtonsUnavailable
   widget=CommonTextBlock
   properties {
    autoWrapText=true
    mobileFontSizeMultiplier=1.33
    text=You can ask your parent or guardian for more time when you have less than 30 minutes left.
    colorAndOpacity {colorUseRule=UseColor_Specified, specifiedColor=<rgba 0.5365 0.5365 0.5365 1>}
    font {
     bForceMonospaced=false
     bMaterialIsStencil=false
     fontMaterial=null
     fontObject=&HeadingNow
     letterSpacing=0
     monospacedWidth=1
     size=13.5
     skewAmount=0
     typefaceFontName=74 Regular
     outlineSettings {
      bApplyOutlineToDropShadows=false
      bMiteredCorners=false
      bSeparateFillAlpha=false
      outlineMaterial=null
      outlineSize=2
      outlineColor=<rgba 0 0 0 0.15>
     }
    }
    shadowOffset=<0 0>
   }
   slot {class=HorizontalBoxSlot, properties {padding=<ltrb 8 0 0 0>, size {sizeRule=Fill, value=0.9}}}
  }
  Container_Buttons {
   widgetClassPath=&VerticalBox
   bIsVariable=true
   parent=Buttons_Switcher
   widget=Container_Buttons
   slot {class=WidgetSwitcherSlot}
  }
  Button_AskMoreTime {
   widgetClassPath=/Sidebar/Buttons/WBP_Sidebar_ButtonGeneric
   bIsVariable=true
   parent=Container_Buttons
   widget=Button_AskMoreTime
   properties {
    bInteractableWhenSelected=false
    bOverride_Cursor=true
    bSelectable=false
    bShouldSelectUponReceivingFocus=false
    icon=/FortUILibrary/Textures/Icons/T_UI_Icon_UIKit_Email_MSDF
    minHeight=48
    minHeightOverride=60
    text=Ask for more time
    useIconBlock=true
    visibility=Visible
    hoveredSlateSoundOverride {resourceObject=/Game/Sounds/Fort_UI_Sounds/FrontEnd/New_FrontEnd/Fortnite_UI_MainMenu_Tab_Hover_01_Cue}
    iconPadding=<ltrb 0 0 8 0>
    pressedSlateSoundOverride {resourceObject=/Game/Sounds/Fort_UI_Sounds/FrontEnd/New_FrontEnd/Fortnite_UI_MainMenu_Generic_01_Cue}
   }
   slot {class=VerticalBoxSlot, properties {padding=<ltrb 0 0 0 12>}}
  }
  Button_AddMoreTime {
   widgetClassPath=&WBP_Sidebar_ButtonGeneric
   bIsVariable=true
   parent=Container_Buttons
   widget=Button_AddMoreTime
   properties {
    bInteractableWhenSelected=false
    bOverride_Cursor=true
    bSelectable=false
    bShouldSelectUponReceivingFocus=false
    icon=/Sidebar/Materials/Icons/Textures/T_UI_Icon_Sidebar_Locked_MSDF
    minHeight=48
    minHeightOverride=60
    text=Add more time
    useIconBlock=true
    visibility=Visible
    hoveredSlateSoundOverride {resourceObject=&Fortnite_UI_MainMenu_Tab_Hover_01_Cue}
    iconPadding=<ltrb 0 0 8 0>
    pressedSlateSoundOverride {resourceObject=&Fortnite_UI_MainMenu_Generic_01_Cue}
   }
   slot {class=VerticalBoxSlot}
  }
  VB_ThymeInfo {
   widgetClassPath=&VerticalBox
   bIsVariable=true
   parent=S_Buttons
   widget=VB_ThymeInfo
   properties {bOverride_Cursor=true}
   slot {class=WidgetSwitcherSlot, properties {padding=<ltrb 30 0 30 0>}}
  }
  Container_Thyme_Disclaimer {
   widgetClassPath=&HorizontalBox
   bIsVariable=true
   parent=VB_ThymeInfo
   widget=Container_Thyme_Disclaimer
   slot {class=VerticalBoxSlot}
  }
  Image_Info_1 {
   widgetClassPath=&Image
   parent=Container_Thyme_Disclaimer
   widget=Image_Info_1
   properties {brush {
     bIsDynamicallyLoaded=false
     drawAs=Image
     imageType=FullColor
     mirroring=NoMirror
     resourceName=None
     resourceObject=&MI_UIKit_Icon_Information
     tiling=NoTile
     imageSize=<32 32>
     margin=<ltrb 0 0 0 0>
     outlineSettings {
      bUseBrushTransparency=false
      roundingType=HalfHeightRadius
      width=0
      color {colorUseRule=UseColor_Specified, specifiedColor=<rgba 0 0 0 0>}
      cornerRadii=<0 0 0 0>
     }
     tintColor {colorUseRule=UseColor_Specified, specifiedColor=<rgba 0.5365 0.5365 0.5365 1>}
     uVRegion {bIsValid=false, max=<0 0>, min=<0 0>}
    }}
   slot {class=HorizontalBoxSlot, properties {horizontalAlignment=HAlign_Left, verticalAlignment=VAlign_Center, size {sizeRule=Automatic, value=0.1}}}
  }
  Text_ButtonsUnavailable {
   widgetClassPath=&CommonTextBlock
   bIsVariable=true
   parent=Container_Thyme_Disclaimer
   widget=Text_ButtonsUnavailable
   properties {
    autoWrapText=true
    mobileFontSizeMultiplier=1.33
    text=You can ask your parent or guardian for more time when you have less than 30 minutes left.
    colorAndOpacity {colorUseRule=UseColor_Specified, specifiedColor=<rgba 0.5365 0.5365 0.5365 1>}
    font {
     bForceMonospaced=false
     bMaterialIsStencil=false
     fontMaterial=null
     fontObject=&HeadingNow
     letterSpacing=0
     monospacedWidth=1
     size=13.5
     skewAmount=0
     typefaceFontName=74 Regular
     outlineSettings {
      bApplyOutlineToDropShadows=false
      bMiteredCorners=false
      bSeparateFillAlpha=false
      outlineMaterial=null
      outlineSize=2
      outlineColor=<rgba 0 0 0 0.15>
     }
    }
    shadowOffset=<0 0>
   }
   slot {class=HorizontalBoxSlot, properties {padding=<ltrb 8 0 0 0>, size {sizeRule=Fill, value=0.9}}}
  }
  Container_Thyme_Buttons {
   widgetClassPath=&VerticalBox
   bIsVariable=true
   parent=VB_ThymeInfo
   widget=Container_Thyme_Buttons
   slot {class=VerticalBoxSlot, properties {padding=<ltrb 0 30 0 0>}}
  }
  Button_Thyme_AskMoreTime {
   widgetClassPath=&WBP_Sidebar_ButtonGeneric
   bIsVariable=true
   parent=Container_Thyme_Buttons
   widget=Button_Thyme_AskMoreTime
   properties {
    bInteractableWhenSelected=false
    bOverride_Cursor=true
    bSelectable=false
    bShouldSelectUponReceivingFocus=false
    icon=&T_UI_Icon_UIKit_Email_MSDF
    minHeight=48
    minHeightOverride=60
    text=Ask for more time
    useIconBlock=true
    visibility=Visible
    hoveredSlateSoundOverride {resourceObject=&Fortnite_UI_MainMenu_Tab_Hover_01_Cue}
    iconPadding=<ltrb 0 0 8 0>
    pressedSlateSoundOverride {resourceObject=&Fortnite_UI_MainMenu_Generic_01_Cue}
   }
   slot {class=VerticalBoxSlot, properties {padding=<ltrb 0 0 0 12>}}
  }
  Button_Thyme_AddMoreTime {
   widgetClassPath=&WBP_Sidebar_ButtonGeneric
   bIsVariable=true
   parent=Container_Thyme_Buttons
   widget=Button_Thyme_AddMoreTime
   properties {
    bInteractableWhenSelected=false
    bOverride_Cursor=true
    bSelectable=false
    bShouldSelectUponReceivingFocus=false
    icon=&T_UI_Icon_Sidebar_Locked_MSDF
    minHeight=48
    minHeightOverride=60
    text=Add more time
    useIconBlock=true
    visibility=Visible
    hoveredSlateSoundOverride {resourceObject=&Fortnite_UI_MainMenu_Tab_Hover_01_Cue}
    iconPadding=<ltrb 0 0 8 0>
    pressedSlateSoundOverride {resourceObject=&Fortnite_UI_MainMenu_Generic_01_Cue}
   }
   slot {class=VerticalBoxSlot}
  }
  Box_EmailSending {
   widgetClassPath=&VerticalBox
   bIsVariable=true
   parent=StateContentsVbox
   widget=Box_EmailSending
   properties {visibility=Collapsed}
   slot {class=VerticalBoxSlot, properties {padding=<ltrb 16 0 16 0>}}
  }
  Spacer {widgetClassPath=&Spacer, parent=Box_EmailSending, widget=Spacer, properties {size=<1 268.1173>}, slot {class=VerticalBoxSlot}}
  VerticalBox_1 {
   widgetClassPath=&VerticalBox
   parent=Box_EmailSending
   widget=VerticalBox_1
   properties {bOverride_Cursor=true}
   slot {class=VerticalBoxSlot, properties {horizontalAlignment=HAlign_Center}}
  }
  WBP_UIKit_Throbber {
   widgetClassPath=/FortUILibrary/UIKit/Throbber/WBP_UIKit_Throbber
   bIsVariable=true
   parent=VerticalBox_1
   widget=WBP_UIKit_Throbber
   slot {class=VerticalBoxSlot, properties {padding=<ltrb 0 0 0 24>}}
  }
  Text_SendingEmail {
   widgetClassPath=&CommonTextBlock
   parent=VerticalBox_1
   widget=Text_SendingEmail
   properties {
    mobileFontSizeMultiplier=1.19
    text=Sending email
    font {
     bForceMonospaced=false
     bMaterialIsStencil=false
     fontMaterial=&MI_Sidebar_FontColor_Secondary
     fontObject=&HeadingNow
     letterSpacing=0
     monospacedWidth=1
     size=15.75
     skewAmount=0
     typefaceFontName=74 Regular
     outlineSettings {
      bApplyOutlineToDropShadows=false
      bMiteredCorners=false
      bSeparateFillAlpha=false
      outlineMaterial=null
      outlineSize=2
      outlineColor=<rgba 0 0 0 0.15>
     }
    }
    shadowOffset=<0 0>
   }
   slot {class=VerticalBoxSlot}
  }
  Box_ContentEmailSent {
   widgetClassPath=&VerticalBox
   bIsVariable=true
   parent=StateContentsVbox
   widget=Box_ContentEmailSent
   properties {visibility=Collapsed}
   slot {class=VerticalBoxSlot}
  }
  Spacer_151 {
   widgetClassPath=&Spacer
   parent=Box_ContentEmailSent
   widget=Spacer_151
   properties {size=<1 268.1173>}
   slot {class=VerticalBoxSlot}
  }
  HorizontalBox_1 {
   widgetClassPath=&HorizontalBox
   parent=Box_ContentEmailSent
   widget=HorizontalBox_1
   slot {class=VerticalBoxSlot, properties {horizontalAlignment=HAlign_Center}}
  }
  Image_Tick {
   widgetClassPath=&Image
   bIsVariable=true
   parent=HorizontalBox_1
   widget=Image_Tick
   properties {brush {
     bIsDynamicallyLoaded=false
     drawAs=Image
     imageType=FullColor
     mirroring=NoMirror
     resourceName=None
     resourceObject=/FortUILibrary/Textures/MaterialInstances/MI_UIKit_Icon_Check
     tiling=NoTile
     imageSize=<32 32>
     margin=<ltrb 0 0 0 0>
     outlineSettings {
      bUseBrushTransparency=false
      roundingType=HalfHeightRadius
      width=0
      color {colorUseRule=UseColor_Specified, specifiedColor=<rgba 0 0 0 0>}
      cornerRadii=<0 0 0 0>
     }
     tintColor {colorUseRule=UseColor_Specified, specifiedColor=<rgba 1 1 1 1>}
     uVRegion {bIsValid=false, max=<0 0>, min=<0 0>}
    }}
   slot {class=HorizontalBoxSlot, properties {horizontalAlignment=HAlign_Center, verticalAlignment=VAlign_Center, padding=<ltrb 0 0 8 0>}}
  }
  Text_EmailSent {
   widgetClassPath=&CommonTextBlock
   bIsVariable=true
   parent=HorizontalBox_1
   widget=Text_EmailSent
   properties {
    mobileFontSizeMultiplier=1.33
    text=Email sent
    font {
     bForceMonospaced=false
     bMaterialIsStencil=false
     fontMaterial=null
     fontObject=&HeadingNow
     letterSpacing=0
     monospacedWidth=1
     size=20.25
     skewAmount=0
     typefaceFontName=75 Medium
     outlineSettings {
      bApplyOutlineToDropShadows=false
      bMiteredCorners=false
      bSeparateFillAlpha=false
      outlineMaterial=null
      outlineSize=0
      outlineColor=<rgba 0 0 0 0.15>
     }
    }
    shadowOffset=<0 0>
   }
   slot {class=HorizontalBoxSlot, properties {horizontalAlignment=HAlign_Left, verticalAlignment=VAlign_Center}}
  }
  Text_EmailDescription {
   widgetClassPath=&CommonTextBlock
   bIsVariable=true
   parent=Box_ContentEmailSent
   widget=Text_EmailDescription
   properties {
    autoWrapText=true
    justification=Center
    mobileFontSizeMultiplier=1.41
    text="We’ve sent an email to your parent or guardian to ask for more time."
    font {
     bForceMonospaced=false
     bMaterialIsStencil=false
     fontMaterial=&MI_Sidebar_FontColor_Secondary
     fontObject=&HeadingNow
     letterSpacing=0
     monospacedWidth=1
     size=12.75
     skewAmount=0
     typefaceFontName=74 Regular
     outlineSettings {
      bApplyOutlineToDropShadows=false
      bMiteredCorners=false
      bSeparateFillAlpha=false
      outlineMaterial=null
      outlineSize=0
      outlineColor=<rgba 0 0 0 0.15>
     }
    }
    shadowOffset=<0 0>
   }
   slot {class=VerticalBoxSlot, properties {padding=<ltrb 30 12 30 0>}}
  }
  Button_Done {
   widgetClassPath=&WBP_Sidebar_ButtonGeneric
   bIsVariable=true
   parent=Box_ContentEmailSent
   widget=Button_Done
   properties {
    bInteractableWhenSelected=false
    bOverride_Cursor=true
    bSelectable=false
    bShouldSelectUponReceivingFocus=false
    minHeight=48
    minHeightOverride=60
    minWidthOverride=300
    mobileMinWidth=300
    text=Done
    visibility=Visible
    hoveredSlateSoundOverride {resourceObject=&Fortnite_UI_MainMenu_Tab_Hover_01_Cue}
    pressedSlateSoundOverride {resourceObject=&Fortnite_UI_MainMenu_Generic_01_Cue}
   }
   slot {class=VerticalBoxSlot, properties {horizontalAlignment=HAlign_Center, padding=<ltrb 0 48 0 0>}}
  }
 ]
}
