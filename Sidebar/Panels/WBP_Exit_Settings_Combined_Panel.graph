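# Sidebar panel widget combining settings, exit, leave and island/social buttons, shown in menus across lobby, Creative and Save the World.
# Button clicks route to FortSidebarSettingsVM actions, sidebar navigation pushes, confirmation view models or async-created modal panels.
# Helper functions derive Creative start/end game state from the minigame and build mission-leave dialogs with a rewards list.
# Delegates refresh visibility on input method, moderator mode and pawn changes.
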WidgetBlueprint /Sidebar/Panels/WBP_Exit_Settings_Combined_Panel
graphs [
 BP_GetDesiredFocusTarget {nodes [
   Get Desired Focus Target()
   n1 = Get Button_StartGame() -> (Button_StartGame:WBP UIKit Button Generic Object Reference)
   > n2 = Branch(Condition=IsVisible(self=@n1))
   > Return Node!(execute, ReturnValue:Widget Object Reference=@n1)
   n3 = Get!(execute:Exec=@n2.else) -> (then:Exec, else:Exec, LastFocusedButton:User Widget Object Reference)
   > n4 = Branch(Condition=IsVisible(self=@n3))
   > Return Node(ReturnValue=Get LastFocusedButton() -> (LastFocusedButton:User Widget Object Reference))
   n5 = For Each Loop(
    Exec=[@n4.else, @n3.else]
    Array:Array of Widget Object References=GetAllChildren(self=Get VBox_Exit() -> (VBox_Exit:Vertical Box Object Reference))
   ) -> (Array Element:Widget Object Reference)
   > n6 = Cast To CommonButtonBase!(execute, Object=@n5) -> (then, CastFailed, AsCommon Button Base:Common Button Base Object Reference) {is=title K2Node_DynamicCast}
   > Branch(Condition=IsVisible(self=@n6))
   > Return Node(ReturnValue=@n5)
   Return Node(execute=@n5.Completed, ReturnValue=Get Button_SupportAndRefunds() -> (Button_SupportAndRefunds:WBP Sidebar Button Generic Object Reference))
  ]}
 ChooseRecommendOrNotInterested {nodes [
   ChooseRecommendOrNotInterested!() -> (then, TargetButton:WBP Sidebar Button Generic Object Reference, OppositeButton:WBP Sidebar Button Generic Object Reference) {is=title K2Node_FunctionEntry}
   n1 = Get OppositeButton() -> (OppositeButton:WBP Sidebar Button Generic Object Reference)
   > Set ButtonType!(execute, ButtonType:E_UI_Sidebar_Button_Type Enum=NewEnumerator0, self:WBP Sidebar Button Generic Object Reference=@n1) -> (then, Output_Get:E_UI_Sidebar_Button_Type Enum=NewEnumerator0)
   > InitializeButtonVisuals(self:WBP Sidebar Button Generic Object Reference=@n1)
   > n2 = Set IsOneButtonSelected(IsOneButtonSelected:bool=GetIsEnabled!(self:Widget Object Reference=@n1) -> (ReturnValue:bool)) -> (Output_Get:bool)
   > SetIsEnabled(self=@n1, bInIsEnabled=NOT Boolean(A=@n2))
   n3 = Get TargetButton() -> (TargetButton:WBP Sidebar Button Generic Object Reference)
   > Set ButtonType(
    ButtonType=Select(
     Option 0:E_UI_Sidebar_Button_Type Enum=NewEnumerator0
     Option 1:E_UI_Sidebar_Button_Type Enum=NewEnumerator2
     Index=Get IsOneButtonSelected() -> (IsOneButtonSelected:bool)
    ) -> (ReturnValue:E_UI_Sidebar_Button_Type Enum)
    self=@n3
   ) -> (Output_Get=NewEnumerator0)
   > InitializeButtonVisuals(self=@n3)
  ]}
 CreateRewardsListWidget {nodes [
   CreateRewardsListWidget() {is=title K2Node_FunctionEntry}
   > n1 = Cast To FortGameStateZone!(execute, Object=GetGameState!() -> (ReturnValue:Game State Base Object Reference)) -> (then, CastFailed, AsFort Game State Zone:Fort Game State Zone Object Reference) {is=title K2Node_DynamicCast}
   > Set GameStateZone(GameStateZone:Fort Game State Zone Object Reference=@n1) -> (Output_Get:Fort Game State Zone Object Reference)
   > n2 = Create Widget!(
    execute:Exec
    Class:User Widget Class Reference=/Sidebar/Panels/Overlay/WBP_Confirmation_RewardsList.WBP_Confirmation_RewardsList_C
    OwningPlayer:Player Controller Object Reference=GetOwningPlayer()
    ShowMissionAlertMessage:bool=False
    Use Verbose Mode:bool=False
   ) -> (then:Exec, ReturnValue:WBP Confirmation Rewards List Object Reference)
   > Set RewardsList(RewardsList:WBP Confirmation Rewards List Object Reference=@n2) -> (Output_Get:WBP Confirmation Rewards List Object Reference)
   > n3 = GetActiveModifiers!(execute, self:Fort Game State Zone Object Reference=Get GameStateZone() -> (GameStateZone:Fort Game State Zone Object Reference)) -> (then, OutActiveModifiers:Array of Fort Gameplay Modifier Item Definition Object References)
   > Update!(
    execute
    self:WBP Confirmation Rewards List Object Reference=Get RewardsList() -> (RewardsList:WBP Confirmation Rewards List Object Reference)
    Mission Rewards:Array of Fort Mission Reward Info Structures=GetMissionRewardsInfo!(self:Fort HUDContext Object Reference=Get FortHUDContext!() -> (ReturnValue:Fort HUDContext Object Reference) {is=title K2Node_GetBlueprintContextEx}) -> (RewardList:Array of Fort Mission Reward Info Structures)
    Gameplay Modifiers:Array of Fort Gameplay Modifier Item Definition Object References=@n3
   ) -> (then)
   > Return Node(ReturnValue=Get RewardsList())
  ]}
 Creative_Get Start Game Button State {nodes [
   Creative_Get Start Game Button State() {is=title K2Node_FunctionEntry}
   > SetVisibility(self=Get Button_StartGame(), InVisibility=Collapsed)
   n1 = GetMinigame!(ContextActor:Actor Object Reference=GetOwningPlayer()) -> (ReturnValue:Fort Minigame Object Reference)
   > Is Valid(InputObject=@n1)
   > Bind Event to State Changed Delegate(self:Fort Minigame Object Reference=@n1, Delegate=Create Event()) {is=title K2Node_AddDelegate}
   > Set Creative_Machine(Creative_Machine:Fort Minigame Settings Building Object Reference=GetMinigameSettingsMachine!(ContextActor:Actor Object Reference=@n1) -> (ReturnValue:Fort Minigame Settings Building Object Reference)) -> (Output_Get:Fort Minigame Settings Building Object Reference)
   > n2 = Cast To FortPlayerControllerAthena(Object=GetOwningPlayer())
   > Is Valid(InputObject=@n2)
   n3 = GetMinigame!(self:Fort Player Controller Gameplay Object Reference=@n2) -> (ReturnValue)
   > Is Valid(InputObject=@n3)
   > n4 = Branch(Condition=IsCuratedHub!(self:Fort Minigame Object Reference=@n3) -> (ReturnValue:bool))
   n5 = Branch(execute=@n4.else, Condition=NOT Boolean(A=CanStartMinigame!(self:Fort Minigame Object Reference=@n3) -> (ReturnValue:bool)))
   SetVisibility(execute=[@n7.NewEnumerator0, @n7.NewEnumerator3, @n7.NewEnumerator6, @n5.then], self=Get Button_StartGame(), InVisibility=Collapsed)
   n6 = Set GameButtonState(
    execute=@n5.else
    GameButtonState:MinigameButtonsState Enum=Select!(
     PreGame:MinigameButtonsState Enum=NewEnumerator2
     Setup:MinigameButtonsState Enum=NewEnumerator6
     Transitioning:MinigameButtonsState Enum=NewEnumerator6
     WaitingForCameras:MinigameButtonsState Enum=NewEnumerator6
     Warmup:MinigameButtonsState Enum=NewEnumerator1
     InProgress:MinigameButtonsState Enum=NewEnumerator1
     PostGameTimeDilation:MinigameButtonsState Enum=NewEnumerator6
     PostRoundEnd:MinigameButtonsState Enum=NewEnumerator6
     PostGameEnd:MinigameButtonsState Enum=NewEnumerator3
     PostGameAbandon:MinigameButtonsState Enum=NewEnumerator3
     PostGameReset:MinigameButtonsState Enum=NewEnumerator6
     Teardown:MinigameButtonsState Enum=NewEnumerator6
     Index:EFortMinigameState Enum=Get CurrentState!(self:Fort Minigame Object Reference=@n3) -> (CurrentState:EFortMinigameState Enum=PreGame)
    ) -> (ReturnValue:MinigameButtonsState Enum)
   ) -> (Output_Get:MinigameButtonsState Enum=NewEnumerator6)
   > n7 = Switch on MinigameButtonsState!(execute, Selection:MinigameButtonsState Enum=@n6) -> (
    NewEnumerator6:Exec
    NewEnumerator0:Exec
    NewEnumerator1:Exec
    NewEnumerator2:Exec
    NewEnumerator3:Exec
    NewEnumerator4:Exec
    NewEnumerator5:Exec
   ) {is=title K2Node_SwitchEnum}
   n8 = Get Button_StartGame()
   SetText!(execute=@n7.NewEnumerator1, self:WBP UIKit Button Generic Base Object Reference=@n8, Text:Text="LOCTEXT(\"End Game\")") -> (then)
   SetText(execute=@n7.NewEnumerator2, self=@n8, Text="LOCTEXT(\"Start Game\")")
  ]}
 Creative_OnMinigameStateChanged {nodes [
   Creative_OnMinigameStateChanged!() -> (then, Minigame:Fort Minigame Object Reference, MinigameState:EFortMinigameState Enum=PreGame) {is=title K2Node_FunctionEntry}
   > Creative_Get Start Game Button State() {is=title K2Node_CallFunction}
   > Update Minigame Visibility()
  ]}
 Creative_isCreativeOptionsNotEmpty {nodes [
   Creative_isCreativeOptionsNotEmpty() {is=title K2Node_FunctionEntry}
   > n1 = For Each Loop with Break!(
    Exec:Exec
    Array:Array of Widget Object References=GetAllChildren(self=Get Vbox_CreativeOptions() -> (Vbox_CreativeOptions:Vertical Box Object Reference))
    Break:Exec=@n3.then
   ) -> (LoopBody:Exec, Array Element:Widget Object Reference, Array Index:int, Completed:Exec)
   Return Node!(execute=@n1.Completed, isCreativeOptionsNotEmpty:bool=Get Has Visible Button() -> (Has Visible Button:bool))
   > n2 = Branch(Condition="Not Equal (Enum)"!(
     A:ESlateVisibility Enum=Get Visibility!(self:Widget Object Reference=@n1) -> (Visibility:ESlateVisibility Enum=Visible)
     B:ESlateVisibility Enum=Collapsed
    ) -> (ReturnValue:bool))
   > n3 = Set Has Visible Button(Has Visible Button:bool=true) -> (Output_Get:bool)
   Set Has Visible Button(execute=@n2.else, Has Visible Button=false)
  ]}
 EventGraph {nodes [
   "On Clicked (Button_Exit)"!() -> (OutputDelegate, then, Button:Common Button Base Object Reference) {is=title K2Node_ComponentBoundEvent}
   > NavigateToConfirmPanel!(execute, self, Class:Sidebar Confirmation Panel VM Class Reference=/Sidebar/Panels/Overlay/Data/ConfirmationVM_Exit.ConfirmationVM_Exit_C) -> (then)
   "On Clicked (Button_LeaveExperience)"!() -> (OutputDelegate, then, Button:Common Button Base Object Reference) {is=title K2Node_ComponentBoundEvent}
   > NavigateToConfirmPanel(Class=/Sidebar/Panels/Overlay/Data/ConfirmationVM_BackToLobby.ConfirmationVM_BackToLobby_C)
   "On Clicked (Button_Settings)"!() -> (OutputDelegate, then, Button:Common Button Base Object Reference) {is=title K2Node_ComponentBoundEvent}
   > RequestCloseSidebar(self:Fort Sidebar VM Object Reference=Get FortSidebarVM() -> (FortSidebarVM:Fort Sidebar VM Object Reference))
   > RequestOpenSettings!(
    execute
    self:Fort Settings Context Object Reference=Get FortSettingsContext!() -> (ReturnValue:Fort Settings Context Object Reference) {is=title K2Node_GetBlueprintContextEx}
    ContextWidget:Common User Widget Object Reference=Self-Reference()
    SettingsTabId:Name=None
    SettingToSelect:String
   ) -> (then)
   "On Clicked (Button_TimeLimits)"!() -> (OutputDelegate, then, Button:Common Button Base Object Reference) {is=title K2Node_ComponentBoundEvent}
   > n1 = Construct Object from Class!(
    execute
    Class=/Sidebar/Shared/Navigation/BP_NavigationPoint.BP_NavigationPoint_C
    self
    SocialUser:Fort Social User Object Reference
    ConversationId:Name=None
    ConfirmationVM:Sidebar Confirmation Panel VM Object Reference
    ReservedSpots:int=0
    SocialItemVM:Fort Social Item VM Object Reference
    NavigationTag:Gameplay Tag Structure=(TagName="UI.StateTree.TransitionTag.Sidebar.Settings.TimeLimits")
   ) -> (then, ReturnValue:BP Navigation Point Object Reference)
   > PushNavigation!(
    execute
    self:Fort Sidebar Navigation VM Object Reference=Get FortSidebarNavigationVM() -> (FortSidebarNavigationVM:Fort Sidebar Navigation VM Object Reference)
    NavigationPoint:Fort Sidebar Navigation Point Object Reference=@n1
   ) -> (then)
   "On Clicked (Button_ControlsHelp)"!() -> (OutputDelegate, then, Button:Common Button Base Object Reference) {is=title K2Node_ComponentBoundEvent}
   n2 = GetSoftWidgetClassForAlias!(
    self:Fort UIData Configuration Context Object Reference=Get FortUIDataConfigurationContext!() -> (ReturnValue:Fort UIData Configuration Context Object Reference) {is=title K2Node_GetBlueprintContextEx}
    WidgetAlias:"Gameplay Tag Structure (by ref)"=MakeLiteralGameplayTag!(Value:Gameplay Tag Structure=(TagName="MobileHUD.TouchControlsHelp")) -> (ReturnValue:Gameplay Tag Structure)
   ) -> (ReturnValue:User Widget Soft Class Reference)
   > n3 = CreateWidgetAsync!(
    execute:Exec
    UserWidgetSoftClass:User Widget Soft Class Reference=Select!(
     Option 0:User Widget Soft Class Reference=Get MobileControlsHelpClass() -> (MobileControlsHelpClass:User Widget Soft Class Reference)
     Option 1:User Widget Soft Class Reference=@n2
     Index:bool=IsValidSoftClassReference!(SoftClassReference:"Object Soft Class Reference (by ref)"=@n2) -> (ReturnValue:bool)
    ) -> (ReturnValue:User Widget Soft Class Reference)
    OwningPlayer:Player Controller Object Reference=GetOwningPlayer()
    bSuspendInputUntilComplete:bool=true
   ) -> (then:Exec, OnComplete:Exec, OnFailure:Exec, UserWidget:User Widget Object Reference) {is=title K2Node_AsyncAction}
   TryPushWidgetAsModalPanel!(execute=@n3.OnComplete, PushWidget:User Widget Object Reference=@n3) -> (then)
   > RequestCloseSidebar(self=Get FortSidebarVM())
   "On Clicked (Button_Cheats)"!() -> (OutputDelegate, then, Button:Common Button Base Object Reference) {is=title K2Node_ComponentBoundEvent}
   > RequestCloseSidebar(self=Get FortSidebarVM())
   > n4 = CreateWidgetAsync(
    UserWidgetSoftClass=/FortniteCoreUI/Settings/SettingsScreen_GameCheats.SettingsScreen_GameCheats_C
    OwningPlayer=GetOwningPlayer()
    bSuspendInputUntilComplete=true
   )
   TryPushWidgetAsActivatablePanelForPlayer!(
    execute=@n4.OnComplete
    OwningPlayer:Player Controller Object Reference=GetOwningPlayer()
    PushWidget:User Widget Object Reference=@n4
    Header:EActivatePanelOption Enum=PlatformDefault
    Footer:EActivatePanelOption Enum=Show
    ChatWidget:EActivatePanelOption Enum=PlatformDefault
   ) -> (then)
   "On Clicked (Button_Locker)"!() -> (OutputDelegate, then, Button:Common Button Base Object Reference) {is=title K2Node_ComponentBoundEvent}
   > RequestCloseSidebar(self=Get FortSidebarVM())
   > ExecuteAction!(
    execute
    self:Fort Sidebar Settings VM Object Reference=Get FortSidebarSettingsVM() -> (FortSidebarSettingsVM:Fort Sidebar Settings VM Object Reference)
    SidebarAction:ESidebarAction Enum=Locker
   ) -> (then)
   "On Clicked (Button_HudLayoutTool)"!() -> (OutputDelegate, then, Button:Common Button Base Object Reference) {is=title K2Node_ComponentBoundEvent}
   n5 = Get FortUIDataConfigurationContext()
   n6 = GetSoftWidgetClassForAlias(self=@n5, WidgetAlias=MakeLiteralGameplayTag(Value=(TagName="MobileHUD.HUDLayoutTool")))
   n7 = Select(
    Option 0=GetHUDLayoutToolV2SoftClass!(self:Fort UIData Configuration Context Object Reference=@n5) -> (ReturnValue:Fort Mobile HUDLayout Tool V2 Soft Class Reference)
    Option 1=@n6
    Index=IsValidSoftClassReference(SoftClassReference=@n6)
   )
   > Branch(Condition=IsValidSoftClassReference(SoftClassReference=@n7))
   > n8 = CreateWidgetAsync(UserWidgetSoftClass=@n7, OwningPlayer=GetOwningPlayer(), bSuspendInputUntilComplete=true)
   TryPushWidgetAsActivatablePanel!(
    execute=@n8.OnComplete
    PushWidget:User Widget Object Reference=@n8
    Header:EActivatePanelOption Enum=Show
    Footer:EActivatePanelOption Enum=Show
    ChatWidget:EActivatePanelOption Enum=Hide
   ) -> (then)
   > RequestCloseSidebar(self=Get FortSidebarVM())
   n9 = "On Text Commited (InputField_Console)"!() -> (OutputDelegate, then, Text:Text, Commit method:ETextCommit Enum=Default) {is=title K2Node_ComponentBoundEvent}
   > n10 = Switch on ETextCommit!(execute, Selection:ETextCommit Enum=@n9.Commit method) -> (Default:Exec, OnEnter:Exec, OnUserMovedFocus:Exec, OnCleared:Exec) {is=title K2Node_SwitchEnum}
   ExecuteConsoleCommand!(execute=@n10.OnEnter, Command="To String (Text)"(InText=@n9.Text), SpecificPlayer) -> (then)
   "On Clicked (Button_ModeratorMode)"!() -> (OutputDelegate, then, Button:Common Button Base Object Reference) {is=title K2Node_ComponentBoundEvent}
   > Branch(Condition=Get bAllowCreativeModeratorMode!(self:Fort Runtime Options Object Reference=GetRuntimeOptions()) -> (bAllowCreativeModeratorMode:bool))
   > n11 = Cast To FortPlayerControllerGameplay!(execute, Object=GetOwningPlayer()) -> (then, CastFailed, AsFort Player Controller Gameplay:Fort Player Controller Gameplay Object Reference) {is=title K2Node_DynamicCast}
   > SetCreativeModeratorModeActive!(
    execute
    self:Fort Player Controller Gameplay Object Reference=@n11
    bActive:bool=NOT Boolean(A=Get Is Creative Moderator Mode Active() -> (Is Creative Moderator Mode Active:bool))
   ) -> (then)
   > Update Moderator Mode Visuals!(execute, self, is Loading:bool=true, is Moderator Mode Active:bool=Get Is Creative Moderator Mode Active()) -> (then)
   "On Clicked (Button_Help)"!() -> (OutputDelegate, then, Button:Common Button Base Object Reference) {is=title K2Node_ComponentBoundEvent}
   > RequestCloseSidebar(self=Get FortSidebarVM())
   > ExecuteAction(self=Get FortSidebarSettingsVM(), SidebarAction=StwHelp)
   "On Clicked (Button_Boosts)"!() -> (OutputDelegate, then, Button:Common Button Base Object Reference) {is=title K2Node_ComponentBoundEvent}
   > RequestCloseSidebar(self=Get FortSidebarVM())
   > ExecuteAction(self=Get FortSidebarSettingsVM(), SidebarAction=StwBoosts)
   "On Clicked (Button_LeaveAction)"!() -> (OutputDelegate, then, Button:Common Button Base Object Reference) {is=title K2Node_ComponentBoundEvent}
   > n12 = Switch on EAthenaMainMenuLeaveButtonText!(
    execute
    Selection:EAthenaMainMenuLeaveButtonText Enum=Get Current Leave Action() -> (Current Leave Action:EAthenaMainMenuLeaveButtonText Enum=LeaveAction_ReturnToLobby)
   ) -> (
    LeaveAction_ReturnToLobby:Exec
    LeaveAction_Replay:Exec
    LeaveAction_PlaylistCreative:Exec
    LeaveAction_PlaylistCreative_LTM:Exec
    LeaveAction_PlaylistDefault:Exec
    LeaveAction_InParty:Exec
    LeaveAction_Mission:Exec
    LeaveAction_StormShield:Exec
    LeaveAction_Default:Exec
   ) {is=title K2Node_SwitchEnum}
   n13 = Get!(execute=[
     @n12.LeaveAction_Default
     @n12.LeaveAction_InParty
     @n12.LeaveAction_PlaylistCreative
     @n12.LeaveAction_PlaylistCreative_LTM
     @n12.LeaveAction_PlaylistDefault
     @n12.LeaveAction_Replay
    ]) -> (then, else, Leave Action Class:Sidebar Confirmation Panel VM Class Reference)
   > NavigateToConfirmPanel(Class=@n13)
   n14 = CreateRewardsListWidget!(execute=[@n12.LeaveAction_Mission, @n12.LeaveAction_StormShield], self) -> (then, ReturnValue:Widget Object Reference) {is=title K2Node_CallFunction}
   n15 = GetOutpostConfirmationDialogText!() -> (OutTitle:Text, OutDescription:Text)
   > n16 = ShowSimpleConfirmationDialog_CustomInput!(
    execute:Exec
    Icon:Slate Brush Structure
    Title:Text=@n15
    Message:Text=@n15.OutDescription
    ConfirmAction:Name=Quit
    DeclineAction:Name=Cancel
    AdditionalContent:Widget Object Reference=@n14
    LeftAdditionalContent:Widget Object Reference
    Cancelable:bool=false
    TargetPlayer:Local Player Object Reference
   ) -> (then:Exec, DialogResult:Exec, Result:EFortDialogResult Enum, ResultName:Name) {is=title K2Node_AsyncAction}
   Branch(execute=@n16.DialogResult, Condition="Equal (Enum)"(A:EFortDialogResult Enum=@n16, B:EFortDialogResult Enum=Confirmed))
   > ExecuteAction(self=Get FortSidebarSettingsVM(), SidebarAction=AthenaMainMenuLeave)
   ExecuteAction(execute=@n12.LeaveAction_ReturnToLobby, self=Get FortSidebarSettingsVM(), SidebarAction=AthenaMainMenuLeave)
   "On Clicked (Button_Reporting)"!() -> (OutputDelegate, then, Button:Common Button Base Object Reference) {is=title K2Node_ComponentBoundEvent}
   > n17 = Construct Object from Class(
    Class=&BP_NavigationPoint_C
    ConversationId=None
    ReservedSpots=0
    NavigationTag=(TagName="UI.StateTree.TransitionTag.Sidebar.Settings.Reporting")
   )
   > PushNavigation(self=Get FortSidebarNavigationVM(), NavigationPoint=@n17)
   "On Clicked (Button_ManageUnlockedContent)"!() -> (OutputDelegate, then, Button:Common Button Base Object Reference) {is=title K2Node_ComponentBoundEvent}
   > ExecuteAction(self=Get FortSidebarSettingsVM(), SidebarAction=ManageUnlockedContent)
   > RequestCloseSidebar(self=Get FortSidebarVM())
   "On Clicked (Button_Respawn)"!() -> (OutputDelegate, then, Button:Common Button Base Object Reference) {is=title K2Node_ComponentBoundEvent}
   > n18 = CreateWidgetAsync(
    UserWidgetSoftClass=Get RespawnClass() -> (RespawnClass:User Widget Soft Class Reference)
    OwningPlayer=GetOwningPlayer()
    bSuspendInputUntilComplete=true
   )
   TryPushWidgetAsModalPanel(execute=@n18.OnComplete, PushWidget=@n18)
   > RequestCloseSidebar(self=Get FortSidebarVM())
   "On Clicked (Button_TeamSelect)"!() -> (OutputDelegate, then, Button:Common Button Base Object Reference) {is=title K2Node_ComponentBoundEvent}
   > n19 = CreateWidgetAsync(
    UserWidgetSoftClass=Get TeamSelectWidgetClass() -> (TeamSelectWidgetClass:User Widget Soft Class Reference)
    OwningPlayer=GetOwningPlayer()
    bSuspendInputUntilComplete=true
   )
   TryPushWidgetAsModalPanel(execute=@n19.OnComplete, PushWidget=@n19)
   > RequestCloseSidebar(self=Get FortSidebarVM())
   "On Clicked (Button_IslandSettings)"!() -> (OutputDelegate, then, Button:Common Button Base Object Reference) {is=title K2Node_ComponentBoundEvent}
   > RequestCloseSidebar(self=Get FortSidebarVM())
   > n20 = Cast To FortPlayerController(Object=GetOwningPlayer())
   > n21 = Branch(Condition=ShouldAddTabsToMapScreen!(PlayerController:Player Controller Object Reference=@n20) -> (ReturnValue:bool))
   n22 = GetMiniMapComponent!(self:Fort Player Controller Object Reference=@n20) -> (ReturnValue:Fort Controller Component Minimap Object Reference)
   > Is Valid(InputObject=@n22)
   > OpenFullScreenMapTab!(execute, self:Fort Controller Component Minimap Object Reference=@n22, TabId:Name=GetMyIslandTabID!() -> (ReturnValue:Name=None)) -> (then)
   OpenInventory!(execute=@n21.else, self:Fort Player Controller Object Reference=@n20, InventoryTabId:Name=GetMyIslandTabID() -> (ReturnValue=None)) -> (then)
   "On Clicked (Button_StartGame)"!() -> (OutputDelegate, then, Button:Common Button Base Object Reference) {is=title K2Node_ComponentBoundEvent}
   > n23 = Switch on MinigameButtonsState(Selection=Get GameButtonState() -> (GameButtonState:MinigameButtonsState Enum=NewEnumerator6))
   n24 = Get Creative_Machine() -> (Creative_Machine:Fort Minigame Settings Building Object Reference)
   Is Valid(exec=@n23.NewEnumerator1, InputObject=@n24)
   > n25 = AbandonMinigame_BP(self:Fort Minigame Settings Building Object Reference=@n24)
   RequestCloseSidebar(execute=[@n27.then, @n25.then], self=Get FortSidebarVM())
   n26 = Get Creative_Machine()
   Is Valid(exec=@n23.NewEnumerator2, InputObject=@n26)
   > n27 = StartMinigame_BP(self:Fort Minigame Settings Building Object Reference=@n26)
   "On Clicked (Button_BackToHub)"!() -> (OutputDelegate, then, Button:Common Button Base Object Reference) {is=title K2Node_ComponentBoundEvent}
   > ExecuteAction(self=Get FortSidebarSettingsVM(), SidebarAction=BackToHub)
   > RequestCloseSidebar(self=Get FortSidebarVM())
   "On Clicked (Button_ParentalControls)"!() -> (OutputDelegate, then, Button:Common Button Base Object Reference) {is=title K2Node_ComponentBoundEvent}
   > RequestToOpenParentalControls!(execute) -> (then)
   > RequestCloseSidebar(self=Get FortSidebarVM())
   "On Clicked (Button_Legal)"!() -> (OutputDelegate, then, Button:Common Button Base Object Reference) {is=title K2Node_ComponentBoundEvent}
   > n28 = Construct Object from Class(
    Class=&BP_NavigationPoint_C
    ConversationId=None
    ReservedSpots=0
    NavigationTag=(TagName="UI.StateTree.TransitionTag.Sidebar.Settings.Legal")
   )
   > PushNavigation(self=Get FortSidebarNavigationVM(), NavigationPoint=@n28)
   "On Clicked (Button_SupportAndRefunds)"!() -> (OutputDelegate, then, Button:Common Button Base Object Reference) {is=title K2Node_ComponentBoundEvent}
   > n29 = Construct Object from Class(
    Class=&BP_NavigationPoint_C
    ConversationId=None
    ReservedSpots=0
    NavigationTag=(TagName="UI.StateTree.TransitionTag.Sidebar.Settings.SupportAndRefunds")
   )
   > PushNavigation(self=Get FortSidebarNavigationVM(), NavigationPoint=@n29)
   "On Clicked (Button_FollowDeveloper)"!() -> (OutputDelegate, then, Button:Common Button Base Object Reference) {is=title K2Node_ComponentBoundEvent}
   > ExecuteAction(self=Get FortSidebarSettingsVM(), SidebarAction=FollowDeveloper)
   > PlayFavoriteAudio!(
    execute
    self
    IsFavorite:bool=Get bIsFollowingDeveloper!(self:Fort Sidebar Settings VM Object Reference=Get FortSidebarSettingsVM()) -> (bIsFollowingDeveloper:bool)
   ) -> (then)
   "On Clicked (Button_LeaveParty)"!() -> (OutputDelegate, then, Button:Common Button Base Object Reference) {is=title K2Node_ComponentBoundEvent}
   > Set SkipMainMenuLeave(SkipMainMenuLeave:bool=true) -> (Output_Get:bool)
   > NavigateToConfirmPanel(Class=/Sidebar/Panels/Overlay/Data/ConfirmationVM_LeaveParty.ConfirmationVM_LeaveParty_C)
   "On Clicked (Button_Sleep)"!() -> (OutputDelegate, then, Button:Common Button Base Object Reference) {is=title K2Node_ComponentBoundEvent}
   > ExecuteAction(self=Get FortSidebarSettingsVM(), SidebarAction=Parsley)
   "On Clicked (Button_SaveIsland)"!() -> (OutputDelegate, then, Button:Common Button Base Object Reference) {is=title K2Node_ComponentBoundEvent}
   > ExecuteAction(self=Get FortSidebarSettingsVM(), SidebarAction=Favorite)
   > PlayFavoriteAudio(IsFavorite=Get bIsFavorite!(self:Fort Sidebar Settings VM Object Reference=Get FortSidebarSettingsVM()) -> (bIsFavorite:bool))
   "On Clicked (Button_Like)"!() -> (OutputDelegate, then, Button:Common Button Base Object Reference) {is=title K2Node_ComponentBoundEvent}
   > ChooseRecommendOrNotInterested!(
    execute
    self
    TargetButton:WBP Sidebar Button Generic Object Reference=Get Button_Like() -> (Button_Like:WBP Sidebar Button Generic Object Reference)
    OppositeButton:WBP Sidebar Button Generic Object Reference=Get Button_Dislike() -> (Button_Dislike:WBP Sidebar Button Generic Object Reference)
   ) -> (then) {is=title K2Node_CallFunction}
   > ExecuteAction(self=Get FortSidebarSettingsVM(), SidebarAction=Like)
   > n30 = Get!(execute) -> (then, else, Sidebar:WBP Sidebar Object Reference)
   > n31 = Set isLiked!(
    execute
    isLiked:bool=NOT Boolean(A=Get isLiked!(self:WBP Sidebar Object Reference=@n30) -> (isLiked:bool))
    self:WBP Sidebar Object Reference=@n30
   ) -> (then, Output_Get:bool)
   > PlayLikedAudio!(execute, self, IsLiked:bool=@n31) -> (then)
   "On Clicked (Button_Dislike)"!() -> (OutputDelegate, then, Button:Common Button Base Object Reference) {is=title K2Node_ComponentBoundEvent}
   > ChooseRecommendOrNotInterested(TargetButton=Get Button_Dislike(), OppositeButton=Get Button_Like())
   > ExecuteAction(self=Get FortSidebarSettingsVM(), SidebarAction=NotInterested)
   > n32 = Get()
   > n33 = Set isDisliked!(
    execute
    isDisliked:bool=NOT Boolean(A=Get isDisliked!(self:WBP Sidebar Object Reference=@n32) -> (isDisliked:bool))
    self:WBP Sidebar Object Reference=@n32
   ) -> (then, Output_Get:bool)
   > PlayDislikedAudio!(execute, self, IsDisliked:bool=@n33) -> (then)
   "On Clicked (Button_Logout)"!() -> (OutputDelegate, then, Button:Common Button Base Object Reference) {is=title K2Node_ComponentBoundEvent}
   > NavigateToConfirmPanel(Class=/Sidebar/Panels/Overlay/Data/ConfirmationVM_Logout.ConfirmationVM_Logout_C)
   n34 = Handle On Is Creative Moderator Mode Active Changed!() -> (OutputDelegate, then, bIsCreativeModeratorModeActive:bool=~) {is=title K2Node_CustomEvent}
   > n35 = Set Is Creative Moderator Mode Active(Is Creative Moderator Mode Active:bool=@n34.bIsCreativeModeratorModeActive) -> (Output_Get:bool)
   > Update Moderator Mode Visuals(is Loading=false, is Moderator Mode Active=@n35)
   On Possessed Pawn Changed!() -> (OutputDelegate, then, OldPawn:Pawn Object Reference, NewPawn:Pawn Object Reference) {is=title K2Node_CustomEvent}
   > Update Respawn Visibility()
   n36 = On Input Method Changed!() -> (OutputDelegate, then, New Input Type:ECommonInputType Enum) {is=title K2Node_CustomEvent}
   > SetVisibility(
    self=Get Button_Settings() -> (Button_Settings:WBP Sidebar Button Generic Object Reference)
    InVisibility=Select!(
     MouseAndKeyboard:ESlateVisibility Enum=Collapsed
     Gamepad:ESlateVisibility Enum=Visible
     Touch:ESlateVisibility Enum=Collapsed
     Count:ESlateVisibility Enum=Visible
     Index:ECommonInputType Enum=@n36.New Input Type
    ) -> (ReturnValue:ESlateVisibility Enum)
   )
   n37 = On Registry Button Clicked!() -> (OutputDelegate, then, Button:Common Button Base Object Reference) {is=title K2Node_CustomEvent}
   > n38 = Cast To WBP_Sidebar_ButtonGeneric!(execute, Object=@n37.Button) -> (then, CastFailed, AsWBP Sidebar Button Generic:WBP Sidebar Button Generic Object Reference) {is=title K2Node_DynamicCast}
   > n39 = For Each Loop!(Exec, Array:Array of Registry Button Structures=Get Registry Buttons() -> (Registry Buttons:Array of Registry Button Structures)) -> (LoopBody, Array Element_SettingRegistryName:Name, Array Element_SettingRegistryDisplayName:Text, Array Index, Completed)
   > Branch(Condition="Equal Exactly (Text)"!(
     A:"Text (by ref)"=@n39.Array Element_SettingRegistryDisplayName
     B:"Text (by ref)"=Get Text(self:WBP UIKit Button Generic Base Object Reference=@n38)
    ) -> (ReturnValue:bool))
   > ExecuteSettingRegistryAction!(execute, self:Fort Sidebar Settings VM Object Reference=Get FortSidebarSettingsVM(), SettingRegistryActionName:Name=@n39) -> (then)
   Event Destruct()
   n40 = Create Event()
   n41 = Get CommonInputSubsystem()
   > Unbind Event from On Input Method Changed(self=@n41, Delegate=@n40)
   n42 = Get CreativeHUDLContext!() -> (ReturnValue:Creative HUDLContext Object Reference) {is=title K2Node_GetBlueprintContextEx}
   > Unbind all Events from On Is Creative Moderator Mode Active Changed!(execute:Exec, self:Creative HUDLContext Object Reference=@n42) -> (then:Exec) {is=title K2Node_ClearDelegate}
   Event Construct()
   > Bind Event to On Input Method Changed(self=@n41, Delegate=@n40)
   > Bind Event to On Is Creative Moderator Mode Active Changed(self:Creative HUDLContext Object Reference=@n42, Delegate=Create Event()) {is=title K2Node_AddDelegate}
   > n43 = Cast To FortPlayerControllerGameplay(Object=GetOwningPlayer())
   > Handle On Is Creative Moderator Mode Active Changed!(
    execute
    self
    bIsCreativeModeratorModeActive:bool=IsCreativeModeratorModeActive!(self:Fort Player Controller Gameplay Object Reference=@n43) -> (ReturnValue:bool)
   ) -> (then) {is=title K2Node_CallFunction}
   > Bind Event to On Possessed Pawn Changed(self:Controller Object Reference=@n43, Delegate=Create Event()) {is=title K2Node_AddDelegate}
   > n44 = FindParentWidgetOfType!(execute, StartingWidget:Widget Object Reference=Self-Reference(), Type:Widget Class Reference=/Sidebar/WBP_Sidebar.WBP_Sidebar_C) -> (then, ReturnValue:WBP Sidebar Object Reference)
   > n45 = Set Sidebar(Sidebar:WBP Sidebar Object Reference=@n44) -> (Output_Get:WBP Sidebar Object Reference)
   > Update Support Buttons Visuals!(execute, self, isLiked:bool=Get isLiked(self=@n45), isDisliked:bool=Get isDisliked(self=@n45)) -> (then)
   Event On Activated()
   > n46 = Set Current Input Type(Current Input Type:ECommonInputType Enum=GetCurrentInputType(self=Get CommonInputSubsystem())) -> (Output_Get:ECommonInputType Enum)
   > On Input Method Changed!(execute, self, New Input Type:ECommonInputType Enum=@n46) -> (then) {is=title K2Node_CallFunction}
   n47 = Get FortSidebarSettingsVM()
   > SetupMainMenuUI(self:Fort Sidebar Settings VM Object Reference=@n47)
   > SetupExitUI(self:Fort Sidebar Settings VM Object Reference=@n47)
   > Update Settings Visibility!(execute:Exec) -> (then:Exec) {is=title K2Node_MacroInstance}
   > SetFavoriteButton!(execute, self, is Favorite:bool=Get bIsFavorite(self=Get FortSidebarSettingsVM())) -> (then)
   > SidebarFeatureFinishedLoading(self:Fort Sidebar VM Object Reference=Get FortSidebarVM())
   > SetLeavePartyButtonText()
   Event OnInitialized()
   > UpdateButtonTimeLimitText()
   > SetRenderTransform!(
    execute
    self:Widget Object Reference=Get Block_Icon!(self:WBP UIKit Button Generic Base Object Reference=Get Button_Dislike()) -> (Block_Icon:WBP UIKit Block Image Object Reference)
    InTransform_Translation:Vector 2D Structure=(X=0,Y=0)
    InTransform_Scale:Vector 2D Structure=(X=1,Y=1)
    InTransform_Shear:Vector 2D Structure=(X=0,Y=0)
    InTransform_Angle:float=180
   ) -> (then)
   > SetBlocksTiming!(
    execute
    self:UIKit Modular Button Object Reference=[Get Button_Like(), Get Button_Dislike(), Get Button_SaveIsland() -> (Button_SaveIsland:WBP Sidebar Button Generic Object Reference)]
    Timing:"UIKit Block Timing Structure (by ref)"=Make UIKit Block Timing!(
     Hovering:float=0
     Unhovering:float=0
     Focusing:float=0.1
     Unfocusing:float=0.15
     Pressing:float=0.1
     Releasing:float=0.1
     Disabling:float=0.1
     Enabling:float=0.1
     Selecting:float=0.15
     DeselectingIdle:float=0.1
     DeselectingFocused:float=0.1
    ) -> (UIKitBlockTiming:UIKit Block Timing Structure) {is=title K2Node_MakeStruct}
   ) -> (then)
   Event PreConstruct()
   > SetMobileStyle() {is=title K2Node_CallFunction}
  ]}
 Execute Leave Action {nodes [
   n1 = Execute Leave Action!() -> (then, Confrimation Action:Name) {is=title K2Node_FunctionEntry}
   > n2 = Branch(Condition="Equal (Name)"(A=@n1, B=LeaveAction))
   ExecuteAction(execute=[@n2.then, @n3.else], self=Get FortSidebarSettingsVM(), SidebarAction=AthenaMainMenuLeave)
   Branch(execute=@n2.else, Condition="Equal (Name)"(A=@n1, B=LeaveParty))
   > n3 = Branch(Condition=Get SkipMainMenuLeave() -> (SkipMainMenuLeave:bool))
   > ExecuteAction(self=Get FortSidebarSettingsVM(), SidebarAction=LeaveParty)
  ]}
 GetControlsHelpButtonDesiredVisibility {nodes [
   GetControlsHelpButtonDesiredVisibility() {is=title K2Node_FunctionEntry}
   n1 = IsUsingCompactUI()
   n2 = GetRuntimeOptions()
   n3 = IsCurrentContextProductTagAllowedByLists!(
    AllowedTagList:"Gameplay Tag Container Structure (by ref)"=Get ControlHelp_AllowedProductTags!(self:Fort Runtime Options Object Reference=@n2) -> (ControlHelp_AllowedProductTags:Gameplay Tag Container Structure)
    DisallowedTagList:"Gameplay Tag Container Structure (by ref)"=Get ControlHelp_DisallowedProductTags!(self:Fort Runtime Options Object Reference=@n2) -> (ControlHelp_DisallowedProductTags:Gameplay Tag Container Structure)
    bMatchDefaultTag:bool=true
   ) -> (ReturnValue:bool)
   > Return Node!(
    execute
    Visibility:ESlateVisibility Enum=Select!(Option 0:ESlateVisibility Enum=Collapsed, Option 1:ESlateVisibility Enum=SelfHitTestInvisible, Index:bool=AND Boolean(A=@n1, B=@n3)) -> (ReturnValue)
   )
  ]}
 GetHLTButtonDesiredVisibility {nodes [
   GetHLTButtonDesiredVisibility() {is=title K2Node_FunctionEntry}
   n1 = IsUsingCompactUI()
   n2 = AND Boolean(A=ShouldForceHUDLayoutToolButtonVisible!() -> (ReturnValue:bool), B=@n1)
   n3 = GetRuntimeOptions()
   n4 = AND Boolean(
    A=@n1
    B=IsCurrentContextProductTagAllowedByLists(
     AllowedTagList=GetHUDLayoutTool_AllowedProductTags!(self:Fort Runtime Options Object Reference=@n3) -> (OutTagContainer:Gameplay Tag Container Structure)
     DisallowedTagList=GetHUDLayoutTool_DisallowedProductTags!(self:Fort Runtime Options Object Reference=@n3) -> (OutTagContainer:Gameplay Tag Container Structure)
     bMatchDefaultTag=true
    )
   )
   > Return Node(Visibility=Select(Option 0=Collapsed, Option 1=SelfHitTestInvisible, Index=OR Boolean(A=@n2, B=@n4)))
  ]}
 GetOutpostConfirmationDialogText {nodes [
   GetOutpostConfirmationDialogText() {is=title K2Node_FunctionEntry}
   n1 = Get FortPartyContext!() -> (ReturnValue:Fort Party Context Object Reference) {is=title K2Node_GetBlueprintContextEx}
   > Set bIsInParty(bIsInParty:bool="integer > integer"(A=GetNumPlayersInLocalParty!(self:Fort Party Context Object Reference=@n1) -> (ReturnValue:int=0), B=1) {is=title K2Node_CallFunction}) -> (Output_Get:bool)
   > n2 = IsOutpostOwner!(execute, self) -> (then, ReturnValue:bool)
   > n3 = Branch(Condition=@n2)
   > Return Node!(
    execute
    OutTitle:Text=Get StormshieldLeaveTitle() -> (StormshieldLeaveTitle:Text)
    OutDescription:Text=Select(
     Option 0:Text="LOCTEXT(\"Leave the Storm Shield and return to Homebase.\")"
     Option 1:Text="LOCTEXT(\"Leave the Storm Shield and return to Homebase with your Party.\")"
     Index=Get bIsInParty() -> (bIsInParty:bool)
    ) -> (ReturnValue:Text)
   )
   n4 = IsInOutpostZone!(self:Fort Global UIContext Object Reference=Get FortGlobalUIContext!() -> (ReturnValue:Fort Global UIContext Object Reference) {is=title K2Node_GetBlueprintContextEx}) -> (ReturnValue:bool)
   n5 = Get MissionLeaveDescription() -> (MissionLeaveDescription:Text)
   n6 = Get StormshieldLeaveDescription() -> (StormshieldLeaveDescription:Text)
   n7 = Get FortGlobalUIContext()
   Return Node(
    execute=@n3.else
    OutTitle=Select(Option 0="LOCTEXT(\"Leave Mission?\")", Option 1="LOCTEXT(\"Leave Storm Shield?\")", Index=IsInOutpostZone(self=@n7))
    OutDescription=Select(
     Option 0=Select(Option 0=@n5, Option 1=@n6, Index=@n4)
     Option 1=Get HordModeLeaveDescription() -> (HordModeLeaveDescription:Text)
     Index=IsInNeighborhood!(self) -> (bIsInNeighborhood:bool)
    )
   )
  ]}
 IsInNeighborhood {nodes [
   IsInNeighborhood() {is=title K2Node_FunctionEntry}
   > n1 = Get Deployable Base Manager!(execute) -> (then, ReturnValue:Fort Deployable Base Manager Object Reference) {is=title K2Node_CallFunction}
   > n2 = Set DeployableBaseManager(DeployableBaseManager:Fort Deployable Base Manager Object Reference=@n1) -> (Output_Get:Fort Deployable Base Manager Object Reference)
   > n3 = Is Valid(InputObject=@n2)
   Return Node!(execute=@n3.Is Not Valid, bIsInNeighborhood:bool=false)
   n4 = Get DeployableBaseManager() -> (DeployableBaseManager:Fort Deployable Base Manager Object Reference)
   > Return Node(bIsInNeighborhood="Equal (Enum)"(
     A:EDeployableBaseUseType Enum=GetDeployableBaseUseType!(self:Fort Deployable Base Manager Object Reference=@n4) -> (ReturnValue:EDeployableBaseUseType Enum=Neighborhood)
     B:EDeployableBaseUseType Enum=Neighborhood
    ))
  ]}
 IsOutpostOwner {nodes [
   IsOutpostOwner() {is=title K2Node_FunctionEntry}
   n1 = GetComponentByClass!(
    self:Actor Object Reference=GetOwningPlayer()
    ComponentClass:Actor Component Class Reference=/Script/FortniteGame.FortControllerComponent_OwnedZone
   ) -> (ReturnValue:Fort Controller Component Owned Zone Object Reference)
   > n2 = Is Valid(InputObject=@n1)
   Return Node!(execute=@n2.Is Not Valid, ReturnValue:bool=false)
   > Return Node(ReturnValue=IsZoneOwner!(self:Fort Controller Component Owned Zone Object Reference=@n1) -> (ReturnValue:bool))
  ]}
 NavigateFromExpressSupportButtons {nodes [
   NavigateFromExpressSupportButtons!() -> (then, Navigation:EUINavigation Enum) {is=title K2Node_FunctionEntry}
   > n1 = Reverse for Each Loop!(Exec:Exec, Array:Array of Widget Object References=GetAllChildren(self=Get VBox_Exit())) -> (LoopBody:Exec, ArrayIndex:int, ArrayElement:Widget Object Reference, Completed:Exec) {is=title K2Node_MacroInstance}
   > n2 = Cast To CommonButtonBase(Object=@n1.ArrayElement)
   > Branch(Condition=IsVisible(self=@n2))
   > PrintString!(execute, InString=GetDisplayName(Object=@n1.ArrayElement), bPrintToScreen, bPrintToLog, TextColor, Duration, Key) -> (then)
   > Return Node(ReturnValue:Widget Object Reference=@n1.ArrayElement)
  ]}
 NavigateFromSettingsToTabGroup {nodes [
   NavigateFromSettingsToTabGroup!() -> (then, Navigation:EUINavigation Enum) {is=title K2Node_FunctionEntry}
   > n1 = FindParentWidgetOfType(StartingWidget=Self-Reference(), Type=&WBP_Sidebar_C)
   > n2 = NavigationToTabGroup!(execute, self:WBP Sidebar Object Reference=@n1, Navigation:EUINavigation Enum=Up) -> (then, ReturnValue:Widget Object Reference)
   > Return Node(ReturnValue=@n2)
  ]}
 NavigateToConfirmPanel {nodes [
   n1 = NavigateToConfirmPanel!() -> (then, Class:Sidebar Confirmation Panel VM Class Reference) {is=title K2Node_FunctionEntry}
   > n2 = Construct Object from Class!(execute, Class=@n1, self) -> (then, ReturnValue:Sidebar Confirmation Panel VM Object Reference)
   > NavigateToConfirmPanel!(
    execute
    SelfReference:Object Reference=Self-Reference()
    SidebarConfirmationVM:Sidebar Confirmation Panel VM Object Reference=@n2
    SidebarNavigationVM:Fort Sidebar Navigation VM Object Reference=Get FortSidebarNavigationVM()
   ) -> (then) {is=title K2Node_CallFunction}
  ]}
 OnSetLastFocusedButton {nodes [
   n1 = OnSetLastFocusedButton!() -> (then, Type:EUINavigation Enum, OldFocusedWidget:Widget Object Reference, NewFocusedWidget:Widget Object Reference) {is=title K2Node_FunctionEntry}
   n2 = GetOuterUserWidget!(Widget:Widget Object Reference=@n1.NewFocusedWidget) -> (ReturnValue:User Widget Object Reference)
   > Branch(Condition=IsValid(Object=@n2))
   > Cast To WBP_Sidebar_ButtonGeneric(Object=@n2)
   > Set LastFocusedButton(LastFocusedButton:User Widget Object Reference=@n2) -> (Output_Get:User Widget Object Reference)
  ]}
 PlayDislikedAudio {nodes [
   n1 = PlayDislikedAudio!() -> (then, IsDisliked:bool=~) {is=title K2Node_FunctionEntry}
   > PrintText!(
    execute
    InText=Format Text!(Format="LOCTEXT(\"PlayDislikedAudio, IsDisiked = {IsDisliked}\")", IsDisliked:bool=@n1) -> (Result)
    bPrintToScreen
    bPrintToLog
    TextColor
    Duration
    Key
   ) -> (then)
   > n2 = Branch(Condition=@n1)
   > PlaySound2D(Sound=Get Disliked Sound() -> (Disliked Sound:Sound Base Object Reference))
   PlaySound2D(execute=@n2.else, Sound=Get Undisliked Sound() -> (Undisliked Sound:Sound Base Object Reference))
  ]}
 PlayFavoriteAudio {nodes [
   n1 = PlayFavoriteAudio!() -> (then, IsFavorite:bool=~) {is=title K2Node_FunctionEntry}
   > PrintText(InText=Format Text!(Format="LOCTEXT(\"PlayFavoriteAudio, IsFavorite = {IsFavorite}\")", IsFavorite:bool=@n1) -> (Result))
   > n2 = Branch(Condition=@n1)
   > PlaySound2D(Sound=Get Favorited Sound() -> (Favorited Sound:Sound Base Object Reference))
   PlaySound2D(execute=@n2.else, Sound=Get Unfavorited Sound() -> (Unfavorited Sound:Sound Base Object Reference))
  ]}
 PlayLikedAudio {nodes [
   n1 = PlayLikedAudio!() -> (then, IsLiked:bool=~) {is=title K2Node_FunctionEntry}
   > PrintText(InText=Format Text!(Format="LOCTEXT(\"PlayLikedAudio, IsLiked = {IsLiked}\")", IsLiked:Wildcard) -> (Result))
   > n2 = Branch(Condition=@n1)
   > PlaySound2D(Sound=Get Liked Sound() -> (Liked Sound:Sound Base Object Reference))
   PlaySound2D(execute=@n2.else, Sound=Get Unliked Sound() -> (Unliked Sound:Sound Base Object Reference))
  ]}
 Set Button Visibilities From Actions {nodes [
   n1 = Set Button Visibilities From Actions!() -> (then, Sidebar Actions:Set of ESidebarAction Enums=SupportCreator) {is=title K2Node_FunctionEntry}
   > Set Exit Button Visibilities From Actions!(execute, self, Actions:Set of ESidebarAction Enums=@n1) -> (then) {is=title K2Node_CallFunction}
   > SetVisibility(
    self=Get Button_TimeLimits() -> (Button_TimeLimits:WBP Sidebar Button Generic Object Reference)
    InVisibility=Select(
     Option 0:ESlateVisibility Enum=Collapsed
     Option 1:ESlateVisibility Enum=Visible
     Index=Contains Item!(TargetSet:Set of ESidebarAction Enums=@n1, ItemToFind:"ESidebarAction Enum (by ref)"=Cilantro) -> (ReturnValue:bool) {is=title K2Node_CallFunction}
    ) -> (ReturnValue:ESlateVisibility Enum)
   )
   > SetVisibility(
    self=Get Button_ManageUnlockedContent() -> (Button_ManageUnlockedContent:WBP Sidebar Button Generic Object Reference)
    InVisibility=Select(Option 0=Collapsed, Option 1=Visible, Index=Contains Item(TargetSet=@n1, ItemToFind=ManageUnlockedContent))
   )
   > Creative_Get Start Game Button State()
   > SetVisibility(
    self=Get Button_ControlsHelp() -> (Button_ControlsHelp:WBP Sidebar Button Generic Object Reference)
    InVisibility=GetControlsHelpButtonDesiredVisibility!() -> (Visibility:ESlateVisibility Enum=Visible) {is=title K2Node_CallFunction}
   )
   > SetVisibility(
    self=Get Button_HudLayoutTool() -> (Button_HudLayoutTool:WBP Sidebar Button Generic Object Reference)
    InVisibility=GetHLTButtonDesiredVisibility!() -> (Visibility:ESlateVisibility Enum=Visible) {is=title K2Node_CallFunction}
   )
   > RequestRefreshFocus(self:Common Activatable Widget Object Reference)
   > SidebarFeatureFinishedLoading(self=Get FortSidebarVM())
  ]}
 Set Exit Button Visibilities From Actions {nodes [
   n1 = Set Exit Button Visibilities From Actions!() -> (then, Actions:Set of ESidebarAction Enums=SupportCreator) {is=title K2Node_FunctionEntry}
   > SetVisibility(
    self=Get Button_Exit() -> (Button_Exit:WBP Sidebar Button Generic Object Reference)
    InVisibility=Select(Option 0=Collapsed, Option 1=Visible, Index=Contains Item(TargetSet=@n1, ItemToFind=ExitGame))
   )
   > SetVisibility(
    self=Get Button_Sleep() -> (Button_Sleep:WBP Sidebar Button Generic Object Reference)
    InVisibility=Select(Option 0=Collapsed, Option 1=Visible, Index=Contains Item(TargetSet=@n1, ItemToFind=Parsley))
   )
   > SetVisibility(
    self=Get Button_LeaveAction() -> (Button_LeaveAction:WBP Sidebar Button Generic Object Reference)
    InVisibility=Select(Option 0=Collapsed, Option 1=Visible, Index=Contains Item(TargetSet=@n1, ItemToFind=AthenaMainMenuLeave))
   )
   > SetVisibility(
    self=Get Button_LeaveExperience() -> (Button_LeaveExperience:WBP Sidebar Button Generic Object Reference)
    InVisibility=Select(Option 0=Collapsed, Option 1=Visible, Index=Contains Item(TargetSet=@n1, ItemToFind=LeaveExperience))
   )
   n2 = Contains Item(TargetSet=@n1, ItemToFind=Favorite)
   > SetVisibility(self=Get Button_SaveIsland(), InVisibility=Select(Option 0=Collapsed, Option 1=Visible, Index=@n2))
   n3 = Contains Item(TargetSet=@n1, ItemToFind=Like)
   > SetVisibility(self=[Get Button_Like(), Get Button_Dislike()], InVisibility=Select(Option 0=Collapsed, Option 1=Visible, Index=@n3))
   n4 = Contains Item(TargetSet=@n1, ItemToFind=FollowDeveloper)
   > SetVisibility(
    self=Get Button_FollowDeveloper() -> (Button_FollowDeveloper:WBP Sidebar Button Generic Object Reference)
    InVisibility=Select(Option 0=Collapsed, Option 1=Visible, Index=@n4)
   )
   > SetVisibility(
    self=Get Vbox_ExpressSupportButtons() -> (Vbox_ExpressSupportButtons:Horizontal Box Object Reference)
    InVisibility=Select(Option 0=Collapsed, Option 1=SelfHitTestInvisible, Index=OR Boolean!(A=@n2, B=@n3, C:bool=@n4) -> (ReturnValue))
   )
   > SetVisibility(
    self=Get Button_Logout() -> (Button_Logout:WBP Sidebar Button Generic Object Reference)
    InVisibility=Select(Option 0=Collapsed, Option 1=Visible, Index=Contains Item(TargetSet=@n1, ItemToFind=Logout))
   )
   n5 = Contains Item(TargetSet=@n1, ItemToFind=LeaveParty)
   n6 = "Not Equal (Enum)"(
    A:EAthenaMainMenuLeaveButtonText Enum=GetAthenaMainMenuLeaveActionText!(self:Fort Sidebar Settings VM Object Reference=Get FortSidebarSettingsVM()) -> (ReturnValue:EAthenaMainMenuLeaveButtonText Enum=LeaveAction_ReturnToLobby)
    B:EAthenaMainMenuLeaveButtonText Enum=LeaveAction_Default
   )
   > SetVisibility(
    self=Get Button_LeaveParty() -> (Button_LeaveParty:WBP Sidebar Button Generic Object Reference)
    InVisibility=Select(Option 0=Collapsed, Option 1=Visible, Index=AND Boolean(A=@n5, B=@n6))
   )
   > RequestRefreshFocus()
  ]}
 Set Leave Action {nodes [
   n1 = Set Leave Action!() -> (then, Action:EAthenaMainMenuLeaveButtonText Enum=LeaveAction_ReturnToLobby) {is=title K2Node_FunctionEntry}
   > n2 = Set Current Leave Action(Current Leave Action:EAthenaMainMenuLeaveButtonText Enum=@n1) -> (Output_Get:EAthenaMainMenuLeaveButtonText Enum=LeaveAction_ReturnToLobby)
   > SetText(
    self=Get Button_LeaveAction()
    Text=Find(
     TargetMap:Map of EAthenaMainMenuLeaveButtonText Enums to Texts=Get Leave Action Text() -> (Leave Action Text:Map of EAthenaMainMenuLeaveButtonText Enums to Texts=LeaveAction_ReturnToLobby)
     Key:"EAthenaMainMenuLeaveButtonText Enum (by ref)"=@n2
    ) -> (Value:Text)
   )
   > Set Leave Action Class(Leave Action Class:Sidebar Confirmation Panel VM Class Reference=Find(
     TargetMap:Map of EAthenaMainMenuLeaveButtonText Enums to Sidebar Confirmation Panel VM Class References=Get Leave Action() -> (Leave Action:Map of EAthenaMainMenuLeaveButtonText Enums to Sidebar Confirmation Panel VM Class References=LeaveAction_ReturnToLobby)
     Key=@n2
    ) -> (Value:Sidebar Confirmation Panel VM Class Reference)) -> (Output_Get:Sidebar Confirmation Panel VM Class Reference)
   > Set SkipMainMenuLeave(SkipMainMenuLeave=false)
  ]}
 SetFavoriteButton {nodes [
   SetFavoriteButton!() -> (then, is Favorite:bool=~) {is=title K2Node_FunctionEntry}
   > Set ButtonType(
    ButtonType=Select(
     Option 0:E_UI_Sidebar_Button_Type Enum=NewEnumerator0
     Option 1:E_UI_Sidebar_Button_Type Enum=NewEnumerator2
     Index=Get is Favorite() -> (is Favorite:bool)
    ) -> (ReturnValue:E_UI_Sidebar_Button_Type Enum)
    self=Get Button_SaveIsland()
   ) -> (Output_Get=NewEnumerator0)
   > SetIsSelected!(execute, self:Common Button Base Object Reference=Get Button_SaveIsland(), InSelected:bool=Get is Favorite(), bGiveClickFeedback:bool=false) -> (then)
   > SetIcon!(
    execute
    self:WBP UIKit Button Generic Base Object Reference=Get Button_SaveIsland()
    Icon:Object Reference=Select(
     Option 0:Object Reference=/FortUILibrary/Textures/Icons/T_UI_Icon_UIKit_FavoriteOutline_MSDF
     Option 1:Object Reference=/FortUILibrary/Textures/Icons/T_UI_Icon_UIKit_Favorite_MSDF
     Index=Get is Favorite()
    ) -> (ReturnValue:Object Reference)
   ) -> (then)
   > InitializeButtonVisuals(self=Get Button_SaveIsland())
  ]}
 SetFollowDeveloperButton {nodes [
   n1 = SetFollowDeveloperButton!() -> (then, IsFollowingDeveloper:bool=~) {is=title K2Node_FunctionEntry}
   > SetText(
    self=Get Button_FollowDeveloper()
    Text=Select(Option 0:Text="LOCTEXT(\"Follow Developer\")", Option 1:Text="LOCTEXT(\"Unfollow Developer\")", Index=@n1) -> (ReturnValue:Text)
   )
   > SetIsSelected(self=Get Button_FollowDeveloper(), InSelected=@n1, bGiveClickFeedback=false)
   > InitializeButtonVisuals(self=Get Button_FollowDeveloper())
  ]}
 SetLeavePartyButtonText {nodes [
   SetLeavePartyButtonText() {is=title K2Node_FunctionEntry}
   n1 = Find(
    TargetMap:Map of EAthenaMainMenuLeaveButtonText Enums to Texts=Get Leave Action Text() -> (Leave Action Text=LeaveAction_ReturnToLobby)
    Key=LeaveAction_Default
   ) -> (Value:Text)
   > Branch(Condition=@n1.ReturnValue)
   > SetText(self=Get Button_LeaveParty(), Text=@n1)
  ]}
 SetMobileStyle {nodes [
   SetMobileStyle() {is=title K2Node_FunctionEntry}
   > Branch(Condition=IsUsingCompactUI())
   n1 = Make Margin(Left=24, Top=24, Right=24, Bottom=24)
   > Set MobileButtonPadding!(execute, MobileButtonPadding:Margin Structure=@n1, self:WBP UIKit Button Generic Base Object Reference=Get Button_SaveIsland()) -> (then, Output_Get:Margin Structure)
   > Set MobileButtonPadding(MobileButtonPadding=@n1, self=Get Button_Like())
   > Set MobileButtonPadding(MobileButtonPadding=@n1, self=Get Button_Dislike())
  ]}
 SetRegistryButtons {nodes [
   n1 = SetRegistryButtons!() -> (then, Registry Buttons:Array of Registry Button Structures) {is=title K2Node_FunctionEntry}
   > n2 = Set Registry Buttons(Registry Buttons:Array of Registry Button Structures=@n1) -> (Output_Get:Array of Registry Button Structures)
   > ClearChildren(self=Get Vbox_RegistryButtons() -> (Vbox_RegistryButtons:Vertical Box Object Reference))
   > n3 = Branch(Condition=NOT Boolean(A=Is Empty!(TargetArray:Array of Registry Button Structures=@n2) -> (ReturnValue:bool)))
   n4 = Get Vbox_RegistryButtons()
   > SetVisibility(self=@n4, InVisibility=SelfHitTestInvisible)
   > n5 = For Each Loop(Array=Get Registry Buttons())
   > n6 = Construct Object from Class!(
    execute
    Class=/Sidebar/Buttons/WBP_Sidebar_ButtonGeneric.WBP_Sidebar_ButtonGeneric_C
    self
    Style:Common Button Style Class Reference=/FortUILibrary/UIKit/Buttons/Blocks/ButtonStyle_UIKit_Empty.ButtonStyle_UIKit_Empty_C
    bApplyAlphaOnDisable:bool=False
    bLocked:bool=False
    bSelectable:bool=True
    bShouldSelectUponReceivingFocus:bool=True
    bInteractableWhenSelected:bool=True
    bToggleable:bool=False
    bTriggerClickedAfterSelection:bool=False
    bDisplayInputActionWhenNotInteractable:bool=True
    bHideInputActionWithKeyboard:bool=False
    bShouldUseFallbackDefaultInputAction:bool=True
    bRequiresHold:bool=False
    HoldData:Common UIHold Data Class Reference=&CommonUI_HoldData_Base_C
    ClickMethod:EButtonClickMethod Enum=DownAndUp
    InputPriority:int=0
    TriggeringInputAction:Data Table Row Handle Structure
   ) -> (then, ReturnValue:WBP Sidebar Button Generic Object Reference)
   > SetText(self=@n6, Text=@n5.Array Element_SettingRegistryDisplayName)
   > Bind Event to On Clicked(self:Common Button Base Object Reference=@n6, Delegate=Create Event()) {is=title K2Node_AddDelegate}
   > AddChild(self=Get Vbox_RegistryButtons(), Content=@n6)
   SetVisibility(execute=@n3.else, self=@n4, InVisibility=Collapsed)
  ]}
 Update Back to Hub Visibility {nodes [
   Update Back to Hub Visibility() {is=title K2Node_FunctionEntry}
   > n1 = Cast To FortGameStateAthena!(execute, Object=GetGameState()) -> (then, CastFailed, AsFort Game State Athena:Fort Game State Athena Object Reference) {is=title K2Node_DynamicCast}
   > Set Local_GameState(Local_GameState:Fort Game State Athena Object Reference=@n1) -> (Output_Get:Fort Game State Athena Object Reference)
   > n2 = Cast To FortPlayerControllerAthena(Object=GetOwningPlayer())
   > n3 = Set Local_PlayerController(Local_PlayerController:Fort Player Controller Athena Object Reference=@n2) -> (Output_Get:Fort Player Controller Athena Object Reference)
   > n4 = Cast To FortPlayerState!(execute, Object=Get PlayerState!(self:Controller Object Reference=@n3) -> (PlayerState:Player State Object Reference)) -> (then, CastFailed, AsFort Player State:Fort Player State Object Reference) {is=title K2Node_DynamicCast}
   > Set Local_PlayerState(Local_PlayerState:Fort Player State Object Reference=@n4) -> (Output_Get:Fort Player State Object Reference)
   n5 = Get Local_GameState() -> (Local_GameState:Fort Game State Athena Object Reference)
   n6 = IsTeleportToCreativeHubAllowed!(
    self:Fort Game State Athena Object Reference=@n5
    PlayerState:Fort Player State Object Reference=Get Local_PlayerState() -> (Local_PlayerState:Fort Player State Object Reference)
   ) -> (ReturnValue:bool)
   n7 = CanRespawnOnStarterIsland!(self:Fort Game State Athena Object Reference=@n5) -> (ReturnValue:bool)
   > SetVisibility(
    self=Get Button_BackToHub() -> (Button_BackToHub:WBP Sidebar Button Generic Object Reference)
    InVisibility=Select(Option 0:ESlateVisibility Enum=Collapsed, Option 1:ESlateVisibility Enum=SelfHitTestInvisible, Index=AND Boolean(A=@n7, B=@n6)) -> (ReturnValue:ESlateVisibility Enum)
   )
   SetVisibility(execute=[@n4.CastFailed, @n2.CastFailed, @n1.CastFailed], self=Get Button_BackToHub(), InVisibility=Collapsed)
  ]}
 Update Minigame Visibility {nodes [
   Update Minigame Visibility() {is=title K2Node_FunctionEntry}
   > n1 = Cast To FortGameStateAthena(Object=GetGameState())
   > Set Local Game State(Local Game State:Fort Game State Athena Object Reference=@n1) -> (Output_Get:Fort Game State Athena Object Reference)
   > n2 = Cast To FortPlayerControllerAthena(Object=GetOwningPlayer())
   > Set Local Player Controller(Local Player Controller:Fort Player Controller Athena Object Reference=@n2) -> (Output_Get:Fort Player Controller Athena Object Reference)
   n3 = Get Local Player Controller() -> (Local Player Controller:Fort Player Controller Athena Object Reference)
   n4 = Get Local Game State() -> (Local Game State:Fort Game State Athena Object Reference)
   > SetVisibility(
    self=Get Button_StartGame()
    InVisibility=Select(
     Option 0=Collapsed
     Option 1=Visible
     Index=isMinigameVisible!(Player:Fort Player Controller Athena Object Reference=@n3, FortGameStateAthena:Fort Game State Athena Object Reference=@n4) -> (bResult:bool)
    )
   )
  ]}
 Update Moderator Mode Visibility {nodes [
   Update Moderator Mode Visibility() {is=title K2Node_FunctionEntry}
   > n1 = Cast To FortPlayerControllerAthena(Object=GetOwningPlayer())
   n2 = NOT Boolean(A=IsMobileGame())
   n3 = IsCreativeModeratorModeEnabled!(self:Fort Player Controller Gameplay Object Reference=@n1) -> (ReturnValue:bool)
   n4 = IsCreativeModeratorModeAllowed!(self:Fort Player Controller Gameplay Object Reference=@n1) -> (ReturnValue:bool)
   > SetVisibility(
    self=Get Button_ModeratorMode() -> (Button_ModeratorMode:WBP Sidebar Button Generic Object Reference)
    InVisibility=Select(Option 0=Collapsed, Option 1=SelfHitTestInvisible, Index=AND Boolean!(A=@n2, B=@n4, C:bool=@n3) -> (ReturnValue))
   )
  ]}
 Update Moderator Mode Visuals {nodes [
   Update Moderator Mode Visuals!() -> (then, is Loading:bool=~, is Moderator Mode Active:bool=~) {is=title K2Node_FunctionEntry}
   n1 = Get is Moderator Mode Active() -> (is Moderator Mode Active:bool)
   n2 = Get Button_ModeratorMode()
   > SetText(
    self=@n2
    Text=SelectText(
     A=SelectText(A="LOCTEXT(\"Disabling Moderator mode\")", B="LOCTEXT(\"Enabling Moderator mode\")", bPickA=@n1)
     B=SelectText(A="LOCTEXT(\"Disable Moderator mode\")", B="LOCTEXT(\"Enable Moderator mode\")", bPickA=@n1)
     bPickA=Get is Loading() -> (is Loading:bool)
    )
   )
   > Set ButtonType(
    ButtonType=Select(Option 0:E_UI_Sidebar_Button_Type Enum=NewEnumerator0, Option 1:E_UI_Sidebar_Button_Type Enum=NewEnumerator3, Index=Get is Loading()) -> (ReturnValue:E_UI_Sidebar_Button_Type Enum)
    self=@n2
   ) -> (Output_Get=NewEnumerator0)
   > SetIsLocked!(execute, self:Common Button Base Object Reference=@n2, bInIsLocked:bool=Get is Loading()) -> (then)
   > SetIconMaterial!(
    execute
    self:WBP UIKit Button Generic Base Object Reference=@n2
    Material Instance:Material Instance Object Reference=Select(
     Option 0:Material Instance Object Reference
     Option 1:Material Instance Object Reference=/Game/UI/DiscoveryShared/Art/MI_UI_Discovery_Throbber
     Index=Get is Loading()
    ) -> (ReturnValue:Material Instance Object Reference)
   ) -> (then)
   > InitializeButtonVisuals(self=@n2)
  ]}
 Update Respawn Visibility {nodes [
   Update Respawn Visibility() {is=title K2Node_FunctionEntry}
   > n1 = Cast To FortGameStateAthena(Object=GetGameState())
   > n2 = Cast To FortRemoteControlledPawnAthena!(execute, Object=GetOwningPlayerPawn!(self:User Widget Object Reference) -> (ReturnValue:Pawn Object Reference)) -> (then, CastFailed, AsFort Remote Controlled Pawn Athena:Fort Remote Controlled Pawn Athena Object Reference) {is=title K2Node_DynamicCast}
   SetVisibility(
    execute=[@n5.CastFailed, @n4.Is Not Valid, @n2.then, @n1.CastFailed]
    self=Get Button_Respawn() -> (Button_Respawn:WBP Sidebar Button Generic Object Reference)
    InVisibility=Collapsed
   )
   n3 = GetOwningPlayer()
   n4 = Is Valid(exec=@n2.CastFailed, InputObject=@n3)
   > n5 = Cast To FortPlayerState(Object=Get PlayerState(self=@n3))
   > Set PlayerTriggeredSpawnAllowed(PlayerTriggeredSpawnAllowed:bool=IsPlayerTriggeredRespawnAllowed!(self:Fort Game State Athena Object Reference=@n1, PlayerState:Fort Player State Object Reference=@n5) -> (ReturnValue:bool)) -> (Output_Get:bool)
   n6 = IsRespawningAllowed!(self:Fort Game State Zone Object Reference=@n1, PlayerState:Fort Player State Object Reference=@n5) -> (ReturnValue:bool)
   n7 = Get PlayerTriggeredSpawnAllowed() -> (PlayerTriggeredSpawnAllowed:bool)
   > SetVisibility(
    self=Get Button_Respawn()
    InVisibility=Select(
     Option 0:ESlateVisibility Enum=Collapsed
     Option 1:ESlateVisibility Enum=SelfHitTestInvisible
     Index=AND Boolean!(A=@n6, B=@n7) -> (ReturnValue)
    ) -> (ReturnValue:ESlateVisibility Enum)
   )
  ]}
 Update Support Buttons Visuals {nodes [
   Update Support Buttons Visuals!() -> (then, isLiked:bool=~, isDisliked:bool=~) {is=title K2Node_FunctionEntry}
   n1 = GetSelected!(self:Common Button Base Object Reference=Get Button_Like()) -> (ReturnValue:bool)
   > n2 = Branch(Condition=AND Boolean(A=Get isLiked!() -> (isLiked), B=NOT Boolean(A=@n1)))
   > ChooseRecommendOrNotInterested(TargetButton=Get Button_Like(), OppositeButton=Get Button_Dislike())
   > SetIsSelected(self=Get Button_Like(), InSelected=true, bGiveClickFeedback=false)
   n3 = GetSelected(self=Get Button_Dislike())
   Branch(execute=@n2.else, Condition=AND Boolean(A=Get isDisliked!() -> (isDisliked), B=NOT Boolean(A=@n3)))
   > ChooseRecommendOrNotInterested(TargetButton=Get Button_Dislike(), OppositeButton=Get Button_Like())
   > SetIsSelected(self=Get Button_Dislike(), InSelected=true, bGiveClickFeedback=false)
  ]}
 Update Team Select Visibility {nodes [
   Update Team Select Visibility() {is=title K2Node_FunctionEntry}
   n1 = GetOwningPlayer()
   > n2 = Cast To FortPlayerControllerAthena(Object=@n1)
   > n3 = Cast To FortGameStateAthena(Object=GetGameState())
   n4 = IsCreativeModeratorModeActive(self=@n2)
   n5 = Get Controlled Pawn!(self:Controller Object Reference=@n1) -> (ReturnValue:Pawn Object Reference) {is=title K2Node_CallFunction}
   n6 = AND Boolean!(
    A=IsTeamSwitchAllowed!(self:Fort Game State Athena Object Reference=@n3, SwitchingPawn:Pawn Object Reference=@n5) -> (ReturnValue:bool)
    B=NOT Boolean(A=Cast To FortRemoteControlledPawnAthena!(Object=@n5) -> (AsFort Remote Controlled Pawn Athena, bSuccess:bool=~).bSuccess)
    C:bool=GetShowTeamSelectButton!() -> (ReturnValue:bool)
   ) -> (ReturnValue)
   > SetVisibility(
    self=Get Button_TeamSelect() -> (Button_TeamSelect:WBP Sidebar Button Generic Object Reference)
    InVisibility=Select(Option 0=Collapsed, Option 1=SelfHitTestInvisible, Index=OR Boolean!(A=@n6, B=@n4) -> (ReturnValue))
   )
   SetVisibility(execute=[@n3.CastFailed, @n2.CastFailed], self=Get Button_TeamSelect(), InVisibility=Collapsed)
  ]}
 UpdateButtonTimeLimitText {nodes [
   UpdateButtonTimeLimitText() {is=title K2Node_FunctionEntry}
   n1 = GetRuntimeOptions()
   > SetText(
    self=Get Button_TimeLimits()
    Text=SelectText(
     A="LOCTEXT(\"Time limits && reports\")"
     B="LOCTEXT(\"Time limits\")"
     bPickA=Get bEnableThyme!(self:Fort Runtime Options Object Reference=@n1) -> (bEnableThyme:bool)
    )
   )
  ]}
 UpdateSettingsVisibility {nodes [
   Inputs()
   > Update Minigame Visibility() {is=title K2Node_CallFunction}
   > Update Moderator Mode Visibility() {is=title K2Node_CallFunction}
   > Update Respawn Visibility() {is=title K2Node_CallFunction}
   > Update Back to Hub Visibility() {is=title K2Node_CallFunction}
   > Update Team Select Visibility() {is=title K2Node_CallFunction}
   n1 = Get FortGlobalUIContext() -> (ReturnValue:Fort Global UIContext Object Reference) {is=title K2Node_GetSubsystem}
   > SetVisibility(
    self=Get Button_ParentalControls() -> (Button_ParentalControls:WBP Sidebar Button Generic Object Reference)
    InVisibility=Select(Option 0=Visible, Option 1=Collapsed, Index=IsInZone!(self:Fort Global UIContext Object Reference=@n1) -> (ReturnValue:bool))
   )
   > SetVisibility(
    self=[Get HeaderText() -> (HeaderText:Common Text Object Reference), Get Vbox_CreativeOptions()]
    InVisibility=Select(
     Option 0=Collapsed
     Option 1=Visible
     Index=Creative_isCreativeOptionsNotEmpty!() -> (isCreativeOptionsNotEmpty:bool) {is=title K2Node_CallFunction}
    )
   )
   > n2 = Cast To FortPlayerControllerAthena(Object=GetOwningPlayer())
   > n3 = SetVisibility(
    self=Get Button_IslandSettings() -> (Button_IslandSettings:WBP Sidebar Button Generic Object Reference)
    InVisibility=Select(
     Option 0=Collapsed
     Option 1=SelfHitTestInvisible
     Index=IsPlayerInAnExperienceTheyOwn!(self:Fort Player Controller Object Reference=@n2) -> (ReturnValue:bool)
    )
   )
   n4 = GetRuntimeOptions()
   SetVisibility(
    execute=[@n3.then, @n2.CastFailed]
    self=Get Vbox_Developer() -> (Vbox_Developer:Vertical Box Object Reference)
    InVisibility=Select(Option 0=Visible, Option 1=Collapsed, Index=IsShippingBuild!(self:Fort Runtime Options Object Reference=@n4) -> (ReturnValue:bool))
   )
   > n5 = Branch(Condition=IsInSubGame!(SubGame:ESubGame Enum=Campaign) -> (ReturnValue:bool))
   n6 = Get Vbox_SaveTheWorld() -> (Vbox_SaveTheWorld:Vertical Box Object Reference)
   > n7 = SetVisibility(self=@n6, InVisibility=Visible)
   Outputs!(then:Exec=[@n7.then, @n8.then])
   n8 = SetVisibility(execute=@n5.else, self=@n6, InVisibility=Collapsed)
  ]}
 isMinigameVisible {nodes [
   isMinigameVisible!() -> (then, Player:Fort Player Controller Athena Object Reference, FortGameStateAthena:Fort Game State Athena Object Reference) {is=title K2Node_FunctionEntry}
   > n1 = Sequence()
   > Get!(execute) -> (then, else, Player:Fort Player Controller Athena Object Reference)
   > Set HasPermissionToStart(HasPermissionToStart:bool=CanOperateOnMinigame!(PlayerController:Fort Player Controller Object Reference=Get Player() -> (Player:Fort Player Controller Athena Object Reference)) -> (ReturnValue:bool)) -> (Output_Get:bool)
   > n2 = Set Minigame(Minigame:Fort Minigame Object Reference=GetMinigame!(ContextActor:Actor Object Reference=GetOwningPlayer()) -> (ReturnValue)) -> (Output_Get:Fort Minigame Object Reference)
   > Is Valid(InputObject=@n2)
   n3 = Get Player()
   > n4 = Set bIsModerator(bIsModerator:bool=IsCreativeModeratorModeEnabled(self=@n3)) -> (Output_Get:bool)
   n5 = HasStarted!(self:Fort Minigame Object Reference=@n2) -> (ReturnValue:bool)
   > Branch(Condition=OR Boolean!(A=@n4, B=CanOperateOnMinigame(PlayerController=@n3), C:bool=NOT Boolean(A=@n5)) -> (ReturnValue))
   n6 = IsCuratedHub(self=Get Minigame() -> (Minigame:Fort Minigame Object Reference))
   n7 = NOT Boolean(A=IsInPublicMatchmadeCreativeMinigame!(self:Fort Player Controller Athena Object Reference=Get Player()) -> (ReturnValue:bool))
   n8 = Get bIsModerator() -> (bIsModerator:bool)
   > Set bShow(bShow:bool=AND Boolean(A=OR Boolean!(A=@n8, B=@n7) -> (ReturnValue), B=NOT Boolean(A=@n6), C=Get HasPermissionToStart() -> (HasPermissionToStart:bool))) -> (Output_Get:bool)
   Return Node!(execute=@n1.then_1, bResult:bool=Get bShow() -> (bShow:bool))
  ]}
]
variables [
 Creative_FortPlayerControllerAthena {type=FortPlayerControllerAthena, default=null}
 Creative_Machine {type=FortMinigameSettingsBuilding, default=null}
 Current Input Type {type=ECommonInputType, default=MouseAndKeyboard}
 Current Leave Action {type=EAthenaMainMenuLeaveButtonText, default=LeaveAction_ReturnToLobby}
 Disliked Sound {type=SoundBase, default=/Game/Sounds/Fort_UI_Sounds/FrontEnd/MultiProduct/Common/Buttons/MSS_UI_MP_Button_Click_Dislike}
 Favorited Sound {type=SoundBase, default=/Game/Sounds/Fort_UI_Sounds/FrontEnd/MultiProduct/Common/Buttons/MSS_UI_MP_Button_Click_Favorite}
 Fort Player Controller Athena {type=FortPlayerControllerAthena, default=null}
 GameButtonState {type=MinigameButtonsState, default=Invalid}
 HordModeLeaveDescription {type=string, default=Your fort design will be saved and it will be used in the Challenge the Horde mission.}
 Is Creative Moderator Mode Active {type=boolean}
 LastFocusedButton {type=UserWidget, default=null}
 Leave Action {default {
   leaveAction_Default=&ConfirmationVM_LeaveParty_C
   leaveAction_InParty=/Sidebar/Panels/Overlay/Data/ConfirmationVM_LeaveWithParty.ConfirmationVM_LeaveWithParty_C
   leaveAction_Mission=/Sidebar/Panels/Overlay/Data/ConfirmationVM_LeaveMission.ConfirmationVM_LeaveMission_C
   leaveAction_PlaylistCreative=/Sidebar/Panels/Overlay/Data/ConfirmationVM_LeaveCreative.ConfirmationVM_LeaveCreative_C
   leaveAction_PlaylistCreative_LTM=/Sidebar/Panels/Overlay/Data/ConfirmationVM_LeaveCreativeLTM.ConfirmationVM_LeaveCreativeLTM_C
   leaveAction_PlaylistDefault=/Sidebar/Panels/Overlay/Data/ConfirmationVM_LeaveMatch.ConfirmationVM_LeaveMatch_C
   leaveAction_Replay=/Sidebar/Panels/Overlay/Data/ConfirmationVM_LeaveReplay.ConfirmationVM_LeaveReplay_C
   leaveAction_ReturnToLobby=&ConfirmationVM_BackToLobby_C
   leaveAction_StormShield=/Sidebar/Panels/Overlay/Data/ConfirmationVM_LeaveStormShield.ConfirmationVM_LeaveStormShield_C
  }}
 Leave Action Class {type=Class, default=null}
 Leave Action Text {default {
   leaveAction_Default=Leave party
   leaveAction_InParty=Leave with party
   leaveAction_Mission=Leave mission
   leaveAction_PlaylistCreative=Leave Creative
   leaveAction_PlaylistCreative_LTM=Leave game
   leaveAction_PlaylistDefault=Leave match
   leaveAction_Replay=Leave replay
   leaveAction_ReturnToLobby=Return to lobby
   leaveAction_StormShield=Leave Storm Shield
  }}
 Liked Sound {type=SoundBase, default=/Game/Sounds/Fort_UI_Sounds/FrontEnd/MultiProduct/Common/Buttons/MSS_UI_MP_Button_Click_Like}
 MissionLeaveDescription {type=string, default="Leave the mission and return to Homebase?\r\nYou will forfeit mission rewards and any XP gained."}
 MissionLeaveDescriptionParty {type=string, default="Leave your party and the mission, returning you to Homebase?\r\nYou will forfeit mission rewards and any XP gained."}
 MobileControlsHelpClass {type=Class, default=/FortniteCoreUI/Frontend/BacchusControlsHelp.BacchusControlsHelp_C}
 Registry Buttons {type="array<RegistryButton>", default []}
 RespawnClass {type=Class, default=/FortniteCoreUI/Frontend/RespawnPromptWidget.RespawnPromptWidget_C}
 Sidebar {type=WBP_Sidebar_C, default=null}
 SkipMainMenuLeave {type=boolean}
 StormshieldLeaveDescription {type=string, default=Leave the Storm Shield and return to Homebase?}
 StormshieldLeaveDescriptionParty {type=string, default="Leave your party and the Storm Shield, returning you to Homebase?"}
 StormshieldLeaveTitle {type=string, default=Leave Storm Shield?}
 TeamSelectWidgetClass {type=Class, default=/FortniteCoreUI/Frontend/TeamSelect/TeamSelectWidget.TeamSelectWidget_C}
 Undisliked Sound {type=SoundBase, default=&MSS_UI_MP_Button_Click_Dislike}
 Unfavorited Sound {type=SoundBase, default=/Game/Sounds/Fort_UI_Sounds/FrontEnd/MultiProduct/Common/Buttons/MSS_UI_MP_Button_Click_Unfavorite}
 Unliked Sound {type=SoundBase, default=&MSS_UI_MP_Button_Click_Dislike}
]
widgets {
 info {
  inheritedWidgetCount=0
  namedSlotCount=1
  parentClass=&CommonActivatableWidget
  rootWidgetClass=/Sidebar/Shared/WBP_Sidebar_PanelLayout.WBP_Sidebar_PanelLayout_C
  widgetCount=45
 }
 widgets [
  WBP_Sidebar_PanelLayout {
   widgetClassPath=/Sidebar/Shared/WBP_Sidebar_PanelLayout
   bIsVariable=true
   widget=WBP_Sidebar_PanelLayout
   properties {headingText="", showBackButton=false}
  }
  SB_Content {
   widgetClassPath=&ScrollBox
   namedSlotHost=WBP_Sidebar_PanelLayout
   widget=SB_Content
   properties {
    bOverride_Cursor=true
    clipping=Inherit
    consumeMouseWheel=Always
    scrollBarVisibility=Hidden
    scrollWhenFocusChanges=AnimatedScroll
    scrollbarPadding=<ltrb 0 0 0 0>
    scrollbarThickness=<0 0>
    widgetStyle {
     barThickness=9
     bottomShadowBrush {
      bIsDynamicallyLoaded=false
      drawAs=NoDrawType
      imageType=FullColor
      mirroring=NoMirror
      resourceName=../../../Engine/Content/Slate/Common/ScrollBoxShadowBottom.png
      resourceObject=null
      tiling=NoTile
      imageSize=<16 8>
      margin=<ltrb 0.5 0 0.5 1>
      outlineSettings {
       bUseBrushTransparency=false
       roundingType=HalfHeightRadius
       width=0
       color {colorUseRule=UseColor_Specified, specifiedColor=<rgba 0 0 0 0>}
       cornerRadii=<0 0 0 0>
      }
      tintColor {colorUseRule=UseColor_Specified, specifiedColor=<rgba 1 1 1 1>}
      uVRegion {bIsValid=false, max=<0 0>, min=<0 0>}
     }
     horizontalScrolledContentPadding=<ltrb 0 0 1 0>
     leftShadowBrush {
      bIsDynamicallyLoaded=false
      drawAs=NoDrawType
      imageType=FullColor
      mirroring=NoMirror
      resourceName=../../../Engine/Content/Slate/Common/ScrollBoxShadowLeft.png
      resourceObject=null
      tiling=NoTile
      imageSize=<8 16>
      margin=<ltrb 1 0.5 0 0.5>
      outlineSettings {
       bUseBrushTransparency=false
       roundingType=HalfHeightRadius
       width=0
       color {colorUseRule=UseColor_Specified, specifiedColor=<rgba 0 0 0 0>}
       cornerRadii=<0 0 0 0>
      }
      tintColor {colorUseRule=UseColor_Specified, specifiedColor=<rgba 1 1 1 1>}
      uVRegion {bIsValid=false, max=<0 0>, min=<0 0>}
     }
     rightShadowBrush {
      bIsDynamicallyLoaded=false
      drawAs=NoDrawType
      imageType=FullColor
      mirroring=NoMirror
      resourceName=../../../Engine/Content/Slate/Common/ScrollBoxShadowRight.png
      resourceObject=null
      tiling=NoTile
      imageSize=<8 16>
      margin=<ltrb 0 0.5 1 0.5>
      outlineSettings {
       bUseBrushTransparency=false
       roundingType=HalfHeightRadius
       width=0
       color {colorUseRule=UseColor_Specified, specifiedColor=<rgba 0 0 0 0>}
       cornerRadii=<0 0 0 0>
      }
      tintColor {colorUseRule=UseColor_Specified, specifiedColor=<rgba 1 1 1 1>}
      uVRegion {bIsValid=false, max=<0 0>, min=<0 0>}
     }
     topShadowBrush {
      bIsDynamicallyLoaded=false
      drawAs=NoDrawType
      imageType=FullColor
      mirroring=NoMirror
      resourceName=../../../Engine/Content/Slate/Common/ScrollBoxShadowTop.png
      resourceObject=null
      tiling=NoTile
      imageSize=<16 8>
      margin=<ltrb 0.5 1 0.5 0>
      outlineSettings {
       bUseBrushTransparency=false
       roundingType=HalfHeightRadius
       width=0
       color {colorUseRule=UseColor_Specified, specifiedColor=<rgba 0 0 0 0>}
       cornerRadii=<0 0 0 0>
      }
      tintColor {colorUseRule=UseColor_Specified, specifiedColor=<rgba 1 1 1 1>}
      uVRegion {bIsValid=false, max=<0 0>, min=<0 0>}
     }
     verticalScrolledContentPadding=<ltrb 0 0 0 1>
    }
   }
  }
  FortScrollNavigationPanel {
   widgetClassPath=/Script/SidebarUI.FortScrollNavigationPanel
   bIsVariable=true
   parent=SB_Content
   widget=FortScrollNavigationPanel
   properties {visibility=Visible}
   slot {class=ScrollBoxSlot, properties {padding=<ltrb 18 18 18 12>}}
   uIComponents [{
     component=&self:UIComponentWidgetBlueprintExtension_0.ComponentContainer.NavigationUIComponent_0
     componentClassPath=/Script/UMG.NavigationUIComponent
    }]
  }
  VBox_Exit {
   widgetClassPath=&VerticalBox
   bIsVariable=true
   parent=FortScrollNavigationPanel
   widget=VBox_Exit
   properties {bOverride_Cursor=true, navigation=&self:WidgetTree.VBox_Exit.WidgetNavigation_1}
   slot {class=VerticalBoxSlot, properties {padding=<ltrb 0 12 0 0>}}
   uIComponents [{
     component=&self:UIComponentWidgetBlueprintExtension_0.ComponentContainer.NavigationUIComponent_1
     componentClassPath=&NavigationUIComponent
    }]
  }
  Button_Exit {
   widgetClassPath=/Sidebar/Buttons/WBP_Sidebar_ButtonGeneric
   bIsVariable=true
   parent=VBox_Exit
   widget=Button_Exit
   properties {
    buttonType=Exit
    minHeightOverride=60
    text=Close Fortnite
    font {
     bForceMonospaced=false
     bMaterialIsStencil=false
     fontMaterial=/Sidebar/Materials/Fonts/MI_Sidebar_FontColor_Primary
     fontObject=&HeadingNow
     letterSpacing=10
     monospacedWidth=1
     size=15.75
     skewAmount=0
     typefaceFontName=74 Regular
     outlineSettings {
      bApplyOutlineToDropShadows=false
      bMiteredCorners=false
      bSeparateFillAlpha=false
      outlineMaterial=/FortUILibrary/Systems/ModularButton/MaterialInstances/MI_UIKit_TextOutline
      outlineSize=0
      outlineColor=<rgba 0 0 0 1>
     }
    }
    secondFont {
     bForceMonospaced=false
     bMaterialIsStencil=false
     fontMaterial=/Sidebar/Materials/Fonts/MI_Sidebar_FontColor_Secondary
     fontObject=&Burbank
     letterSpacing=48
     monospacedWidth=1
     size=21
     skewAmount=0
     typefaceFontName=BurbankRegular-Black
     outlineSettings {
      bApplyOutlineToDropShadows=false
      bMiteredCorners=false
      bSeparateFillAlpha=false
      outlineMaterial=null
      outlineSize=0
      outlineColor=<rgba 0 0 0 1>
     }
    }
   }
   slot {class=VerticalBoxSlot, properties {padding=<ltrb 0 0 0 12>, size {sizeRule=Fill, value=1}}}
  }
  Button_LeaveAction {
   widgetClassPath=&WBP_Sidebar_ButtonGeneric
   bIsVariable=true
   parent=VBox_Exit
   widget=Button_LeaveAction
   properties {
    minHeightOverride=60
    text=Return to lobby
    visibility=Collapsed
    font {
     bForceMonospaced=false
     bMaterialIsStencil=false
     fontMaterial=&MI_Sidebar_FontColor_Primary
     fontObject=&HeadingNow
     letterSpacing=10
     monospacedWidth=1
     size=15.75
     skewAmount=0
     typefaceFontName=74 Regular
     outlineSettings {
      bApplyOutlineToDropShadows=false
      bMiteredCorners=false
      bSeparateFillAlpha=false
      outlineMaterial=&MI_UIKit_TextOutline
      outlineSize=0
      outlineColor=<rgba 0 0 0 1>
     }
    }
    secondFont {
     bForceMonospaced=false
     bMaterialIsStencil=false
     fontMaterial=&MI_Sidebar_FontColor_Secondary
     fontObject=&Burbank
     letterSpacing=48
     monospacedWidth=1
     size=21
     skewAmount=0
     typefaceFontName=BurbankRegular-Black
     outlineSettings {
      bApplyOutlineToDropShadows=false
      bMiteredCorners=false
      bSeparateFillAlpha=false
      outlineMaterial=null
      outlineSize=0
      outlineColor=<rgba 0 0 0 1>
     }
    }
   }
   slot {class=VerticalBoxSlot, properties {padding=<ltrb 0 0 0 12>, size {sizeRule=Fill, value=1}}}
  }
  Button_LeaveExperience {
   widgetClassPath=&WBP_Sidebar_ButtonGeneric
   bIsVariable=true
   parent=VBox_Exit
   widget=Button_LeaveExperience
   properties {
    minHeightOverride=60
    text=Leave experience
    visibility=Collapsed
    font {
     bForceMonospaced=false
     bMaterialIsStencil=false
     fontMaterial=&MI_Sidebar_FontColor_Primary
     fontObject=&HeadingNow
     letterSpacing=10
     monospacedWidth=1
     size=15.75
     skewAmount=0
     typefaceFontName=74 Regular
     outlineSettings {
      bApplyOutlineToDropShadows=false
      bMiteredCorners=false
      bSeparateFillAlpha=false
      outlineMaterial=&MI_UIKit_TextOutline
      outlineSize=0
      outlineColor=<rgba 0 0 0 1>
     }
    }
    secondFont {
     bForceMonospaced=false
     bMaterialIsStencil=false
     fontMaterial=&MI_Sidebar_FontColor_Secondary
     fontObject=&Burbank
     letterSpacing=48
     monospacedWidth=1
     size=21
     skewAmount=0
     typefaceFontName=BurbankRegular-Black
     outlineSettings {
      bApplyOutlineToDropShadows=false
      bMiteredCorners=false
      bSeparateFillAlpha=false
      outlineMaterial=null
      outlineSize=0
      outlineColor=<rgba 0 0 0 1>
     }
    }
   }
   slot {class=VerticalBoxSlot, properties {padding=<ltrb 0 0 0 12>, size {sizeRule=Fill, value=1}}}
  }
  Button_LeaveParty {
   widgetClassPath=&WBP_Sidebar_ButtonGeneric
   bIsVariable=true
   parent=VBox_Exit
   widget=Button_LeaveParty
   properties {
    minHeightOverride=60
    text="[PH] Leave Epic Party"
    visibility=Collapsed
    font {
     bForceMonospaced=false
     bMaterialIsStencil=false
     fontMaterial=&MI_Sidebar_FontColor_Primary
     fontObject=&HeadingNow
     letterSpacing=10
     monospacedWidth=1
     size=15.75
     skewAmount=0
     typefaceFontName=74 Regular
     outlineSettings {
      bApplyOutlineToDropShadows=false
      bMiteredCorners=false
      bSeparateFillAlpha=false
      outlineMaterial=&MI_UIKit_TextOutline
      outlineSize=0
      outlineColor=<rgba 0 0 0 1>
     }
    }
    secondFont {
     bForceMonospaced=false
     bMaterialIsStencil=false
     fontMaterial=&MI_Sidebar_FontColor_Secondary
     fontObject=&Burbank
     letterSpacing=48
     monospacedWidth=1
     size=21
     skewAmount=0
     typefaceFontName=BurbankRegular-Black
     outlineSettings {
      bApplyOutlineToDropShadows=false
      bMiteredCorners=false
      bSeparateFillAlpha=false
      outlineMaterial=null
      outlineSize=0
      outlineColor=<rgba 0 0 0 1>
     }
    }
   }
   slot {class=VerticalBoxSlot, properties {padding=<ltrb 0 0 0 12>, size {sizeRule=Fill, value=1}}}
  }
  Button_Logout {
   widgetClassPath=&WBP_Sidebar_ButtonGeneric
   bIsVariable=true
   parent=VBox_Exit
   widget=Button_Logout
   properties {
    minHeightOverride=60
    text=Logout
    font {
     bForceMonospaced=false
     bMaterialIsStencil=false
     fontMaterial=&MI_Sidebar_FontColor_Primary
     fontObject=&HeadingNow
     letterSpacing=10
     monospacedWidth=1
     size=15.75
     skewAmount=0
     typefaceFontName=74 Regular
     outlineSettings {
      bApplyOutlineToDropShadows=false
      bMiteredCorners=false
      bSeparateFillAlpha=false
      outlineMaterial=&MI_UIKit_TextOutline
      outlineSize=0
      outlineColor=<rgba 0 0 0 1>
     }
    }
    secondFont {
     bForceMonospaced=false
     bMaterialIsStencil=false
     fontMaterial=&MI_Sidebar_FontColor_Secondary
     fontObject=&Burbank
     letterSpacing=48
     monospacedWidth=1
     size=21
     skewAmount=0
     typefaceFontName=BurbankRegular-Black
     outlineSettings {
      bApplyOutlineToDropShadows=false
      bMiteredCorners=false
      bSeparateFillAlpha=false
      outlineMaterial=null
      outlineSize=0
      outlineColor=<rgba 0 0 0 1>
     }
    }
   }
   slot {class=VerticalBoxSlot, properties {padding=<ltrb 0 0 0 12>, size {sizeRule=Fill, value=1}}}
  }
  Button_Sleep {
   widgetClassPath=&WBP_Sidebar_ButtonGeneric
   bIsVariable=true
   parent=VBox_Exit
   widget=Button_Sleep
   properties {
    minHeightOverride=60
    text=Enter sleep mode
    font {
     bForceMonospaced=false
     bMaterialIsStencil=false
     fontMaterial=&MI_Sidebar_FontColor_Primary
     fontObject=&HeadingNow
     letterSpacing=10
     monospacedWidth=1
     size=15.75
     skewAmount=0
     typefaceFontName=74 Regular
     outlineSettings {
      bApplyOutlineToDropShadows=false
      bMiteredCorners=false
      bSeparateFillAlpha=false
      outlineMaterial=&MI_UIKit_TextOutline
      outlineSize=0
      outlineColor=<rgba 0 0 0 1>
     }
    }
    secondFont {
     bForceMonospaced=false
     bMaterialIsStencil=false
     fontMaterial=&MI_Sidebar_FontColor_Secondary
     fontObject=&Burbank
     letterSpacing=48
     monospacedWidth=1
     size=21
     skewAmount=0
     typefaceFontName=BurbankRegular-Black
     outlineSettings {
      bApplyOutlineToDropShadows=false
      bMiteredCorners=false
      bSeparateFillAlpha=false
      outlineMaterial=null
      outlineSize=0
      outlineColor=<rgba 0 0 0 1>
     }
    }
   }
   slot {class=VerticalBoxSlot, properties {padding=<ltrb 0 0 0 12>, size {sizeRule=Fill, value=1}}}
  }
  CurrentActivityText {
   widgetClassPath=&CommonTextBlock
   bIsVariable=true
   parent=VBox_Exit
   widget=CurrentActivityText
   properties {
    bAutoCollapseWithEmptyText=true
    bIsEnabled=false
    font {
     bForceMonospaced=false
     bMaterialIsStencil=false
     fontMaterial=null
     fontObject=&HeadingNow
     letterSpacing=0
     monospacedWidth=1
     size=22.5
     skewAmount=0
     typefaceFontName=75 Medium
     outlineSettings {
      bApplyOutlineToDropShadows=false
      bMiteredCorners=false
      bSeparateFillAlpha=false
      outlineMaterial=null
      outlineSize=0
      outlineColor=<rgba 0 0 0 1>
     }
    }
    shadowOffset=<0 0>
   }
   slot {class=VerticalBoxSlot, properties {padding=<ltrb 24 24 18 12>}}
  }
  Vbox_ExpressSupportButtons {
   widgetClassPath=&HorizontalBox
   bIsVariable=true
   parent=FortScrollNavigationPanel
   widget=Vbox_ExpressSupportButtons
   properties {navigation=&self:WidgetTree.Vbox_ExpressSupportButtons.WidgetNavigation_1}
   slot {class=VerticalBoxSlot, properties {padding=<ltrb 0 0 0 8>}}
  }
  Button_Like {
   widgetClassPath=&WBP_Sidebar_ButtonGeneric
   bIsVariable=true
   parent=Vbox_ExpressSupportButtons
   widget=Button_Like
   properties {
    bShouldSelectUponReceivingFocus=false
    bToggleable=true
    icon=/FortUILibrary/Textures/Icons/T_UI_Icon_UIKit_Like_MSDF
    iconMaterial=/Sidebar/Materials/Icons/MI_Sidebar_LikeIcon
    minHeightOverride=105
    mobileFontSizeMultiplierOverride=1.5
    overrideDefaultMobileFontSizeMultiplier=true
    text=Like
    textBlockAlignment=Bottom
    useIconBlock=true
    visibility=Visible
    buttonPadding=<ltrb 12 12 12 12>
    font {
     bForceMonospaced=false
     bMaterialIsStencil=false
     fontMaterial=&MI_Sidebar_FontColor_Primary
     fontObject=&HeadingNow
     letterSpacing=10
     monospacedWidth=1
     size=12
     skewAmount=0
     typefaceFontName=74 Regular
     outlineSettings {
      bApplyOutlineToDropShadows=false
      bMiteredCorners=false
      bSeparateFillAlpha=false
      outlineMaterial=&MI_UIKit_TextOutline
      outlineSize=0
      outlineColor=<rgba 0 0 0 1>
     }
    }
    iconPadding=<ltrb 0 0 0 6>
    iconSize=<36 36>
    secondFont {
     bForceMonospaced=false
     bMaterialIsStencil=false
     fontMaterial=&MI_Sidebar_FontColor_Secondary
     fontObject=&Burbank
     letterSpacing=48
     monospacedWidth=1
     size=21
     skewAmount=0
     typefaceFontName=BurbankRegular-Black
     outlineSettings {
      bApplyOutlineToDropShadows=false
      bMiteredCorners=false
      bSeparateFillAlpha=false
      outlineMaterial=null
      outlineSize=0
      outlineColor=<rgba 0 0 0 1>
     }
    }
   }
   slot {class=HorizontalBoxSlot, properties {padding=<ltrb 0 0 6 0>, size {sizeRule=Fill, value=1}}}
  }
  Button_Dislike {
   widgetClassPath=&WBP_Sidebar_ButtonGeneric
   bIsVariable=true
   parent=Vbox_ExpressSupportButtons
   widget=Button_Dislike
   properties {
    bShouldSelectUponReceivingFocus=false
    bToggleable=true
    icon=/FortUILibrary/Textures/Icons/T_UI_Icon_UIKit_Dislike_MSDF
    iconMaterial=&MI_Sidebar_LikeIcon
    minHeightOverride=105
    mobileFontSizeMultiplierOverride=1.5
    overrideDefaultMobileFontSizeMultiplier=true
    text=Dislike
    textBlockAlignment=Bottom
    useIconBlock=true
    visibility=Visible
    buttonPadding=<ltrb 12 12 12 12>
    font {
     bForceMonospaced=false
     bMaterialIsStencil=false
     fontMaterial=&MI_Sidebar_FontColor_Primary
     fontObject=&HeadingNow
     letterSpacing=10
     monospacedWidth=1
     size=12
     skewAmount=0
     typefaceFontName=74 Regular
     outlineSettings {
      bApplyOutlineToDropShadows=false
      bMiteredCorners=false
      bSeparateFillAlpha=false
      outlineMaterial=&MI_UIKit_TextOutline
      outlineSize=0
      outlineColor=<rgba 0 0 0 1>
     }
    }
    iconPadding=<ltrb 0 0 0 6>
    iconSize=<36 36>
    secondFont {
     bForceMonospaced=false
     bMaterialIsStencil=false
     fontMaterial=&MI_Sidebar_FontColor_Secondary
     fontObject=&Burbank
     letterSpacing=48
     monospacedWidth=1
     size=21
     skewAmount=0
     typefaceFontName=BurbankRegular-Black
     outlineSettings {
      bApplyOutlineToDropShadows=false
      bMiteredCorners=false
      bSeparateFillAlpha=false
      outlineMaterial=null
      outlineSize=0
      outlineColor=<rgba 0 0 0 1>
     }
    }
   }
   slot {class=HorizontalBoxSlot, properties {padding=<ltrb 6 0 6 0>, size {sizeRule=Fill, value=1}}}
  }
  Button_FollowDeveloper {
   widgetClassPath=&WBP_Sidebar_ButtonGeneric
   bIsVariable=true
   parent=Vbox_ExpressSupportButtons
   widget=Button_FollowDeveloper
   properties {
    bShouldSelectUponReceivingFocus=false
    bToggleable=true
    bTriggerClickedAfterSelection=true
    icon=/FortUILibrary/Textures/Icons/T_UI_Icon_UIKit_Add_MSDF
    minHeightOverride=105
    mobileFontSizeMultiplierOverride=1.5
    overrideDefaultMobileFontSizeMultiplier=true
    text=Follow Developer
    textBlockAlignment=Bottom
    useIconBlock=true
    visibility=Visible
    buttonPadding=<ltrb 12 12 12 12>
    font {
     bForceMonospaced=false
     bMaterialIsStencil=false
     fontMaterial=&MI_Sidebar_FontColor_Secondary
     fontObject=&HeadingNow
     letterSpacing=10
     monospacedWidth=1
     size=12
     skewAmount=0
     typefaceFontName=74 Regular
     outlineSettings {
      bApplyOutlineToDropShadows=false
      bMiteredCorners=false
      bSeparateFillAlpha=false
      outlineMaterial=&MI_UIKit_TextOutline
      outlineSize=0
      outlineColor=<rgba 0 0 0 1>
     }
    }
    iconPadding=<ltrb 0 0 0 6>
    iconSize=<36 36>
    secondFont {
     bForceMonospaced=false
     bMaterialIsStencil=false
     fontMaterial=&MI_Sidebar_FontColor_Secondary
     fontObject=&Burbank
     letterSpacing=48
     monospacedWidth=1
     size=21
     skewAmount=0
     typefaceFontName=BurbankRegular-Black
     outlineSettings {
      bApplyOutlineToDropShadows=false
      bMiteredCorners=false
      bSeparateFillAlpha=false
      outlineMaterial=null
      outlineSize=0
      outlineColor=<rgba 0 0 0 1>
     }
    }
   }
   slot {class=HorizontalBoxSlot, properties {padding=<ltrb 6 0 6 0>, size {sizeRule=Fill, value=1}}}
  }
  Button_SaveIsland {
   widgetClassPath=&WBP_Sidebar_ButtonGeneric
   bIsVariable=true
   parent=Vbox_ExpressSupportButtons
   widget=Button_SaveIsland
   properties {
    bShouldSelectUponReceivingFocus=false
    bToggleable=true
    bTriggerClickedAfterSelection=true
    icon=&T_UI_Icon_UIKit_FavoriteOutline_MSDF
    iconMaterial=/Sidebar/Materials/Icons/MI_Sidebar_FavoriteIcon
    minHeightOverride=105
    mobileFontSizeMultiplierOverride=1.5
    overrideDefaultMobileFontSizeMultiplier=true
    text=Save the island
    textBlockAlignment=Bottom
    useIconBlock=true
    visibility=Visible
    buttonPadding=<ltrb 12 12 12 12>
    font {
     bForceMonospaced=false
     bMaterialIsStencil=false
     fontMaterial=&MI_Sidebar_FontColor_Primary
     fontObject=&HeadingNow
     letterSpacing=10
     monospacedWidth=1
     size=12
     skewAmount=0
     typefaceFontName=74 Regular
     outlineSettings {
      bApplyOutlineToDropShadows=false
      bMiteredCorners=false
      bSeparateFillAlpha=false
      outlineMaterial=&MI_UIKit_TextOutline
      outlineSize=0
      outlineColor=<rgba 0 0 0 1>
     }
    }
    iconPadding=<ltrb 0 0 0 6>
    iconSize=<36 36>
    secondFont {
     bForceMonospaced=false
     bMaterialIsStencil=false
     fontMaterial=&MI_Sidebar_FontColor_Secondary
     fontObject=&Burbank
     letterSpacing=48
     monospacedWidth=1
     size=21
     skewAmount=0
     typefaceFontName=BurbankRegular-Black
     outlineSettings {
      bApplyOutlineToDropShadows=false
      bMiteredCorners=false
      bSeparateFillAlpha=false
      outlineMaterial=null
      outlineSize=0
      outlineColor=<rgba 0 0 0 1>
     }
    }
   }
   slot {class=HorizontalBoxSlot, properties {padding=<ltrb 6 0 0 0>, size {sizeRule=Fill, value=1}}}
  }
  WidgetSlottingHostOverlay_BelowActionButtons {
   widgetClassPath=/Script/WidgetSlotting.WidgetSlottingHostOverlay
   parent=FortScrollNavigationPanel
   widget=WidgetSlottingHostOverlay_BelowActionButtons
   properties {
    bIsEnabled=false
    slottedWidgetDefaultHorizontalAlignment=HAlign_Fill
    slottedWidgetDefaultVerticalAlignment=VAlign_Bottom
    hostTag {tagName=UI.Sidebar.SubPanel.Exit.BelowActionButtons}
   }
   slot {class=VerticalBoxSlot}
  }
  Divider {
   widgetClassPath=&Image
   bIsVariable=true
   parent=FortScrollNavigationPanel
   widget=Divider
   properties {
    visibility=HitTestInvisible
    brush {
     bIsDynamicallyLoaded=false
     drawAs=Image
     imageType=FullColor
     mirroring=NoMirror
     resourceName=None
     resourceObject=/Sidebar/Materials/MI_Sidebar_Divider
     tiling=NoTile
     imageSize=<32 2>
     margin=<ltrb 0 0 0 0>
     outlineSettings {
      bUseBrushTransparency=false
      roundingType=HalfHeightRadius
      width=0
      color {colorUseRule=UseColor_Specified, specifiedColor=<rgba 0 0 0 0>}
      cornerRadii=<0 0 0 0>
     }
     tintColor {colorUseRule=UseColor_Specified, specifiedColor=<rgba 1 1 1 1>}
     uVRegion {bIsValid=false, max=<0 0>, min=<0 0>}
    }
   }
   slot {class=VerticalBoxSlot, properties {verticalAlignment=VAlign_Bottom, padding=<ltrb 0 18 0 28>}}
  }
  VBox_Menu {widgetClassPath=&VerticalBox, bIsVariable=true, parent=FortScrollNavigationPanel, widget=VBox_Menu, slot {class=VerticalBoxSlot}}
  Button_Settings {
   widgetClassPath=&WBP_Sidebar_ButtonGeneric
   bIsVariable=true
   parent=VBox_Menu
   widget=Button_Settings
   properties {
    icon=/FortUILibrary/Textures/Icons/T_UI_Icon_UIKit_GeneralSettings_MSDF
    minHeightOverride=60
    text=Settings
    useIconBlock=true
    buttonPadding=<ltrb 24 4 24 4>
    iconPadding=<ltrb 0 0 4 0>
   }
   slot {class=VerticalBoxSlot, properties {padding=<ltrb 0 0 0 12>}}
  }
  Button_ParentalControls {
   widgetClassPath=&WBP_Sidebar_ButtonGeneric
   bIsVariable=true
   parent=VBox_Menu
   widget=Button_ParentalControls
   properties {minHeightOverride=60, text=Parental controls}
   slot {class=VerticalBoxSlot, properties {padding=<ltrb 0 0 0 12>}}
  }
  Button_ManageUnlockedContent {
   widgetClassPath=&WBP_Sidebar_ButtonGeneric
   bIsVariable=true
   parent=VBox_Menu
   widget=Button_ManageUnlockedContent
   properties {minHeightOverride=60, text=Manage unlocked content}
   slot {class=VerticalBoxSlot, properties {padding=<ltrb 0 0 0 12>}}
  }
  Button_TimeLimits {
   widgetClassPath=&WBP_Sidebar_ButtonGeneric
   bIsVariable=true
   parent=VBox_Menu
   widget=Button_TimeLimits
   properties {minHeightOverride=60, text=Time limits}
   slot {class=VerticalBoxSlot, properties {padding=<ltrb 0 0 0 12>}}
  }
  Button_SupportAndRefunds {
   widgetClassPath=&WBP_Sidebar_ButtonGeneric
   bIsVariable=true
   parent=VBox_Menu
   widget=Button_SupportAndRefunds
   properties {minHeightOverride=60, text=Support and refunds}
   slot {class=VerticalBoxSlot, properties {padding=<ltrb 0 0 0 12>}}
  }
  Button_Reporting {
   widgetClassPath=&WBP_Sidebar_ButtonGeneric
   bIsVariable=true
   parent=VBox_Menu
   widget=Button_Reporting
   properties {minHeightOverride=60, text=Report}
   slot {class=VerticalBoxSlot, properties {padding=<ltrb 0 0 0 12>}}
  }
  Button_Legal {
   widgetClassPath=&WBP_Sidebar_ButtonGeneric
   bIsVariable=true
   parent=VBox_Menu
   widget=Button_Legal
   properties {minHeightOverride=60, text=Legal}
   slot {class=VerticalBoxSlot, properties {padding=<ltrb 0 0 0 12>}}
  }
  Button_ControlsHelp {
   widgetClassPath=&WBP_Sidebar_ButtonGeneric
   bIsVariable=true
   parent=VBox_Menu
   widget=Button_ControlsHelp
   properties {minHeightOverride=60, text=Controls help}
   slot {class=VerticalBoxSlot, properties {padding=<ltrb 0 0 0 12>}}
  }
  Button_HudLayoutTool {
   widgetClassPath=&WBP_Sidebar_ButtonGeneric
   bIsVariable=true
   parent=VBox_Menu
   widget=Button_HudLayoutTool
   properties {minHeightOverride=60, text=HUD Layout Tool}
   slot {class=VerticalBoxSlot, properties {padding=<ltrb 0 0 0 12>}}
  }
  Vbox_SaveTheWorld {
   widgetClassPath=&VerticalBox
   bIsVariable=true
   parent=FortScrollNavigationPanel
   widget=Vbox_SaveTheWorld
   slot {class=VerticalBoxSlot}
  }
  Button_Help {
   widgetClassPath=&WBP_Sidebar_ButtonGeneric
   bIsVariable=true
   parent=Vbox_SaveTheWorld
   widget=Button_Help
   properties {minHeightOverride=60, text=Help}
   slot {class=VerticalBoxSlot, properties {padding=<ltrb 0 0 0 12>}}
  }
  Button_Boosts {
   widgetClassPath=&WBP_Sidebar_ButtonGeneric
   bIsVariable=true
   parent=Vbox_SaveTheWorld
   widget=Button_Boosts
   properties {minHeightOverride=60, text=Boosts}
   slot {class=VerticalBoxSlot, properties {padding=<ltrb 0 0 0 12>}}
  }
  Vbox_Developer {
   widgetClassPath=&VerticalBox
   bIsVariable=true
   parent=FortScrollNavigationPanel
   widget=Vbox_Developer
   slot {class=VerticalBoxSlot}
  }
  Button_Cheats {
   widgetClassPath=&WBP_Sidebar_ButtonGeneric
   bIsVariable=true
   parent=Vbox_Developer
   widget=Button_Cheats
   properties {minHeightOverride=60, text="Cheats (developer)"}
   slot {class=VerticalBoxSlot, properties {padding=<ltrb 0 0 0 12>}}
  }
  Button_Locker {
   widgetClassPath=&WBP_Sidebar_ButtonGeneric
   bIsVariable=true
   parent=Vbox_Developer
   widget=Button_Locker
   properties {minHeightOverride=60, text="Locker (developer)"}
   slot {class=VerticalBoxSlot, properties {padding=<ltrb 0 0 0 12>}}
  }
  InputField_Console {
   widgetClassPath=/Sidebar/Buttons/WBP_Sidebar_InputField
   bIsVariable=true
   parent=Vbox_Developer
   widget=InputField_Console
   properties {
    hint Text=Enter console commands...
    icon=/Game/ArtTools/Tech_Anim/shelf_ue4TechAnim/icons/consoleCmds_icon
    maxTextCharacters=200
    iconProperties {
     material=/FortUILibrary/Systems/InputField/MaterialInstances/MI_UIKit_InputFieldIcon
     image Size=<50 50>
     placement {
      horizontal Alignment=HAlign_Center
      vertical Alignment=VAlign_Center
      grid {column=0, column Span=0, layer=0, row=1, row Span=0, nudge=<0 0>}
      padding=<ltrb 10 0 0 0>
     }
    }
   }
   slot {class=VerticalBoxSlot, properties {padding=<ltrb 0 0 0 12>}}
  }
  HeaderText {
   widgetClassPath=&CommonTextBlock
   bIsVariable=true
   parent=FortScrollNavigationPanel
   widget=HeaderText
   properties {
    bIsEnabled=false
    text=Creative Options
    visibility=HitTestInvisible
    font {
     bForceMonospaced=false
     bMaterialIsStencil=false
     fontMaterial=null
     fontObject=&HeadingNow
     letterSpacing=0
     monospacedWidth=1
     size=18
     skewAmount=0
     typefaceFontName=75 Medium
     outlineSettings {
      bApplyOutlineToDropShadows=false
      bMiteredCorners=false
      bSeparateFillAlpha=false
      outlineMaterial=null
      outlineSize=0
      outlineColor=<rgba 0 0 0 1>
     }
    }
    shadowOffset=<0 0>
   }
   slot {class=VerticalBoxSlot, properties {padding=<ltrb 0 24 0 24>}}
  }
  Button_StartGame {
   widgetClassPath=&WBP_UIKit_Button_Generic
   bIsVariable=true
   parent=FortScrollNavigationPanel
   widget=Button_StartGame
   properties {
    bHideInputAction=true
    backgroundMaterial=/FortUILibrary/UIKit/Buttons/MaterialInstances/CTAButtons/MI_UIKit_BgLoud
    minHeightOverride=60
    mobileMinHeight=80
    text=Start game
    textTransformPolicy=None
    useIconBlock=false
    visibility=Collapsed
    buttonPadding=<ltrb 24 8 24 8>
    font {
     bForceMonospaced=false
     bMaterialIsStencil=false
     fontMaterial=/FortUILibrary/UIKit/Buttons/MaterialInstances/CTAButtons/MI_UIKit_TextQuiet
     fontObject=&HeadingNow
     letterSpacing=10
     monospacedWidth=1
     size=15
     skewAmount=0
     typefaceFontName=75 Medium
     outlineSettings {
      bApplyOutlineToDropShadows=false
      bMiteredCorners=false
      bSeparateFillAlpha=false
      outlineMaterial=null
      outlineSize=0
      outlineColor=<rgba 0 0 0 1>
     }
    }
   }
   slot {class=VerticalBoxSlot, properties {padding=<ltrb 0 0 0 8>}}
  }
  Button_ModeratorMode {
   widgetClassPath=&WBP_Sidebar_ButtonGeneric
   bIsVariable=true
   parent=FortScrollNavigationPanel
   widget=Button_ModeratorMode
   properties {iconMaterial=null, minHeightOverride=60, text=Moderator Mode, useIconBlock=true, visibility=Collapsed, iconPadding=<ltrb 0 0 4 0>}
   slot {class=VerticalBoxSlot, properties {padding=<ltrb 0 0 0 6>}}
  }
  Vbox_RegistryButtons {
   widgetClassPath=&VerticalBox
   bIsVariable=true
   parent=FortScrollNavigationPanel
   widget=Vbox_RegistryButtons
   properties {visibility=Collapsed}
   slot {class=VerticalBoxSlot, properties {padding=<ltrb 0 0 0 12>}}
  }
  Vbox_CreativeOptions {
   widgetClassPath=&VerticalBox
   bIsVariable=true
   parent=FortScrollNavigationPanel
   widget=Vbox_CreativeOptions
   slot {class=VerticalBoxSlot}
  }
  Button_Respawn {
   widgetClassPath=&WBP_Sidebar_ButtonGeneric
   bIsVariable=true
   parent=Vbox_CreativeOptions
   widget=Button_Respawn
   properties {minHeightOverride=60, text=Respawn, visibility=Collapsed}
   slot {class=VerticalBoxSlot, properties {padding=<ltrb 0 0 0 6>}}
  }
  Button_TeamSelect {
   widgetClassPath=&WBP_Sidebar_ButtonGeneric
   bIsVariable=true
   parent=Vbox_CreativeOptions
   widget=Button_TeamSelect
   properties {minHeightOverride=60, text=Team select, visibility=Collapsed}
   slot {class=VerticalBoxSlot, properties {padding=<ltrb 0 0 0 6>}}
  }
  Button_IslandSettings {
   widgetClassPath=&WBP_Sidebar_ButtonGeneric
   bIsVariable=true
   parent=Vbox_CreativeOptions
   widget=Button_IslandSettings
   properties {minHeightOverride=60, text=Island settings, visibility=Collapsed}
   slot {class=VerticalBoxSlot, properties {padding=<ltrb 0 0 0 6>}}
  }
  Button_BackToHub {
   widgetClassPath=&WBP_Sidebar_ButtonGeneric
   bIsVariable=true
   parent=Vbox_CreativeOptions
   widget=Button_BackToHub
   properties {minHeightOverride=60, text=Back to hub, visibility=Collapsed}
   slot {class=VerticalBoxSlot, properties {padding=<ltrb 0 0 0 6>}}
  }
  CreativeOptionWidgetSlottingHostOverlay {
   widgetClassPath=&WidgetSlottingHostOverlay
   parent=Vbox_CreativeOptions
   widget=CreativeOptionWidgetSlottingHostOverlay
   properties {
    slottedWidgetDefaultHorizontalAlignment=HAlign_Fill
    slottedWidgetDefaultVerticalAlignment=VAlign_Fill
    hostTag {tagName=DynamicUI.HUD.Fort.Athena.Sidebar.CreativeOptions.Slot}
   }
   slot {class=VerticalBoxSlot, properties {padding=<ltrb 0 0 0 6>}}
  }
 ]
}
