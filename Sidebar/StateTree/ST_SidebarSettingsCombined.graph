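# StateTree that drives the combined Settings section of the sidebar, with states for the main panel, Legal, Support and Refunds, Reporting and Time Limits.
# Each state slots a full-size, per-player widget into the UI.Sidebar.SubPanel host, with instance caching only on the main panel.
# Events lead from the main state to each subpage, and a Root event transition returns to it.
# TimeLimits reuses the ReportingPanel cache entry name.
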StateTree /Sidebar/StateTree/ST_SidebarSettingsCombined
states [
 Root {
  id=#1
  type=STATE
  children [
   #2
   #3
   #4
   #5
   #6
  ]
  tasks []
  transitions [{id=#7, trigger=ON_EVENT, state {id=#2, name=SidebarSettings}}]
 }
 SidebarSettings {
  id=#2
  type=STATE
  tasks [{
    id=#8
    description=UIState Tree Widget Slotting Task
    properties {
     iD=#9
     instance {
      _structType=/Script/UIStateTree.UIStateTreeWidgetSlottingTaskInstanceData
      bIsTaskActive=false
      widgets [{
        bAutoActivateWidget=true
        bCacheInstance=true
        bFill=true
        bIsPlayerSpecific=true
        bIsWidgetEnabled=true
        bKeepConstructed=true
        cacheEntryName=SettingsCombinedPanel
        memoryTag=None
        widget=/Sidebar/Panels/WBP_Exit_Settings_Combined_Panel.WBP_Exit_Settings_Combined_Panel_C
        hostTag {tagName=UI.Sidebar.SubPanel}
        padding=<ltrb 0 0 0 0>
       }]
     }
     node {
      name=None
      _structType=/Script/UIStateTree.UIStateTreeWidgetSlottingTask
      bCanEditConsideredForCompletion=true
      bConsideredForCompletion=false
      bHasTaskCompletionDelegateDispatcher=false
      bTaskEnabled=true
      transitionHandlingPriority=Normal
      bindingsBatch {value=65535}
      executionRuntimeTemplateIndex {value=65535}
      instanceDataHandle {index=65535, source=None, stateHandle {index=65535}}
      instanceTemplateIndex {value=65535}
      outputBindingsBatch {value=65535}
     }
    }
   }]
  transitions [
   {id=#10, trigger=ON_EVENT, state {id=#3, name=Legal}}
   {id=#11, trigger=ON_EVENT, state {id=#4, name=SupportAndRefunds}}
   {id=#12, trigger=ON_EVENT, state {id=#5, name=Reporting}}
   {id=#13, trigger=ON_EVENT, state {id=#6, name=TimeLimits}}
  ]
 }
 Legal {
  id=#3
  type=STATE
  tasks [{
    id=#14
    description=UIState Tree Widget Slotting Task
    properties {
     iD=#15
     instance {
      _structType=&UIStateTreeWidgetSlottingTaskInstanceData
      bIsTaskActive=false
      widgets [{
        bAutoActivateWidget=true
        bCacheInstance=false
        bFill=true
        bIsPlayerSpecific=true
        bIsWidgetEnabled=true
        bKeepConstructed=true
        cacheEntryName=LegalPanel
        memoryTag=None
        widget=/Sidebar/Panels/Settings/WBP_Settings_Legal.WBP_Settings_Legal_C
        hostTag {tagName=UI.Sidebar.SubPanel}
        padding=<ltrb 0 0 0 0>
       }]
     }
     node {
      name=None
      _structType=&UIStateTreeWidgetSlottingTask
      bCanEditConsideredForCompletion=true
      bConsideredForCompletion=false
      bHasTaskCompletionDelegateDispatcher=false
      bTaskEnabled=true
      transitionHandlingPriority=Normal
      bindingsBatch {value=65535}
      executionRuntimeTemplateIndex {value=65535}
      instanceDataHandle {index=65535, source=None, stateHandle {index=65535}}
      instanceTemplateIndex {value=65535}
      outputBindingsBatch {value=65535}
     }
    }
   }]
  transitions []
 }
 SupportAndRefunds {
  id=#4
  type=STATE
  tasks [{
    id=#16
    description=UIState Tree Widget Slotting Task
    properties {
     iD=#17
     instance {
      _structType=&UIStateTreeWidgetSlottingTaskInstanceData
      bIsTaskActive=false
      widgets [{
        bAutoActivateWidget=true
        bCacheInstance=false
        bFill=true
        bIsPlayerSpecific=true
        bIsWidgetEnabled=true
        bKeepConstructed=true
        cacheEntryName=SupportAndRefundsPanel
        memoryTag=None
        widget=/Sidebar/Panels/Settings/WBP_Settings_SupportAndRefunds.WBP_Settings_SupportAndRefunds_C
        hostTag {tagName=UI.Sidebar.SubPanel}
        padding=<ltrb 0 0 0 0>
       }]
     }
     node {
      name=None
      _structType=&UIStateTreeWidgetSlottingTask
      bCanEditConsideredForCompletion=true
      bConsideredForCompletion=false
      bHasTaskCompletionDelegateDispatcher=false
      bTaskEnabled=true
      transitionHandlingPriority=Normal
      bindingsBatch {value=65535}
      executionRuntimeTemplateIndex {value=65535}
      instanceDataHandle {index=65535, source=None, stateHandle {index=65535}}
      instanceTemplateIndex {value=65535}
      outputBindingsBatch {value=65535}
     }
    }
   }]
  transitions []
 }
 Reporting {
  id=#5
  type=STATE
  tasks [{
    id=#18
    description=UIState Tree Widget Slotting Task
    properties {
     iD=#19
     instance {
      _structType=&UIStateTreeWidgetSlottingTaskInstanceData
      bIsTaskActive=false
      widgets [{
        bAutoActivateWidget=true
        bCacheInstance=false
        bFill=true
        bIsPlayerSpecific=true
        bIsWidgetEnabled=true
        bKeepConstructed=true
        cacheEntryName=ReportingPanel
        memoryTag=None
        widget=/Sidebar/Panels/Settings/WBP_Settings_Reporting.WBP_Settings_Reporting_C
        hostTag {tagName=UI.Sidebar.SubPanel}
        padding=<ltrb 0 0 0 0>
       }]
     }
     node {
      name=None
      _structType=&UIStateTreeWidgetSlottingTask
      bCanEditConsideredForCompletion=true
      bConsideredForCompletion=false
      bHasTaskCompletionDelegateDispatcher=false
      bTaskEnabled=true
      transitionHandlingPriority=Normal
      bindingsBatch {value=65535}
      executionRuntimeTemplateIndex {value=65535}
      instanceDataHandle {index=65535, source=None, stateHandle {index=65535}}
      instanceTemplateIndex {value=65535}
      outputBindingsBatch {value=65535}
     }
    }
   }]
  transitions []
 }
 TimeLimits {
  id=#6
  type=STATE
  tasks [{
    id=#20
    description=UIState Tree Widget Slotting Task
    properties {
     iD=#21
     instance {
      _structType=&UIStateTreeWidgetSlottingTaskInstanceData
      bIsTaskActive=false
      widgets [{
        bAutoActivateWidget=true
        bCacheInstance=false
        bFill=true
        bIsPlayerSpecific=true
        bIsWidgetEnabled=true
        bKeepConstructed=true
        cacheEntryName=ReportingPanel
        memoryTag=None
        widget=/Sidebar/Panels/Settings/WBP_Settings_TimeLimits.WBP_Settings_TimeLimits_C
        hostTag {tagName=UI.Sidebar.SubPanel}
        padding=<ltrb 0 0 0 0>
       }]
     }
     node {
      name=None
      _structType=&UIStateTreeWidgetSlottingTask
      bCanEditConsideredForCompletion=true
      bConsideredForCompletion=false
      bHasTaskCompletionDelegateDispatcher=false
      bTaskEnabled=true
      transitionHandlingPriority=Normal
      bindingsBatch {value=65535}
      executionRuntimeTemplateIndex {value=65535}
      instanceDataHandle {index=65535, source=None, stateHandle {index=65535}}
      instanceTemplateIndex {value=65535}
      outputBindingsBatch {value=65535}
     }
    }
   }]
  transitions []
 }
]
