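# Activatable sidebar panel for setting a nickname on another player, showing their profile header, a nickname input field and a Done button.
# On activation it takes the social user from the current navigation point and initializes the social item and nickname viewmodels that feed the header and field.
# Done or the field's changed-nickname event calls TryChangeNickname, popping sidebar navigation if it reports a change.
# The input field is the desired focus target.
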WidgetBlueprint /Sidebar/Panels/Social/Profile/WBP_Profile_OtherPlayer_Nickname
graphs [
 BP_GetDesiredFocusTarget {nodes [
   Get Desired Focus Target()
   > Return Node!(
    execute
    ReturnValue:Widget Object Reference=Get Sidebar_InputField() -> (Sidebar_InputField:WBP Sidebar Nickname Input Field Object Reference)
   )
  ]}
 EventGraph {nodes [
   n1 = "On Clicked (Button_Done)"!() -> (OutputDelegate, then, Button:Common Button Base Object Reference) {is=title K2Node_ComponentBoundEvent}
   n2 = TryChangeNickname!(
    execute=[@n1.then, @n3.then]
    self:WBP Sidebar Nickname Input Field Object Reference=Get Sidebar_InputField()
    VM:Nickname VM Object Reference=Get FortNicknameVM() -> (FortNicknameVM:Nickname VM Object Reference)
   ) -> (then, Changed:bool)
   > Branch(Condition=@n2)
   > PopNavigation(self:Fort Sidebar Navigation VM Object Reference=Get FortSidebarNavigationVM() -> (FortSidebarNavigationVM:Fort Sidebar Navigation VM Object Reference))
   n3 = "On Changed Nickname (Sidebar_InputField)"() {is=title K2Node_ComponentBoundEvent}
   Event On Activated()
   > n4 = GetCurrentNavigationPoint!(execute, self:Fort Sidebar Navigation VM Object Reference=Get FortSidebarNavigationVM()) -> (then, ReturnValue:Fort Sidebar Navigation Point Object Reference)
   > n5 = Cast To BP_NavigationPoint!(execute, Object=@n4) -> (then, CastFailed, AsBP Navigation Point:BP Navigation Point Object Reference) {is=title K2Node_DynamicCast}
   n6 = Get FortSocialItemVM() -> (FortSocialItemVM:Fort Social Item VM Object Reference)
   n7 = Get SocialUser!(self:BP Navigation Point Object Reference=@n5) -> (SocialUser:Fort Social User Object Reference)
   > Init!(execute, self:Fort Social Item VM Object Reference=@n6, InTargetUser:Fort Social User Object Reference=@n7) -> (then)
   > Initialize Header With Fort Social Item VM!(
    execute
    self:WBP Profile Header Object Reference=Get Profile_Header() -> (Profile_Header:WBP Profile Header Object Reference)
    FortSocialItemVM:Fort Social Item VM Object Reference=@n6
   ) -> (then)
   n8 = Get FortNicknameVM()
   > SetSocialUser!(execute, self:Fort Social User Viewmodel Object Reference=@n8, InSocialUser:Fort Social User Object Reference=@n7) -> (then)
   > SetVM!(execute, self:WBP Sidebar Nickname Input Field Object Reference=Get Sidebar_InputField(), FortNicknameVM:Nickname VM Object Reference=@n8) -> (then)
  ]}
]
variables []
widgets {
 info {
  inheritedWidgetCount=0
  namedSlotCount=1
  parentClass=&CommonActivatableWidget
  rootWidgetClass=/Sidebar/Shared/WBP_Sidebar_PanelLayout.WBP_Sidebar_PanelLayout_C
  widgetCount=5
 }
 widgets [
  PanelLayout {widgetClassPath=/Sidebar/Shared/WBP_Sidebar_PanelLayout, bIsVariable=true, widget=PanelLayout, properties {headingText=""}}
  Content {
   widgetClassPath=&VerticalBox
   namedSlotHost=PanelLayout
   widget=Content
   properties {navigation=&self:WidgetTree.Content.WidgetNavigation_0}
  }
  Profile_Header {
   widgetClassPath=&/WBP_Profile_Header
   bIsVariable=true
   parent=Content
   widget=Profile_Header
   slot {class=VerticalBoxSlot, properties {padding=<ltrb 0 30 0 30>}}
  }
  Sidebar_InputField {
   widgetClassPath=&/WBP_Sidebar_NicknameInputField
   bIsVariable=true
   parent=Content
   widget=Sidebar_InputField
   properties {bSelectable=false}
   slot {class=VerticalBoxSlot, properties {padding=<ltrb 18 36 18 12>}}
  }
  Button_Done {
   widgetClassPath=/Sidebar/Buttons/WBP_Sidebar_ButtonGeneric
   bIsVariable=true
   parent=Content
   widget=Button_Done
   properties {bShouldUseFallbackDefaultInputAction=false, minHeightOverride=60, text=Done}
   slot {class=VerticalBoxSlot, properties {padding=<ltrb 18 0 18 0>}}
  }
 ]
}
