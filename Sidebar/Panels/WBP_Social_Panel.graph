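# Social tab panel of the sidebar, an activatable widget showing a tree list of social headers and players.
# Item events push profile, nickname or confirmation navigation; confirmed leave, remove and block actions execute here.
# It saves scroll offset on the navigation point and filters headers by active voice channel.
# Friends-list updates run only while active, sorting depends on hover, and first scroll restores clipping and dismisses a voice-chat coach mark.
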WidgetBlueprint /Sidebar/Panels/WBP_Social_Panel
graphs [
 BP_GetDesiredFocusTarget {nodes [
   Get Desired Focus Target()
   > Return Node!(
    execute
    ReturnValue:Widget Object Reference=Get TreeView!(self:Fort Polymorphic Tree View Object Reference=Get WBP_Poly_TreeView_PlayerList() -> (WBP_Poly_TreeView_PlayerList:WBP Sidebar Social Tree List Object Reference)) -> (TreeView:Tree View Object Reference)
   )
  ]}
 EventGraph {nodes [
   n1 = "Leave Party Clicked (WBP_Poly_TreeView_PlayerList)"!() -> (OutputDelegate, then, Action:Name=None) {is=title K2Node_ComponentBoundEvent}
   > n2 = Switch on Name!(execute:Exec, Selection:Name=@n1.Action) -> (Default:Exec, Disconnect:Exec, EndLocalPlay:Exec, LeaveParty:Exec) {is=title K2Node_SwitchName, pinNames [Disconnect,EndLocalPlay,LeaveParty]}
   NavigateToConfirmPanel!(
    execute=[@n2.Default, @n2.LeaveParty]
    self
    Class:Sidebar Confirmation Panel VM Class Reference=/Sidebar/Panels/Overlay/Data/ConfirmationVM_LeaveParty.ConfirmationVM_LeaveParty_C
   ) -> (then)
   NavigateToConfirmPanel(execute=@n2.Disconnect, Class=/Sidebar/Panels/Overlay/Data/ConfirmationVM_Disconnect.ConfirmationVM_Disconnect_C)
   NavigateToConfirmPanel(execute=@n2.EndLocalPlay, Class=/Sidebar/Panels/Overlay/Data/ConfirmationVM_EndLocalPlay.ConfirmationVM_EndLocalPlay_C)
   n3 = "Item Clicked (WBP_Poly_TreeView_PlayerList)"!() -> (OutputDelegate, then, ItemVM:Fort Social Item VM Object Reference) {is=title K2Node_ComponentBoundEvent}
   > SaveSocialScrollOffset()
   > NavigateToPlayerProfile!(
    execute
    SelfReference:Object Reference=Self-Reference()
    SocialItemVM:Fort Social Item VM Object Reference=@n3.ItemVM
    SidebarNavigationVM:Fort Sidebar Navigation VM Object Reference=Get FortSidebarNavigationVM() -> (FortSidebarNavigationVM:Fort Sidebar Navigation VM Object Reference)
    Sidebar VM:Fort Sidebar VM Object Reference=Get FortSidebarVM() -> (FortSidebarVM:Fort Sidebar VM Object Reference)
   ) -> (then)
   n4 = "Item Edit Nickname (WBP_Poly_TreeView_PlayerList)"!() -> (OutputDelegate, then, SocialUser:Fort Social User Object Reference) {is=title K2Node_ComponentBoundEvent}
   > Navigate to Edit Nickname!(execute:Exec, SocialUser:Fort Social User Object Reference=@n4.SocialUser) -> (then:Exec) {is=title K2Node_MacroInstance}
   n5 = "Item Remove Friend (WBP_Poly_TreeView_PlayerList)"!() -> (OutputDelegate, then, SocialUser:Fort Social User Object Reference, Interaction:Fort Social Interaction VM Object Reference) {is=title K2Node_ComponentBoundEvent}
   > Set RemoveFriendInteraction(RemoveFriendInteraction:Fort Social Interaction VM Object Reference=@n5.Interaction) -> (Output_Get:Fort Social Interaction VM Object Reference)
   > Navigate to Remove Friend!(execute:Exec, self:Fort Social User Object Reference=@n5.SocialUser) -> (then:Exec) {is=title K2Node_MacroInstance}
   n6 = "Item Block Player (WBP_Poly_TreeView_PlayerList)"!() -> (OutputDelegate, then, SocialUser:Fort Social User Object Reference, Interaction:Fort Social Interaction VM Object Reference) {is=title K2Node_ComponentBoundEvent}
   > Set BlockPlayerInteraction(BlockPlayerInteraction:Fort Social Interaction VM Object Reference=@n6.Interaction) -> (Output_Get:Fort Social Interaction VM Object Reference)
   > Navigate to Block Player!(execute:Exec, self:Fort Social User Object Reference=@n6.SocialUser) -> (then:Exec) {is=title K2Node_MacroInstance}
   n7 = "On Item Hovered (WBP_Poly_TreeView_PlayerList)"!() -> (OutputDelegate, then, Index:int=0) {is=title K2Node_ComponentBoundEvent}
   > SetAllowListSortDuringUpdate!(
    execute
    self:Fort Social Header List VM Object Reference=Get FortSocialFriendsListVM() -> (FortSocialFriendsListVM:Fort Social Friends List VM Object Reference)
    bIsEnabled:bool="Equal (Integer)"(A=@n7.Index)
   ) -> (then)
   n8 = CustomEvent!() -> (OutputDelegate, then, ItemOffset:float=~, DistanceRemaining:float=~) {is=title K2Node_CustomEvent}
   > Unbind Event from On List View Scrolled(self:List View Object Reference=Get TreeView(self=Get WBP_Poly_TreeView_PlayerList()), Delegate=@n8) {is=title K2Node_RemoveDelegate}
   > SetClipping!(
    execute
    self:Widget Object Reference=Get List_Clipping!(self:WBP Sidebar Panel Layout Object Reference=Get PanelLayout() -> (PanelLayout:WBP Sidebar Panel Layout Object Reference)) -> (List_Clipping:Grid Panel Object Reference)
    InClipping:EWidgetClipping Enum=ClipToBounds
   ) -> (then)
   n9 = Get WBP_Poly_TreeView_PlayerList()
   > n10 = Set Cached Displayed Entry Widgets(Cached Displayed Entry Widgets:Array of User Widget Object References=GetDisplayedEntryWidgets!(self:List View Base Object Reference=Get TreeView(self=@n9)) -> (ReturnValue:Array of User Widget Object References)) -> (Output_Get:Array of User Widget Object References)
   > n11 = For Each Loop with Break!(Exec:Exec, Array:Array of User Widget Object References=@n10, Break:Exec=@n13.then) -> (LoopBody:Exec, Array Element:User Widget Object Reference, Array Index:int, Completed:Exec)
   > n12 = Cast To WBP_PartyAndGame_Controls!(execute, Object=@n11) -> (then, CastFailed, AsWBP Party and Game Controls:WBP Party And Game Controls Object Reference) {is=title K2Node_DynamicCast}
   > DeactivateWidget(self=Get WBP_VoiceChat_ControlsMoved_CoachMark!(self:WBP Party And Game Controls Object Reference=@n12) -> (WBP_VoiceChat_ControlsMoved_CoachMark:WBP Voice Chat Controls Moved Coach Mark Object Reference))
   > n13 = RequestRefreshFocus(self:Common Activatable Widget Object Reference)
   Event Construct()
   > Init(self:Fort Social Friends List VM Object Reference=Get FortSocialFriendsListVM())
   > Initialize(self:Voice Channel Management VM Object Reference=Get VoiceChannelManagementVM() -> (VoiceChannelManagementVM:Voice Channel Management VM Object Reference))
   Event On Activated()
   n14 = Get FortSocialFriendsListVM()
   > SetAllowListAutoUpdate!(execute, self:Fort Social Header List VM Object Reference=@n14, bIsEnabled:bool=true) -> (then)
   > SetAllowListSortDuringUpdate(self=@n14, bIsEnabled=true)
   > UpdateNow(self:Fort Social Header List VM Object Reference=@n14)
   n15 = Get WBP_Poly_TreeView_PlayerList()
   > n16 = Branch(Condition=Get Regenerate On Activated!(self:WBP Sidebar Social Tree List Object Reference=@n15) -> (Regenerate On Activated:bool))
   > RegenerateAllEntries(self:List View Base Object Reference=Get TreeView(self=@n15))
   > n17 = Set Regenerate On Activated!(execute, Regenerate On Activated:bool=false, self:WBP Sidebar Social Tree List Object Reference=@n15) -> (then, Output_Get:bool)
   n18 = GetScrollOffset!(self:List View Base Object Reference=Get TreeView(self=Get WBP_Poly_TreeView_PlayerList())) -> (ReturnValue:float)
   SetClipping(
    execute=[@n16.else, @n17.then]
    self=Get List_Clipping(self=Get PanelLayout())
    InClipping=Select(Option 0:EWidgetClipping Enum=Inherit, Option 1:EWidgetClipping Enum=ClipToBounds, Index="float > float"(A=@n18)) -> (ReturnValue:EWidgetClipping Enum)
   )
   > Bind Event to On List View Scrolled(self:List View Object Reference=Get TreeView(self=Get WBP_Poly_TreeView_PlayerList()), Delegate=@n8) {is=title K2Node_AddDelegate}
   > SidebarFeatureFinishedLoading(self:Fort Sidebar VM Object Reference=Get FortSidebarVM())
   Event On Deactivated()
   n19 = Get FortSocialFriendsListVM()
   > SetAllowListAutoUpdate(self=@n19, bIsEnabled=false)
   > SetAllowListSortDuringUpdate(self=@n19, bIsEnabled=false)
  ]}
 GetSocialTreeViewHeaders {nodes [
   n1 = GetSocialTreeViewHeaders!() -> (
    then
    VMHeaderArray:Array of Fort Social Header Item VM Object References
    CurrentActiveVoiceChannel:ESidebarVoiceChannelType Enum=None
    PendingInvites:Array of Fort Social User Object References
   ) {is=title K2Node_FunctionEntry}
   > Clear(TargetArray:Array of Fort Social Header Item VM Object References=Get ModifiedArrayOfVisibleHeaders() -> (ModifiedArrayOfVisibleHeaders:Array of Fort Social Header Item VM Object References))
   > n2 = For Each Loop(Array:Array of Fort Social Header Item VM Object References=@n1) -> (Array Element:Fort Social Header Item VM Object Reference)
   n3 = IsInZone!(self:Fort Global UIContext Object Reference=Get FortGlobalUIContext!() -> (ReturnValue:Fort Global UIContext Object Reference) {is=title K2Node_GetBlueprintContextEx}) -> (ReturnValue:bool)
   n4 = Get HeaderItemType!(self:Fort Social Header Item VM Object Reference=@n2) -> (HeaderItemType:ESocialHeaderItemType Enum=Unknown)
   n5 = Get HeaderItemType(self=@n2) -> (HeaderItemType=Unknown)
   n6 = OR Boolean!(
    A="Equal (Enum)"(A:ESocialHeaderItemType Enum=@n5, B:ESocialHeaderItemType Enum=PartyMembers)
    B="Equal (Enum)"(A=@n5, B=NativeEpicPartyMembers)
    C:bool=AND Boolean(A="Equal (Enum)"(A=@n4, B=GameVoiceChannels), B=NOT Boolean(A=@n3))
   ) -> (ReturnValue)
   n7 = "Equal (Enum)"(
    A:ESidebarVoiceChannelType Enum=Get CurrentActiveVoiceChannel() -> (CurrentActiveVoiceChannel:ESidebarVoiceChannelType Enum=None)
    B:ESidebarVoiceChannelType Enum=Game
   )
   n8 = Get CurrentActiveVoiceChannel() -> (CurrentActiveVoiceChannel=None)
   n9 = OR Boolean!(A="Equal (Enum)"(A=@n8, B=Party), B="Equal (Enum)"(A=@n8, B=EpicParty)) -> (ReturnValue)
   n10 = Get HeaderItemType(self=@n2) -> (HeaderItemType=Unknown)
   n11 = OR Boolean(A="Equal (Enum)"(A:ESocialHeaderItemType Enum=@n10, B:ESocialHeaderItemType Enum=TeamMembers), B="Equal (Enum)"(A=@n10, B=GameVoiceChannels))
   > n12 = Branch(Condition=OR Boolean(A=AND Boolean(A=@n6, B=@n7), B=AND Boolean(A=@n11, B=@n9)))
   Add(
    execute=@n12.else
    TargetArray:Array of Fort Social Header Item VM Object References=Get ModifiedArrayOfVisibleHeaders()
    NewItem:"Fort Social Header Item VM Object Reference (by ref)"=@n2
   ) {is=title K2Node_CallArrayFunction}
   Return Node!(execute=@n2.Completed, OutHeaderArray:Array of Fort Social Header Item VM Object References=Get ModifiedArrayOfVisibleHeaders())
  ]}
 HandleConfirmationOverlayActions {nodes [
   n1 = HandleConfirmationOverlayActions!() -> (then, ActionID:Name) {is=title K2Node_FunctionEntry}
   > Allow Handle Overlay Confirmation!(Exec:Exec, Action ID:Name=@n1) -> (then:Exec) {is=title K2Node_MacroInstance}
   > n2 = Branch(Condition="Equal (Name)"(A=@n1, B=LeaveParty))
   > n3 = LeaveParty(self:Fort Social Epic Party Header Item VM Object Reference=Get EpicPartyHeaderItemViewModel!(self:WBP Sidebar Social Tree List Object Reference=Get WBP_Poly_TreeView_PlayerList()) -> (EpicPartyHeaderItemViewModel:Fort Social Epic Party Header Item VM Object Reference))
   Branch(
    execute=[@n7.then, @n6.then, @n3.then]
    Condition=Get DisableListSortDuringScroll!(self:WBP Sidebar Social Tree List Object Reference=Get WBP_Poly_TreeView_PlayerList()) -> (DisableListSortDuringScroll:bool)
   )
   n4 = Get FortSocialFriendsListVM()
   > SetAllowListSortDuringUpdate(self=@n4, bIsEnabled=true)
   > UpdateNow(self=@n4)
   n5 = Branch(execute=@n2.else, Condition="Equal (Name)"(A=@n1, B=RemoveFriend))
   > n6 = ExecuteInteraction(self:Fort Social Interaction VM Object Reference=Get RemoveFriendInteraction() -> (RemoveFriendInteraction:Fort Social Interaction VM Object Reference))
   Branch(execute=@n5.else, Condition="Equal (Name)"(A=@n1, B=Block))
   > n7 = ExecuteInteraction(self=Get BlockPlayerInteraction() -> (BlockPlayerInteraction:Fort Social Interaction VM Object Reference))
  ]}
 NavigateToBlockPlayer {nodes [
   n1 = Inputs!() -> (execute, self:Fort Social User Object Reference)
   > n2 = Construct Object from Class(Class=/Sidebar/Panels/Overlay/Data/ConfirmationVM_BlockPlayer.ConfirmationVM_BlockPlayer_C) -> (ReturnValue:Confirmation VM Block Player Object Reference)
   > Set Description!(
    execute
    Description:Text=Format Text!(
     Format="LOCTEXT(\"Would you like to block {PlayerName}?\")"
     PlayerName:String=GetNicknameOrDisplayName!(self:Fort Social User Object Reference=@n1) -> (ReturnValue:String)
    ) -> (Result)
    self:Sidebar Confirmation Panel VM Object Reference=@n2
   ) -> (then, Output_Get:Text)
   > NavigateToConfirmPanel!(
    execute
    SelfReference:Object Reference=Self-Reference()
    SidebarConfirmationVM:Sidebar Confirmation Panel VM Object Reference=@n2
    SidebarNavigationVM:Fort Sidebar Navigation VM Object Reference=Get FortSidebarNavigationVM()
   ) -> (then)
   > Outputs!(then:Exec)
  ]}
 NavigateToConfirmPanel {nodes [
   n1 = NavigateToConfirmPanel!() -> (then, Class:Sidebar Confirmation Panel VM Class Reference) {is=title K2Node_FunctionEntry}
   > n2 = Construct Object from Class(Class=@n1) -> (ReturnValue:Sidebar Confirmation Panel VM Object Reference)
   > NavigateToConfirmPanel(SelfReference=Self-Reference(), SidebarConfirmationVM=@n2, SidebarNavigationVM=Get FortSidebarNavigationVM()) {is=title K2Node_CallFunction}
  ]}
 NavigateToEditNickname {nodes [
   n1 = Inputs!() -> (execute, SocialUser:Fort Social User Object Reference)
   > n2 = Construct Object from Class!(
    execute
    Class=/Sidebar/Shared/Navigation/BP_NavigationPoint.BP_NavigationPoint_C
    self
    SocialUser:Fort Social User Object Reference=@n1
    ConversationId:Name=None
    ConfirmationVM:Sidebar Confirmation Panel VM Object Reference
    ReservedSpots:int=0
    SocialItemVM:Fort Social Item VM Object Reference
    NavigationTag:Gameplay Tag Structure=(TagName="UI.StateTree.TransitionTag.Sidebar.Social.OtherPlayerNickname")
   ) -> (then, ReturnValue:BP Navigation Point Object Reference)
   > PushNavigation!(
    execute
    self:Fort Sidebar Navigation VM Object Reference=Get FortSidebarNavigationVM()
    NavigationPoint:Fort Sidebar Navigation Point Object Reference=@n2
   ) -> (then)
   > Outputs()
  ]}
 NavigateToRemoveFriend {nodes [
   n1 = Inputs!() -> (execute, self:Fort Social User Object Reference)
   > n2 = Construct Object from Class!(execute, Class=/Sidebar/Panels/Overlay/Data/ConfirmationVM_RemoveFriend.ConfirmationVM_RemoveFriend_C, self) -> (then, ReturnValue:Confirmation VM Remove Friend Object Reference)
   > Set Description(
    Description=Format Text!(Format="LOCTEXT(\"Are you sure you’d like to remove {FriendName} as a friend?\")", FriendName:String=GetNicknameOrDisplayName(self=@n1)) -> (Result)
    self=@n2
   )
   > n3 = Construct Object from Class!(
    execute
    Class=&BP_NavigationPoint_C
    self
    SocialUser:Fort Social User Object Reference
    ConversationId:Name=None
    ConfirmationVM:Sidebar Confirmation Panel VM Object Reference=@n2
    ReservedSpots:int=0
    SocialItemVM:Fort Social Item VM Object Reference
    NavigationTag:Gameplay Tag Structure=(TagName="UI.StateTree.TransitionTag.Sidebar.Overlay.Confirm")
   ) -> (then, ReturnValue:BP Navigation Point Object Reference)
   > PushNavigation(self=Get FortSidebarNavigationVM(), NavigationPoint=@n3)
   > Outputs()
  ]}
 SaveSocialScrollOffset {nodes [
   SaveSocialScrollOffset() {is=title K2Node_FunctionEntry}
   > n1 = GetScrollOffset!(execute:Exec, self:WBP Sidebar Social Tree List Object Reference=Get WBP_Poly_TreeView_PlayerList()) -> (then:Exec, Scroll Offset:float)
   > n2 = GetCurrentNavigationPoint!(execute, self:Fort Sidebar Navigation VM Object Reference=Get FortSidebarNavigationVM()) -> (then, ReturnValue:Fort Sidebar Navigation Point Object Reference)
   > n3 = Cast To BP_NavigationPoint!(execute, Object=@n2) -> (then, CastFailed, AsBP Navigation Point:BP Navigation Point Object Reference) {is=title K2Node_DynamicCast}
   > Set ScrollOffset!(execute, ScrollOffset:double=@n1, self:BP Navigation Point Object Reference=@n3) -> (then, Output_Get:double)
  ]}
 SetSocialHeaderItems {nodes [
   n1 = SetSocialHeaderItems!() -> (then, HeaderItems:Array of Fort Social Header Item VM Object References) {is=title K2Node_FunctionEntry}
   > SetTreeViewItems!(
    execute
    self:WBP Sidebar Social Tree List Object Reference=Get WBP_Poly_TreeView_PlayerList()
    OuterArray:Array of Fort Social Header Item VM Object References=@n1
   ) -> (then)
   > RequestRefreshFocus()
  ]}
 SetSocialScrollOffset {nodes [
   SetSocialScrollOffset() {is=title K2Node_FunctionEntry}
   > n1 = GetCurrentNavigationPoint(self=Get FortSidebarNavigationVM())
   > n2 = Cast To BP_NavigationPoint(Object=@n1)
   > SetScrollOffset!(
    execute
    self:WBP Sidebar Social Tree List Object Reference=Get WBP_Poly_TreeView_PlayerList()
    ScrollOffset:float=Get ScrollOffset!(self:BP Navigation Point Object Reference=@n2) -> (ScrollOffset:double)
   ) -> (then)
  ]}
]
variables [
 BlockPlayerInteraction {type=FortSocialInteractionVM, default=null}
 Cached Displayed Entry Widgets {type="array<UserWidget>", default []}
 RemoveFriendInteraction {type=FortSocialInteractionVM, default=null}
]
widgets {
 info {
  inheritedWidgetCount=0
  namedSlotCount=1
  parentClass=&CommonActivatableWidget
  rootWidgetClass=/Sidebar/Shared/WBP_Sidebar_PanelLayout.WBP_Sidebar_PanelLayout_C
  widgetCount=2
 }
 widgets [
  PanelLayout {
   widgetClassPath=/Sidebar/Shared/WBP_Sidebar_PanelLayout
   bIsVariable=true
   widget=PanelLayout
   properties {headingText=Social, showBackButton=false}
  }
  WBP_Poly_TreeView_PlayerList {
   widgetClassPath=/Sidebar/Panels/Social/List/WBP_Sidebar_SocialTreeList
   bIsVariable=true
   namedSlotHost=PanelLayout
   widget=WBP_Poly_TreeView_PlayerList
   properties {
    bIsFocusable=true
    isSocialTab=true
    itemClassToEntryData {
     &Class'/Script/SidebarUI.FortSocialHeaderItemVM' {bIsEntrySelectable=true, desiredEntryClass=/Sidebar/Panels/Social/List/WBP_Sidebar_SocialHeader.WBP_Sidebar_SocialHeader_C}
     &Class'/Script/SidebarUI.FortSocialItemVM' {bIsEntrySelectable=true, desiredEntryClass=/Sidebar/Panels/Social/List/WBP_Sidebar_SocialPlayer.WBP_Sidebar_SocialPlayer_C}
    }
   }
  }
 ]
}
