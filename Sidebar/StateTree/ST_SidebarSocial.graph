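# UI state tree for the sidebar's social section, deciding which social panel fills the UI.Sidebar.SubPanel host.
# Top slots the cached social panel and holds event transitions to its sibling states under SocialRoot.
# OtherPlayer, LocalPlayer and Party are linked-asset states.
# OtherPlayerNickname and VoiceDisclosure slot their widgets directly into the same host without instance caching, and define no transitions of their own.
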StateTree /Sidebar/StateTree/ST_SidebarSocial
states [
 SocialRoot {
  id=#1
  type=STATE
  children [
   #2
   #3
   #4
   #5
   #6
   #7
  ]
  tasks []
  transitions []
 }
 Top {
  id=#2
  type=STATE
  tasks [{
    id=#8
    description=UIState Tree Widget Slotting Task
    properties {
     iD=#9
     instance {
      _structType=/Script/UIStateTree.UIStateTreeWidgetSlottingTaskInstanceData
      bIsTaskActive=false
      widgets [{
        bAutoActivateWidget=true
        bCacheInstance=true
        bFill=true
        bIsPlayerSpecific=true
        bIsWidgetEnabled=true
        bKeepConstructed=true
        cacheEntryName=SocialPanel
        memoryTag=None
        widget=/Sidebar/Panels/WBP_Social_Panel.WBP_Social_Panel_C
        hostTag {tagName=UI.Sidebar.SubPanel}
        padding=<ltrb 0 0 0 0>
       }]
     }
     node {
      name=None
      _structType=/Script/UIStateTree.UIStateTreeWidgetSlottingTask
      bCanEditConsideredForCompletion=true
      bConsideredForCompletion=false
      bHasTaskCompletionDelegateDispatcher=false
      bTaskEnabled=true
      transitionHandlingPriority=Normal
      bindingsBatch {value=65535}
      executionRuntimeTemplateIndex {value=65535}
      instanceDataHandle {index=65535, source=None, stateHandle {index=65535}}
      instanceTemplateIndex {value=65535}
      outputBindingsBatch {value=65535}
     }
    }
   }]
  transitions [
   {id=#10, trigger=ON_EVENT, state {id=#3, name=OtherPlayer}}
   {id=#11, trigger=ON_EVENT, state {id=#4, name=LocalPlayer}}
   {id=#12, trigger=ON_EVENT, state {id=#6, name=Party}}
   {id=#13, trigger=ON_EVENT, state {id=#7, name=VoiceDisclosure}}
   {id=#14, trigger=ON_EVENT, state {id=#5, name=OtherPlayerNickname}}
  ]
 }
 OtherPlayer {id=#3, type=LINKED_ASSET, transitions []}
 LocalPlayer {id=#4, type=LINKED_ASSET, transitions []}
 OtherPlayerNickname {
  id=#5
  type=STATE
  tasks [{
    id=#15
    description=UIState Tree Widget Slotting Task
    properties {
     iD=#16
     instance {
      _structType=&UIStateTreeWidgetSlottingTaskInstanceData
      bIsTaskActive=false
      widgets [{
        bAutoActivateWidget=true
        bCacheInstance=false
        bFill=true
        bIsPlayerSpecific=true
        bIsWidgetEnabled=true
        bKeepConstructed=true
        cacheEntryName=OtherPlayerPanel
        memoryTag=None
        widget=/Sidebar/Panels/Social/Profile/WBP_Profile_OtherPlayer_Nickname.WBP_Profile_OtherPlayer_Nickname_C
        hostTag {tagName=UI.Sidebar.SubPanel}
        padding=<ltrb 0 0 0 0>
       }]
     }
     node {
      name=None
      _structType=&UIStateTreeWidgetSlottingTask
      bCanEditConsideredForCompletion=true
      bConsideredForCompletion=false
      bHasTaskCompletionDelegateDispatcher=false
      bTaskEnabled=true
      transitionHandlingPriority=Normal
      bindingsBatch {value=65535}
      executionRuntimeTemplateIndex {value=65535}
      instanceDataHandle {index=65535, source=None, stateHandle {index=65535}}
      instanceTemplateIndex {value=65535}
      outputBindingsBatch {value=65535}
     }
    }
   }]
  transitions []
 }
 Party {id=#6, type=LINKED_ASSET, transitions []}
 VoiceDisclosure {
  id=#7
  type=STATE
  tasks [{
    id=#17
    description=UIState Tree Widget Slotting Task
    properties {
     iD=#18
     instance {
      _structType=&UIStateTreeWidgetSlottingTaskInstanceData
      bIsTaskActive=false
      widgets [{
        bAutoActivateWidget=true
        bCacheInstance=false
        bFill=true
        bIsPlayerSpecific=true
        bIsWidgetEnabled=true
        bKeepConstructed=false
        cacheEntryName=None
        memoryTag=None
        widget=/Sidebar/Panels/Social/VoiceReport/WBP_VoiceReport_Disclosure.WBP_VoiceReport_Disclosure_C
        hostTag {tagName=UI.Sidebar.SubPanel}
        padding=<ltrb 0 0 0 0>
       }]
     }
     node {
      name=None
      _structType=&UIStateTreeWidgetSlottingTask
      bCanEditConsideredForCompletion=true
      bConsideredForCompletion=false
      bHasTaskCompletionDelegateDispatcher=false
      bTaskEnabled=true
      transitionHandlingPriority=Normal
      bindingsBatch {value=65535}
      executionRuntimeTemplateIndex {value=65535}
      instanceDataHandle {index=65535, source=None, stateHandle {index=65535}}
      instanceTemplateIndex {value=65535}
      outputBindingsBatch {value=65535}
     }
    }
   }]
  transitions []
 }
]
