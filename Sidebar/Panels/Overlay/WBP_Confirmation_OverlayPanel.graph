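# Sidebar overlay confirmation dialog with title, description, optional sub-description and up to three buttons.
# On activation it takes the confirmation VM from the current navigation point; SetConfirmationButtons maps button VMs to text, style and visibility.
# Confirms execute their VM action or, for HandleInBlueprint, publish an action ID on the navigation VM, then pop to NavigateBackTag or back one.
# Every exit, including Decline and gamepad Back, dispatches the overlay-closed transition and clears that ID.
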WidgetBlueprint /Sidebar/Panels/Overlay/WBP_Confirmation_OverlayPanel
graphs [
 BP_GetDesiredFocusTarget {nodes [
   Get Desired Focus Target()
   > Return Node!(execute, ReturnValue:Widget Object Reference=Get Button_Decline() -> (Button_Decline:WBP Sidebar Button Generic Object Reference))
  ]}
 EventGraph {nodes [
   n1 = "On Clicked (Button_Decline)"!() -> (OutputDelegate, then, Button:Common Button Base Object Reference) {is=title K2Node_ComponentBoundEvent}
   n2 = Get FortSidebarNavigationVM() -> (FortSidebarNavigationVM:Fort Sidebar Navigation VM Object Reference)
   n3 = PopNavigation(execute=[@n7.then, @n9.else, @n1.then, @n14.then], self:Fort Sidebar Navigation VM Object Reference=@n2)
   Dispatch Transition Event!(
    execute=[@n3.then, @n9.then]
    self:UIState Tree Manager Object Reference=Get UIStateTreeManager() -> (ReturnValue:UIState Tree Manager Object Reference) {is=title K2Node_GetSubsystem}
    TransitionTag:Gameplay Tag Structure=(TagName="UI.StateTree.TransitionTag.Sidebar.Overlay.Closed")
    StateTreeOwner:Player Controller Object Reference
    AnimationHints:Gameplay Tag Container Structure
    IncompatibleAnimationHints:Gameplay Tag Container Structure
   ) -> (then)
   > SetOverlayConfirmationActionID!(execute, self:Fort Sidebar Navigation VM Object Reference=Get FortSidebarNavigationVM(), InOverlayConfirmationActionID:Name=None) -> (then)
   "On Clicked (Button_ConfirmSecond)"!() -> (OutputDelegate, then, Button:Common Button Base Object Reference) {is=title K2Node_ComponentBoundEvent}
   > n4 = Branch(Condition="Not Equal (Enum)"!(
     A:ESidebarConfirmAction Enum=Get SecondaryConfirmationAction() -> (SecondaryConfirmationAction:ESidebarConfirmAction Enum=None)
     B:ESidebarConfirmAction Enum=HandleInBlueprint
    ) -> (ReturnValue:bool))
   > n5 = ExecuteAction!(
    execute
    self:Sidebar Confirmation Panel VM Object Reference=Get FortSidebarConfirmPanelVM() -> (FortSidebarConfirmPanelVM:Sidebar Confirmation Panel VM Object Reference)
    Action:ESidebarConfirmAction Enum=Get SecondaryConfirmationAction() -> (SecondaryConfirmationAction=None)
   ) -> (then)
   n6 = Get NavigateBackTag!(self:Sidebar Confirmation Panel VM Object Reference=Get FortSidebarConfirmPanelVM()) -> (NavigateBackTag:Gameplay Tag Structure)
   n7 = Branch(execute=[@n13.then, @n5.then, @n12.then, @n10.then], Condition="Equal (GameplayTag)"(A=@n6))
   n8 = PopToNavigationTag!(execute=@n7.else, self:Fort Sidebar Navigation VM Object Reference=@n2, NavigationTag:Gameplay Tag Structure=@n6) -> (then, ReturnValue:bool)
   > n9 = Branch(Condition=@n8)
   n10 = SetOverlayConfirmationActionID(
    execute=@n4.else
    self=Get FortSidebarNavigationVM()
    InOverlayConfirmationActionID=Get SecondaryConfirmationID() -> (SecondaryConfirmationID:Name=None)
   )
   "On Clicked (Button_Confirm)"!() -> (OutputDelegate, then, Button:Common Button Base Object Reference) {is=title K2Node_ComponentBoundEvent}
   > n11 = Branch(Condition="Not Equal (Enum)"(A=Get PrimaryConfirmationAction() -> (PrimaryConfirmationAction:ESidebarConfirmAction Enum=None), B=HandleInBlueprint))
   > n12 = ExecuteAction(self=Get FortSidebarConfirmPanelVM(), Action=Get PrimaryConfirmationAction() -> (PrimaryConfirmationAction=None))
   n13 = SetOverlayConfirmationActionID(
    execute=@n11.else
    self=Get FortSidebarNavigationVM()
    InOverlayConfirmationActionID=Get PrimaryConfirmationID() -> (PrimaryConfirmationID:Name=None)
   )
   n14 = "On Clicked (DecoyButton_BackGamepad)"!() -> (OutputDelegate, then, Button:Common Button Base Object Reference) {is=title K2Node_ComponentBoundEvent}
   Event On Activated()
   > n15 = GetCurrentNavigationPoint!(execute, self:Fort Sidebar Navigation VM Object Reference=Get FortSidebarNavigationVM()) -> (then, ReturnValue:Fort Sidebar Navigation Point Object Reference)
   > n16 = Cast To BP_NavigationPoint!(execute, Object=@n15) -> (then, CastFailed, AsBP Navigation Point:BP Navigation Point Object Reference) {is=title K2Node_DynamicCast}
   > Set FortSidebarConfirmPanelVM(FortSidebarConfirmPanelVM:Sidebar Confirmation Panel VM Object Reference=Get ConfirmationVM!(self:BP Navigation Point Object Reference=@n16) -> (ConfirmationVM:Sidebar Confirmation Panel VM Object Reference)) -> (Output_Get:Sidebar Confirmation Panel VM Object Reference)
  ]}
 SetActionButtonStyling {nodes [
   n1 = SetActionButtonStyling!() -> (then, Button:WBP Sidebar Button Generic Object Reference, ButtonStyle:ESidebarConfirmActionButtonStyle Enum=Default) {is=title K2Node_FunctionEntry}
   > UpdateButtonType!(
    execute
    self:WBP Sidebar Button Generic Object Reference=@n1
    ButtonType:E_UI_Sidebar_Button_Type Enum=Select(
     Option 0:E_UI_Sidebar_Button_Type Enum=NewEnumerator0
     Option 1:E_UI_Sidebar_Button_Type Enum=NewEnumerator8
     Index="Equal (Enum)"(A:ESidebarConfirmActionButtonStyle Enum=@n1.ButtonStyle, B:ESidebarConfirmActionButtonStyle Enum=Destructive)
    ) -> (ReturnValue:E_UI_Sidebar_Button_Type Enum)
   ) -> (then)
  ]}
 SetConfirmationButtons {nodes [
   n1 = SetConfirmationButtons!() -> (then, Buttons:Array of Sidebar Confirmation Button VM Object References) {is=title K2Node_FunctionEntry}
   > n2 = Sequence!(execute) -> (then_0, then_1, then_2:Exec)
   n3 = "Get (a copy)"(Array:Array of Sidebar Confirmation Button VM Object References=@n1) -> (Output:Sidebar Confirmation Button VM Object Reference)
   > Is Valid(InputObject=@n3)
   > n4 = Set PrimaryConfirmationAction(PrimaryConfirmationAction:ESidebarConfirmAction Enum=Get ActionType!(self:Sidebar Confirmation Button VM Object Reference=@n3) -> (ActionType:ESidebarConfirmAction Enum=None)) -> (Output_Get:ESidebarConfirmAction Enum=None)
   > Set PrimaryConfirmationID(PrimaryConfirmationID:Name=Get ActionID!(self:Sidebar Confirmation Button VM Object Reference=@n3) -> (ActionID:Name=None)) -> (Output_Get:Name=None)
   n5 = "Equal (Enum)"(
    A:ESidebarConfirmAction Enum=@n4
    B:ESidebarConfirmAction Enum=Literal enum ESidebarConfirmAction!(Enum:ESidebarConfirmAction Enum=None) -> (ReturnValue:ESidebarConfirmAction Enum) {is=title K2Node_EnumLiteral}
   )
   n6 = Get Button_Confirm() -> (Button_Confirm:WBP Sidebar Button Generic Object Reference)
   > SetVisibility(
    self=@n6
    InVisibility=Select(Option 0:ESlateVisibility Enum=SelfHitTestInvisible, Option 1:ESlateVisibility Enum=Collapsed, Index=@n5) -> (ReturnValue:ESlateVisibility Enum)
   )
   > SetActionButtonStyling!(
    execute
    self
    Button:WBP Sidebar Button Generic Object Reference=@n6
    ButtonStyle:ESidebarConfirmActionButtonStyle Enum=Get ActionButtonStyle!(self:Sidebar Confirmation Button VM Object Reference=@n3) -> (ActionButtonStyle:ESidebarConfirmActionButtonStyle Enum=Default)
   ) -> (then) {is=title K2Node_CallFunction}
   > n7 = Branch(Condition=@n5)
   n8 = Get Text(self:Sidebar Confirmation Button VM Object Reference=@n3)
   > SetText!(execute, self:WBP UIKit Button Generic Base Object Reference=Get Button_Decline(), Text:Text=@n8) -> (then)
   SetText(execute=@n7.else, self=@n6, Text=@n8)
   n9 = "Get (a copy)"(Array=@n1, Dimension 1=1)
   Is Valid(exec=@n2.then_1, InputObject=@n9)
   > n10 = Set SecondaryConfirmationAction(SecondaryConfirmationAction:ESidebarConfirmAction Enum=Get ActionType(self=@n9) -> (ActionType=None)) -> (Output_Get:ESidebarConfirmAction Enum=None)
   > Set SecondaryConfirmationID(SecondaryConfirmationID:Name=Get ActionID(self=@n9) -> (ActionID=None)) -> (Output_Get:Name=None)
   n11 = "Equal (Enum)"(A=@n10, B=Literal enum ESidebarConfirmAction(Enum=None))
   n12 = Get Button_ConfirmSecond() -> (Button_ConfirmSecond:WBP Sidebar Button Generic Object Reference)
   > SetVisibility(self=@n12, InVisibility=Select(Option 0=SelfHitTestInvisible, Option 1=Collapsed, Index=@n11))
   > SetActionButtonStyling(Button=@n12, ButtonStyle=Get ActionButtonStyle(self=@n9) -> (ActionButtonStyle=Default))
   > n13 = Branch(Condition=@n11)
   n14 = Get Text(self=@n9)
   > SetText(self=Get Button_Decline(), Text=@n14)
   SetText(execute=@n13.else, self=@n12, Text=@n14)
   n15 = "Get (a copy)"(Array=@n1, Dimension 1=2)
   Is Valid(exec=@n2.then_2, InputObject=@n15)
   n16 = Get Button_Decline()
   > SetText(self=@n16, Text=Get Text(self=@n15))
   > SetActionButtonStyling(Button=@n16, ButtonStyle=Get ActionButtonStyle(self=@n15) -> (ActionButtonStyle=Default))
  ]}
 SetSubDescription {nodes [
   n1 = SetSubDescription!() -> (then, SubDescriptionText:Text) {is=title K2Node_FunctionEntry}
   > n2 = Branch(Condition=TextIsEmpty(InText=@n1))
   n3 = Get SubDescription() -> (SubDescription:Common Text Object Reference)
   > Set Text(self=@n3)
   > SetVisibility(self=@n3, InVisibility=Collapsed)
   Set Text(execute=@n2.else, Text=@n1, self=@n3)
   > SetVisibility(self=@n3, InVisibility=SelfHitTestInvisible)
  ]}
]
variables [
 PrimaryConfirmationAction {type=ESidebarConfirmAction, default=None}
 PrimaryConfirmationID {type=string, default=None}
 SecondaryConfirmationAction {type=ESidebarConfirmAction, default=None}
 SecondaryConfirmationID {type=string, default=None}
]
widgets {
 info {inheritedWidgetCount=0, parentClass=&CommonActivatableWidget, rootWidgetClass=&StackBox, widgetCount=9}
 widgets [
  Root_NavStop {
   widgetClassPath=&StackBox
   widget=Root_NavStop
   properties {
    bOverride_Cursor=true
    navigation=&self:WidgetTree.Root_NavStop.WidgetNavigation_0
    orientation=Orient_Vertical
   }
  }
  Title {
   widgetClassPath=&CommonTextBlock
   bIsVariable=true
   parent=Root_NavStop
   widget=Title
   properties {
    mobileFontSizeMultiplier=1.2
    text=Exit
    colorAndOpacity {colorUseRule=UseColor_Specified, specifiedColor=<rgba 1 1 1 0.95>}
    font {
     bForceMonospaced=false
     bMaterialIsStencil=false
     fontMaterial=null
     fontObject=&HeadingNow
     letterSpacing=0
     monospacedWidth=1
     size=22.5
     skewAmount=0
     typefaceFontName=75 Medium
     outlineSettings {
      bApplyOutlineToDropShadows=false
      bMiteredCorners=false
      bSeparateFillAlpha=false
      outlineMaterial=null
      outlineSize=0
      outlineColor=<rgba 0 0 0 1>
     }
    }
    shadowOffset=<0 0>
   }
   slot {class=StackBoxSlot}
  }
  Description {
   widgetClassPath=&CommonTextBlock
   bIsVariable=true
   parent=Root_NavStop
   widget=Description
   properties {
    autoWrapText=true
    lineHeightPercentage=1.2
    mobileFontSizeMultiplier=1.41
    text=Description text
    font {
     bForceMonospaced=false
     bMaterialIsStencil=false
     fontMaterial=/Sidebar/Materials/Fonts/MI_Sidebar_FontColor_Secondary
     fontObject=&HeadingNow
     letterSpacing=0
     monospacedWidth=1
     size=12.75
     skewAmount=0
     typefaceFontName=74 Regular
     outlineSettings {
      bApplyOutlineToDropShadows=false
      bMiteredCorners=false
      bSeparateFillAlpha=false
      outlineMaterial=null
      outlineSize=0
      outlineColor=<rgba 0 0 0 1>
     }
    }
    shadowOffset=<0 0>
   }
   slot {class=StackBoxSlot, properties {padding=<ltrb 0 12 0 0>}}
  }
  SubDescription {
   widgetClassPath=&CommonTextBlock
   bIsVariable=true
   parent=Root_NavStop
   widget=SubDescription
   properties {
    autoWrapText=true
    lineHeightPercentage=1.2
    mobileFontSizeMultiplier=1.41
    text=SubDescription text
    font {
     bForceMonospaced=false
     bMaterialIsStencil=false
     fontMaterial=&MI_Sidebar_FontColor_Secondary
     fontObject=&HeadingNow
     letterSpacing=0
     monospacedWidth=1
     size=12.75
     skewAmount=0
     typefaceFontName=74 Regular
     outlineSettings {
      bApplyOutlineToDropShadows=false
      bMiteredCorners=false
      bSeparateFillAlpha=false
      outlineMaterial=null
      outlineSize=0
      outlineColor=<rgba 0 0 0 1>
     }
    }
    shadowOffset=<0 0>
   }
   slot {class=StackBoxSlot, properties {padding=<ltrb 0 12 0 0>}}
  }
  StackBox_1 {
   widgetClassPath=&StackBox
   parent=Root_NavStop
   widget=StackBox_1
   properties {bOverride_Cursor=true, orientation=Orient_Vertical}
   slot {class=StackBoxSlot, properties {padding=<ltrb 0 36 0 0>}}
  }
  Button_Confirm {
   widgetClassPath=/Sidebar/Buttons/WBP_Sidebar_ButtonGeneric
   bIsVariable=true
   parent=StackBox_1
   widget=Button_Confirm
   properties {
    buttonType=Alert
    minHeightOverride=60
    text=Yes
    font {
     bForceMonospaced=false
     bMaterialIsStencil=false
     fontMaterial=/FortUILibrary/UIKit/Buttons/MaterialInstances/CTAButtons/MI_UIKit_TextQuiet_Prompt
     fontObject=&HeadingNow
     letterSpacing=10
     monospacedWidth=1
     size=15.75
     skewAmount=0
     typefaceFontName=74 Regular
     outlineSettings {
      bApplyOutlineToDropShadows=false
      bMiteredCorners=false
      bSeparateFillAlpha=false
      outlineMaterial=/FortUILibrary/Systems/ModularButton/MaterialInstances/MI_UIKit_TextOutline
      outlineSize=0
      outlineColor=<rgba 0 0 0 1>
     }
    }
   }
   slot {class=StackBoxSlot, properties {padding=<ltrb 0 0 0 12>, size {sizeRule=Fill, value=1}}}
  }
  Button_ConfirmSecond {
   widgetClassPath=&WBP_Sidebar_ButtonGeneric
   bIsVariable=true
   parent=StackBox_1
   widget=Button_ConfirmSecond
   properties {
    buttonType=Alert
    minHeightOverride=60
    text=Yes
    font {
     bForceMonospaced=false
     bMaterialIsStencil=false
     fontMaterial=&MI_UIKit_TextQuiet_Prompt
     fontObject=&HeadingNow
     letterSpacing=10
     monospacedWidth=1
     size=15.75
     skewAmount=0
     typefaceFontName=74 Regular
     outlineSettings {
      bApplyOutlineToDropShadows=false
      bMiteredCorners=false
      bSeparateFillAlpha=false
      outlineMaterial=&MI_UIKit_TextOutline
      outlineSize=0
      outlineColor=<rgba 0 0 0 1>
     }
    }
   }
   slot {class=StackBoxSlot, properties {padding=<ltrb 0 0 0 12>, size {sizeRule=Fill, value=1}}}
  }
  Button_Decline {
   widgetClassPath=&WBP_Sidebar_ButtonGeneric
   bIsVariable=true
   parent=StackBox_1
   widget=Button_Decline
   properties {
    minHeightOverride=60
    text=No
    font {
     bForceMonospaced=false
     bMaterialIsStencil=false
     fontMaterial=&MI_UIKit_TextQuiet_Prompt
     fontObject=&HeadingNow
     letterSpacing=10
     monospacedWidth=1
     size=15.75
     skewAmount=0
     typefaceFontName=74 Regular
     outlineSettings {
      bApplyOutlineToDropShadows=false
      bMiteredCorners=false
      bSeparateFillAlpha=false
      outlineMaterial=&MI_UIKit_TextOutline
      outlineSize=0
      outlineColor=<rgba 0 0 0 1>
     }
    }
   }
   slot {class=StackBoxSlot, properties {size {sizeRule=Fill, value=1}}}
  }
  DecoyButton_BackGamepad {
   widgetClassPath=/Game/UI/Frontend/BottomBar/WBP_BottomBarDecoyButton
   bIsVariable=true
   parent=Root_NavStop
   widget=DecoyButton_BackGamepad
   properties {
    bDisplayInActionBar=false
    holdData=&CommonUI_HoldData_Base_C
    triggeringInputAction {dataTable=&InputActionDataTable, rowName=Back}
   }
   slot {class=StackBoxSlot}
  }
 ]
}
