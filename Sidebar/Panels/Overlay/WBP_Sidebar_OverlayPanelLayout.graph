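# Activatable sidebar overlay layout that frames content slotted into the UI.Sidebar.OverlayPanel host as a bottom-anchored panel.
# InitializeVisuals sets colors and rounded corners on a full-size tint backplate and two stacked panel backplates, at init and in designer preview.
# A transparent ClickCatcher button layered beneath the slotting host pops sidebar navigation and dispatches the Overlay.Closed transition to the UI State Tree.
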WidgetBlueprint /Sidebar/Panels/Overlay/WBP_Sidebar_OverlayPanelLayout
graphs [
 EventGraph {nodes [
   "On Clicked (ClickCatcher)"() {is=title K2Node_ComponentBoundEvent}
   > PopNavigation(self:Fort Sidebar Navigation VM Object Reference=Get FortSidebarNavigationVM() -> (FortSidebarNavigationVM:Fort Sidebar Navigation VM Object Reference))
   > Dispatch Transition Event!(
    execute
    self:UIState Tree Manager Object Reference=Get UIStateTreeManager() -> (ReturnValue:UIState Tree Manager Object Reference) {is=title K2Node_GetSubsystem}
    TransitionTag:Gameplay Tag Structure=(TagName="UI.StateTree.TransitionTag.Sidebar.Overlay.Closed")
    StateTreeOwner:Player Controller Object Reference
    AnimationHints:Gameplay Tag Container Structure
    IncompatibleAnimationHints:Gameplay Tag Container Structure
   ) -> (then)
   n1 = Event PreConstruct()
   > Branch(Condition=@n1.IsDesignTime)
   > InitializeVisuals()
   Event OnInitialized()
   > InitializeVisuals()
  ]}
 InitializeVisuals {nodes [
   InitializeVisuals() {is=title K2Node_FunctionEntry}
   n1 = Get Backplate_Panel() -> (Backplate_Panel:WBP Sidebar Block Backplate Object Reference)
   > SetCornerRadius!(execute, self:WBP Sidebar Block Backplate Object Reference=@n1, Radius In Pixels:bool=true, Corner Radius:double=30) -> (then)
   > SetBackplateColor!(execute, self:WBP Sidebar Block Backplate Object Reference=@n1, Backplate Switch:E_UI_Sidebar_Button_Backplate_Color Enum=NewEnumerator1) -> (then)
   > SetSecondaryColor!(execute, self:WBP Sidebar Block Backplate Object Reference=@n1, Secondary Color:Linear Color Structure=(R=0.9911,G=0.15,B=0.1413,A=1)) -> (then)
   n2 = Get Backplate_Tint() -> (Backplate_Tint:WBP Sidebar Block Backplate Object Reference)
   > SetCornerRadius(self=@n2, Radius In Pixels=true, Corner Radius=36)
   > SetBackplateColor(self=@n2, Backplate Switch=NewEnumerator8)
   n3 = Get Backplate_UnderneathPanel() -> (Backplate_UnderneathPanel:WBP Sidebar Block Backplate Object Reference)
   > SetCornerRadius(self=@n3, Radius In Pixels=true, Corner Radius=30)
   > SetBackplateColor(self=@n3, Backplate Switch=NewEnumerator11)
  ]}
]
variables []
widgets {
 info {inheritedWidgetCount=0, parentClass=&CommonActivatableWidget, rootWidgetClass=&GridPanel, widgetCount=6}
 widgets [
  Grid_Root {widgetClassPath=&GridPanel, widget=Grid_Root, properties {columnFill [1], rowFill [1,0]}}
  Backplate_Tint {
   widgetClassPath=/Sidebar/Buttons/Blocks/WBP_Sidebar_Block_Backplate
   bIsVariable=true
   parent=Grid_Root
   widget=Backplate_Tint
   slot {class=GridSlot, properties {rowSpan=2}}
  }
  Backplate_UnderneathPanel {
   widgetClassPath=&WBP_Sidebar_Block_Backplate
   bIsVariable=true
   parent=Grid_Root
   widget=Backplate_UnderneathPanel
   slot {class=GridSlot, properties {row=1, padding=<ltrb 6 6 6 6>}}
  }
  Backplate_Panel {
   widgetClassPath=&WBP_Sidebar_Block_Backplate
   bIsVariable=true
   parent=Grid_Root
   widget=Backplate_Panel
   slot {class=GridSlot, properties {row=1, padding=<ltrb 6 6 6 6>}}
  }
  ClickCatcher {
   widgetClassPath=&Button
   parent=Grid_Root
   widget=ClickCatcher
   properties {backgroundColor=<rgba 1 1 1 0>}
   slot {class=GridSlot, properties {layer=1, rowSpan=2}}
  }
  SlottingHost {
   widgetClassPath=/Script/WidgetSlotting.WidgetSlottingHostOverlay
   bIsVariable=true
   parent=Grid_Root
   widget=SlottingHost
   properties {hostTag {tagName=UI.Sidebar.OverlayPanel}}
   slot {class=GridSlot, properties {layer=3, row=1, verticalAlignment=VAlign_Bottom, padding=<ltrb 32 32 32 32>}}
  }
 ]
}
