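# Sidebar profile panel for the local player, with header, current activity, Online/Away tabs and party joinability, participation and mutual-friends setting buttons.
# Focus path activates it; activation takes the social user from the navigation point, falling back to the owning player, and initializes viewmodels and child widgets.
# Name-to-enum maps translate tab and participation choices into viewmodel updates.
# An animated material scalar shifts the status tab backplate.
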WidgetBlueprint /Sidebar/Panels/Social/Profile/WBP_Profile_LocalPlayer
animations [Anim_OnlineStatusBackplate {
  binding_count=1
  length_seconds=0.3
  object=&self:Anim_OnlineStatusBackplate
  sequence {
   bindings [OnlineStatus_Backplate {
     id=#1
     possessed_class=Image
     tracks [{
       type=MovieSceneWidgetMaterialTrack
       sections [{
         type=MovieSceneParameterSection
         end_frame=30
         start_frame=0
         channels [Control {type=MovieSceneScriptingFloatChannel, thinned_from=3, keys [[0,0],[30,1]]}]
        }]
      }]
    }]
   display_rate {numerator=100}
  }
 }]
graphs [
 BP_GetDesiredFocusTarget {nodes [
   Get Desired Focus Target()
   n1 = Get OnlineStatus_TabGroup() -> (OnlineStatus_TabGroup:WBP UIKit Tab Group Object Reference)
   > Return Node!(
    execute
    ReturnValue:Widget Object Reference=GetTabButtonBaseByID!(
     self:Common Tab List Widget Base Object Reference=@n1
     TabNameID:Name=GetSelectedTabId!(self:Common Tab List Widget Base Object Reference=@n1) -> (ReturnValue:Name=None)
    ) -> (ReturnValue:Common Button Base Object Reference)
   )
  ]}
 BindEnableParticipation {nodes [
   n1 = BindEnableParticipation!() -> (then, NumPlayers:int) {is=title K2Node_FunctionEntry}
   > SetIsEnabled(
    self=Get Button_Participation() -> (Button_Participation:WBP Sidebar Button With Context Menu Object Reference)
    bInIsEnabled=Select(Option 0:bool=true, Option 1:bool=false, Index="integer < integer"(A=@n1, B=2)) -> (ReturnValue:bool)
   )
  ]}
 ConvertSocialSettingToJoinableVisibility {nodes [
   n1 = ConvertSocialSettingToJoinableVisibility!() -> (then, Setting Value:Text, Possible Values:Array of Texts) {is=title K2Node_FunctionEntry}
   n2 = "Get (a copy)"(Array:Array of Texts=@n1.Possible Values) -> (Output:Text)
   > Return Node!(
    execute
    Visibility:ESlateVisibility Enum=Select(
     Option 0:ESlateVisibility Enum=Collapsed
     Option 1:ESlateVisibility Enum=SelfHitTestInvisible
     Index="Equal Exactly (Text)"!(A:"Text (by ref)"=@n1, B:"Text (by ref)"=@n2) -> (ReturnValue:bool)
    ) -> (ReturnValue:ESlateVisibility Enum)
   )
  ]}
 EventGraph {nodes [
   n1 = "On Tab Selected (OnlineStatus_TabGroup)"!() -> (OutputDelegate, then, TabId:Name=None) {is=title K2Node_ComponentBoundEvent}
   > SetOnlineStatusVM!(execute, self, TabID:Name=@n1.TabId) -> (then)
   n2 = "On Button Option Clicked (Button_Participation)"!() -> (OutputDelegate, then, Button ID:Name=None) {is=title K2Node_ComponentBoundEvent}
   n3 = Find(
    TargetMap:Map of Names to EGameReadiness Enums=Get ParticipationMap() -> (ParticipationMap:Map of Names to EGameReadiness Enums)
    Key:"Name (by ref)"=@n2.Button ID
   ) -> (Value:EGameReadiness Enum=NotReady)
   > Branch(Condition=@n3.ReturnValue)
   > SetGameReadiness!(
    execute
    self:Fort Social User Party View Model Object Reference=Get ParticipationSettingVM!(self:Fort Local Player Profile VM Object Reference=Get FortLocalPlayerProfileVM() -> (FortLocalPlayerProfileVM:Fort Local Player Profile VM Object Reference)) -> (ParticipationSettingVM:Fort Social User Party View Model Object Reference)
    Participation:EGameReadiness Enum=@n3
   ) -> (then)
   Event OnInitialized()
   > Parent: OnInitialized() {is=title K2Node_CallParentFunction}
   > n4 = GetConsoleVariableBoolValue!(execute, VariableName:String=fort.UseEpicParties) -> (then, ReturnValue:bool)
   > Set with Broadcast IsUsingEpicParties!(execute, IsUsingEpicParties:bool=@n4) -> (then, Output_Get:bool)
   Event OnAddedToFocusPath!() -> (OutputDelegate, then, InFocusEvent:Focus Event Structure) {is=title K2Node_Event}
   > ActivateWidget()
   Event On Activated()
   > Set Social User!(execute:Exec) -> (then:Exec) {is=title K2Node_MacroInstance}
   > Set Local Player Settings!(execute:Exec) -> (then:Exec) {is=title K2Node_MacroInstance}
   > SidebarFeatureFinishedLoading(self:Fort Sidebar VM Object Reference=Get FortSidebarVM() -> (FortSidebarVM:Fort Sidebar VM Object Reference))
   Event OnRemovedFromFocusPath!() -> (OutputDelegate, then, InFocusEvent:Focus Event Structure) {is=title K2Node_Event}
   > DeactivateWidget()
   Event Construct()
   > Set with Broadcast ShowBackButton!(
    execute
    ShowBackButton:bool=Get IsUsingEpicParties() -> (IsUsingEpicParties:bool)
    self:WBP Sidebar Panel Layout Object Reference=Get Root() -> (Root:WBP Sidebar Panel Layout Object Reference)
   ) -> (then, Output_Get:bool)
  ]}
 NavigateUpFromButtons {nodes [
   NavigateUpFromButtons!() -> (then, Navigation:EUINavigation Enum) {is=title K2Node_FunctionEntry}
   n1 = Get OnlineStatus_TabGroup()
   > Return Node!(execute, ReturnValue:Widget Object Reference=GetTabButtonBaseByID(self=@n1, TabNameID=GetSelectedTabId(self=@n1) -> (ReturnValue=None)))
  ]}
 SetLocalPlayerSettings {nodes [
   Inputs()
   n1 = Get OnlineStatus_Backplate() -> (OnlineStatus_Backplate:Image Object Reference)
   > SetScalarParameterValue(
    self=GetDynamicMaterial!(Brush:"Slate Brush Structure (by ref)"=Get Brush!(self:Image Object Reference=@n1) -> (Brush:Slate Brush Structure)) -> (ReturnValue)
    ParameterName=Control
    Value=Select!(
     Online:float=0
     Away:float=1
     Offline:float=1
     Blocked:float=1
     Index:EPresenceIndicatorState Enum=Get OnlineStatus() -> (OnlineStatus:EPresenceIndicatorState Enum=Online)
    ) -> (ReturnValue:float)
   )
   > SetButtonVM!(
    execute
    self:WBP Sidebar Button With Context Menu Object Reference=Get Button_PartyJoinability() -> (Button_PartyJoinability:WBP Sidebar Button With Context Menu Object Reference)
    ViewModel:Fort Social Setting Discrete VM Object Reference=Get PartyJoinabilitySettingVM!(self:Fort Local Player Profile VM Object Reference=Get FortLocalPlayerProfileVM()) -> (PartyJoinabilitySettingVM:Fort Social Setting Discrete VM Object Reference)
   ) -> (then)
   > SetButtonVM(
    self=Get Button_MutualFriends() -> (Button_MutualFriends:WBP Sidebar Button With Context Menu Object Reference)
    ViewModel=Get ShowMutualFriendsSettingVM!(self:Fort Local Player Profile VM Object Reference=Get FortLocalPlayerProfileVM()) -> (ShowMutualFriendsSettingVM:Fort Social Setting Discrete VM Object Reference)
   )
   n2 = Get ParticipationMap()
   > n3 = Values!(execute, TargetMap:Map of Names to EGameReadiness Enums=@n2) -> (then, Values:Array of EGameReadiness Enums)
   > n4 = Keys!(execute, TargetMap:Map of Names to EGameReadiness Enums=@n2) -> (then, Keys:Array of Names)
   n5 = GetGameReadiness!(self:Fort Social User Party View Model Object Reference=Get ParticipationSettingVM(self=Get FortLocalPlayerProfileVM())) -> (ReturnValue:EGameReadiness Enum=NotReady)
   > SetLocalActiveOption!(
    execute
    self:WBP Sidebar Button With Context Menu Object Reference=Get Button_Participation()
    Name:Name="Get (a ref)"(
     Array:Array of Names=@n4
     Dimension 1=Find Item!(TargetArray:Array of EGameReadiness Enums=@n3, ItemToFind:"EGameReadiness Enum (by ref)"=@n5) -> (ReturnValue)
    ) -> (Output:"Name (by ref)")
   ) -> (then)
   > Outputs!(then:Exec)
  ]}
 SetOnlineStatus {nodes [
   n1 = SetOnlineStatus!() -> (then, Status:EPresenceIndicatorState Enum=Online) {is=title K2Node_FunctionEntry}
   > n2 = Set OnlineStatus(OnlineStatus:EPresenceIndicatorState Enum=@n1) -> (Output_Get:EPresenceIndicatorState Enum=Online)
   > SelectTabByIndex!(
    execute
    self:UIKit Tab Group Object Reference=Get OnlineStatus_TabGroup()
    Index:int=Select(Online:int=0, Away:int=1, Offline:int=1, Blocked:int=1, Index=@n2) -> (ReturnValue:int)
    bSuppressClickFeedback:bool=true
   ) -> (then, ReturnValue:bool)
  ]}
 SetOnlineStatusVM {nodes [
   n1 = SetOnlineStatusVM!() -> (then, TabID:Name) {is=title K2Node_FunctionEntry}
   n2 = Get FortLocalPlayerProfileVM()
   n3 = Find(
    TargetMap:Map of Names to EPresenceIndicatorState Enums=Get OnlineStatusMap() -> (OnlineStatusMap:Map of Names to EPresenceIndicatorState Enums)
    Key=@n1
   ) -> (Value:EPresenceIndicatorState Enum=Online)
   > Branch(Condition=AND Boolean(A=IsValid(Object=@n2), B=@n3.ReturnValue))
   > Branch(Condition="Not Equal (Enum)"!(A:EPresenceIndicatorState Enum=@n3, B:EPresenceIndicatorState Enum=Get OnlineStatus() -> (OnlineStatus=Online)) -> (ReturnValue:bool))
   > n4 = Set OnlineStatus(OnlineStatus=@n3) -> (Output_Get=Online)
   > SetOnlineStatus!(
    execute
    self:Fort Online Status Viewmodel Object Reference=Get OnlineStatusVM!(self:Fort Local Player Profile VM Object Reference=Get FortLocalPlayerProfileVM()) -> (OnlineStatusVM:Fort Online Status Viewmodel Object Reference)
    InOnlineStatus:EPresenceIndicatorState Enum=@n4
   ) -> (then) {is=title K2Node_CallFunction}
   > n5 = Switch on EPresenceIndicatorState!(execute, Selection:EPresenceIndicatorState Enum=@n4) -> (Online:Exec, Away:Exec, Offline:Exec, Blocked:Exec) {is=title K2Node_SwitchEnum}
   n6 = Get Anim_OnlineStatusBackplate() -> (Anim_OnlineStatusBackplate:Widget Animation Object Reference)
   PlayAnimationForward(execute=@n5.Away, InAnimation=@n6)
   PlayAnimationReverse(execute=@n5.Online, InAnimation=@n6)
  ]}
 SetSocialUser {nodes [
   Inputs()
   > n1 = GetCurrentNavigationPoint!(
    execute
    self:Fort Sidebar Navigation VM Object Reference=Get FortSidebarNavigationVM() -> (FortSidebarNavigationVM:Fort Sidebar Navigation VM Object Reference)
   ) -> (then, ReturnValue:Fort Sidebar Navigation Point Object Reference)
   > n2 = Cast To BP_NavigationPoint!(execute, Object=@n1) -> (then, CastFailed, AsBP Navigation Point:BP Navigation Point Object Reference) {is=title K2Node_DynamicCast}
   > Set with Broadcast LocalPlayerTabView!(
    execute
    LocalPlayerTabView:bool="Equal (GameplayTag)"(
     A=Get NavigationTag!(self:Fort Sidebar Navigation Point Object Reference=@n2) -> (NavigationTag:Gameplay Tag Structure)
     B=(TagName="UI.StateTree.TransitionTag.Sidebar.LocalPlayerTab")
    )
   ) -> (then, Output_Get:bool)
   n3 = Get SocialUser!(self:BP Navigation Point Object Reference=@n2) -> (SocialUser:Fort Social User Object Reference)
   > n4 = Is Valid(InputObject=@n3)
   n5 = GetOwningPlayer()
   n6 = SetSocialUserAndVM!(
    execute=@n4.Is Not Valid
    self
    SocialUser:Fort Social User Object Reference=GetSocialUserFromController!(FortPC:Fort Player Controller Object Reference=GetFortPlayerControllerFromActor!(Actor:Actor Object Reference=@n5) -> (ReturnValue:Fort Player Controller Object Reference)) -> (ReturnValue:Fort Social User Object Reference)
   ) -> (then)
   Outputs(then=[@n7.then, @n6.then])
   n7 = SetSocialUserAndVM(execute=@n4.Is Valid, SocialUser=@n3)
  ]}
 SetSocialUserAndVM {nodes [
   n1 = SetSocialUserAndVM!() -> (then, SocialUser:Fort Social User Object Reference) {is=title K2Node_FunctionEntry}
   > SetSocialUser!(execute, self:Fort Social User Viewmodel Object Reference=Get FortLocalPlayerProfileVM(), InSocialUser:Fort Social User Object Reference=@n1) -> (then)
   n2 = Get FortSocialItemVM() -> (FortSocialItemVM:Fort Social Item VM Object Reference)
   > Init!(execute, self:Fort Social Item VM Object Reference=@n2, InTargetUser:Fort Social User Object Reference=@n1) -> (then)
   > Initialize Header With Fort Social Item VM!(
    execute
    self:WBP Profile Header Object Reference=Get Profile_Header() -> (Profile_Header:WBP Profile Header Object Reference)
    FortSocialItemVM:Fort Social Item VM Object Reference=@n2
   ) -> (then)
   > SetSocialUser!(
    execute
    self:WBP Sidebar Current Activity Local Player Object Reference=Get LocalPlayer_CurrentActivity() -> (LocalPlayer_CurrentActivity:WBP Sidebar Current Activity Local Player Object Reference)
    Social User:Fort Social User Object Reference=@n1
   ) -> (then)
  ]}
]
variables [
 IsUsingEpicParties {type=boolean}
 LocalPlayerTabView {type=boolean}
 OnlineStatus {type=EPresenceIndicatorState, default=Online}
 OnlineStatusMap {type="map<EPresenceIndicatorState>", default {away=Away, online=Online}}
 ParticipationMap {type="map<EGameReadiness>", default {not_ready=NotReady, sitting_out=SittingOut}}
]
widgets {
 info {
  inheritedWidgetCount=0
  namedSlotCount=1
  parentClass=&CommonActivatableWidget
  rootWidgetClass=/Sidebar/Shared/WBP_Sidebar_PanelLayout.WBP_Sidebar_PanelLayout_C
  widgetCount=13
 }
 widgets [
  Root {widgetClassPath=/Sidebar/Shared/WBP_Sidebar_PanelLayout, bIsVariable=true, widget=Root, properties {headingText=""}}
  ScrollBox_0 {
   widgetClassPath=&ScrollBox
   namedSlotHost=Root
   widget=ScrollBox_0
   properties {
    clipping=Inherit
    scrollWhenFocusChanges=AnimatedScroll
    visibility=SelfHitTestInvisible
    widgetBarStyle {
     thickness=8
     draggedThumbImage {
      bIsDynamicallyLoaded=false
      drawAs=Box
      imageType=FullColor
      mirroring=NoMirror
      resourceName=../../../Engine/Content/Slate/Common/Scrollbar_Thumb.png
      resourceObject=null
      tiling=NoTile
      imageSize=<0 0>
      margin=<ltrb 0.25 0.25 0.25 0.25>
      outlineSettings {
       bUseBrushTransparency=false
       roundingType=HalfHeightRadius
       width=0
       color {colorUseRule=UseColor_Specified, specifiedColor=<rgba 0 0 0 0>}
       cornerRadii=<0 0 0 0>
      }
      tintColor {colorUseRule=UseColor_Specified, specifiedColor=<rgba 1 1 1 1>}
      uVRegion {bIsValid=false, max=<0 0>, min=<0 0>}
     }
     horizontalBackgroundImage {
      bIsDynamicallyLoaded=false
      drawAs=NoDrawType
      imageType=NoImage
      mirroring=NoMirror
      resourceName=None
      resourceObject=null
      tiling=NoTile
      imageSize=<0 0>
      margin=<ltrb 0 0 0 0>
      outlineSettings {
       bUseBrushTransparency=false
       roundingType=HalfHeightRadius
       width=0
       color {colorUseRule=UseColor_Specified, specifiedColor=<rgba 0 0 0 0>}
       cornerRadii=<0 0 0 0>
      }
      tintColor {colorUseRule=UseColor_Specified, specifiedColor=<rgba 1 1 1 1>}
      uVRegion {bIsValid=false, max=<0 0>, min=<0 0>}
     }
     horizontalBottomSlotImage {
      bIsDynamicallyLoaded=false
      drawAs=NoDrawType
      imageType=FullColor
      mirroring=NoMirror
      resourceName=../../../Engine/Content/Slate/Common/Scrollbar_Background_Horizontal.png
      resourceObject=null
      tiling=NoTile
      imageSize=<8 8>
      margin=<ltrb 0 0 0 0>
      outlineSettings {
       bUseBrushTransparency=false
       roundingType=HalfHeightRadius
       width=0
       color {colorUseRule=UseColor_Specified, specifiedColor=<rgba 0 0 0 0>}
       cornerRadii=<0 0 0 0>
      }
      tintColor {colorUseRule=UseColor_Specified, specifiedColor=<rgba 1 1 1 1>}
      uVRegion {bIsValid=false, max=<0 0>, min=<0 0>}
     }
     horizontalTopSlotImage {
      bIsDynamicallyLoaded=false
      drawAs=NoDrawType
      imageType=FullColor
      mirroring=NoMirror
      resourceName=../../../Engine/Content/Slate/Common/Scrollbar_Background_Horizontal.png
      resourceObject=null
      tiling=NoTile
      imageSize=<8 8>
      margin=<ltrb 0 0 0 0>
      outlineSettings {
       bUseBrushTransparency=false
       roundingType=HalfHeightRadius
       width=0
       color {colorUseRule=UseColor_Specified, specifiedColor=<rgba 0 0 0 0>}
       cornerRadii=<0 0 0 0>
      }
      tintColor {colorUseRule=UseColor_Specified, specifiedColor=<rgba 1 1 1 1>}
      uVRegion {bIsValid=false, max=<0 0>, min=<0 0>}
     }
     hoveredThumbImage {
      bIsDynamicallyLoaded=false
      drawAs=Box
      imageType=FullColor
      mirroring=NoMirror
      resourceName=../../../Engine/Content/Slate/Common/Scrollbar_Thumb.png
      resourceObject=null
      tiling=NoTile
      imageSize=<0 0>
      margin=<ltrb 0.25 0.25 0.25 0.25>
      outlineSettings {
       bUseBrushTransparency=false
       roundingType=HalfHeightRadius
       width=0
       color {colorUseRule=UseColor_Specified, specifiedColor=<rgba 0 0 0 0>}
       cornerRadii=<0 0 0 0>
      }
      tintColor {colorUseRule=UseColor_Specified, specifiedColor=<rgba 1 1 1 1>}
      uVRegion {bIsValid=false, max=<0 0>, min=<0 0>}
     }
     normalThumbImage {
      bIsDynamicallyLoaded=false
      drawAs=Box
      imageType=FullColor
      mirroring=NoMirror
      resourceName=../../../Engine/Content/Slate/Common/Scrollbar_Thumb.png
      resourceObject=null
      tiling=NoTile
      imageSize=<0 0>
      margin=<ltrb 0.25 0.25 0.25 0.25>
      outlineSettings {
       bUseBrushTransparency=false
       roundingType=HalfHeightRadius
       width=0
       color {colorUseRule=UseColor_Specified, specifiedColor=<rgba 0 0 0 0>}
       cornerRadii=<0 0 0 0>
      }
      tintColor {colorUseRule=UseColor_Specified, specifiedColor=<rgba 1 1 1 1>}
      uVRegion {bIsValid=false, max=<0 0>, min=<0 0>}
     }
     verticalBackgroundImage {
      bIsDynamicallyLoaded=false
      drawAs=NoDrawType
      imageType=NoImage
      mirroring=NoMirror
      resourceName=None
      resourceObject=null
      tiling=NoTile
      imageSize=<0 0>
      margin=<ltrb 0 0 0 0>
      outlineSettings {
       bUseBrushTransparency=false
       roundingType=HalfHeightRadius
       width=0
       color {colorUseRule=UseColor_Specified, specifiedColor=<rgba 0 0 0 0>}
       cornerRadii=<0 0 0 0>
      }
      tintColor {colorUseRule=UseColor_Specified, specifiedColor=<rgba 1 1 1 1>}
      uVRegion {bIsValid=false, max=<0 0>, min=<0 0>}
     }
     verticalBottomSlotImage {
      bIsDynamicallyLoaded=false
      drawAs=NoDrawType
      imageType=FullColor
      mirroring=NoMirror
      resourceName=../../../Engine/Content/Slate/Common/Scrollbar_Background_Vertical.png
      resourceObject=null
      tiling=NoTile
      imageSize=<8 8>
      margin=<ltrb 0 0 0 0>
      outlineSettings {
       bUseBrushTransparency=false
       roundingType=HalfHeightRadius
       width=0
       color {colorUseRule=UseColor_Specified, specifiedColor=<rgba 0 0 0 0>}
       cornerRadii=<0 0 0 0>
      }
      tintColor {colorUseRule=UseColor_Specified, specifiedColor=<rgba 1 1 1 1>}
      uVRegion {bIsValid=false, max=<0 0>, min=<0 0>}
     }
     verticalTopSlotImage {
      bIsDynamicallyLoaded=false
      drawAs=NoDrawType
      imageType=FullColor
      mirroring=NoMirror
      resourceName=../../../Engine/Content/Slate/Common/Scrollbar_Background_Vertical.png
      resourceObject=null
      tiling=NoTile
      imageSize=<8 8>
      margin=<ltrb 0 0 0 0>
      outlineSettings {
       bUseBrushTransparency=false
       roundingType=HalfHeightRadius
       width=0
       color {colorUseRule=UseColor_Specified, specifiedColor=<rgba 0 0 0 0>}
       cornerRadii=<0 0 0 0>
      }
      tintColor {colorUseRule=UseColor_Specified, specifiedColor=<rgba 1 1 1 1>}
      uVRegion {bIsValid=false, max=<0 0>, min=<0 0>}
     }
    }
    widgetStyle {
     barThickness=9
     bottomShadowBrush {
      bIsDynamicallyLoaded=false
      drawAs=NoDrawType
      imageType=FullColor
      mirroring=NoMirror
      resourceName=../../../Engine/Content/Slate/Common/ScrollBoxShadowBottom.png
      resourceObject=null
      tiling=NoTile
      imageSize=<16 8>
      margin=<ltrb 0.5 0 0.5 1>
      outlineSettings {
       bUseBrushTransparency=false
       roundingType=HalfHeightRadius
       width=0
       color {colorUseRule=UseColor_Specified, specifiedColor=<rgba 0 0 0 0>}
       cornerRadii=<0 0 0 0>
      }
      tintColor {colorUseRule=UseColor_Specified, specifiedColor=<rgba 1 1 1 1>}
      uVRegion {bIsValid=false, max=<0 0>, min=<0 0>}
     }
     horizontalScrolledContentPadding=<ltrb 0 0 1 0>
     leftShadowBrush {
      bIsDynamicallyLoaded=false
      drawAs=NoDrawType
      imageType=FullColor
      mirroring=NoMirror
      resourceName=../../../Engine/Content/Slate/Common/ScrollBoxShadowLeft.png
      resourceObject=null
      tiling=NoTile
      imageSize=<8 16>
      margin=<ltrb 1 0.5 0 0.5>
      outlineSettings {
       bUseBrushTransparency=false
       roundingType=HalfHeightRadius
       width=0
       color {colorUseRule=UseColor_Specified, specifiedColor=<rgba 0 0 0 0>}
       cornerRadii=<0 0 0 0>
      }
      tintColor {colorUseRule=UseColor_Specified, specifiedColor=<rgba 1 1 1 1>}
      uVRegion {bIsValid=false, max=<0 0>, min=<0 0>}
     }
     rightShadowBrush {
      bIsDynamicallyLoaded=false
      drawAs=NoDrawType
      imageType=FullColor
      mirroring=NoMirror
      resourceName=../../../Engine/Content/Slate/Common/ScrollBoxShadowRight.png
      resourceObject=null
      tiling=NoTile
      imageSize=<8 16>
      margin=<ltrb 0 0.5 1 0.5>
      outlineSettings {
       bUseBrushTransparency=false
       roundingType=HalfHeightRadius
       width=0
       color {colorUseRule=UseColor_Specified, specifiedColor=<rgba 0 0 0 0>}
       cornerRadii=<0 0 0 0>
      }
      tintColor {colorUseRule=UseColor_Specified, specifiedColor=<rgba 1 1 1 1>}
      uVRegion {bIsValid=false, max=<0 0>, min=<0 0>}
     }
     topShadowBrush {
      bIsDynamicallyLoaded=false
      drawAs=NoDrawType
      imageType=FullColor
      mirroring=NoMirror
      resourceName=../../../Engine/Content/Slate/Common/ScrollBoxShadowTop.png
      resourceObject=null
      tiling=NoTile
      imageSize=<16 8>
      margin=<ltrb 0.5 1 0.5 0>
      outlineSettings {
       bUseBrushTransparency=false
       roundingType=HalfHeightRadius
       width=0
       color {colorUseRule=UseColor_Specified, specifiedColor=<rgba 0 0 0 0>}
       cornerRadii=<0 0 0 0>
      }
      tintColor {colorUseRule=UseColor_Specified, specifiedColor=<rgba 1 1 1 1>}
      uVRegion {bIsValid=false, max=<0 0>, min=<0 0>}
     }
     verticalScrolledContentPadding=<ltrb 0 0 0 1>
    }
   }
  }
  ProfileLayout {
   widgetClassPath=&GridPanel
   parent=ScrollBox_0
   widget=ProfileLayout
   properties {columnFill [1], rowFill [0,1]}
   slot {class=ScrollBoxSlot}
  }
  Profile_Header {
   widgetClassPath=&/WBP_Profile_Header
   bIsVariable=true
   parent=ProfileLayout
   widget=Profile_Header
   slot {class=GridSlot, properties {verticalAlignment=VAlign_Top, padding=<ltrb 0 72 0 0>}}
  }
  StackBox_0 {
   widgetClassPath=&StackBox
   parent=ProfileLayout
   widget=StackBox_0
   properties {
    bOverride_Cursor=true
    navigation=&self:WidgetTree.StackBox_0.WidgetNavigation_0
    orientation=Orient_Vertical
   }
   slot {class=GridSlot, properties {row=1, verticalAlignment=VAlign_Top}}
  }
  LocalPlayer_CurrentActivity {
   widgetClassPath=&/WBP_Sidebar_CurrentActivity_LocalPlayer
   bIsVariable=true
   parent=StackBox_0
   widget=LocalPlayer_CurrentActivity
   slot {class=StackBoxSlot, properties {padding=<ltrb 18 30 18 0>}}
  }
  GridPanel_0 {
   widgetClassPath=&GridPanel
   parent=StackBox_0
   widget=GridPanel_0
   properties {columnFill [1]}
   slot {class=StackBoxSlot, properties {padding=<ltrb 0 30 0 15>}}
  }
  OnlineStatus_TabGroup {
   widgetClassPath=/FortUILibrary/UIKit/Tabs/WBP_UIKit_TabGroup
   bIsVariable=true
   parent=GridPanel_0
   widget=OnlineStatus_TabGroup
   properties {
    animateReservedPromptSpace=false
    backplateBrightness=Dark
    buttonClass=/Sidebar/Buttons/WBP_Sidebar_ButtonOnlineStatus.WBP_Sidebar_ButtonOnlineStatus_C
    buttonHorizontalAlignment=HAlign_Fill
    inputActions=None
    isNaturalNavigation=true
    tabGroupHorizontalAlignment=HAlign_Fill
    tabSpacing=4
    useBackplate=false
    useInitialTabs=true
    useMinButtonSize=false
    usePromptBackplate=false
    iconSize=<18 18>
    initialTabsData [
     {
      bEnableTabId=true
      tabId=online
      tabViewModel=&self:WidgetTree.OnlineStatus_TabGroup.UIKitButtonViewModel_0
     }
     {
      bEnableTabId=true
      tabId=away
      tabViewModel=&self:WidgetTree.OnlineStatus_TabGroup.UIKitButtonViewModel_2
     }
    ]
    minButtonSize=<0 80>
    tabGroupPadding=<ltrb 16 16 16 16>
   }
   slot {class=GridSlot, properties {layer=1, verticalAlignment=VAlign_Top}}
  }
  OnlineStatus_Backplate {
   widgetClassPath=&Image
   bIsVariable=true
   parent=GridPanel_0
   widget=OnlineStatus_Backplate
   properties {brush {
     bIsDynamicallyLoaded=false
     drawAs=Image
     imageType=FullColor
     mirroring=NoMirror
     resourceName=None
     resourceObject=/Sidebar/Materials/MI_Sidebar_TabSwitcherBackplate
     tiling=NoTile
     imageSize=<32 32>
     margin=<ltrb 0 0 0 0>
     outlineSettings {
      bUseBrushTransparency=false
      roundingType=HalfHeightRadius
      width=0
      color {colorUseRule=UseColor_Specified, specifiedColor=<rgba 0 0 0 0>}
      cornerRadii=<0 0 0 0>
     }
     tintColor {colorUseRule=UseColor_Specified, specifiedColor=<rgba 1 1 1 1>}
     uVRegion {bIsValid=false, max=<0 0>, min=<0 0>}
    }}
   slot {class=GridSlot, properties {layer=-10, padding=<ltrb 8 8 8 8>}}
  }
  StackBox_Buttons_CustomNav {
   widgetClassPath=&StackBox
   parent=StackBox_0
   widget=StackBox_Buttons_CustomNav
   properties {
    navigation=&self:WidgetTree.StackBox_Buttons_CustomNav.WidgetNavigation_0
    orientation=Orient_Vertical
   }
   slot {class=StackBoxSlot, properties {size {sizeRule=Fill, value=1}}}
  }
  Button_PartyJoinability {
   widgetClassPath=/Sidebar/Buttons/WBP_Sidebar_ButtonWithContextMenu
   bIsVariable=true
   parent=StackBox_Buttons_CustomNav
   widget=Button_PartyJoinability
   properties {bNavigateToNextWidgetOnDisable=true, optionName=Party joinability}
   slot {class=StackBoxSlot}
  }
  Button_Participation {
   widgetClassPath=&WBP_Sidebar_ButtonWithContextMenu
   bIsVariable=true
   parent=StackBox_Buttons_CustomNav
   widget=Button_Participation
   properties {bNavigateToNextWidgetOnDisable=true, optionName=Participation, useLocalOptions=true, options {not_ready=Playing, sitting_out=Sitting Out}}
   slot {class=StackBoxSlot, properties {padding=<ltrb 0 15 0 0>}}
  }
  Button_MutualFriends {
   widgetClassPath=&WBP_Sidebar_ButtonWithContextMenu
   bIsVariable=true
   parent=StackBox_Buttons_CustomNav
   widget=Button_MutualFriends
   properties {
    bNavigateToNextWidgetOnDisable=true
    optionDescription=Choose who will be able to see mutual friends on your profile.
    optionName=Show mutual friends
   }
   slot {class=StackBoxSlot, properties {padding=<ltrb 0 15 0 0>}}
  }
 ]
}
