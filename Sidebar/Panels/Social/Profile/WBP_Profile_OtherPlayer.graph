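# Sidebar profile panel for another player, showing header, current activity for friends, friend-request notice, mutual friends and social actions.
# On activation it reads the target from the navigation point and initializes the social view models.
# Interaction buttons are built at runtime, mapped to their VMs, with overflow behind More options.
# Clicks execute interactions or push nickname and confirmation panels; scroll-on-focus applies only to gamepad.
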WidgetBlueprint /Sidebar/Panels/Social/Profile/WBP_Profile_OtherPlayer
graphs [
 BP_GetDesiredFocusTarget {nodes [
   Get Desired Focus Target()
   > n1 = Branch(Condition=HasAnyChildren!(self:Panel Widget Object Reference=Get VB_Interactions() -> (VB_Interactions:Vertical Box Object Reference)) -> (ReturnValue:bool))
   > Return Node!(
    execute
    ReturnValue:Widget Object Reference=GetChildAt!(self:Panel Widget Object Reference=Get VB_Interactions(), Index:int=0) -> (ReturnValue:Widget Object Reference)
   )
   n2 = Get TreeView!(self:Fort Polymorphic Tree View Object Reference=Get MutualFriends() -> (MutualFriends:WBP Sidebar Social Tree List Object Reference)) -> (TreeView:Tree View Object Reference)
   n3 = Branch(execute=@n1.else, Condition="integer > integer"(A=GetNumItems!(self:List View Object Reference=@n2) -> (ReturnValue:int=0)))
   n4 = GetIndexForItem!(
    self:List View Object Reference=@n2
    Item:Object Reference=Get Selected Item!(self:List View Object Reference=@n2) -> (ReturnValue:Object Reference) {is=title K2Node_CallFunction}
   ) -> (ReturnValue:int=0)
   > SetSelectedIndex!(
    execute
    self:List View Object Reference=@n2
    Index:int=Select(Option 0:int=0, Option 1:int=@n4, Index="integer >= integer"(A=@n4)) -> (ReturnValue:int)
   ) -> (then)
   > Return Node(ReturnValue=@n2)
   Return Node(execute=@n3.else)
  ]}
 ConstructMutualFriendsHeader {nodes [
   Inputs()
   > n1 = Get!(execute:Exec) -> (then:Exec, else:Exec, MutualFriendsHeader:BP Sidebar Custom Header Entry VM Object Reference)
   Outputs!(then:Exec=[@n1.then, @n3.then])
   n2 = Construct Object from Class(execute=@n1.else, Class=/Sidebar/Panels/Social/List/Data/BP_Sidebar_CustomHeaderEntryVM.BP_Sidebar_CustomHeaderEntryVM_C) -> (ReturnValue:BP Sidebar Custom Header Entry VM Object Reference)
   > Set with Broadcast DisplayName!(execute, DisplayName:Text="LOCTEXT(\"Top mutual friends\")", self:BP Sidebar Custom Header Entry VM Object Reference=@n2) -> (then, Output_Get:Text)
   > n3 = Set MutualFriendsHeader(MutualFriendsHeader:BP Sidebar Custom Header Entry VM Object Reference=@n2) -> (Output_Get:BP Sidebar Custom Header Entry VM Object Reference)
  ]}
 EventGraph {nodes [
   "On Loading Switcher Active Widget Changed (PanelLayout)"!() -> (OutputDelegate, then, IsAsyncActionInProgress:bool) {is=title K2Node_ComponentBoundEvent}
   > RequestRefreshFocus(self:Common Activatable Widget Object Reference)
   "On Clicked (WBP_Sidebar_Button_MoreOptions)"!() -> (OutputDelegate, then, Button:Common Button Base Object Reference) {is=title K2Node_ComponentBoundEvent}
   > n1 = Flip Flop!(None:Exec) -> (A:Exec, B:Exec, IsA:bool=~) {is=title K2Node_MacroInstance}
   SetText!(
    execute=[@n1.A, @n1.B]
    self:WBP UIKit Button Generic Base Object Reference=Get WBP_Sidebar_Button_MoreOptions() -> (WBP_Sidebar_Button_MoreOptions:WBP Sidebar Button Generic Object Reference)
    Text:Text=SelectText(A="LOCTEXT(\"Fewer options\")", B="LOCTEXT(\"More options\")", bPickA=@n1)
   ) -> (then)
   > SetVisibility(
    self=Get VB_MoreInteractions() -> (VB_MoreInteractions:Vertical Box Object Reference)
    InVisibility=Select(Option 0:ESlateVisibility Enum=Collapsed, Option 1:ESlateVisibility Enum=SelfHitTestInvisible, Index=@n1) -> (ReturnValue:ESlateVisibility Enum)
   )
   n2 = "Item Clicked (MutualFriends)"!() -> (OutputDelegate, then, ItemVM:Fort Social Item VM Object Reference) {is=title K2Node_ComponentBoundEvent}
   > Navigate to Mutual Friend!(execute:Exec, self:Fort Social Item VM Object Reference=@n2.ItemVM) -> (then:Exec) {is=title K2Node_MacroInstance}
   n3 = Sidebar Interaction OnClicked!() -> (OutputDelegate, then, Button:Common Button Base Object Reference) {is=title K2Node_CustomEvent}
   > Execute Interaction Button!(execute, self, SocialButton:Common Button Base Object Reference=@n3.Button) -> (then)
   n4 = OnHandleInputMethodChanged!() -> (OutputDelegate, then, NewInputType:ECommonInputType Enum) {is=title K2Node_CustomEvent}
   > SetScrollWhenFocusChanges!(
    execute
    self:Scroll Box Object Reference=Get ScrollBox_NoteBreaksVirtualizingMutualFriendList() -> (ScrollBox_NoteBreaksVirtualizingMutualFriendList:Scroll Box Object Reference)
    NewScrollWhenFocusChanges:EScrollWhenFocusChanges Enum=Select!(
     MouseAndKeyboard:EScrollWhenFocusChanges Enum=NoScroll
     Gamepad:EScrollWhenFocusChanges Enum=AnimatedScroll
     Touch:EScrollWhenFocusChanges Enum=NoScroll
     Count:EScrollWhenFocusChanges Enum=NoScroll
     Index:ECommonInputType Enum=@n4.NewInputType
    ) -> (ReturnValue:EScrollWhenFocusChanges Enum)
   ) -> (then)
   Event Destruct()
   n5 = Create Event()
   n6 = Get CommonInputSubsystem()
   > Unbind Event from On Input Method Changed(self=@n6, Delegate=@n5)
   Event Construct()
   > Bind Event to On Input Method Changed(self=@n6, Delegate=@n5)
   > OnHandleInputMethodChanged!(execute, self, NewInputType:ECommonInputType Enum=GetCurrentInputType(self=@n6)) -> (then) {is=title K2Node_CallFunction}
   Event On Activated()
   > Init(self:Player Playtime View Model Object Reference=Get PlayerPlaytimeVM() -> (PlayerPlaytimeVM:Player Playtime View Model Object Reference))
   > n7 = GetCurrentNavigationPoint!(
    execute
    self:Fort Sidebar Navigation VM Object Reference=Get FortSidebarNavigationVM() -> (FortSidebarNavigationVM:Fort Sidebar Navigation VM Object Reference)
   ) -> (then, ReturnValue:Fort Sidebar Navigation Point Object Reference)
   > n8 = Cast To BP_NavigationPoint!(execute, Object=@n7) -> (then, CastFailed, AsBP Navigation Point:BP Navigation Point Object Reference) {is=title K2Node_DynamicCast}
   > Delay(Duration=0.3)
   > Set with Broadcast ShowAsyncActionSpinner!(
    execute
    ShowAsyncActionSpinner:bool=false
    self:WBP Sidebar Panel Layout Object Reference=Get PanelLayout() -> (PanelLayout:WBP Sidebar Panel Layout Object Reference)
   ) -> (then, Output_Get:bool)
   > n9 = Get!(execute, self:BP Navigation Point Object Reference=@n8) -> (then, else, SocialItemVM:Fort Social Item VM Object Reference)
   n10 = Get FortSocialItemVM() -> (FortSocialItemVM:Fort Social Item VM Object Reference)
   > n11 = Init!(
    execute
    self:Fort Social Item VM Object Reference=@n10
    InTargetUser:Fort Social User Object Reference=Get TargetUser!(self:Fort Social Item VM Object Reference=@n9) -> (TargetUser:Fort Social User Weak object)
   ) -> (then)
   n12 = Get FortSocialItemVM()
   Initialize Header With Fort Social Item VM!(
    execute=[@n11.then, @n19.then]
    self:WBP Profile Header Object Reference=Get Profile_Header() -> (Profile_Header:WBP Profile Header Object Reference)
    FortSocialItemVM:Fort Social Item VM Object Reference=@n12
   ) -> (then)
   n13 = Get TargetUser(self=@n12)
   > SetSocialUser!(
    execute
    self:Fort Social User Viewmodel Object Reference=Get FortSocialUserMutualFriendsVM() -> (FortSocialUserMutualFriendsVM:Fort Social User Mutual Friends Viewmodel Object Reference)
    InSocialUser:Fort Social User Object Reference=@n13
   ) -> (then)
   > SetSocialUser!(
    execute
    self:WBP Sidebar Current Activity Other Player Object Reference=Get OtherPlayer_CurrentActivity() -> (OtherPlayer_CurrentActivity:WBP Sidebar Current Activity Other Player Object Reference)
    Social User:Fort Social User Object Reference=@n13
   ) -> (then)
   > n14 = GetConsoleVariableBoolValue!(execute, VariableName:String=SocialUI.RegisterProfileDelegatesPermanently) -> (then, ReturnValue:bool)
   > n15 = Branch(Condition=@n14)
   Set Rich Presence Visibility!(execute:Exec=[@n17.then, @n15.then]) -> (then:Exec) {is=title K2Node_MacroInstance}
   n16 = Get FortSocialItemVM()
   > ActivateInteractionList(self:Fort Social Item VM Object Reference=@n16)
   > SetAnalyticsActionSource!(execute, self:Fort Social Item VM Object Reference=@n16, ActionSource:String=Sidebar) -> (then)
   n17 = RegisterProfileDelegates(execute=@n15.else, self:Fort Social Item VM Object Reference=@n12)
   n18 = Get!(execute=@n9.else, self=@n8) -> (then, else, SocialUser:Fort Social User Object Reference)
   > n19 = Init(self=@n10, InTargetUser=@n18)
   Event On Deactivated()
   > n20 = GetConsoleVariableBoolValue(VariableName=SocialUI.RegisterProfileDelegatesPermanently)
   > n21 = Branch(Condition=@n20)
   Set with Broadcast ShowAsyncActionSpinner(execute=[@n22.then, @n21.then], ShowAsyncActionSpinner=true, self=Get PanelLayout())
   n22 = ReleaseProfileDelegates(execute=@n21.else, self:Fort Social Item VM Object Reference=Get FortSocialItemVM())
   Event OnInitialized()
   > Construct Mutual Friends Header!(execute:Exec) -> (then:Exec) {is=title K2Node_MacroInstance}
   Event PreConstruct()
   > SetVisibility(self=Get WBP_Sidebar_Button_MoreOptions(), InVisibility=Collapsed)
   > SetCornerRadius!(
    execute
    self:WBP Sidebar Block Backplate Object Reference=Get FriendRequest_Backplate() -> (FriendRequest_Backplate:WBP Sidebar Block Backplate Object Reference)
    Radius In Pixels:bool=true
    Corner Radius:double=12
   ) -> (then)
   > SetBackplateColor!(
    execute
    self:WBP Sidebar Block Backplate Object Reference=Get FriendRequest_Backplate()
    Backplate Switch:E_UI_Sidebar_Button_Backplate_Color Enum=NewEnumerator9
   ) -> (then)
  ]}
 Execute Interaction Button {nodes [
   n1 = Execute Interaction Button!() -> (then, SocialButton:Common Button Base Object Reference) {is=title K2Node_FunctionEntry}
   > n2 = Cast To WBP_Sidebar_ButtonGeneric!(execute, Object=@n1) -> (then, CastFailed, AsWBP Sidebar Button Generic:WBP Sidebar Button Generic Object Reference) {is=title K2Node_DynamicCast}
   n3 = Find(
    TargetMap:Map of WBP Sidebar Button Generic Object References to Fort Social Interaction VM Object References=Get InteractionsMap() -> (InteractionsMap:Map of WBP Sidebar Button Generic Object References to Fort Social Interaction VM Object References)
    Key:"WBP Sidebar Button Generic Object Reference (by ref)"=@n2
   ) -> (Value:Fort Social Interaction VM Object Reference)
   > Branch(Condition=@n3.ReturnValue)
   > n4 = Branch(Condition="Equal (Enum)"(
     A:ESocialInteractionType Enum=Get InteractionType!(self:Fort Social Interaction VM Object Reference=@n3) -> (InteractionType:ESocialInteractionType Enum=AcceptEpicPartyInvite)
     B:ESocialInteractionType Enum=EditNickname
    ))
   > Navigate to Edit Nickname!(execute:Exec) -> (then:Exec) {is=title K2Node_MacroInstance}
   n5 = Branch(execute=@n4.else, Condition="Equal (Enum)"(A=Get InteractionType(self=@n3) -> (InteractionType=AcceptEpicPartyInvite), B=RemoveFriend))
   > Navigate to Remove Friend!(execute:Exec) -> (then:Exec) {is=title K2Node_MacroInstance}
   n6 = Branch(
    execute=@n5.else
    Condition="Equal (Enum)"(A=Get InteractionType(self=@n3) -> (InteractionType=AcceptEpicPartyInvite), B=DeclineFriendInvite)
   )
   > Navigate to Decline Friend Request!(execute:Exec) -> (then:Exec) {is=title K2Node_MacroInstance}
   n7 = Branch(execute=@n6.else, Condition="Equal (Enum)"(A=Get InteractionType(self=@n3) -> (InteractionType=AcceptEpicPartyInvite), B=Block))
   > Navigate to Block Player!(execute:Exec) -> (then:Exec) {is=title K2Node_MacroInstance}
   n8 = Get InteractionType(self=@n3) -> (InteractionType=AcceptEpicPartyInvite)
   Set ShouldCloseCurrentPanel(
    execute=@n7.else
    ShouldCloseCurrentPanel:bool=OR Boolean!(
     A="Equal (Enum)"(A=@n8, B=AcceptFriendInvite)
     B="Equal (Enum)"(A=@n8, B=DismissJoinRequest)
     C:bool="Equal (Enum)"(A=@n8, B=DismissEpicPartyJoinRequest)
     D:bool="Equal (Enum)"(A=@n8, B=RejectPartyInvite)
     E:bool="Equal (Enum)"(A=@n8, B=RejectEpicPartyInvite)
     F:bool="Equal (Enum)"(A=@n8, B=InviteToParty)
     G:bool="Equal (Enum)"(A=@n8, B=InviteToNativeEpicParty)
     H:bool="Equal (Enum)"(A=@n8, B=RequestToJoinParty)
     I:bool="Equal (Enum)"(A=@n8, B=RequestToJoinEpicParty)
    ) -> (ReturnValue)
   ) -> (Output_Get:bool)
   > ExecuteInteraction(self:Fort Social Interaction VM Object Reference=@n3)
   > Branch(Condition=Get ShouldCloseCurrentPanel() -> (ShouldCloseCurrentPanel:bool))
   > PopNavigation(self:Fort Sidebar Navigation VM Object Reference=Get FortSidebarNavigationVM())
  ]}
 ExecuteSocialInteraction {nodes [
   n1 = ExecuteSocialInteraction!() -> (then, InteractionType:ESocialInteractionType Enum=AcceptEpicPartyInvite) {is=title K2Node_FunctionEntry}
   > n2 = For Each Loop with Break!(
    Exec:Exec
    Array:Array of Fort Social Interaction VM Object References=Get Interactions!(self:Fort Social Item VM Object Reference=Get FortSocialItemVM()) -> (Interactions:Array of Fort Social Interaction VM Object References)
    Break:Exec=@n3.then
   ) -> (LoopBody:Exec, Array Element:Fort Social Interaction VM Object Reference, Array Index:int, Completed:Exec)
   > Branch(Condition="Equal (Enum)"(A=Get InteractionType(self=@n2) -> (InteractionType=AcceptEpicPartyInvite), B=@n1))
   > n3 = ExecuteInteraction(self=@n2)
  ]}
 GetInteractionButtonAvailable {nodes [
   GetInteractionButtonAvailable!() -> (then, InteractionType:ESocialInteractionType Enum=AcceptEpicPartyInvite) {is=title K2Node_FunctionEntry}
   n1 = Get InteractionType!() -> (InteractionType=AcceptEpicPartyInvite)
   n2 = OR Boolean!(
    A="Equal (Enum)"(A=@n1, B=SpectateFriendsGame)
    B="Equal (Enum)"(A=@n1, B=JoinAndSpectateFriend)
    C="Equal (Enum)"(A=@n1, B=JoinMatchViaPresence)
    D="Equal (Enum)"(A=@n1, B=JoinEpicPartyAndSpectateFriend)
   ) -> (ReturnValue)
   n3 = "Equal (Enum)"(
    A:EPlaytimeState Enum=Get PlayState!(self:Player Playtime View Model Object Reference=Get PlayerPlaytimeVM()) -> (PlayState:EPlaytimeState Enum=Unblocked)
    B:EPlaytimeState Enum=PlayTime_Blocked
   )
   > Return Node!(execute, IsAvailable:bool=NOT Boolean(A=AND Boolean(A=@n2, B=@n3)))
  ]}
 HandleConfirmationOverlayActions {nodes [
   n1 = HandleConfirmationOverlayActions!() -> (then, ActionID:Name) {is=title K2Node_FunctionEntry}
   > Allow Handle Overlay Confirmation!(Exec:Exec, Action ID:Name=@n1) -> (then:Exec) {is=title K2Node_MacroInstance}
   > n2 = Branch(Condition="Equal (Name)"(A=@n1, B=RemoveFriend))
   > ExecuteSocialInteraction!(execute, self, InteractionType:ESocialInteractionType Enum=RemoveFriend) -> (then) {is=title K2Node_CallFunction}
   n3 = Branch(execute=@n2.else, Condition="Equal (Name)"(A=@n1, B=Block))
   > ExecuteSocialInteraction(InteractionType=Block)
   Branch(execute=@n3.else, Condition="Equal (Name)"(A=@n1, B=DeclineFriendRequest))
   > ExecuteSocialInteraction(InteractionType=DeclineFriendInvite)
   > PopToNavigationTag!(
    execute
    self:Fort Sidebar Navigation VM Object Reference=Get FortSidebarNavigationVM()
    NavigationTag:Gameplay Tag Structure=(TagName="UI.StateTree.TransitionTag.Sidebar.AddFriend")
   ) -> (then, ReturnValue:bool)
  ]}
 NavigateToBlockPlayer {nodes [
   Inputs()
   > n1 = Construct Object from Class(Class=/Sidebar/Panels/Overlay/Data/ConfirmationVM_BlockPlayer.ConfirmationVM_BlockPlayer_C) -> (ReturnValue:Confirmation VM Block Player Object Reference)
   n2 = Get TargetUser(self=Get FortSocialItemVM())
   > Set Description!(
    execute
    Description:Text=Format Text!(
     Format="LOCTEXT(\"Would you like to block {PlayerName}?\")"
     PlayerName:String=GetNicknameOrDisplayName!(self:Fort Social User Object Reference=@n2) -> (ReturnValue:String)
    ) -> (Result)
    self:Sidebar Confirmation Panel VM Object Reference=@n1
   ) -> (then, Output_Get:Text)
   > NavigateToConfirmPanel!(
    execute
    SelfReference:Object Reference=Self-Reference()
    SidebarConfirmationVM:Sidebar Confirmation Panel VM Object Reference=@n1
    SidebarNavigationVM:Fort Sidebar Navigation VM Object Reference=Get FortSidebarNavigationVM()
   ) -> (then)
   > Outputs()
  ]}
 NavigateToDeclineFriendRequest {nodes [
   Inputs()
   > n1 = Construct Object from Class(Class=/Sidebar/Panels/Overlay/Data/ConfirmationVM_DeclineFriendRequest.ConfirmationVM_DeclineFriendRequest_C) -> (ReturnValue:Confirmation VM Decline Friend Request Object Reference)
   n2 = Get TargetUser(self=Get FortSocialItemVM())
   > Set Description(
    Description=Format Text!(Format="LOCTEXT(\"Are you sure you’d like to decline {FriendName} friend request?\")", FriendName:String=GetNicknameOrDisplayName(self=@n2)) -> (Result)
    self=@n1
   )
   > NavigateToConfirmPanel(SelfReference=Self-Reference(), SidebarConfirmationVM=@n1, SidebarNavigationVM=Get FortSidebarNavigationVM())
   > Outputs()
  ]}
 NavigateToEditNickname {nodes [
   Inputs()
   > n1 = Construct Object from Class!(
    execute
    Class=/Sidebar/Shared/Navigation/BP_NavigationPoint.BP_NavigationPoint_C
    self
    SocialUser:Fort Social User Object Reference=Get TargetUser(self=Get FortSocialItemVM())
    ConversationId:Name=None
    ConfirmationVM:Sidebar Confirmation Panel VM Object Reference
    ReservedSpots:int=0
    SocialItemVM:Fort Social Item VM Object Reference
    NavigationTag:Gameplay Tag Structure=(TagName="UI.StateTree.TransitionTag.Sidebar.Social.OtherPlayerNickname")
   ) -> (then, ReturnValue:BP Navigation Point Object Reference)
   > PushNavigation!(
    execute
    self:Fort Sidebar Navigation VM Object Reference=Get FortSidebarNavigationVM()
    NavigationPoint:Fort Sidebar Navigation Point Object Reference=@n1
   ) -> (then)
   > Outputs()
  ]}
 NavigateToMutualFriend {nodes [
   n1 = Inputs!() -> (execute, self:Fort Social Item VM Object Reference)
   > n2 = Construct Object from Class!(execute, Class=/Script/SidebarUI.FortSocialItemVM, self) -> (then, ReturnValue:Fort Social Item VM Object Reference)
   n3 = Get TargetUser(self=@n1)
   > Init(self=@n2, InTargetUser=@n3)
   > n4 = Construct Object from Class!(
    execute
    Class=&BP_NavigationPoint_C
    self
    SocialUser:Fort Social User Object Reference=@n3
    ConversationId:Name=None
    ConfirmationVM:Sidebar Confirmation Panel VM Object Reference
    ReservedSpots:int=0
    SocialItemVM:Fort Social Item VM Object Reference=@n2
    NavigationTag:Gameplay Tag Structure=(TagName="UI.StateTree.TransitionTag.Sidebar.Social.OtherPlayer")
   ) -> (then, ReturnValue:BP Navigation Point Object Reference)
   > PushNavigation(self=Get FortSidebarNavigationVM(), NavigationPoint=@n4)
   > Outputs()
  ]}
 NavigateToRemoveFriend {nodes [
   Inputs!() -> (execute)
   > n1 = Construct Object from Class!(execute, Class=/Sidebar/Panels/Overlay/Data/ConfirmationVM_RemoveFriend.ConfirmationVM_RemoveFriend_C, self) -> (then, ReturnValue:Confirmation VM Remove Friend Object Reference)
   n2 = Get TargetUser(self=Get FortSocialItemVM())
   > Set Description(
    Description=Format Text(Format="LOCTEXT(\"Are you sure you’d like to remove {FriendName} as a friend?\")", FriendName=GetNicknameOrDisplayName(self=@n2))
    self=@n1
   )
   > NavigateToConfirmPanel(SelfReference=Self-Reference(), SidebarConfirmationVM=@n1, SidebarNavigationVM=Get FortSidebarNavigationVM())
   > Outputs()
  ]}
 Populate Interactions {nodes [
   Populate Interactions!() -> (then, Interactions:Array of Fort Social Interaction VM Object References) {is=title K2Node_FunctionEntry}
   > ClearChildren(self=[Get VB_Interactions(), Get VB_MoreInteractions()])
   > Clear!(execute, TargetMap:Map of WBP Sidebar Button Generic Object References to Fort Social Interaction VM Object References=Get InteractionsMap()) -> (then) {is=title K2Node_CallFunction}
   > SetJoinableVisibility!(
    execute
    self:WBP Sidebar Current Activity Base Object Reference=Get OtherPlayer_CurrentActivity()
    Visibility:ESlateVisibility Enum=Collapsed
   ) -> (then)
   > SetVisibility(self=Get O_FriendRequest() -> (O_FriendRequest:Overlay Object Reference), InVisibility=Collapsed)
   n1 = Get Interactions!() -> (Interactions)
   n2 = GetRuntimeOptions()
   > Set IsMoreInteractions(IsMoreInteractions:bool="integer > integer"(
     A=Length(TargetArray:Array of Fort Social Interaction VM Object References=@n1)
     B=Get SidebarOtherPlayer_NumSocialInteractionsToShow!(self:Fort Runtime Options Object Reference=@n2) -> (SidebarOtherPlayer_NumSocialInteractionsToShow:int=0)
    )) -> (Output_Get:bool)
   > n3 = For Each Loop(Array:Array of Fort Social Interaction VM Object References=@n1) -> (Array Element:Fort Social Interaction VM Object Reference)
   RequestRefreshFocus(execute=@n3.Completed)
   n4 = Branch(
    execute=@n3.LoopBody
    Condition="Equal (Enum)"(
     A:ESocialInteractionType Enum=Get InteractionType!(self:Fort Social Interaction VM Object Reference=@n3) -> (InteractionType=AcceptEpicPartyInvite)
     B:ESocialInteractionType Enum=AcceptFriendInvite
    )
   )
   > n5 = Set Friend Request Visuals Visibility!(execute, self, Header Item:ESocialHeaderItemType Enum=FriendInvites) -> (then) {is=title K2Node_CallFunction}
   Branch(
    execute=[@n5.then, @n4.else]
    Condition=GetInteractionButtonAvailable!(InteractionType:ESocialInteractionType Enum=Get InteractionType(self=@n3) -> (InteractionType=AcceptEpicPartyInvite)) -> (IsAvailable:bool) {is=title K2Node_CallFunction}
   )
   > n6 = Create Widget!(
    execute:Exec
    Class:User Widget Class Reference=/Sidebar/Buttons/WBP_Sidebar_ButtonGeneric.WBP_Sidebar_ButtonGeneric_C
    OwningPlayer:Player Controller Object Reference=GetOwningPlayer()
    Style:Common Button Style Class Reference=/FortUILibrary/UIKit/Buttons/Blocks/ButtonStyle_UIKit_Empty.ButtonStyle_UIKit_Empty_C
    bApplyAlphaOnDisable:bool=False
    bLocked:bool=False
    bSelectable:bool=True
    bShouldSelectUponReceivingFocus:bool=True
    bInteractableWhenSelected:bool=True
    bToggleable:bool=False
    bTriggerClickedAfterSelection:bool=False
    bDisplayInputActionWhenNotInteractable:bool=false
    bHideInputActionWithKeyboard:bool=False
    bShouldUseFallbackDefaultInputAction:bool=True
    bRequiresHold:bool=False
    HoldData:Common UIHold Data Class Reference=&CommonUI_HoldData_Base_C
    ClickMethod:EButtonClickMethod Enum=DownAndUp
    InputPriority:int=0
    TriggeringInputAction:Data Table Row Handle Structure
   ) -> (then:Exec, ReturnValue:WBP Sidebar Button Generic Object Reference)
   > n7 = Branch(Condition="Equal (Enum)"(A=Get InteractionType(self=@n3) -> (InteractionType=AcceptEpicPartyInvite), B=PromoteToEpicPartyLeader))
   > n8 = Set Epic Party Leader Button Style!(execute:Exec, self:WBP UIKit Button Generic Base Object Reference=@n6) -> (then:Exec) {is=title K2Node_MacroInstance}
   Add!(
    execute=[@n8.then, @n7.else]
    TargetMap:Map of WBP Sidebar Button Generic Object References to Fort Social Interaction VM Object References=Get InteractionsMap()
    Key:"WBP Sidebar Button Generic Object Reference (by ref)"=@n6
    Value:"Fort Social Interaction VM Object Reference (by ref)"=@n3
   ) -> (then) {is=title K2Node_CallFunction}
   > SetText(self=@n6, Text=GetDisplayName!(self:Fort Social Interaction VM Object Reference=@n3) -> (ReturnValue:Text))
   > SetPadding!(
    execute
    self:User Widget Object Reference=@n6
    InPadding_Left:float=0
    InPadding_Top:float=0
    InPadding_Right:float=0
    InPadding_Bottom:float=12
   ) -> (then)
   > Bind Event to On Clicked(self:Common Button Base Object Reference=@n6, Delegate=Create Event()) {is=title K2Node_AddDelegate}
   n9 = int - int(A=Get SidebarOtherPlayer_NumSocialInteractionsToShow(self=GetRuntimeOptions()) -> (SidebarOtherPlayer_NumSocialInteractionsToShow=0))
   > n10 = Branch(Condition=AND Boolean(A="integer > integer"(A=@n3.Array Index, B=@n9), B=Get IsMoreInteractions() -> (IsMoreInteractions:bool)))
   > AddChild(self=Get VB_MoreInteractions(), Content=@n6)
   > n11 = SetVisibility(self=Get WBP_Sidebar_Button_MoreOptions(), InVisibility=SelfHitTestInvisible)
   n12 = Get InteractionType(self=@n3) -> (InteractionType=AcceptEpicPartyInvite)
   Branch(
    execute=[@n11.then, @n13.then]
    Condition=OR Boolean!(A="Equal (Enum)"(A=@n12, B=JoinParty), B="Equal (Enum)"(A=@n12, B=JoinNativeEpicParty)) -> (ReturnValue)
   )
   > SetJoinableVisibility(self=Get OtherPlayer_CurrentActivity(), Visibility=SelfHitTestInvisible)
   n13 = AddChild(execute=@n10.else, self=Get VB_Interactions(), Content=@n6)
  ]}
 PopulateMutualFriendsInsideHeader {nodes [
   n1 = PopulateMutualFriendsInsideHeader!() -> (then, Mutual Friends:Array of Fort Social Item VM Object References) {is=title K2Node_FunctionEntry}
   n2 = Is Empty!(TargetArray:Array of Fort Social Item VM Object References=@n1) -> (ReturnValue:bool)
   > SetVisibility(
    self=Get MutualFriends()
    InVisibility=Select!(Option 0:ESlateVisibility Enum=SelfHitTestInvisible, Option 1:ESlateVisibility Enum=Collapsed, Index:bool=@n2) -> (ReturnValue:ESlateVisibility Enum)
   )
   > n3 = Branch(Condition=@n2)
   > ClearListItems(self=Get TreeView(self=Get MutualFriends()))
   n4 = Get MutualFriendsHeader() -> (MutualFriendsHeader:BP Sidebar Custom Header Entry VM Object Reference)
   Set with Broadcast ChildrenItems!(
    execute=@n3.else
    ChildrenItems:Array of MVVM Base Viewmodel Object References=Get Mutual Friends() -> (Mutual Friends:Array of Fort Social Item VM Object References)
    self:BP Sidebar Custom Header Entry VM Object Reference=@n4
   ) -> (then, Output_Get:Array of MVVM Base Viewmodel Object References)
   n5 = Get TreeView(self=Get MutualFriends())
   > Set List Items!(
    execute
    self:List View Object Reference=@n5
    InListItems:Array of Object References=Make Array("[0]":Object Reference=@n4) -> (Array:Array of Object References)
   ) -> (then) {is=title K2Node_CallFunction}
   > ExpandAll(self:Tree View Object Reference=@n5)
  ]}
 Set Friend Request Visuals Visibility {nodes [
   n1 = Set Friend Request Visuals Visibility!() -> (then, Header Item:ESocialHeaderItemType Enum=Unknown) {is=title K2Node_FunctionEntry}
   > SetVisibility(
    self=Get O_FriendRequest()
    InVisibility=Select(
     Option 0=Collapsed
     Option 1=HitTestInvisible
     Index="Equal (Enum)"(A:ESocialHeaderItemType Enum=@n1, B:ESocialHeaderItemType Enum=FriendInvites)
    )
   )
  ]}
 SetEpicPartyLeaderButtonStyle {nodes [
   n1 = Inputs!() -> (execute, self:WBP UIKit Button Generic Base Object Reference)
   > Set IconMaterial!(
    execute
    IconMaterial:Material Instance Object Reference=/Sidebar/Materials/Icons/MI_Sidebar_Icon_Crown
    self:WBP UIKit Button Generic Base Object Reference=@n1
   ) -> (then, Output_Get:Material Instance Object Reference)
   > Set UseIconBlock!(execute, UseIconBlock:bool=true, self:WBP UIKit Button Generic Base Object Reference=@n1) -> (then, Output_Get:bool)
   > SetIconPadding!(
    execute
    self:WBP UIKit Button Generic Base Object Reference=@n1
    Padding_Left:float=0
    Padding_Top:float=0
    Padding_Right:float=6
    Padding_Bottom:float=0
   ) -> (then)
   > Outputs()
  ]}
 SetRichPresenceVisibility {nodes [
   Inputs!() -> (execute)
   n1 = Get FortSocialUserInfoViewModel!(self:WBP Sidebar Current Activity Other Player Object Reference=Get OtherPlayer_CurrentActivity()) -> (FortSocialUserInfoViewModel:Fort Social User Info View Model Object Reference)
   > SetVisibility(
    self=Get OtherPlayer_CurrentActivity()
    InVisibility=Select(
     Option 0=Collapsed
     Option 1=SelfHitTestInvisible
     Index=Get bIsFriend!(self:Fort Social User Info View Model Object Reference=@n1) -> (bIsFriend:bool)
    )
   )
   > Outputs()
  ]}
]
variables [
 InteractionsMap {default {}}
 MutualFriendsHeader {type=BP_Sidebar_CustomHeaderEntryVM_C, default=null}
 OneVOne_Interaction {type=FortSocialInteractionVM, default=null}
 PlayTimeOutInteractionIndex {type=integer}
]
widgets {
 info {
  inheritedWidgetCount=0
  namedSlotCount=1
  parentClass=&CommonActivatableWidget
  rootWidgetClass=/Sidebar/Shared/WBP_Sidebar_PanelLayout.WBP_Sidebar_PanelLayout_C
  widgetCount=15
 }
 widgets [
  PanelLayout {
   widgetClassPath=/Sidebar/Shared/WBP_Sidebar_PanelLayout
   bIsVariable=true
   widget=PanelLayout
   properties {headingText="", showAsyncActionSpinner=true, startInAsyncActionSpinner=true}
  }
  ScrollBox_NoteBreaksVirtualizingMutualFriendList {
   widgetClassPath=&ScrollBox
   bIsVariable=true
   namedSlotHost=PanelLayout
   widget=ScrollBox_NoteBreaksVirtualizingMutualFriendList
   properties {
    clipping=Inherit
    navigationDestination=Center
    scrollWhenFocusChanges=AnimatedScroll
    widgetBarStyle {
     thickness=8
     draggedThumbImage {
      bIsDynamicallyLoaded=false
      drawAs=Box
      imageType=FullColor
      mirroring=NoMirror
      resourceName=../../../Engine/Content/Slate/Common/Scrollbar_Thumb.png
      resourceObject=null
      tiling=NoTile
      imageSize=<0 0>
      margin=<ltrb 0.25 0.25 0.25 0.25>
      outlineSettings {
       bUseBrushTransparency=false
       roundingType=HalfHeightRadius
       width=0
       color {colorUseRule=UseColor_Specified, specifiedColor=<rgba 0 0 0 0>}
       cornerRadii=<0 0 0 0>
      }
      tintColor {colorUseRule=UseColor_Specified, specifiedColor=<rgba 1 1 1 1>}
      uVRegion {bIsValid=false, max=<0 0>, min=<0 0>}
     }
     horizontalBackgroundImage {
      bIsDynamicallyLoaded=false
      drawAs=NoDrawType
      imageType=NoImage
      mirroring=NoMirror
      resourceName=None
      resourceObject=null
      tiling=NoTile
      imageSize=<0 0>
      margin=<ltrb 0 0 0 0>
      outlineSettings {
       bUseBrushTransparency=false
       roundingType=HalfHeightRadius
       width=0
       color {colorUseRule=UseColor_Specified, specifiedColor=<rgba 0 0 0 0>}
       cornerRadii=<0 0 0 0>
      }
      tintColor {colorUseRule=UseColor_Specified, specifiedColor=<rgba 1 1 1 1>}
      uVRegion {bIsValid=false, max=<0 0>, min=<0 0>}
     }
     horizontalBottomSlotImage {
      bIsDynamicallyLoaded=false
      drawAs=NoDrawType
      imageType=FullColor
      mirroring=NoMirror
      resourceName=../../../Engine/Content/Slate/Common/Scrollbar_Background_Horizontal.png
      resourceObject=null
      tiling=NoTile
      imageSize=<8 8>
      margin=<ltrb 0 0 0 0>
      outlineSettings {
       bUseBrushTransparency=false
       roundingType=HalfHeightRadius
       width=0
       color {colorUseRule=UseColor_Specified, specifiedColor=<rgba 0 0 0 0>}
       cornerRadii=<0 0 0 0>
      }
      tintColor {colorUseRule=UseColor_Specified, specifiedColor=<rgba 1 1 1 1>}
      uVRegion {bIsValid=false, max=<0 0>, min=<0 0>}
     }
     horizontalTopSlotImage {
      bIsDynamicallyLoaded=false
      drawAs=NoDrawType
      imageType=FullColor
      mirroring=NoMirror
      resourceName=../../../Engine/Content/Slate/Common/Scrollbar_Background_Horizontal.png
      resourceObject=null
      tiling=NoTile
      imageSize=<8 8>
      margin=<ltrb 0 0 0 0>
      outlineSettings {
       bUseBrushTransparency=false
       roundingType=HalfHeightRadius
       width=0
       color {colorUseRule=UseColor_Specified, specifiedColor=<rgba 0 0 0 0>}
       cornerRadii=<0 0 0 0>
      }
      tintColor {colorUseRule=UseColor_Specified, specifiedColor=<rgba 1 1 1 1>}
      uVRegion {bIsValid=false, max=<0 0>, min=<0 0>}
     }
     hoveredThumbImage {
      bIsDynamicallyLoaded=false
      drawAs=Box
      imageType=FullColor
      mirroring=NoMirror
      resourceName=../../../Engine/Content/Slate/Common/Scrollbar_Thumb.png
      resourceObject=null
      tiling=NoTile
      imageSize=<0 0>
      margin=<ltrb 0.25 0.25 0.25 0.25>
      outlineSettings {
       bUseBrushTransparency=false
       roundingType=HalfHeightRadius
       width=0
       color {colorUseRule=UseColor_Specified, specifiedColor=<rgba 0 0 0 0>}
       cornerRadii=<0 0 0 0>
      }
      tintColor {colorUseRule=UseColor_Specified, specifiedColor=<rgba 1 1 1 1>}
      uVRegion {bIsValid=false, max=<0 0>, min=<0 0>}
     }
     normalThumbImage {
      bIsDynamicallyLoaded=false
      drawAs=Box
      imageType=FullColor
      mirroring=NoMirror
      resourceName=../../../Engine/Content/Slate/Common/Scrollbar_Thumb.png
      resourceObject=null
      tiling=NoTile
      imageSize=<0 0>
      margin=<ltrb 0.25 0.25 0.25 0.25>
      outlineSettings {
       bUseBrushTransparency=false
       roundingType=HalfHeightRadius
       width=0
       color {colorUseRule=UseColor_Specified, specifiedColor=<rgba 0 0 0 0>}
       cornerRadii=<0 0 0 0>
      }
      tintColor {colorUseRule=UseColor_Specified, specifiedColor=<rgba 1 1 1 1>}
      uVRegion {bIsValid=false, max=<0 0>, min=<0 0>}
     }
     verticalBackgroundImage {
      bIsDynamicallyLoaded=false
      drawAs=NoDrawType
      imageType=NoImage
      mirroring=NoMirror
      resourceName=None
      resourceObject=null
      tiling=NoTile
      imageSize=<0 0>
      margin=<ltrb 0 0 0 0>
      outlineSettings {
       bUseBrushTransparency=false
       roundingType=HalfHeightRadius
       width=0
       color {colorUseRule=UseColor_Specified, specifiedColor=<rgba 0 0 0 0>}
       cornerRadii=<0 0 0 0>
      }
      tintColor {colorUseRule=UseColor_Specified, specifiedColor=<rgba 1 1 1 1>}
      uVRegion {bIsValid=false, max=<0 0>, min=<0 0>}
     }
     verticalBottomSlotImage {
      bIsDynamicallyLoaded=false
      drawAs=NoDrawType
      imageType=FullColor
      mirroring=NoMirror
      resourceName=../../../Engine/Content/Slate/Common/Scrollbar_Background_Vertical.png
      resourceObject=null
      tiling=NoTile
      imageSize=<8 8>
      margin=<ltrb 0 0 0 0>
      outlineSettings {
       bUseBrushTransparency=false
       roundingType=HalfHeightRadius
       width=0
       color {colorUseRule=UseColor_Specified, specifiedColor=<rgba 0 0 0 0>}
       cornerRadii=<0 0 0 0>
      }
      tintColor {colorUseRule=UseColor_Specified, specifiedColor=<rgba 1 1 1 1>}
      uVRegion {bIsValid=false, max=<0 0>, min=<0 0>}
     }
     verticalTopSlotImage {
      bIsDynamicallyLoaded=false
      drawAs=NoDrawType
      imageType=FullColor
      mirroring=NoMirror
      resourceName=../../../Engine/Content/Slate/Common/Scrollbar_Background_Vertical.png
      resourceObject=null
      tiling=NoTile
      imageSize=<8 8>
      margin=<ltrb 0 0 0 0>
      outlineSettings {
       bUseBrushTransparency=false
       roundingType=HalfHeightRadius
       width=0
       color {colorUseRule=UseColor_Specified, specifiedColor=<rgba 0 0 0 0>}
       cornerRadii=<0 0 0 0>
      }
      tintColor {colorUseRule=UseColor_Specified, specifiedColor=<rgba 1 1 1 1>}
      uVRegion {bIsValid=false, max=<0 0>, min=<0 0>}
     }
    }
    widgetStyle {
     barThickness=0
     bottomShadowBrush {
      bIsDynamicallyLoaded=false
      drawAs=NoDrawType
      imageType=FullColor
      mirroring=NoMirror
      resourceName=../../../Engine/Content/Slate/Common/ScrollBoxShadowBottom.png
      resourceObject=null
      tiling=NoTile
      imageSize=<16 8>
      margin=<ltrb 0.5 0 0.5 1>
      outlineSettings {
       bUseBrushTransparency=false
       roundingType=HalfHeightRadius
       width=0
       color {colorUseRule=UseColor_Specified, specifiedColor=<rgba 0 0 0 0>}
       cornerRadii=<0 0 0 0>
      }
      tintColor {colorUseRule=UseColor_Specified, specifiedColor=<rgba 1 1 1 1>}
      uVRegion {bIsValid=false, max=<0 0>, min=<0 0>}
     }
     horizontalScrolledContentPadding=<ltrb 0 0 1 0>
     leftShadowBrush {
      bIsDynamicallyLoaded=false
      drawAs=NoDrawType
      imageType=FullColor
      mirroring=NoMirror
      resourceName=../../../Engine/Content/Slate/Common/ScrollBoxShadowLeft.png
      resourceObject=null
      tiling=NoTile
      imageSize=<8 16>
      margin=<ltrb 1 0.5 0 0.5>
      outlineSettings {
       bUseBrushTransparency=false
       roundingType=HalfHeightRadius
       width=0
       color {colorUseRule=UseColor_Specified, specifiedColor=<rgba 0 0 0 0>}
       cornerRadii=<0 0 0 0>
      }
      tintColor {colorUseRule=UseColor_Specified, specifiedColor=<rgba 1 1 1 1>}
      uVRegion {bIsValid=false, max=<0 0>, min=<0 0>}
     }
     rightShadowBrush {
      bIsDynamicallyLoaded=false
      drawAs=Box
      imageType=FullColor
      mirroring=NoMirror
      resourceName=../../../Engine/Content/Slate/Common/ScrollBoxShadowRight.png
      resourceObject=null
      tiling=NoTile
      imageSize=<8 16>
      margin=<ltrb 0 0.5 1 0.5>
      outlineSettings {
       bUseBrushTransparency=false
       roundingType=HalfHeightRadius
       width=0
       color {colorUseRule=UseColor_Specified, specifiedColor=<rgba 0 0 0 0>}
       cornerRadii=<0 0 0 0>
      }
      tintColor {colorUseRule=UseColor_Specified, specifiedColor=<rgba 1 1 1 1>}
      uVRegion {bIsValid=false, max=<0 0>, min=<0 0>}
     }
     topShadowBrush {
      bIsDynamicallyLoaded=false
      drawAs=NoDrawType
      imageType=FullColor
      mirroring=NoMirror
      resourceName=../../../Engine/Content/Slate/Common/ScrollBoxShadowTop.png
      resourceObject=null
      tiling=NoTile
      imageSize=<16 8>
      margin=<ltrb 0.5 1 0.5 0>
      outlineSettings {
       bUseBrushTransparency=false
       roundingType=HalfHeightRadius
       width=0
       color {colorUseRule=UseColor_Specified, specifiedColor=<rgba 0 0 0 0>}
       cornerRadii=<0 0 0 0>
      }
      tintColor {colorUseRule=UseColor_Specified, specifiedColor=<rgba 1 1 1 1>}
      uVRegion {bIsValid=false, max=<0 0>, min=<0 0>}
     }
     verticalScrolledContentPadding=<ltrb 0 0 0 1>
    }
   }
  }
  Content {widgetClassPath=&VerticalBox, parent=ScrollBox_NoteBreaksVirtualizingMutualFriendList, widget=Content, slot {class=ScrollBoxSlot}}
  Profile_Header {
   widgetClassPath=&/WBP_Profile_Header
   bIsVariable=true
   parent=Content
   widget=Profile_Header
   slot {class=VerticalBoxSlot, properties {padding=<ltrb 0 30 0 0>}}
  }
  O_FriendRequest {
   widgetClassPath=&Overlay
   bIsVariable=true
   parent=Content
   widget=O_FriendRequest
   slot {class=VerticalBoxSlot, properties {padding=<ltrb 18 30 18 8>}}
  }
  FriendRequest_Backplate {
   widgetClassPath=/Sidebar/Buttons/Blocks/WBP_Sidebar_Block_Backplate
   bIsVariable=true
   parent=O_FriendRequest
   widget=FriendRequest_Backplate
   slot {class=OverlaySlot, properties {horizontalAlignment=HAlign_Fill, verticalAlignment=VAlign_Fill}}
  }
  VB_FriendRequest {
   widgetClassPath=&VerticalBox
   parent=O_FriendRequest
   widget=VB_FriendRequest
   slot {class=OverlaySlot, properties {horizontalAlignment=HAlign_Fill, padding=<ltrb 18 18 18 18>}}
  }
  Text_FriendRequest_Title {
   widgetClassPath=&CommonTextBlock
   bIsVariable=true
   parent=VB_FriendRequest
   widget=Text_FriendRequest_Title
   properties {
    bAutoCollapseWithEmptyText=true
    bIsScrollingEnabled=false
    mobileFontSizeMultiplier=1.33
    scrollStyle=/FortUILibrary/Styles/Typography/Foundation/TextScrollStyle-NoFade.TextScrollStyle-NoFade_C
    text="Sent you a friend request!"
    textOverflowPolicy=Ellipsis
    font {
     bForceMonospaced=false
     bMaterialIsStencil=false
     fontMaterial=/Sidebar/Materials/Fonts/MI_Sidebar_FontColor_Primary
     fontObject=&HeadingNow
     letterSpacing=0
     monospacedWidth=1
     size=15.75
     skewAmount=0
     typefaceFontName=75 Medium
     outlineSettings {
      bApplyOutlineToDropShadows=false
      bMiteredCorners=false
      bSeparateFillAlpha=false
      outlineMaterial=null
      outlineSize=0
      outlineColor=<rgba 0 0 0 1>
     }
    }
    shadowOffset=<0 0>
   }
   slot {class=VerticalBoxSlot, properties {padding=<ltrb 0 0 0 8>}}
  }
  Text_FriendRequest_Body {
   widgetClassPath=&CommonTextBlock
   bIsVariable=true
   parent=VB_FriendRequest
   widget=Text_FriendRequest_Body
   properties {
    autoWrapText=true
    bAutoCollapseWithEmptyText=true
    bIsScrollingEnabled=false
    lineHeightPercentage=1.1
    mobileFontSizeMultiplier=1.41
    scrollStyle=&TextScrollStyle-NoFade_C
    text="Friends can see your online status, form parties together. and chat with you depending on your communication settings. You can report players from their profile."
    textOverflowPolicy=Ellipsis
    colorAndOpacity {colorUseRule=UseColor_Specified, specifiedColor=<rgba 1 1 1 0.5>}
    font {
     bForceMonospaced=false
     bMaterialIsStencil=false
     fontMaterial=/Sidebar/Materials/Fonts/MI_Sidebar_FontColor_Secondary
     fontObject=&HeadingNow
     letterSpacing=0
     monospacedWidth=1
     size=12.75
     skewAmount=0
     typefaceFontName=74 Regular
     outlineSettings {
      bApplyOutlineToDropShadows=false
      bMiteredCorners=false
      bSeparateFillAlpha=false
      outlineMaterial=/FortUILibrary/Systems/ModularButton/MaterialInstances/MI_UIKit_TextOutline
      outlineSize=0
      outlineColor=<rgba 0 0 0 1>
     }
    }
    shadowOffset=<0 0>
   }
   slot {class=VerticalBoxSlot}
  }
  OtherPlayer_CurrentActivity {
   widgetClassPath=&/WBP_Sidebar_CurrentActivity_OtherPlayer
   bIsVariable=true
   parent=Content
   widget=OtherPlayer_CurrentActivity
   slot {class=VerticalBoxSlot, properties {padding=<ltrb 18 30 18 0>}}
  }
  Interactions {
   widgetClassPath=&VerticalBox
   bIsVariable=true
   parent=Content
   widget=Interactions
   slot {class=VerticalBoxSlot, properties {padding=<ltrb 18 30 18 18>}}
  }
  VB_Interactions {widgetClassPath=&VerticalBox, bIsVariable=true, parent=Interactions, widget=VB_Interactions, slot {class=VerticalBoxSlot}}
  WBP_Sidebar_Button_MoreOptions {
   widgetClassPath=/Sidebar/Buttons/WBP_Sidebar_ButtonGeneric
   bIsVariable=true
   parent=Interactions
   widget=WBP_Sidebar_Button_MoreOptions
   properties {bShouldUseFallbackDefaultInputAction=false, text=More options}
   slot {class=VerticalBoxSlot}
  }
  VB_MoreInteractions {
   widgetClassPath=&VerticalBox
   bIsVariable=true
   parent=Interactions
   widget=VB_MoreInteractions
   properties {visibility=Collapsed}
   slot {class=VerticalBoxSlot, properties {padding=<ltrb 0 24 0 0>}}
  }
  MutualFriends {
   widgetClassPath=/Sidebar/Panels/Social/List/WBP_Sidebar_SocialTreeList
   bIsVariable=true
   parent=Content
   widget=MutualFriends
   properties {bIsVolatile=true}
   slot {class=VerticalBoxSlot, properties {padding=<ltrb 0 18 0 0>}}
  }
 ]
}
