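# WBP_Exit_Panel is the sidebar's Exit panel, with quit, leave, logout and sleep buttons plus island like, dislike, favorite and follow-developer toggles.
# An ESidebarAction set controls button visibility, and FortSidebarSettingsVM executes actions.
# The leave-action enum maps to button text and confirmation view-model classes, while Mission and Storm Shield instead open a dialog with a rewards list.
# Focus returns to the last visible focused button, with fallbacks.
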WidgetBlueprint /Sidebar/Panels/WBP_Exit_Panel
graphs [
 BP_GetDesiredFocusTarget {nodes [
   Get Desired Focus Target()
   > n1 = Get!(execute:Exec) -> (then:Exec, else:Exec, LastFocusedButton:User Widget Object Reference)
   > n2 = Branch(Condition=IsVisible(self=@n1))
   > Return Node!(execute, ReturnValue:Widget Object Reference=@n1)
   n3 = For Each Loop(
    Exec=[@n2.else, @n1.else]
    Array:Array of Widget Object References=GetAllChildren(self=Get ActionButtonsContainer() -> (ActionButtonsContainer:Vertical Box Object Reference))
   ) -> (Array Element:Widget Object Reference)
   > Branch(Condition=AND Boolean(A="integer > integer"(A=@n3.Array Index), B=IsVisible(self=@n3)))
   > Return Node(ReturnValue=@n3)
   Return Node(execute=@n3.Completed, ReturnValue=Get Button_LeaveAction() -> (Button_LeaveAction:WBP Sidebar Button Generic Object Reference))
  ]}
 ChooseRecommendOrNotInterested {nodes [
   ChooseRecommendOrNotInterested!() -> (then, TargetButton:WBP Sidebar Button Generic Object Reference, OppositeButton:WBP Sidebar Button Generic Object Reference) {is=title K2Node_FunctionEntry}
   n1 = Get OppositeButton() -> (OppositeButton:WBP Sidebar Button Generic Object Reference)
   > Set ButtonType!(execute, ButtonType:E_UI_Sidebar_Button_Type Enum=NewEnumerator0, self:WBP Sidebar Button Generic Object Reference=@n1) -> (then, Output_Get:E_UI_Sidebar_Button_Type Enum=NewEnumerator0)
   > InitializeButtonVisuals(self:WBP Sidebar Button Generic Object Reference=@n1)
   > n2 = Set IsOneButtonSelected(IsOneButtonSelected:bool=GetIsEnabled!(self:Widget Object Reference=@n1) -> (ReturnValue:bool)) -> (Output_Get:bool)
   > SetIsEnabled(self=@n1, bInIsEnabled=NOT Boolean(A=@n2))
   n3 = Get TargetButton() -> (TargetButton:WBP Sidebar Button Generic Object Reference)
   > Set ButtonType(
    ButtonType=Select(
     Option 0:E_UI_Sidebar_Button_Type Enum=NewEnumerator0
     Option 1:E_UI_Sidebar_Button_Type Enum=NewEnumerator2
     Index=Get IsOneButtonSelected() -> (IsOneButtonSelected:bool)
    ) -> (ReturnValue:E_UI_Sidebar_Button_Type Enum)
    self=@n3
   ) -> (Output_Get=NewEnumerator0)
   > InitializeButtonVisuals(self=@n3)
  ]}
 CreateRewardsListWidget {nodes [
   CreateRewardsListWidget() {is=title K2Node_FunctionEntry}
   > n1 = Cast To FortGameStateZone!(execute, Object=GetGameState!() -> (ReturnValue:Game State Base Object Reference)) -> (then, CastFailed, AsFort Game State Zone:Fort Game State Zone Object Reference) {is=title K2Node_DynamicCast}
   > Set GameStateZone(GameStateZone:Fort Game State Zone Object Reference=@n1) -> (Output_Get:Fort Game State Zone Object Reference)
   > n2 = Create Widget!(
    execute:Exec
    Class:User Widget Class Reference=/Sidebar/Panels/Overlay/WBP_Confirmation_RewardsList.WBP_Confirmation_RewardsList_C
    OwningPlayer:Player Controller Object Reference=GetOwningPlayer()
    ShowMissionAlertMessage:bool=False
    Use Verbose Mode:bool=False
   ) -> (then:Exec, ReturnValue:WBP Confirmation Rewards List Object Reference)
   > Set RewardsList(RewardsList:WBP Confirmation Rewards List Object Reference=@n2) -> (Output_Get:WBP Confirmation Rewards List Object Reference)
   > n3 = GetActiveModifiers!(execute, self:Fort Game State Zone Object Reference=Get GameStateZone() -> (GameStateZone:Fort Game State Zone Object Reference)) -> (then, OutActiveModifiers:Array of Fort Gameplay Modifier Item Definition Object References)
   > Update!(
    execute
    self:WBP Confirmation Rewards List Object Reference=Get RewardsList() -> (RewardsList:WBP Confirmation Rewards List Object Reference)
    Mission Rewards:Array of Fort Mission Reward Info Structures=GetMissionRewardsInfo!(self:Fort HUDContext Object Reference=Get FortHUDContext!() -> (ReturnValue:Fort HUDContext Object Reference) {is=title K2Node_GetBlueprintContextEx}) -> (RewardList:Array of Fort Mission Reward Info Structures)
    Gameplay Modifiers:Array of Fort Gameplay Modifier Item Definition Object References=@n3
   ) -> (then)
   > Return Node(ReturnValue=Get RewardsList())
  ]}
 EventGraph {nodes [
   "On Clicked (Button_Exit)"!() -> (OutputDelegate, then, Button:Common Button Base Object Reference) {is=title K2Node_ComponentBoundEvent}
   > NavigateToConfirmPanel!(execute, self, Class:Sidebar Confirmation Panel VM Class Reference=/Sidebar/Panels/Overlay/Data/ConfirmationVM_Exit.ConfirmationVM_Exit_C) -> (then)
   "On Clicked (Button_LeaveExperience)"!() -> (OutputDelegate, then, Button:Common Button Base Object Reference) {is=title K2Node_ComponentBoundEvent}
   > NavigateToConfirmPanel(Class=/Sidebar/Panels/Overlay/Data/ConfirmationVM_BackToLobby.ConfirmationVM_BackToLobby_C)
   "On Clicked (Button_LeaveAction)"!() -> (OutputDelegate, then, Button:Common Button Base Object Reference) {is=title K2Node_ComponentBoundEvent}
   > n1 = Switch on EAthenaMainMenuLeaveButtonText!(
    execute
    Selection:EAthenaMainMenuLeaveButtonText Enum=Get Current Leave Action() -> (Current Leave Action:EAthenaMainMenuLeaveButtonText Enum=LeaveAction_ReturnToLobby)
   ) -> (
    LeaveAction_ReturnToLobby:Exec
    LeaveAction_Replay:Exec
    LeaveAction_PlaylistCreative:Exec
    LeaveAction_PlaylistCreative_LTM:Exec
    LeaveAction_PlaylistDefault:Exec
    LeaveAction_InParty:Exec
    LeaveAction_Mission:Exec
    LeaveAction_StormShield:Exec
    LeaveAction_Default:Exec
   ) {is=title K2Node_SwitchEnum}
   n2 = Get!(execute=[
     @n1.LeaveAction_Default
     @n1.LeaveAction_InParty
     @n1.LeaveAction_PlaylistCreative
     @n1.LeaveAction_PlaylistCreative_LTM
     @n1.LeaveAction_PlaylistDefault
     @n1.LeaveAction_Replay
    ]) -> (then, else, Leave Action Class:Sidebar Confirmation Panel VM Class Reference)
   > NavigateToConfirmPanel(Class=@n2)
   n3 = CreateRewardsListWidget!(execute=[@n1.LeaveAction_Mission, @n1.LeaveAction_StormShield], self) -> (then, ReturnValue:Widget Object Reference) {is=title K2Node_CallFunction}
   n4 = GetOutpostConfirmationDialogText!() -> (OutTitle:Text, OutDescription:Text)
   > n5 = ShowSimpleConfirmationDialog_CustomInput!(
    execute:Exec
    Icon:Slate Brush Structure
    Title:Text=@n4
    Message:Text=@n4.OutDescription
    ConfirmAction:Name=Quit
    DeclineAction:Name=Cancel
    AdditionalContent:Widget Object Reference=@n3
    LeftAdditionalContent:Widget Object Reference
    Cancelable:bool=false
    TargetPlayer:Local Player Object Reference
   ) -> (then:Exec, DialogResult:Exec, Result:EFortDialogResult Enum, ResultName:Name) {is=title K2Node_AsyncAction}
   Branch(execute=@n5.DialogResult, Condition="Equal (Enum)"(A:EFortDialogResult Enum=@n5, B:EFortDialogResult Enum=Confirmed))
   > ExecuteAction!(
    execute
    self:Fort Sidebar Settings VM Object Reference=Get FortSidebarSettingsVM() -> (FortSidebarSettingsVM:Fort Sidebar Settings VM Object Reference)
    SidebarAction:ESidebarAction Enum=AthenaMainMenuLeave
   ) -> (then)
   ExecuteAction(execute=@n1.LeaveAction_ReturnToLobby, self=Get FortSidebarSettingsVM(), SidebarAction=AthenaMainMenuLeave)
   "On Clicked (Button_FollowDeveloper)"!() -> (OutputDelegate, then, Button:Common Button Base Object Reference) {is=title K2Node_ComponentBoundEvent}
   > ExecuteAction(self=Get FortSidebarSettingsVM(), SidebarAction=FollowDeveloper)
   > PlayFavoriteAudio!(
    execute
    self
    IsFavorite:bool=Get bIsFollowingDeveloper!(self:Fort Sidebar Settings VM Object Reference=Get FortSidebarSettingsVM()) -> (bIsFollowingDeveloper:bool)
   ) -> (then)
   "On Clicked (Button_LeaveParty)"!() -> (OutputDelegate, then, Button:Common Button Base Object Reference) {is=title K2Node_ComponentBoundEvent}
   > Set SkipMainMenuLeave(SkipMainMenuLeave:bool=true) -> (Output_Get:bool)
   > NavigateToConfirmPanel(Class=/Sidebar/Panels/Overlay/Data/ConfirmationVM_LeaveParty.ConfirmationVM_LeaveParty_C)
   "On Clicked (Button_Sleep)"!() -> (OutputDelegate, then, Button:Common Button Base Object Reference) {is=title K2Node_ComponentBoundEvent}
   > ExecuteAction(self=Get FortSidebarSettingsVM(), SidebarAction=Parsley)
   "On Clicked (Button_SaveIsland)"!() -> (OutputDelegate, then, Button:Common Button Base Object Reference) {is=title K2Node_ComponentBoundEvent}
   > ExecuteAction(self=Get FortSidebarSettingsVM(), SidebarAction=Favorite)
   > PlayFavoriteAudio(IsFavorite=Get bIsFavorite!(self:Fort Sidebar Settings VM Object Reference=Get FortSidebarSettingsVM()) -> (bIsFavorite:bool))
   "On Clicked (Button_Like)"!() -> (OutputDelegate, then, Button:Common Button Base Object Reference) {is=title K2Node_ComponentBoundEvent}
   > ChooseRecommendOrNotInterested!(
    execute
    self
    TargetButton:WBP Sidebar Button Generic Object Reference=Get Button_Like() -> (Button_Like:WBP Sidebar Button Generic Object Reference)
    OppositeButton:WBP Sidebar Button Generic Object Reference=Get Button_Dislike() -> (Button_Dislike:WBP Sidebar Button Generic Object Reference)
   ) -> (then) {is=title K2Node_CallFunction}
   > ExecuteAction(self=Get FortSidebarSettingsVM(), SidebarAction=Like)
   > n6 = Get!(execute) -> (then, else, Sidebar:WBP Sidebar Object Reference)
   > n7 = Set isLiked!(
    execute
    isLiked:bool=NOT Boolean(A=Get isLiked!(self:WBP Sidebar Object Reference=@n6) -> (isLiked:bool))
    self:WBP Sidebar Object Reference=@n6
   ) -> (then, Output_Get:bool)
   > PlayLikedAudio!(execute, self, IsLiked:bool=@n7) -> (then)
   "On Clicked (Button_Dislike)"!() -> (OutputDelegate, then, Button:Common Button Base Object Reference) {is=title K2Node_ComponentBoundEvent}
   > ChooseRecommendOrNotInterested(TargetButton=Get Button_Dislike(), OppositeButton=Get Button_Like())
   > ExecuteAction(self=Get FortSidebarSettingsVM(), SidebarAction=NotInterested)
   > n8 = Get()
   > n9 = Set isDisliked!(
    execute
    isDisliked:bool=NOT Boolean(A=Get isDisliked!(self:WBP Sidebar Object Reference=@n8) -> (isDisliked:bool))
    self:WBP Sidebar Object Reference=@n8
   ) -> (then, Output_Get:bool)
   > PlayDislikedAudio!(execute, self, IsDisliked:bool=@n9) -> (then)
   "On Clicked (Button_Logout)"!() -> (OutputDelegate, then, Button:Common Button Base Object Reference) {is=title K2Node_ComponentBoundEvent}
   > NavigateToConfirmPanel(Class=/Sidebar/Panels/Overlay/Data/ConfirmationVM_Logout.ConfirmationVM_Logout_C)
   Event On Activated()
   > SetActivityTitle()
   > SetFavoriteButton!(execute, self, is Favorite:bool=Get bIsFavorite(self=Get FortSidebarSettingsVM())) -> (then)
   > SidebarFeatureFinishedLoading(self:Fort Sidebar VM Object Reference=Get FortSidebarVM() -> (FortSidebarVM:Fort Sidebar VM Object Reference))
   Event Construct()
   > SetupExitUI(self:Fort Sidebar Settings VM Object Reference=Get FortSidebarSettingsVM())
   > SetLeavePartyButtonText()
   > n10 = FindParentWidgetOfType!(execute, StartingWidget:Widget Object Reference=Self-Reference(), Type:Widget Class Reference=/Sidebar/WBP_Sidebar.WBP_Sidebar_C) -> (then, ReturnValue:WBP Sidebar Object Reference)
   > n11 = Set Sidebar(Sidebar:WBP Sidebar Object Reference=@n10) -> (Output_Get:WBP Sidebar Object Reference)
   > Update Support Buttons Visuals!(execute, self, isLiked:bool=Get isLiked(self=@n11), isDisliked:bool=Get isDisliked(self=@n11)) -> (then)
   Event PreConstruct()
   > SetMobileStyle() {is=title K2Node_CallFunction}
   Event OnInitialized()
   > SetRenderTransform!(
    execute
    self:Widget Object Reference=Get Block_Icon!(self:WBP UIKit Button Generic Base Object Reference=Get Button_Dislike()) -> (Block_Icon:WBP UIKit Block Image Object Reference)
    InTransform_Translation:Vector 2D Structure=(X=0,Y=0)
    InTransform_Scale:Vector 2D Structure=(X=1,Y=1)
    InTransform_Shear:Vector 2D Structure=(X=0,Y=0)
    InTransform_Angle:float=180
   ) -> (then)
   > SetBlocksTiming!(
    execute
    self:UIKit Modular Button Object Reference=[Get Button_Like(), Get Button_Dislike(), Get Button_SaveIsland() -> (Button_SaveIsland:WBP Sidebar Button Generic Object Reference)]
    Timing:"UIKit Block Timing Structure (by ref)"=Make UIKit Block Timing!(
     Hovering:float=0
     Unhovering:float=0
     Focusing:float=0.1
     Unfocusing:float=0.15
     Pressing:float=0.1
     Releasing:float=0.1
     Disabling:float=0.1
     Enabling:float=0.1
     Selecting:float=0.15
     DeselectingIdle:float=0.1
     DeselectingFocused:float=0.1
    ) -> (UIKitBlockTiming:UIKit Block Timing Structure) {is=title K2Node_MakeStruct}
   ) -> (then)
  ]}
 Execute Leave Action {nodes [
   n1 = Execute Leave Action!() -> (then, Confrimation Action:Name) {is=title K2Node_FunctionEntry}
   > n2 = Branch(Condition="Equal (Name)"(A=@n1, B=LeaveAction))
   ExecuteAction(execute=[@n2.then, @n3.else], self=Get FortSidebarSettingsVM(), SidebarAction=AthenaMainMenuLeave)
   Branch(execute=@n2.else, Condition="Equal (Name)"(A=@n1, B=LeaveParty))
   > n3 = Branch(Condition=Get SkipMainMenuLeave() -> (SkipMainMenuLeave:bool))
   > ExecuteAction(self=Get FortSidebarSettingsVM(), SidebarAction=LeaveParty)
  ]}
 GetOutpostConfirmationDialogText {nodes [
   GetOutpostConfirmationDialogText() {is=title K2Node_FunctionEntry}
   n1 = Get FortPartyContext!() -> (ReturnValue:Fort Party Context Object Reference) {is=title K2Node_GetBlueprintContextEx}
   > Set bIsInParty(bIsInParty:bool="integer > integer"(A=GetNumPlayersInLocalParty!(self:Fort Party Context Object Reference=@n1) -> (ReturnValue:int=0), B=1) {is=title K2Node_CallFunction}) -> (Output_Get:bool)
   > n2 = IsOutpostOwner!(execute, self) -> (then, ReturnValue:bool)
   > n3 = Branch(Condition=@n2)
   > Return Node!(
    execute
    OutTitle:Text=Get StormshieldLeaveTitle() -> (StormshieldLeaveTitle:Text)
    OutDescription:Text=Select(
     Option 0:Text="LOCTEXT(\"Leave the Storm Shield and return to Homebase.\")"
     Option 1:Text="LOCTEXT(\"Leave the Storm Shield and return to Homebase with your Party.\")"
     Index=Get bIsInParty() -> (bIsInParty:bool)
    ) -> (ReturnValue:Text)
   )
   n4 = IsInOutpostZone!(self:Fort Global UIContext Object Reference=Get FortGlobalUIContext!() -> (ReturnValue:Fort Global UIContext Object Reference) {is=title K2Node_GetBlueprintContextEx}) -> (ReturnValue:bool)
   n5 = Get MissionLeaveDescription() -> (MissionLeaveDescription:Text)
   n6 = Get StormshieldLeaveDescription() -> (StormshieldLeaveDescription:Text)
   n7 = Get FortGlobalUIContext()
   Return Node(
    execute=@n3.else
    OutTitle=Select(Option 0="LOCTEXT(\"Leave Mission?\")", Option 1="LOCTEXT(\"Leave Storm Shield?\")", Index=IsInOutpostZone(self=@n7))
    OutDescription=Select(
     Option 0=Select(Option 0=@n5, Option 1=@n6, Index=@n4)
     Option 1=Get HordModeLeaveDescription() -> (HordModeLeaveDescription:Text)
     Index=IsInNeighborhood!(self) -> (bIsInNeighborhood:bool)
    )
   )
  ]}
 IsInNeighborhood {nodes [
   IsInNeighborhood() {is=title K2Node_FunctionEntry}
   > n1 = Get Deployable Base Manager!(execute) -> (then, ReturnValue:Fort Deployable Base Manager Object Reference) {is=title K2Node_CallFunction}
   > n2 = Set DeployableBaseManager(DeployableBaseManager:Fort Deployable Base Manager Object Reference=@n1) -> (Output_Get:Fort Deployable Base Manager Object Reference)
   > n3 = Is Valid(InputObject=@n2)
   Return Node!(execute=@n3.Is Not Valid, bIsInNeighborhood:bool=false)
   n4 = Get DeployableBaseManager() -> (DeployableBaseManager:Fort Deployable Base Manager Object Reference)
   > Return Node(bIsInNeighborhood="Equal (Enum)"(
     A:EDeployableBaseUseType Enum=GetDeployableBaseUseType!(self:Fort Deployable Base Manager Object Reference=@n4) -> (ReturnValue:EDeployableBaseUseType Enum=Neighborhood)
     B:EDeployableBaseUseType Enum=Neighborhood
    ))
  ]}
 IsOutpostOwner {nodes [
   IsOutpostOwner() {is=title K2Node_FunctionEntry}
   n1 = GetComponentByClass!(
    self:Actor Object Reference=GetOwningPlayer()
    ComponentClass:Actor Component Class Reference=/Script/FortniteGame.FortControllerComponent_OwnedZone
   ) -> (ReturnValue:Fort Controller Component Owned Zone Object Reference)
   > n2 = Is Valid(InputObject=@n1)
   Return Node!(execute=@n2.Is Not Valid, ReturnValue:bool=false)
   > Return Node(ReturnValue=IsZoneOwner!(self:Fort Controller Component Owned Zone Object Reference=@n1) -> (ReturnValue:bool))
  ]}
 NavigateToConfirmPanel {nodes [
   n1 = NavigateToConfirmPanel!() -> (then, Class:Sidebar Confirmation Panel VM Class Reference) {is=title K2Node_FunctionEntry}
   > n2 = Construct Object from Class(Class=@n1) -> (ReturnValue:Sidebar Confirmation Panel VM Object Reference)
   > NavigateToConfirmPanel!(
    execute
    SelfReference:Object Reference=Self-Reference()
    SidebarConfirmationVM:Sidebar Confirmation Panel VM Object Reference=@n2
    SidebarNavigationVM:Fort Sidebar Navigation VM Object Reference=Get FortSidebarNavigationVM() -> (FortSidebarNavigationVM:Fort Sidebar Navigation VM Object Reference)
   ) -> (then) {is=title K2Node_CallFunction}
  ]}
 NavigateToExpressSupportButtons {nodes [
   NavigateToExpressSupportButtons!() -> (then, Navigation:EUINavigation Enum) {is=title K2Node_FunctionEntry}
   n1 = Get Button_Like()
   > n2 = Branch(Condition=AND Boolean(A=IsVisible(self=@n1), B=GetIsEnabled(self=@n1)))
   > Return Node(ReturnValue:Widget Object Reference=Get Button_Like())
   n3 = Get Button_Dislike()
   n4 = Branch(execute=@n2.else, Condition=AND Boolean(A=IsVisible(self=@n3), B=GetIsEnabled(self=@n3)))
   > Return Node(ReturnValue=Get Button_Dislike())
   n5 = Get Button_FollowDeveloper() -> (Button_FollowDeveloper:WBP Sidebar Button Generic Object Reference)
   n6 = Branch(execute=@n4.else, Condition=IsVisible(self=@n5))
   > Return Node(ReturnValue=@n5)
   n7 = Get Button_SaveIsland()
   Branch(execute=@n6.else, Condition=IsVisible(self=@n7))
   > Return Node(ReturnValue=@n7)
  ]}
 OnSetLastFocusedButton {nodes [
   n1 = OnSetLastFocusedButton!() -> (then, Type:EUINavigation Enum, OldFocusedWidget:Widget Object Reference, NewFocusedWidget:Widget Object Reference) {is=title K2Node_FunctionEntry}
   n2 = GetOuterUserWidget!(Widget:Widget Object Reference=@n1.NewFocusedWidget) -> (ReturnValue:User Widget Object Reference)
   > Branch(Condition=IsValid(Object=@n2))
   > Set LastFocusedButton(LastFocusedButton:User Widget Object Reference=@n2) -> (Output_Get:User Widget Object Reference)
  ]}
 PlayDislikedAudio {nodes [
   n1 = PlayDislikedAudio!() -> (then, IsDisliked:bool=~) {is=title K2Node_FunctionEntry}
   > PrintText!(
    execute
    InText=Format Text!(Format="LOCTEXT(\"PlayDislikedAudio, IsDisiked = {IsDisliked}\")", IsDisliked:bool=@n1) -> (Result)
    bPrintToScreen
    bPrintToLog
    TextColor
    Duration
    Key
   ) -> (then)
   > n2 = Branch(Condition=@n1)
   > PlaySound2D(Sound=Get Disliked Sound() -> (Disliked Sound:Sound Base Object Reference))
   PlaySound2D(execute=@n2.else, Sound=Get Undisliked Sound() -> (Undisliked Sound:Sound Base Object Reference))
  ]}
 PlayFavoriteAudio {nodes [
   n1 = PlayFavoriteAudio!() -> (then, IsFavorite:bool=~) {is=title K2Node_FunctionEntry}
   > PrintText(InText=Format Text!(Format="LOCTEXT(\"PlayFavoriteAudio, IsFavorite = {IsFavorite}\")", IsFavorite:bool=@n1) -> (Result))
   > n2 = Branch(Condition=@n1)
   > PlaySound2D(Sound=Get Favorited Sound() -> (Favorited Sound:Sound Base Object Reference))
   PlaySound2D(execute=@n2.else, Sound=Get Unfavorited Sound() -> (Unfavorited Sound:Sound Base Object Reference))
  ]}
 PlayLikedAudio {nodes [
   n1 = PlayLikedAudio!() -> (then, IsLiked:bool=~) {is=title K2Node_FunctionEntry}
   > PrintText(InText=Format Text!(Format="LOCTEXT(\"PlayLikedAudio, IsLiked = {IsLiked}\")", IsLiked:Wildcard) -> (Result))
   > n2 = Branch(Condition=@n1)
   > PlaySound2D(Sound=Get Liked Sound() -> (Liked Sound:Sound Base Object Reference))
   PlaySound2D(execute=@n2.else, Sound=Get Unliked Sound() -> (Unliked Sound:Sound Base Object Reference))
  ]}
 Set Button Visibilities From Actions {nodes [
   n1 = Set Button Visibilities From Actions!() -> (then, Actions:Set of ESidebarAction Enums=SupportCreator) {is=title K2Node_FunctionEntry}
   > Set Actions(Actions:Set of ESidebarAction Enums=@n1) -> (Output_Get:Set of ESidebarAction Enums=SupportCreator)
   > SetVisibility(
    self=Get Button_Exit() -> (Button_Exit:WBP Sidebar Button Generic Object Reference)
    InVisibility=Select(
     Option 0:ESlateVisibility Enum=Collapsed
     Option 1:ESlateVisibility Enum=Visible
     Index=Contains Item!(TargetSet:Set of ESidebarAction Enums=@n1, ItemToFind:"ESidebarAction Enum (by ref)"=ExitGame) -> (ReturnValue:bool) {is=title K2Node_CallFunction}
    ) -> (ReturnValue:ESlateVisibility Enum)
   )
   > SetVisibility(
    self=Get Button_Sleep() -> (Button_Sleep:WBP Sidebar Button Generic Object Reference)
    InVisibility=Select(Option 0=Collapsed, Option 1=Visible, Index=Contains Item(TargetSet=@n1, ItemToFind=Parsley))
   )
   > SetVisibility(
    self=Get Button_LeaveAction()
    InVisibility=Select(Option 0=Collapsed, Option 1=Visible, Index=Contains Item(TargetSet=@n1, ItemToFind=AthenaMainMenuLeave))
   )
   > SetVisibility(
    self=Get Button_LeaveExperience() -> (Button_LeaveExperience:WBP Sidebar Button Generic Object Reference)
    InVisibility=Select(Option 0=Collapsed, Option 1=Visible, Index=Contains Item(TargetSet=@n1, ItemToFind=LeaveExperience))
   )
   > SetVisibility(
    self=Get Button_SaveIsland()
    InVisibility=Select(Option 0=Collapsed, Option 1=Visible, Index=Contains Item(TargetSet=@n1, ItemToFind=Favorite))
   )
   > SetVisibility(
    self=[Get Button_Like(), Get Button_Dislike()]
    InVisibility=Select(Option 0=Collapsed, Option 1=Visible, Index=Contains Item(TargetSet=@n1, ItemToFind=Like))
   )
   > SetVisibility(
    self=Get Button_FollowDeveloper()
    InVisibility=Select(Option 0=Collapsed, Option 1=Visible, Index=Contains Item(TargetSet=@n1, ItemToFind=FollowDeveloper))
   )
   > SetVisibility(
    self=Get Button_Logout() -> (Button_Logout:WBP Sidebar Button Generic Object Reference)
    InVisibility=Select(Option 0=Collapsed, Option 1=Visible, Index=Contains Item(TargetSet=@n1, ItemToFind=Logout))
   )
   n2 = Contains Item(TargetSet=@n1, ItemToFind=LeaveParty)
   n3 = "Not Equal (Enum)"!(
    A:EAthenaMainMenuLeaveButtonText Enum=GetAthenaMainMenuLeaveActionText!(self:Fort Sidebar Settings VM Object Reference=Get FortSidebarSettingsVM()) -> (ReturnValue:EAthenaMainMenuLeaveButtonText Enum=LeaveAction_ReturnToLobby)
    B:EAthenaMainMenuLeaveButtonText Enum=LeaveAction_Default
   ) -> (ReturnValue:bool)
   > SetVisibility(
    self=Get Button_LeaveParty() -> (Button_LeaveParty:WBP Sidebar Button Generic Object Reference)
    InVisibility=Select(Option 0=Collapsed, Option 1=Visible, Index=AND Boolean(A=@n2, B=@n3))
   )
   > RequestRefreshFocus(self:Common Activatable Widget Object Reference)
  ]}
 Set Leave Action {nodes [
   n1 = Set Leave Action!() -> (then, Action:EAthenaMainMenuLeaveButtonText Enum=LeaveAction_ReturnToLobby) {is=title K2Node_FunctionEntry}
   > n2 = Set Current Leave Action(Current Leave Action:EAthenaMainMenuLeaveButtonText Enum=@n1) -> (Output_Get:EAthenaMainMenuLeaveButtonText Enum=LeaveAction_ReturnToLobby)
   > SetText!(
    execute
    self:WBP UIKit Button Generic Base Object Reference=Get Button_LeaveAction()
    Text:Text=Find(
     TargetMap:Map of EAthenaMainMenuLeaveButtonText Enums to Texts=Get Leave Action Text() -> (Leave Action Text:Map of EAthenaMainMenuLeaveButtonText Enums to Texts=LeaveAction_ReturnToLobby)
     Key:"EAthenaMainMenuLeaveButtonText Enum (by ref)"=@n2
    ) -> (Value:Text)
   ) -> (then)
   > Set Leave Action Class(Leave Action Class:Sidebar Confirmation Panel VM Class Reference=Find(
     TargetMap:Map of EAthenaMainMenuLeaveButtonText Enums to Sidebar Confirmation Panel VM Class References=Get Leave Action() -> (Leave Action:Map of EAthenaMainMenuLeaveButtonText Enums to Sidebar Confirmation Panel VM Class References=LeaveAction_ReturnToLobby)
     Key=@n2
    ) -> (Value:Sidebar Confirmation Panel VM Class Reference)) -> (Output_Get:Sidebar Confirmation Panel VM Class Reference)
   > Set SkipMainMenuLeave(SkipMainMenuLeave=false)
  ]}
 SetActivityTitle {nodes [
   SetActivityTitle() {is=title K2Node_FunctionEntry}
   > n1 = Is Valid(InputObject=Get Activity!(self:Fort Sidebar Settings VM Object Reference=Get FortSidebarSettingsVM()) -> (Activity:Fort Game Activity Object Reference))
   SetVisibility(execute=@n1.Is Not Valid, self=Get CurrentActivityText() -> (CurrentActivityText:Common Text Object Reference), InVisibility=Collapsed)
   n2 = Get Activity(self=Get FortSidebarSettingsVM())
   "SetText (Text)"(
    execute=@n1.Is Valid
    self=Get CurrentActivityText()
    InText="To Text (String)"(InString=Get Title!(self:Fort Game Activity Object Reference=@n2) -> (Title:String))
   )
  ]}
 SetFavoriteButton {nodes [
   SetFavoriteButton!() -> (then, is Favorite:bool=~) {is=title K2Node_FunctionEntry}
   > Set ButtonType(
    ButtonType=Select(
     Option 0:E_UI_Sidebar_Button_Type Enum=NewEnumerator0
     Option 1:E_UI_Sidebar_Button_Type Enum=NewEnumerator2
     Index=Get is Favorite() -> (is Favorite:bool)
    ) -> (ReturnValue:E_UI_Sidebar_Button_Type Enum)
    self=Get Button_SaveIsland()
   ) -> (Output_Get=NewEnumerator0)
   > SetIsSelected!(execute, self:Common Button Base Object Reference=Get Button_SaveIsland(), InSelected:bool=Get is Favorite(), bGiveClickFeedback:bool=false) -> (then)
   > SetIcon!(
    execute
    self:WBP UIKit Button Generic Base Object Reference=Get Button_SaveIsland()
    Icon:Object Reference=Select(
     Option 0:Object Reference=/FortUILibrary/Textures/Icons/T_UI_Icon_UIKit_FavoriteOutline_MSDF
     Option 1:Object Reference=/FortUILibrary/Textures/Icons/T_UI_Icon_UIKit_Favorite_MSDF
     Index=Get is Favorite()
    ) -> (ReturnValue:Object Reference)
   ) -> (then)
   > InitializeButtonVisuals(self=Get Button_SaveIsland())
  ]}
 SetFollowDeveloperButton {nodes [
   n1 = SetFollowDeveloperButton!() -> (then, IsFollowingDeveloper:bool=~) {is=title K2Node_FunctionEntry}
   > SetText(
    self=Get Button_FollowDeveloper()
    Text=Select(Option 0:Text="LOCTEXT(\"Follow Developer\")", Option 1:Text="LOCTEXT(\"Unfollow Developer\")", Index=@n1) -> (ReturnValue:Text)
   )
   > SetIsSelected(self=Get Button_FollowDeveloper(), InSelected=@n1, bGiveClickFeedback=false)
   > InitializeButtonVisuals(self=Get Button_FollowDeveloper())
  ]}
 SetLeavePartyButtonText {nodes [
   SetLeavePartyButtonText() {is=title K2Node_FunctionEntry}
   n1 = Find(
    TargetMap:Map of EAthenaMainMenuLeaveButtonText Enums to Texts=Get Leave Action Text() -> (Leave Action Text=LeaveAction_ReturnToLobby)
    Key=LeaveAction_Default
   ) -> (Value:Text)
   > Branch(Condition=@n1.ReturnValue)
   > SetText(self=Get Button_LeaveParty(), Text=@n1)
  ]}
 SetMobileStyle {nodes [
   SetMobileStyle() {is=title K2Node_FunctionEntry}
   > Branch(Condition=IsUsingCompactUI())
   n1 = Make Margin(Left=24, Top=24, Right=24, Bottom=24)
   > Set MobileButtonPadding!(execute, MobileButtonPadding:Margin Structure=@n1, self:WBP UIKit Button Generic Base Object Reference=Get Button_SaveIsland()) -> (then, Output_Get:Margin Structure)
   > Set MobileButtonPadding(MobileButtonPadding=@n1, self=Get Button_Like())
   > Set MobileButtonPadding(MobileButtonPadding=@n1, self=Get Button_Dislike())
  ]}
 Update Support Buttons Visuals {nodes [
   Update Support Buttons Visuals!() -> (then, isLiked:bool=~, isDisliked:bool=~) {is=title K2Node_FunctionEntry}
   n1 = GetSelected!(self:Common Button Base Object Reference=Get Button_Like()) -> (ReturnValue:bool)
   > n2 = Branch(Condition=AND Boolean(A=Get isLiked!() -> (isLiked), B=NOT Boolean(A=@n1)))
   > ChooseRecommendOrNotInterested(TargetButton=Get Button_Like(), OppositeButton=Get Button_Dislike())
   > SetIsSelected(self=Get Button_Like(), InSelected=true, bGiveClickFeedback=false)
   n3 = GetSelected(self=Get Button_Dislike())
   Branch(execute=@n2.else, Condition=AND Boolean(A=Get isDisliked!() -> (isDisliked), B=NOT Boolean(A=@n3)))
   > ChooseRecommendOrNotInterested(TargetButton=Get Button_Dislike(), OppositeButton=Get Button_Like())
   > SetIsSelected(self=Get Button_Dislike(), InSelected=true, bGiveClickFeedback=false)
  ]}
]
variables [
 Actions {type="array<ESidebarAction>", default []}
 Current Leave Action {type=EAthenaMainMenuLeaveButtonText, default=LeaveAction_ReturnToLobby}
 Disliked Sound {type=SoundBase, default=/Game/Sounds/Fort_UI_Sounds/FrontEnd/MultiProduct/Common/Buttons/MSS_UI_MP_Button_Click_Dislike}
 Favorited Sound {type=SoundBase, default=/Game/Sounds/Fort_UI_Sounds/FrontEnd/MultiProduct/Common/Buttons/MSS_UI_MP_Button_Click_Favorite}
 HordModeLeaveDescription {type=string, default=Your fort design will be saved and it will be used in the Challenge the Horde mission.}
 isNotInterested {type=boolean}
 isRecommended {type=boolean}
 LastFocusedButton {type=UserWidget, default=null}
 Leave Action {default {
   leaveAction_Default=&ConfirmationVM_LeaveParty_C
   leaveAction_InParty=/Sidebar/Panels/Overlay/Data/ConfirmationVM_LeaveWithParty.ConfirmationVM_LeaveWithParty_C
   leaveAction_Mission=/Sidebar/Panels/Overlay/Data/ConfirmationVM_LeaveMission.ConfirmationVM_LeaveMission_C
   leaveAction_PlaylistCreative=/Sidebar/Panels/Overlay/Data/ConfirmationVM_LeaveCreative.ConfirmationVM_LeaveCreative_C
   leaveAction_PlaylistCreative_LTM=/Sidebar/Panels/Overlay/Data/ConfirmationVM_LeaveCreativeLTM.ConfirmationVM_LeaveCreativeLTM_C
   leaveAction_PlaylistDefault=/Sidebar/Panels/Overlay/Data/ConfirmationVM_LeaveMatch.ConfirmationVM_LeaveMatch_C
   leaveAction_Replay=/Sidebar/Panels/Overlay/Data/ConfirmationVM_LeaveReplay.ConfirmationVM_LeaveReplay_C
   leaveAction_ReturnToLobby=&ConfirmationVM_BackToLobby_C
   leaveAction_StormShield=/Sidebar/Panels/Overlay/Data/ConfirmationVM_LeaveStormShield.ConfirmationVM_LeaveStormShield_C
  }}
 Leave Action Class {type=Class, default=null}
 Leave Action Text {default {
   leaveAction_Default=Leave party
   leaveAction_InParty=Leave with party
   leaveAction_Mission=Leave mission
   leaveAction_PlaylistCreative=Leave Creative
   leaveAction_PlaylistCreative_LTM=Leave game
   leaveAction_PlaylistDefault=Leave match
   leaveAction_Replay=Leave replay
   leaveAction_ReturnToLobby=Return to lobby
   leaveAction_StormShield=Leave Storm Shield
  }}
 Liked Sound {type=SoundBase, default=/Game/Sounds/Fort_UI_Sounds/FrontEnd/MultiProduct/Common/Buttons/MSS_UI_MP_Button_Click_Like}
 MissionLeaveDescription {type=string, default="Are you sure you wish to leave Save the World?\r\nYou will forfeit mission rewards and any XP gained."}
 Sidebar {type=WBP_Sidebar_C, default=null}
 SkipMainMenuLeave {type=boolean}
 StormshieldLeaveDescription {type=string, default=Are you sure you wish to leave Save the World?}
 StormshieldLeaveTitle {type=string, default=Leave Storm Shield?}
 Undisliked Sound {type=SoundBase, default=&MSS_UI_MP_Button_Click_Dislike}
 Unfavorited Sound {type=SoundBase, default=/Game/Sounds/Fort_UI_Sounds/FrontEnd/MultiProduct/Common/Buttons/MSS_UI_MP_Button_Click_Unfavorite}
 Unliked Sound {type=SoundBase, default=&MSS_UI_MP_Button_Click_Dislike}
]
widgets {
 info {
  inheritedWidgetCount=0
  namedSlotCount=1
  parentClass=&CommonActivatableWidget
  rootWidgetClass=/Sidebar/Shared/WBP_Sidebar_PanelLayout.WBP_Sidebar_PanelLayout_C
  widgetCount=18
 }
 widgets [
  PanelLayout {
   widgetClassPath=/Sidebar/Shared/WBP_Sidebar_PanelLayout
   bIsVariable=true
   widget=PanelLayout
   properties {headingText=Exit, showBackButton=false}
  }
  VerticalBox {
   widgetClassPath=&VerticalBox
   namedSlotHost=PanelLayout
   widget=VerticalBox
   properties {bOverride_Cursor=true}
   uIComponents [{
     component=&self:UIComponentWidgetBlueprintExtension_0.ComponentContainer.NavigationUIComponent_0
     componentClassPath=/Script/UMG.NavigationUIComponent
    }]
  }
  ActionButtonsContainer {
   widgetClassPath=&VerticalBox
   bIsVariable=true
   parent=VerticalBox
   widget=ActionButtonsContainer
   properties {navigation=&self:WidgetTree.ActionButtonsContainer.WidgetNavigation_0}
   slot {class=VerticalBoxSlot, properties {padding=<ltrb 18 0 18 12>}}
  }
  TopSpacer {
   widgetClassPath=&Spacer
   bIsVariable=true
   parent=ActionButtonsContainer
   widget=TopSpacer
   slot {class=VerticalBoxSlot, properties {padding=<ltrb 0 12 0 0>}}
  }
  Button_Exit {
   widgetClassPath=/Sidebar/Buttons/WBP_Sidebar_ButtonGeneric
   bIsVariable=true
   parent=ActionButtonsContainer
   widget=Button_Exit
   properties {
    buttonType=Exit
    minHeightOverride=60
    text=Close Fortnite
    font {
     bForceMonospaced=false
     bMaterialIsStencil=false
     fontMaterial=/Sidebar/Materials/Fonts/MI_Sidebar_FontColor_Primary
     fontObject=&HeadingNow
     letterSpacing=10
     monospacedWidth=1
     size=15.75
     skewAmount=0
     typefaceFontName=74 Regular
     outlineSettings {
      bApplyOutlineToDropShadows=false
      bMiteredCorners=false
      bSeparateFillAlpha=false
      outlineMaterial=/FortUILibrary/Systems/ModularButton/MaterialInstances/MI_UIKit_TextOutline
      outlineSize=0
      outlineColor=<rgba 0 0 0 1>
     }
    }
    secondFont {
     bForceMonospaced=false
     bMaterialIsStencil=false
     fontMaterial=/Sidebar/Materials/Fonts/MI_Sidebar_FontColor_Secondary
     fontObject=&Burbank
     letterSpacing=48
     monospacedWidth=1
     size=21
     skewAmount=0
     typefaceFontName=BurbankRegular-Black
     outlineSettings {
      bApplyOutlineToDropShadows=false
      bMiteredCorners=false
      bSeparateFillAlpha=false
      outlineMaterial=null
      outlineSize=0
      outlineColor=<rgba 0 0 0 1>
     }
    }
   }
   slot {class=VerticalBoxSlot, properties {padding=<ltrb 0 0 0 12>, size {sizeRule=Fill, value=1}}}
  }
  Button_LeaveAction {
   widgetClassPath=&WBP_Sidebar_ButtonGeneric
   bIsVariable=true
   parent=ActionButtonsContainer
   widget=Button_LeaveAction
   properties {
    minHeightOverride=60
    text=Return to lobby
    visibility=Collapsed
    font {
     bForceMonospaced=false
     bMaterialIsStencil=false
     fontMaterial=&MI_Sidebar_FontColor_Primary
     fontObject=&HeadingNow
     letterSpacing=10
     monospacedWidth=1
     size=15.75
     skewAmount=0
     typefaceFontName=74 Regular
     outlineSettings {
      bApplyOutlineToDropShadows=false
      bMiteredCorners=false
      bSeparateFillAlpha=false
      outlineMaterial=&MI_UIKit_TextOutline
      outlineSize=0
      outlineColor=<rgba 0 0 0 1>
     }
    }
    secondFont {
     bForceMonospaced=false
     bMaterialIsStencil=false
     fontMaterial=&MI_Sidebar_FontColor_Secondary
     fontObject=&Burbank
     letterSpacing=48
     monospacedWidth=1
     size=21
     skewAmount=0
     typefaceFontName=BurbankRegular-Black
     outlineSettings {
      bApplyOutlineToDropShadows=false
      bMiteredCorners=false
      bSeparateFillAlpha=false
      outlineMaterial=null
      outlineSize=0
      outlineColor=<rgba 0 0 0 1>
     }
    }
   }
   slot {class=VerticalBoxSlot, properties {padding=<ltrb 0 0 0 12>, size {sizeRule=Fill, value=1}}}
  }
  Button_LeaveExperience {
   widgetClassPath=&WBP_Sidebar_ButtonGeneric
   bIsVariable=true
   parent=ActionButtonsContainer
   widget=Button_LeaveExperience
   properties {
    minHeightOverride=60
    text=Leave experience
    visibility=Collapsed
    font {
     bForceMonospaced=false
     bMaterialIsStencil=false
     fontMaterial=&MI_Sidebar_FontColor_Primary
     fontObject=&HeadingNow
     letterSpacing=10
     monospacedWidth=1
     size=15.75
     skewAmount=0
     typefaceFontName=74 Regular
     outlineSettings {
      bApplyOutlineToDropShadows=false
      bMiteredCorners=false
      bSeparateFillAlpha=false
      outlineMaterial=&MI_UIKit_TextOutline
      outlineSize=0
      outlineColor=<rgba 0 0 0 1>
     }
    }
    secondFont {
     bForceMonospaced=false
     bMaterialIsStencil=false
     fontMaterial=&MI_Sidebar_FontColor_Secondary
     fontObject=&Burbank
     letterSpacing=48
     monospacedWidth=1
     size=21
     skewAmount=0
     typefaceFontName=BurbankRegular-Black
     outlineSettings {
      bApplyOutlineToDropShadows=false
      bMiteredCorners=false
      bSeparateFillAlpha=false
      outlineMaterial=null
      outlineSize=0
      outlineColor=<rgba 0 0 0 1>
     }
    }
   }
   slot {class=VerticalBoxSlot, properties {padding=<ltrb 0 0 0 12>, size {sizeRule=Fill, value=1}}}
  }
  Button_LeaveParty {
   widgetClassPath=&WBP_Sidebar_ButtonGeneric
   bIsVariable=true
   parent=ActionButtonsContainer
   widget=Button_LeaveParty
   properties {
    minHeightOverride=60
    text="[PH] Leave Epic Party"
    visibility=Collapsed
    font {
     bForceMonospaced=false
     bMaterialIsStencil=false
     fontMaterial=&MI_Sidebar_FontColor_Primary
     fontObject=&HeadingNow
     letterSpacing=10
     monospacedWidth=1
     size=15.75
     skewAmount=0
     typefaceFontName=74 Regular
     outlineSettings {
      bApplyOutlineToDropShadows=false
      bMiteredCorners=false
      bSeparateFillAlpha=false
      outlineMaterial=&MI_UIKit_TextOutline
      outlineSize=0
      outlineColor=<rgba 0 0 0 1>
     }
    }
    secondFont {
     bForceMonospaced=false
     bMaterialIsStencil=false
     fontMaterial=&MI_Sidebar_FontColor_Secondary
     fontObject=&Burbank
     letterSpacing=48
     monospacedWidth=1
     size=21
     skewAmount=0
     typefaceFontName=BurbankRegular-Black
     outlineSettings {
      bApplyOutlineToDropShadows=false
      bMiteredCorners=false
      bSeparateFillAlpha=false
      outlineMaterial=null
      outlineSize=0
      outlineColor=<rgba 0 0 0 1>
     }
    }
   }
   slot {class=VerticalBoxSlot, properties {padding=<ltrb 0 0 0 12>, size {sizeRule=Fill, value=1}}}
  }
  Button_Logout {
   widgetClassPath=&WBP_Sidebar_ButtonGeneric
   bIsVariable=true
   parent=ActionButtonsContainer
   widget=Button_Logout
   properties {
    minHeightOverride=60
    text=Logout
    font {
     bForceMonospaced=false
     bMaterialIsStencil=false
     fontMaterial=&MI_Sidebar_FontColor_Primary
     fontObject=&HeadingNow
     letterSpacing=10
     monospacedWidth=1
     size=15.75
     skewAmount=0
     typefaceFontName=74 Regular
     outlineSettings {
      bApplyOutlineToDropShadows=false
      bMiteredCorners=false
      bSeparateFillAlpha=false
      outlineMaterial=&MI_UIKit_TextOutline
      outlineSize=0
      outlineColor=<rgba 0 0 0 1>
     }
    }
    secondFont {
     bForceMonospaced=false
     bMaterialIsStencil=false
     fontMaterial=&MI_Sidebar_FontColor_Secondary
     fontObject=&Burbank
     letterSpacing=48
     monospacedWidth=1
     size=21
     skewAmount=0
     typefaceFontName=BurbankRegular-Black
     outlineSettings {
      bApplyOutlineToDropShadows=false
      bMiteredCorners=false
      bSeparateFillAlpha=false
      outlineMaterial=null
      outlineSize=0
      outlineColor=<rgba 0 0 0 1>
     }
    }
   }
   slot {class=VerticalBoxSlot, properties {padding=<ltrb 0 0 0 12>, size {sizeRule=Fill, value=1}}}
  }
  Button_Sleep {
   widgetClassPath=&WBP_Sidebar_ButtonGeneric
   bIsVariable=true
   parent=ActionButtonsContainer
   widget=Button_Sleep
   properties {
    minHeightOverride=60
    text=Enter sleep mode
    font {
     bForceMonospaced=false
     bMaterialIsStencil=false
     fontMaterial=&MI_Sidebar_FontColor_Primary
     fontObject=&HeadingNow
     letterSpacing=10
     monospacedWidth=1
     size=15.75
     skewAmount=0
     typefaceFontName=74 Regular
     outlineSettings {
      bApplyOutlineToDropShadows=false
      bMiteredCorners=false
      bSeparateFillAlpha=false
      outlineMaterial=&MI_UIKit_TextOutline
      outlineSize=0
      outlineColor=<rgba 0 0 0 1>
     }
    }
    secondFont {
     bForceMonospaced=false
     bMaterialIsStencil=false
     fontMaterial=&MI_Sidebar_FontColor_Secondary
     fontObject=&Burbank
     letterSpacing=48
     monospacedWidth=1
     size=21
     skewAmount=0
     typefaceFontName=BurbankRegular-Black
     outlineSettings {
      bApplyOutlineToDropShadows=false
      bMiteredCorners=false
      bSeparateFillAlpha=false
      outlineMaterial=null
      outlineSize=0
      outlineColor=<rgba 0 0 0 1>
     }
    }
   }
   slot {class=VerticalBoxSlot, properties {padding=<ltrb 0 0 0 12>, size {sizeRule=Fill, value=1}}}
  }
  WidgetSlottingHostOverlay_BelowActionButtons {
   widgetClassPath=/Script/WidgetSlotting.WidgetSlottingHostOverlay
   parent=ActionButtonsContainer
   widget=WidgetSlottingHostOverlay_BelowActionButtons
   properties {
    slottedWidgetDefaultHorizontalAlignment=HAlign_Fill
    slottedWidgetDefaultVerticalAlignment=VAlign_Bottom
    hostTag {tagName=UI.Sidebar.SubPanel.Exit.BelowActionButtons}
   }
   slot {class=VerticalBoxSlot}
  }
  CurrentActivityText {
   widgetClassPath=&CommonTextBlock
   bIsVariable=true
   parent=VerticalBox
   widget=CurrentActivityText
   properties {
    text=Battle Royale
    font {
     bForceMonospaced=false
     bMaterialIsStencil=false
     fontMaterial=null
     fontObject=&HeadingNow
     letterSpacing=0
     monospacedWidth=1
     size=22.5
     skewAmount=0
     typefaceFontName=75 Medium
     outlineSettings {
      bApplyOutlineToDropShadows=false
      bMiteredCorners=false
      bSeparateFillAlpha=false
      outlineMaterial=null
      outlineSize=0
      outlineColor=<rgba 0 0 0 1>
     }
    }
    shadowOffset=<0 0>
   }
   slot {class=VerticalBoxSlot, properties {padding=<ltrb 24 24 18 12>}}
  }
  ExpressSupportButtons {
   widgetClassPath=&HorizontalBox
   bIsVariable=true
   parent=VerticalBox
   widget=ExpressSupportButtons
   properties {visibility=Visible}
   slot {class=VerticalBoxSlot, properties {padding=<ltrb 18 12 18 0>}}
  }
  Button_Like {
   widgetClassPath=&WBP_Sidebar_ButtonGeneric
   bIsVariable=true
   parent=ExpressSupportButtons
   widget=Button_Like
   properties {
    bShouldSelectUponReceivingFocus=false
    bToggleable=true
    icon=/FortUILibrary/Textures/Icons/T_UI_Icon_UIKit_Like_MSDF
    iconMaterial=/Sidebar/Materials/Icons/MI_Sidebar_LikeIcon
    minHeightOverride=105
    mobileFontSizeMultiplierOverride=1.5
    overrideDefaultMobileFontSizeMultiplier=true
    text=Like
    textBlockAlignment=Bottom
    useIconBlock=true
    visibility=Visible
    buttonPadding=<ltrb 12 12 12 12>
    font {
     bForceMonospaced=false
     bMaterialIsStencil=false
     fontMaterial=&MI_Sidebar_FontColor_Primary
     fontObject=&HeadingNow
     letterSpacing=10
     monospacedWidth=1
     size=12
     skewAmount=0
     typefaceFontName=74 Regular
     outlineSettings {
      bApplyOutlineToDropShadows=false
      bMiteredCorners=false
      bSeparateFillAlpha=false
      outlineMaterial=&MI_UIKit_TextOutline
      outlineSize=0
      outlineColor=<rgba 0 0 0 1>
     }
    }
    iconPadding=<ltrb 0 0 0 6>
    iconSize=<36 36>
    secondFont {
     bForceMonospaced=false
     bMaterialIsStencil=false
     fontMaterial=&MI_Sidebar_FontColor_Secondary
     fontObject=&Burbank
     letterSpacing=48
     monospacedWidth=1
     size=21
     skewAmount=0
     typefaceFontName=BurbankRegular-Black
     outlineSettings {
      bApplyOutlineToDropShadows=false
      bMiteredCorners=false
      bSeparateFillAlpha=false
      outlineMaterial=null
      outlineSize=0
      outlineColor=<rgba 0 0 0 1>
     }
    }
   }
   slot {class=HorizontalBoxSlot, properties {padding=<ltrb 0 0 6 0>, size {sizeRule=Fill, value=1}}}
  }
  Button_Dislike {
   widgetClassPath=&WBP_Sidebar_ButtonGeneric
   bIsVariable=true
   parent=ExpressSupportButtons
   widget=Button_Dislike
   properties {
    bShouldSelectUponReceivingFocus=false
    bToggleable=true
    icon=/FortUILibrary/Textures/Icons/T_UI_Icon_UIKit_Dislike_MSDF
    iconMaterial=&MI_Sidebar_LikeIcon
    minHeightOverride=105
    mobileFontSizeMultiplierOverride=1.5
    overrideDefaultMobileFontSizeMultiplier=true
    text=Dislike
    textBlockAlignment=Bottom
    useIconBlock=true
    visibility=Visible
    buttonPadding=<ltrb 12 12 12 12>
    font {
     bForceMonospaced=false
     bMaterialIsStencil=false
     fontMaterial=&MI_Sidebar_FontColor_Primary
     fontObject=&HeadingNow
     letterSpacing=10
     monospacedWidth=1
     size=12
     skewAmount=0
     typefaceFontName=74 Regular
     outlineSettings {
      bApplyOutlineToDropShadows=false
      bMiteredCorners=false
      bSeparateFillAlpha=false
      outlineMaterial=&MI_UIKit_TextOutline
      outlineSize=0
      outlineColor=<rgba 0 0 0 1>
     }
    }
    iconPadding=<ltrb 0 0 0 6>
    iconSize=<36 36>
    secondFont {
     bForceMonospaced=false
     bMaterialIsStencil=false
     fontMaterial=&MI_Sidebar_FontColor_Secondary
     fontObject=&Burbank
     letterSpacing=48
     monospacedWidth=1
     size=21
     skewAmount=0
     typefaceFontName=BurbankRegular-Black
     outlineSettings {
      bApplyOutlineToDropShadows=false
      bMiteredCorners=false
      bSeparateFillAlpha=false
      outlineMaterial=null
      outlineSize=0
      outlineColor=<rgba 0 0 0 1>
     }
    }
   }
   slot {class=HorizontalBoxSlot, properties {padding=<ltrb 6 0 0 0>, size {sizeRule=Fill, value=1}}}
  }
  ExpressSupportButtons_Row2 {
   widgetClassPath=&HorizontalBox
   parent=VerticalBox
   widget=ExpressSupportButtons_Row2
   slot {class=VerticalBoxSlot, properties {padding=<ltrb 18 12 18 0>}}
  }
  Button_FollowDeveloper {
   widgetClassPath=&WBP_Sidebar_ButtonGeneric
   bIsVariable=true
   parent=ExpressSupportButtons_Row2
   widget=Button_FollowDeveloper
   properties {
    bShouldSelectUponReceivingFocus=false
    bToggleable=true
    bTriggerClickedAfterSelection=true
    icon=/FortUILibrary/Textures/Icons/T_UI_Icon_UIKit_Add_MSDF
    minHeightOverride=105
    mobileFontSizeMultiplierOverride=1.5
    overrideDefaultMobileFontSizeMultiplier=true
    text=Follow Developer
    textBlockAlignment=Bottom
    useIconBlock=true
    visibility=Collapsed
    buttonPadding=<ltrb 12 12 12 12>
    font {
     bForceMonospaced=false
     bMaterialIsStencil=false
     fontMaterial=&MI_Sidebar_FontColor_Secondary
     fontObject=&HeadingNow
     letterSpacing=10
     monospacedWidth=1
     size=12
     skewAmount=0
     typefaceFontName=74 Regular
     outlineSettings {
      bApplyOutlineToDropShadows=false
      bMiteredCorners=false
      bSeparateFillAlpha=false
      outlineMaterial=&MI_UIKit_TextOutline
      outlineSize=0
      outlineColor=<rgba 0 0 0 1>
     }
    }
    iconPadding=<ltrb 0 0 0 6>
    iconSize=<36 36>
    secondFont {
     bForceMonospaced=false
     bMaterialIsStencil=false
     fontMaterial=&MI_Sidebar_FontColor_Secondary
     fontObject=&Burbank
     letterSpacing=48
     monospacedWidth=1
     size=21
     skewAmount=0
     typefaceFontName=BurbankRegular-Black
     outlineSettings {
      bApplyOutlineToDropShadows=false
      bMiteredCorners=false
      bSeparateFillAlpha=false
      outlineMaterial=null
      outlineSize=0
      outlineColor=<rgba 0 0 0 1>
     }
    }
   }
   slot {class=HorizontalBoxSlot, properties {padding=<ltrb 0 0 6 0>, size {sizeRule=Fill, value=1}}}
  }
  Button_SaveIsland {
   widgetClassPath=&WBP_Sidebar_ButtonGeneric
   bIsVariable=true
   parent=ExpressSupportButtons_Row2
   widget=Button_SaveIsland
   properties {
    bShouldSelectUponReceivingFocus=false
    bToggleable=true
    bTriggerClickedAfterSelection=true
    icon=&T_UI_Icon_UIKit_FavoriteOutline_MSDF
    iconMaterial=/Sidebar/Materials/Icons/MI_Sidebar_FavoriteIcon
    minHeightOverride=105
    mobileFontSizeMultiplierOverride=1.5
    overrideDefaultMobileFontSizeMultiplier=true
    text=Save the island
    textBlockAlignment=Bottom
    useIconBlock=true
    visibility=Visible
    buttonPadding=<ltrb 12 12 12 12>
    font {
     bForceMonospaced=false
     bMaterialIsStencil=false
     fontMaterial=&MI_Sidebar_FontColor_Primary
     fontObject=&HeadingNow
     letterSpacing=10
     monospacedWidth=1
     size=12
     skewAmount=0
     typefaceFontName=74 Regular
     outlineSettings {
      bApplyOutlineToDropShadows=false
      bMiteredCorners=false
      bSeparateFillAlpha=false
      outlineMaterial=&MI_UIKit_TextOutline
      outlineSize=0
      outlineColor=<rgba 0 0 0 1>
     }
    }
    iconPadding=<ltrb 0 0 0 6>
    iconSize=<36 36>
    secondFont {
     bForceMonospaced=false
     bMaterialIsStencil=false
     fontMaterial=&MI_Sidebar_FontColor_Secondary
     fontObject=&Burbank
     letterSpacing=48
     monospacedWidth=1
     size=21
     skewAmount=0
     typefaceFontName=BurbankRegular-Black
     outlineSettings {
      bApplyOutlineToDropShadows=false
      bMiteredCorners=false
      bSeparateFillAlpha=false
      outlineMaterial=null
      outlineSize=0
      outlineColor=<rgba 0 0 0 1>
     }
    }
   }
   slot {class=HorizontalBoxSlot, properties {padding=<ltrb 6 0 0 0>, size {sizeRule=Fill, value=1}}}
  }
 ]
}
