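# Sidebar "Friend requests" panel widget, an activatable screen showing an info block and a social tree list of friend invites.
# It adds a FriendInvites header to the player search view model and passes header arrays to the list.
# Item clicks open player profiles; block requests store the interaction, open a confirmation panel and execute it on confirmation.
# It also toggles a social-ban message and adjusts padding for compact UI.
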WidgetBlueprint /Sidebar/Panels/WBP_FriendRequests_Panel
graphs [
 BP_GetDesiredFocusTarget {nodes [
   Get Desired Focus Target()
   > Return Node!(
    execute
    ReturnValue:Widget Object Reference=Get WBP_Sidebar_SocialTreeList() -> (WBP_Sidebar_SocialTreeList:WBP Sidebar Social Tree List Object Reference)
   )
  ]}
 EventGraph {nodes [
   n1 = "Item Clicked (WBP_Sidebar_SocialTreeList)"!() -> (OutputDelegate, then, ItemVM:Fort Social Item VM Object Reference) {is=title K2Node_ComponentBoundEvent}
   > NavigateToPlayerProfile!(
    execute
    SelfReference:Object Reference=Self-Reference()
    SocialItemVM:Fort Social Item VM Object Reference=@n1.ItemVM
    SidebarNavigationVM:Fort Sidebar Navigation VM Object Reference=Get FortSidebarNavigationVM() -> (FortSidebarNavigationVM:Fort Sidebar Navigation VM Object Reference)
    Sidebar VM:Fort Sidebar VM Object Reference
   ) -> (then)
   n2 = "Item Block Player (WBP_Sidebar_SocialTreeList)"!() -> (OutputDelegate, then, SocialUser:Fort Social User Object Reference, Interaction:Fort Social Interaction VM Object Reference) {is=title K2Node_ComponentBoundEvent}
   > Set BlockPlayerInteraction(BlockPlayerInteraction:Fort Social Interaction VM Object Reference=@n2.Interaction) -> (Output_Get:Fort Social Interaction VM Object Reference)
   > Navigate to Block Player!(exec, SocialUser:Fort Social User Object Reference=@n2.SocialUser) -> (then:Exec) {is=title K2Node_MacroInstance}
   Event PreConstruct()
   > SetMobileStyle() {is=title K2Node_CallFunction}
   > SetBackplateColor!(
    execute
    self:WBP Sidebar Block Backplate Object Reference=Get FriendRequest_Backplate() -> (FriendRequest_Backplate:WBP Sidebar Block Backplate Object Reference)
    Backplate Switch:E_UI_Sidebar_Button_Backplate_Color Enum=NewEnumerator8
   ) -> (then)
   Event Construct()
   > Populate Social Headers()
   Event On Activated()
   > SidebarFeatureFinishedLoading(self:Fort Sidebar VM Object Reference=Get FortSidebarVM() -> (FortSidebarVM:Fort Sidebar VM Object Reference))
  ]}
 HandleConfirmationOverlayActions {nodes [
   n1 = HandleConfirmationOverlayActions!() -> (then, ActionID:Name) {is=title K2Node_FunctionEntry}
   > Allow Handle Overlay Confirmation!(Exec:Exec, Action ID:Name=@n1) -> (then:Exec) {is=title K2Node_MacroInstance}
   > Branch(Condition="Equal (Name)"(A=@n1, B=Block))
   > ExecuteInteraction(self:Fort Social Interaction VM Object Reference=Get BlockPlayerInteraction() -> (BlockPlayerInteraction:Fort Social Interaction VM Object Reference))
  ]}
 NavigateToBlockPlayer {nodes [
   n1 = Inputs!() -> (exec:Exec, SocialUser:Fort Social User Object Reference)
   > n2 = Construct Object from Class(Class=/Sidebar/Panels/Overlay/Data/ConfirmationVM_BlockPlayer.ConfirmationVM_BlockPlayer_C) -> (ReturnValue:Confirmation VM Block Player Object Reference)
   > Set Description!(
    execute
    Description:Text=Format Text!(
     Format="LOCTEXT(\"Would you like to block {PlayerName}?\")"
     PlayerName:String=GetNicknameOrDisplayName!(self:Fort Social User Object Reference=@n1) -> (ReturnValue:String)
    ) -> (Result)
    self:Sidebar Confirmation Panel VM Object Reference=@n2
   ) -> (then, Output_Get:Text)
   > NavigateToConfirmPanel!(
    execute
    SelfReference:Object Reference=Self-Reference()
    SidebarConfirmationVM:Sidebar Confirmation Panel VM Object Reference=@n2
    SidebarNavigationVM:Fort Sidebar Navigation VM Object Reference=Get FortSidebarNavigationVM()
   ) -> (then)
   > Outputs!(then:Exec)
  ]}
 Populate Social Headers {nodes [
   Populate Social Headers() {is=title K2Node_FunctionEntry}
   > AddHeader!(
    execute
    self:Fort Social Header List VM Object Reference=Get FortSocialPlayerSearchVM() -> (FortSocialPlayerSearchVM:Fort Social Player Search VM Object Reference)
    HeaderType:ESocialHeaderItemType Enum=FriendInvites
   ) -> (then, ReturnValue:Fort Social Header Item VM Object Reference)
  ]}
 Search Results {nodes [
   n1 = Search Results!() -> (then, OuterArray:Array of Fort Social Header Item VM Object References) {is=title K2Node_FunctionEntry}
   > SetTreeViewItems!(
    execute
    self:WBP Sidebar Social Tree List Object Reference=Get WBP_Sidebar_SocialTreeList()
    OuterArray:Array of Fort Social Header Item VM Object References=@n1
   ) -> (then)
  ]}
 SetMobileStyle {nodes [
   SetMobileStyle() {is=title K2Node_FunctionEntry}
   > Branch(Condition=IsUsingCompactUI())
   > Set InputFieldHeight(InputFieldHeight:double=80) -> (Output_Get:double)
   n1 = SlotAsOverlaySlot(Widget=Get WBP_Sidebar_SocialTreeList())
   n2 = Get!(self:Overlay Slot Object Reference=@n1) -> (Padding_Left:float=~, Padding_Top:float=~, Padding_Right:float=~, Padding_Bottom:float=~)
   > SetPadding!(
    execute
    self:Overlay Slot Object Reference=@n1
    InPadding_Left:float=@n2
    InPadding_Top:float=20
    InPadding_Right:float=@n2.Padding_Right
    InPadding_Bottom:float=@n2.Padding_Bottom
   ) -> (then)
   n3 = SlotAsGridSlot()
   n4 = Get(self:Grid Slot Object Reference=@n3) -> (Padding_Left=~, Padding_Top=~, Padding_Right=~, Padding_Bottom=~)
   > SetPadding(
    self:Grid Slot Object Reference=@n3
    InPadding_Left=@n4
    InPadding_Top=24
    InPadding_Right=@n4.Padding_Right
    InPadding_Bottom=@n4.Padding_Bottom
   )
  ]}
 SetSocialBanVisibility {nodes [
   SetSocialBanVisibility!() -> (then, IsSocialBanned:bool=~) {is=title K2Node_FunctionEntry}
   > SetVisibility(InVisibility=Select(Option 0=SelfHitTestInvisible, Option 1=Collapsed, Index=Get IsSocialBanned() -> (IsSocialBanned:bool)))
   > SetVisibility(
    self=Get WBP_Sidebar_SocialBan_Message() -> (WBP_Sidebar_SocialBan_Message:WBP Social Ban Message Object Reference)
    InVisibility=Select(Option 0=Collapsed, Option 1=SelfHitTestInvisible, Index=Get IsSocialBanned())
   )
   > RequestRefreshFocus(self:Common Activatable Widget Object Reference)
  ]}
]
variables [
 BlockPlayerInteraction {type=FortSocialInteractionVM, default=null}
 InputFieldHeight {type=number, default=60}
 Social Headers {type="array<ESocialHeaderItemType>", default [FriendInvites]}
]
widgets {
 info {
  inheritedWidgetCount=0
  namedSlotCount=1
  parentClass=&CommonActivatableWidget
  rootWidgetClass=/Sidebar/Shared/WBP_Sidebar_PanelLayout.WBP_Sidebar_PanelLayout_C
  widgetCount=9
 }
 widgets [
  WBP_Sidebar_PanelLayout {
   widgetClassPath=/Sidebar/Shared/WBP_Sidebar_PanelLayout
   bIsVariable=true
   widget=WBP_Sidebar_PanelLayout
   properties {headingText=Friend requests}
  }
  AddFriends_Grid_Root {
   widgetClassPath=&GridPanel
   namedSlotHost=WBP_Sidebar_PanelLayout
   widget=AddFriends_Grid_Root
   properties {columnFill [1], rowFill [0,0,1]}
  }
  WBP_Sidebar_SocialBan_Message {
   widgetClassPath=/FortniteCoreUI/UI/Social/SocialBan/WBP_SocialBan_Message
   bIsVariable=true
   parent=AddFriends_Grid_Root
   widget=WBP_Sidebar_SocialBan_Message
   properties {useFriendsSpecificMessage=true}
   slot {class=GridSlot, properties {layer=1, padding=<ltrb 18 8 18 0>}}
  }
  O_FriendRequest {
   widgetClassPath=&Overlay
   bIsVariable=true
   parent=AddFriends_Grid_Root
   widget=O_FriendRequest
   slot {class=GridSlot, properties {row=1, padding=<ltrb 18 8 18 4>}}
  }
  FriendRequest_Backplate {
   widgetClassPath=/Sidebar/Buttons/Blocks/WBP_Sidebar_Block_Backplate
   bIsVariable=true
   parent=O_FriendRequest
   widget=FriendRequest_Backplate
   slot {class=OverlaySlot, properties {horizontalAlignment=HAlign_Fill, verticalAlignment=VAlign_Fill}}
  }
  VB_FriendRequest {
   widgetClassPath=&VerticalBox
   parent=O_FriendRequest
   widget=VB_FriendRequest
   slot {class=OverlaySlot, properties {horizontalAlignment=HAlign_Fill, padding=<ltrb 18 18 18 18>}}
  }
  Text_FriendRequest_Body {
   widgetClassPath=&CommonTextBlock
   bIsVariable=true
   parent=VB_FriendRequest
   widget=Text_FriendRequest_Body
   properties {
    autoWrapText=true
    bAutoCollapseWithEmptyText=true
    bIsScrollingEnabled=false
    lineHeightPercentage=1.1
    mobileFontSizeMultiplier=1.41
    scrollStyle=/FortUILibrary/Styles/Typography/Foundation/TextScrollStyle-NoFade.TextScrollStyle-NoFade_C
    text="Friends can see your online status, form parties together. and chat with you depending on your communication settings. You can report players from their profile."
    textOverflowPolicy=Ellipsis
    colorAndOpacity {colorUseRule=UseColor_Specified, specifiedColor=<rgba 1 1 1 0.5>}
    font {
     bForceMonospaced=false
     bMaterialIsStencil=false
     fontMaterial=/Sidebar/Materials/Fonts/MI_Sidebar_FontColor_Secondary
     fontObject=&HeadingNow
     letterSpacing=0
     monospacedWidth=1
     size=12.75
     skewAmount=0
     typefaceFontName=74 Regular
     outlineSettings {
      bApplyOutlineToDropShadows=false
      bMiteredCorners=false
      bSeparateFillAlpha=false
      outlineMaterial=/FortUILibrary/Systems/ModularButton/MaterialInstances/MI_UIKit_TextOutline
      outlineSize=0
      outlineColor=<rgba 0 0 0 1>
     }
    }
    shadowOffset=<0 0>
   }
   slot {class=VerticalBoxSlot}
  }
  Border_Clipping {
   widgetClassPath=&Border
   parent=AddFriends_Grid_Root
   widget=Border_Clipping
   properties {
    clipping=ClipToBounds
    visibility=SelfHitTestInvisible
    background {
     bIsDynamicallyLoaded=false
     drawAs=NoDrawType
     imageType=NoImage
     mirroring=NoMirror
     resourceName=None
     resourceObject=null
     tiling=NoTile
     imageSize=<32 32>
     margin=<ltrb 0 0 0 0>
     outlineSettings {
      bUseBrushTransparency=false
      roundingType=HalfHeightRadius
      width=0
      color {colorUseRule=UseColor_Specified, specifiedColor=<rgba 0 0 0 0>}
      cornerRadii=<0 0 0 0>
     }
     tintColor {colorUseRule=UseColor_Specified, specifiedColor=<rgba 1 1 1 1>}
     uVRegion {bIsValid=false, max=<0 0>, min=<0 0>}
    }
    padding=<ltrb 18 12 18 0>
   }
   slot {class=GridSlot, properties {row=2, padding=<ltrb 0 8 0 0>}}
  }
  WBP_Sidebar_SocialTreeList {
   widgetClassPath=/Sidebar/Panels/Social/List/WBP_Sidebar_SocialTreeList
   bIsVariable=true
   parent=Border_Clipping
   widget=WBP_Sidebar_SocialTreeList
   properties {bIsFocusable=true}
   slot {class=BorderSlot, properties {padding=<ltrb 18 12 18 0>}}
  }
 ]
}
