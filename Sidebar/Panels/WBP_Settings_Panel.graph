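# WBP_Settings_Panel is the Fortnite sidebar's settings menu, with buttons for settings, help, legal, parental controls, respawn, team select, island and minigame start or end, and moderator tools.
# Update functions set button visibility from game state, minigame, input type, runtime options and sidebar actions.
# Clicks run view-model actions, push navigation points or async-load widgets as panels, often closing the sidebar.
# Registry buttons are built at runtime and matched to actions by display text.
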WidgetBlueprint /Sidebar/Panels/WBP_Settings_Panel
graphs [
 BP_GetDesiredFocusTarget {nodes [
   Get Desired Focus Target()
   n1 = Get Button_StartGame() -> (Button_StartGame:WBP UIKit Button Generic Object Reference)
   > n2 = Branch(Condition=IsVisible(self=@n1))
   > Return Node!(execute, ReturnValue:Widget Object Reference=@n1)
   n3 = Get!(execute:Exec=@n2.else) -> (then:Exec, else:Exec, LastFocusedButton:User Widget Object Reference)
   > n4 = Branch(Condition=IsVisible(self=@n3))
   > Return Node(ReturnValue=Get LastFocusedButton() -> (LastFocusedButton:User Widget Object Reference))
   n5 = For Each Loop(
    Exec=[@n4.else, @n3.else]
    Array:Array of Widget Object References=GetAllChildren(self=Get VBox_Menu() -> (VBox_Menu:Vertical Box Object Reference))
   ) -> (Array Element:Widget Object Reference)
   > n6 = Cast To CommonButtonBase!(execute, Object=@n5) -> (then, CastFailed, AsCommon Button Base:Common Button Base Object Reference) {is=title K2Node_DynamicCast}
   > Branch(Condition=IsVisible(self=@n6))
   > Return Node(ReturnValue=@n5)
   Return Node(execute=@n5.Completed, ReturnValue=Get Button_SupportAndRefunds() -> (Button_SupportAndRefunds:WBP Sidebar Button Generic Object Reference))
  ]}
 Creative_Get Start Game Button State {nodes [
   Creative_Get Start Game Button State() {is=title K2Node_FunctionEntry}
   > SetVisibility(self=Get Button_StartGame(), InVisibility=Collapsed)
   n1 = GetMinigame!(ContextActor:Actor Object Reference=GetOwningPlayer()) -> (ReturnValue:Fort Minigame Object Reference)
   > Is Valid(InputObject=@n1)
   > Bind Event to State Changed Delegate(self:Fort Minigame Object Reference=@n1, Delegate=Create Event()) {is=title K2Node_AddDelegate}
   > Set Creative_Machine(Creative_Machine:Fort Minigame Settings Building Object Reference=GetMinigameSettingsMachine!(ContextActor:Actor Object Reference=@n1) -> (ReturnValue:Fort Minigame Settings Building Object Reference)) -> (Output_Get:Fort Minigame Settings Building Object Reference)
   > n2 = Cast To FortPlayerControllerAthena(Object=GetOwningPlayer())
   > Is Valid(InputObject=@n2)
   n3 = GetMinigame!(self:Fort Player Controller Gameplay Object Reference=@n2) -> (ReturnValue)
   > Is Valid(InputObject=@n3)
   > n4 = Branch(Condition=IsCuratedHub!(self:Fort Minigame Object Reference=@n3) -> (ReturnValue:bool))
   n5 = Branch(execute=@n4.else, Condition=NOT Boolean(A=CanStartMinigame!(self:Fort Minigame Object Reference=@n3) -> (ReturnValue:bool)))
   SetVisibility(execute=[@n7.NewEnumerator0, @n7.NewEnumerator3, @n7.NewEnumerator6, @n5.then], self=Get Button_StartGame(), InVisibility=Collapsed)
   n6 = Set GameButtonState(
    execute=@n5.else
    GameButtonState:MinigameButtonsState Enum=Select!(
     PreGame:MinigameButtonsState Enum=NewEnumerator2
     Setup:MinigameButtonsState Enum=NewEnumerator6
     Transitioning:MinigameButtonsState Enum=NewEnumerator6
     WaitingForCameras:MinigameButtonsState Enum=NewEnumerator6
     Warmup:MinigameButtonsState Enum=NewEnumerator1
     InProgress:MinigameButtonsState Enum=NewEnumerator1
     PostGameTimeDilation:MinigameButtonsState Enum=NewEnumerator6
     PostRoundEnd:MinigameButtonsState Enum=NewEnumerator6
     PostGameEnd:MinigameButtonsState Enum=NewEnumerator3
     PostGameAbandon:MinigameButtonsState Enum=NewEnumerator3
     PostGameReset:MinigameButtonsState Enum=NewEnumerator6
     Teardown:MinigameButtonsState Enum=NewEnumerator6
     Index:EFortMinigameState Enum=Get CurrentState!(self:Fort Minigame Object Reference=@n3) -> (CurrentState:EFortMinigameState Enum=PreGame)
    ) -> (ReturnValue:MinigameButtonsState Enum)
   ) -> (Output_Get:MinigameButtonsState Enum=NewEnumerator6)
   > n7 = Switch on MinigameButtonsState!(execute, Selection:MinigameButtonsState Enum=@n6) -> (
    NewEnumerator6:Exec
    NewEnumerator0:Exec
    NewEnumerator1:Exec
    NewEnumerator2:Exec
    NewEnumerator3:Exec
    NewEnumerator4:Exec
    NewEnumerator5:Exec
   ) {is=title K2Node_SwitchEnum}
   n8 = Get Button_StartGame()
   SetText!(execute=@n7.NewEnumerator1, self:WBP UIKit Button Generic Base Object Reference=@n8, Text:Text="LOCTEXT(\"End Game\")") -> (then)
   SetText(execute=@n7.NewEnumerator2, self=@n8, Text="LOCTEXT(\"Start Game\")")
  ]}
 Creative_OnMinigameStateChanged {nodes [
   Creative_OnMinigameStateChanged!() -> (then, Minigame:Fort Minigame Object Reference, MinigameState:EFortMinigameState Enum=PreGame) {is=title K2Node_FunctionEntry}
   > Creative_Get Start Game Button State() {is=title K2Node_CallFunction}
   > Update Minigame Visibility()
  ]}
 Creative_isCreativeOptionsNotEmpty {nodes [
   Creative_isCreativeOptionsNotEmpty() {is=title K2Node_FunctionEntry}
   > n1 = For Each Loop with Break!(
    Exec:Exec
    Array:Array of Widget Object References=GetAllChildren(self=Get Vbox_CreativeOptions() -> (Vbox_CreativeOptions:Vertical Box Object Reference))
    Break:Exec=@n3.then
   ) -> (LoopBody:Exec, Array Element:Widget Object Reference, Array Index:int, Completed:Exec)
   Return Node!(execute=@n1.Completed, isCreativeOptionsNotEmpty:bool=Get Has Visible Button() -> (Has Visible Button:bool))
   > n2 = Branch(Condition="Not Equal (Enum)"!(
     A:ESlateVisibility Enum=Get Visibility!(self:Widget Object Reference=@n1) -> (Visibility:ESlateVisibility Enum=Visible)
     B:ESlateVisibility Enum=Collapsed
    ) -> (ReturnValue:bool))
   > n3 = Set Has Visible Button(Has Visible Button:bool=true) -> (Output_Get:bool)
   Set Has Visible Button(execute=@n2.else, Has Visible Button=false)
  ]}
 EventGraph {nodes [
   "On Clicked (Button_StartGame)"!() -> (OutputDelegate, then, Button:Common Button Base Object Reference) {is=title K2Node_ComponentBoundEvent}
   > n1 = Switch on MinigameButtonsState(Selection=Get GameButtonState() -> (GameButtonState:MinigameButtonsState Enum=NewEnumerator6))
   n2 = Get Creative_Machine() -> (Creative_Machine:Fort Minigame Settings Building Object Reference)
   Is Valid(exec=@n1.NewEnumerator1, InputObject=@n2)
   > n3 = AbandonMinigame_BP(self:Fort Minigame Settings Building Object Reference=@n2)
   RequestCloseSidebar(execute=[@n5.then, @n3.then], self:Fort Sidebar VM Object Reference=Get FortSidebarVM() -> (FortSidebarVM:Fort Sidebar VM Object Reference))
   n4 = Get Creative_Machine()
   Is Valid(exec=@n1.NewEnumerator2, InputObject=@n4)
   > n5 = StartMinigame_BP(self:Fort Minigame Settings Building Object Reference=@n4)
   "On Clicked (Button_BackToHub)"!() -> (OutputDelegate, then, Button:Common Button Base Object Reference) {is=title K2Node_ComponentBoundEvent}
   > ExecuteAction!(
    execute
    self:Fort Sidebar Settings VM Object Reference=Get FortSidebarSettingsVM() -> (FortSidebarSettingsVM:Fort Sidebar Settings VM Object Reference)
    SidebarAction:ESidebarAction Enum=BackToHub
   ) -> (then)
   > RequestCloseSidebar(self=Get FortSidebarVM())
   "On Clicked (Button_Settings)"!() -> (OutputDelegate, then, Button:Common Button Base Object Reference) {is=title K2Node_ComponentBoundEvent}
   > RequestCloseSidebar(self=Get FortSidebarVM())
   > RequestOpenSettings!(
    execute
    self:Fort Settings Context Object Reference=Get FortSettingsContext!() -> (ReturnValue:Fort Settings Context Object Reference) {is=title K2Node_GetBlueprintContextEx}
    ContextWidget:Common User Widget Object Reference=Self-Reference()
    SettingsTabId:Name=None
    SettingToSelect:String
   ) -> (then)
   "On Clicked (Button_TimeLimits)"!() -> (OutputDelegate, then, Button:Common Button Base Object Reference) {is=title K2Node_ComponentBoundEvent}
   > n6 = Construct Object from Class!(
    execute
    Class=/Sidebar/Shared/Navigation/BP_NavigationPoint.BP_NavigationPoint_C
    self
    SocialUser:Fort Social User Object Reference
    ConversationId:Name=None
    ConfirmationVM:Sidebar Confirmation Panel VM Object Reference
    ReservedSpots:int=0
    SocialItemVM:Fort Social Item VM Object Reference
    NavigationTag:Gameplay Tag Structure=(TagName="UI.StateTree.TransitionTag.Sidebar.Settings.TimeLimits")
   ) -> (then, ReturnValue:BP Navigation Point Object Reference)
   > PushNavigation!(
    execute
    self:Fort Sidebar Navigation VM Object Reference=Get FortSidebarNavigationVM() -> (FortSidebarNavigationVM:Fort Sidebar Navigation VM Object Reference)
    NavigationPoint:Fort Sidebar Navigation Point Object Reference=@n6
   ) -> (then)
   "On Clicked (Button_ControlsHelp)"!() -> (OutputDelegate, then, Button:Common Button Base Object Reference) {is=title K2Node_ComponentBoundEvent}
   n7 = GetSoftWidgetClassForAlias!(
    self:Fort UIData Configuration Context Object Reference=Get FortUIDataConfigurationContext!() -> (ReturnValue:Fort UIData Configuration Context Object Reference) {is=title K2Node_GetBlueprintContextEx}
    WidgetAlias:"Gameplay Tag Structure (by ref)"=MakeLiteralGameplayTag!(Value:Gameplay Tag Structure=(TagName="MobileHUD.TouchControlsHelp")) -> (ReturnValue:Gameplay Tag Structure)
   ) -> (ReturnValue:User Widget Soft Class Reference)
   > n8 = CreateWidgetAsync!(
    execute:Exec
    UserWidgetSoftClass:User Widget Soft Class Reference=Select!(
     Option 0:User Widget Soft Class Reference=Get MobileControlsHelpClass() -> (MobileControlsHelpClass:User Widget Soft Class Reference)
     Option 1:User Widget Soft Class Reference=@n7
     Index:bool=IsValidSoftClassReference!(SoftClassReference:"Object Soft Class Reference (by ref)"=@n7) -> (ReturnValue:bool)
    ) -> (ReturnValue:User Widget Soft Class Reference)
    OwningPlayer:Player Controller Object Reference=GetOwningPlayer()
    bSuspendInputUntilComplete:bool=true
   ) -> (then:Exec, OnComplete:Exec, OnFailure:Exec, UserWidget:User Widget Object Reference) {is=title K2Node_AsyncAction}
   TryPushWidgetAsModalPanel!(execute=@n8.OnComplete, PushWidget:User Widget Object Reference=@n8) -> (then)
   > RequestCloseSidebar(self=Get FortSidebarVM())
   "On Clicked (Button_Cheats)"!() -> (OutputDelegate, then, Button:Common Button Base Object Reference) {is=title K2Node_ComponentBoundEvent}
   > RequestCloseSidebar(self=Get FortSidebarVM())
   > n9 = CreateWidgetAsync(
    UserWidgetSoftClass=/FortniteCoreUI/Settings/SettingsScreen_GameCheats.SettingsScreen_GameCheats_C
    OwningPlayer=GetOwningPlayer()
    bSuspendInputUntilComplete=true
   )
   TryPushWidgetAsActivatablePanelForPlayer!(
    execute=@n9.OnComplete
    OwningPlayer:Player Controller Object Reference=GetOwningPlayer()
    PushWidget:User Widget Object Reference=@n9
    Header:EActivatePanelOption Enum=PlatformDefault
    Footer:EActivatePanelOption Enum=Show
    ChatWidget:EActivatePanelOption Enum=PlatformDefault
   ) -> (then)
   "On Clicked (Button_Locker)"!() -> (OutputDelegate, then, Button:Common Button Base Object Reference) {is=title K2Node_ComponentBoundEvent}
   > RequestCloseSidebar(self=Get FortSidebarVM())
   > ExecuteAction(self=Get FortSidebarSettingsVM(), SidebarAction=Locker)
   "On Clicked (Button_HudLayoutTool)"!() -> (OutputDelegate, then, Button:Common Button Base Object Reference) {is=title K2Node_ComponentBoundEvent}
   n10 = Get FortUIDataConfigurationContext()
   n11 = GetSoftWidgetClassForAlias(self=@n10, WidgetAlias=MakeLiteralGameplayTag(Value=(TagName="MobileHUD.HUDLayoutTool")))
   n12 = Select(
    Option 0=GetHUDLayoutToolV2SoftClass!(self:Fort UIData Configuration Context Object Reference=@n10) -> (ReturnValue:Fort Mobile HUDLayout Tool V2 Soft Class Reference)
    Option 1=@n11
    Index=IsValidSoftClassReference(SoftClassReference=@n11)
   )
   > Branch(Condition=IsValidSoftClassReference(SoftClassReference=@n12))
   > n13 = CreateWidgetAsync(UserWidgetSoftClass=@n12, OwningPlayer=GetOwningPlayer(), bSuspendInputUntilComplete=true)
   TryPushWidgetAsActivatablePanel!(
    execute=@n13.OnComplete
    PushWidget:User Widget Object Reference=@n13
    Header:EActivatePanelOption Enum=Show
    Footer:EActivatePanelOption Enum=Show
    ChatWidget:EActivatePanelOption Enum=Hide
   ) -> (then)
   > RequestCloseSidebar(self=Get FortSidebarVM())
   n14 = "On Text Commited (InputField_Console)"!() -> (OutputDelegate, then, Text:Text, Commit method:ETextCommit Enum=Default) {is=title K2Node_ComponentBoundEvent}
   > n15 = Switch on ETextCommit!(execute, Selection:ETextCommit Enum=@n14.Commit method) -> (Default:Exec, OnEnter:Exec, OnUserMovedFocus:Exec, OnCleared:Exec) {is=title K2Node_SwitchEnum}
   ExecuteConsoleCommand!(execute=@n15.OnEnter, Command="To String (Text)"(InText=@n14.Text), SpecificPlayer) -> (then)
   "On Clicked (Button_ModeratorMode)"!() -> (OutputDelegate, then, Button:Common Button Base Object Reference) {is=title K2Node_ComponentBoundEvent}
   > Branch(Condition=Get bAllowCreativeModeratorMode!(self:Fort Runtime Options Object Reference=GetRuntimeOptions()) -> (bAllowCreativeModeratorMode:bool))
   > n16 = Cast To FortPlayerControllerGameplay!(execute, Object=GetOwningPlayer()) -> (then, CastFailed, AsFort Player Controller Gameplay:Fort Player Controller Gameplay Object Reference) {is=title K2Node_DynamicCast}
   > SetCreativeModeratorModeActive!(
    execute
    self:Fort Player Controller Gameplay Object Reference=@n16
    bActive:bool=NOT Boolean(A=Get Is Creative Moderator Mode Active() -> (Is Creative Moderator Mode Active:bool))
   ) -> (then)
   > Update Moderator Mode Visuals!(execute, self, is Loading:bool=true, is Moderator Mode Active:bool=Get Is Creative Moderator Mode Active()) -> (then)
   "On Clicked (Button_Help)"!() -> (OutputDelegate, then, Button:Common Button Base Object Reference) {is=title K2Node_ComponentBoundEvent}
   > RequestCloseSidebar(self=Get FortSidebarVM())
   > ExecuteAction(self=Get FortSidebarSettingsVM(), SidebarAction=StwHelp)
   "On Clicked (Button_Boosts)"!() -> (OutputDelegate, then, Button:Common Button Base Object Reference) {is=title K2Node_ComponentBoundEvent}
   > RequestCloseSidebar(self=Get FortSidebarVM())
   > ExecuteAction(self=Get FortSidebarSettingsVM(), SidebarAction=StwBoosts)
   "On Clicked (Button_ParentalControls)"!() -> (OutputDelegate, then, Button:Common Button Base Object Reference) {is=title K2Node_ComponentBoundEvent}
   > RequestToOpenParentalControls!(execute) -> (then)
   > RequestCloseSidebar(self=Get FortSidebarVM())
   "On Clicked (Button_ModeratorVLACapture)"!() -> (OutputDelegate, then, Button:Common Button Base Object Reference) {is=title K2Node_ComponentBoundEvent}
   > Branch(Condition=Get bAllowCreativeModeratorMode(self=GetRuntimeOptions()))
   > ToggleCapture(self:VLACapture Interface Subsystem Object Reference=Get VLACaptureInterfaceSubsystem() -> (ReturnValue:VLACapture Interface Subsystem Object Reference) {is=title K2Node_GetSubsystem})
   "On Clicked (Button_Legal)"!() -> (OutputDelegate, then, Button:Common Button Base Object Reference) {is=title K2Node_ComponentBoundEvent}
   > n17 = Construct Object from Class(
    Class=&BP_NavigationPoint_C
    ConversationId=None
    ReservedSpots=0
    NavigationTag=(TagName="UI.StateTree.TransitionTag.Sidebar.Settings.Legal")
   )
   > PushNavigation(self=Get FortSidebarNavigationVM(), NavigationPoint=@n17)
   "On Clicked (Button_SupportAndRefunds)"!() -> (OutputDelegate, then, Button:Common Button Base Object Reference) {is=title K2Node_ComponentBoundEvent}
   > n18 = Construct Object from Class(
    Class=&BP_NavigationPoint_C
    ConversationId=None
    ReservedSpots=0
    NavigationTag=(TagName="UI.StateTree.TransitionTag.Sidebar.Settings.SupportAndRefunds")
   )
   > PushNavigation(self=Get FortSidebarNavigationVM(), NavigationPoint=@n18)
   "On Clicked (Button_Reporting)"!() -> (OutputDelegate, then, Button:Common Button Base Object Reference) {is=title K2Node_ComponentBoundEvent}
   > n19 = Construct Object from Class(
    Class=&BP_NavigationPoint_C
    ConversationId=None
    ReservedSpots=0
    NavigationTag=(TagName="UI.StateTree.TransitionTag.Sidebar.Settings.Reporting")
   )
   > PushNavigation(self=Get FortSidebarNavigationVM(), NavigationPoint=@n19)
   "On Clicked (Button_ManageUnlockedContent)"!() -> (OutputDelegate, then, Button:Common Button Base Object Reference) {is=title K2Node_ComponentBoundEvent}
   > ExecuteAction(self=Get FortSidebarSettingsVM(), SidebarAction=ManageUnlockedContent)
   > RequestCloseSidebar(self=Get FortSidebarVM())
   "On Clicked (Button_Respawn)"!() -> (OutputDelegate, then, Button:Common Button Base Object Reference) {is=title K2Node_ComponentBoundEvent}
   > n20 = CreateWidgetAsync(
    UserWidgetSoftClass=Get RespawnClass() -> (RespawnClass:User Widget Soft Class Reference)
    OwningPlayer=GetOwningPlayer()
    bSuspendInputUntilComplete=true
   )
   TryPushWidgetAsModalPanel(execute=@n20.OnComplete, PushWidget=@n20)
   > RequestCloseSidebar(self=Get FortSidebarVM())
   "On Clicked (Button_TeamSelect)"!() -> (OutputDelegate, then, Button:Common Button Base Object Reference) {is=title K2Node_ComponentBoundEvent}
   > n21 = CreateWidgetAsync(
    UserWidgetSoftClass=Get TeamSelectWidgetClass() -> (TeamSelectWidgetClass:User Widget Soft Class Reference)
    OwningPlayer=GetOwningPlayer()
    bSuspendInputUntilComplete=true
   )
   TryPushWidgetAsModalPanel(execute=@n21.OnComplete, PushWidget=@n21)
   > RequestCloseSidebar(self=Get FortSidebarVM())
   "On Clicked (Button_IslandSettings)"!() -> (OutputDelegate, then, Button:Common Button Base Object Reference) {is=title K2Node_ComponentBoundEvent}
   > RequestCloseSidebar(self=Get FortSidebarVM())
   > n22 = Cast To FortPlayerController(Object=GetOwningPlayer())
   > n23 = Branch(Condition=ShouldAddTabsToMapScreen!(PlayerController:Player Controller Object Reference=@n22) -> (ReturnValue:bool))
   n24 = GetMiniMapComponent!(self:Fort Player Controller Object Reference=@n22) -> (ReturnValue:Fort Controller Component Minimap Object Reference)
   > Is Valid(InputObject=@n24)
   > OpenFullScreenMapTab!(execute, self:Fort Controller Component Minimap Object Reference=@n24, TabId:Name=GetMyIslandTabID!() -> (ReturnValue:Name=None)) -> (then)
   OpenInventory!(execute=@n23.else, self:Fort Player Controller Object Reference=@n22, InventoryTabId:Name=GetMyIslandTabID() -> (ReturnValue=None)) -> (then)
   n25 = Handle On Is Creative Moderator Mode Active Changed!() -> (OutputDelegate, then, bIsCreativeModeratorModeActive:bool=~) {is=title K2Node_CustomEvent}
   > n26 = Set Is Creative Moderator Mode Active(Is Creative Moderator Mode Active:bool=@n25.bIsCreativeModeratorModeActive) -> (Output_Get:bool)
   > Update Moderator Mode Visuals(is Loading=false, is Moderator Mode Active=@n26)
   On Possessed Pawn Changed!() -> (OutputDelegate, then, OldPawn:Pawn Object Reference, NewPawn:Pawn Object Reference) {is=title K2Node_CustomEvent}
   > Update Respawn Visibility()
   n27 = On Input Method Changed!() -> (OutputDelegate, then, New Input Type:ECommonInputType Enum) {is=title K2Node_CustomEvent}
   > SetVisibility(
    self=Get Button_Settings() -> (Button_Settings:WBP Sidebar Button Generic Object Reference)
    InVisibility=Select!(
     MouseAndKeyboard:ESlateVisibility Enum=Collapsed
     Gamepad:ESlateVisibility Enum=Visible
     Touch:ESlateVisibility Enum=Collapsed
     Count:ESlateVisibility Enum=Visible
     Index:ECommonInputType Enum=@n27.New Input Type
    ) -> (ReturnValue:ESlateVisibility Enum)
   )
   n28 = On Registry Button Clicked!() -> (OutputDelegate, then, Button:Common Button Base Object Reference) {is=title K2Node_CustomEvent}
   > n29 = Cast To WBP_Sidebar_ButtonGeneric!(execute, Object=@n28.Button) -> (then, CastFailed, AsWBP Sidebar Button Generic:WBP Sidebar Button Generic Object Reference) {is=title K2Node_DynamicCast}
   > n30 = For Each Loop!(Exec, Array:Array of Registry Button Structures=Get Registry Buttons() -> (Registry Buttons:Array of Registry Button Structures)) -> (LoopBody, Array Element_SettingRegistryName:Name, Array Element_SettingRegistryDisplayName:Text, Array Index, Completed)
   > Branch(Condition="Equal Exactly (Text)"!(
     A:"Text (by ref)"=@n30.Array Element_SettingRegistryDisplayName
     B:"Text (by ref)"=Get Text(self:WBP UIKit Button Generic Base Object Reference=@n29)
    ) -> (ReturnValue:bool))
   > ExecuteSettingRegistryAction!(execute, self:Fort Sidebar Settings VM Object Reference=Get FortSidebarSettingsVM(), SettingRegistryActionName:Name=@n30) -> (then)
   Event Destruct()
   n31 = Create Event()
   n32 = Get CommonInputSubsystem()
   > Unbind Event from On Input Method Changed(self=@n32, Delegate=@n31)
   n33 = Get CreativeHUDLContext!() -> (ReturnValue:Creative HUDLContext Object Reference) {is=title K2Node_GetBlueprintContextEx}
   > Unbind all Events from On Is Creative Moderator Mode Active Changed!(execute:Exec, self:Creative HUDLContext Object Reference=@n33) -> (then:Exec) {is=title K2Node_ClearDelegate}
   n34 = Get VLACaptureInterfaceSubsystem()
   > Unbind all Events from On VLACapture State Changed!(execute:Exec, self:VLACapture Interface Subsystem Object Reference=@n34) -> (then:Exec) {is=title K2Node_ClearDelegate}
   Event Construct()
   > Bind Event to On Input Method Changed(self=@n32, Delegate=@n31)
   > Bind Event to On Is Creative Moderator Mode Active Changed(self:Creative HUDLContext Object Reference=@n33, Delegate=Create Event()) {is=title K2Node_AddDelegate}
   > Bind Event to On VLACapture State Changed(self:VLACapture Interface Subsystem Object Reference=@n34, Delegate=Create Event()) {is=title K2Node_AddDelegate}
   > Update Moderator VLACapture Visuals!(
    execute
    self
    NewCaptureState:EVLACaptureStateInterface Enum=GetCaptureState!(self:VLACapture Interface Subsystem Object Reference=@n34) -> (ReturnValue:EVLACaptureStateInterface Enum=Idle)
   ) -> (then)
   > n35 = Cast To FortPlayerControllerGameplay(Object=GetOwningPlayer())
   > Handle On Is Creative Moderator Mode Active Changed!(
    execute
    self
    bIsCreativeModeratorModeActive:bool=IsCreativeModeratorModeActive!(self:Fort Player Controller Gameplay Object Reference=@n35) -> (ReturnValue:bool)
   ) -> (then) {is=title K2Node_CallFunction}
   > Bind Event to On Possessed Pawn Changed(self:Controller Object Reference=@n35, Delegate=Create Event()) {is=title K2Node_AddDelegate}
   Event On Activated()
   > n36 = Set Current Input Type(Current Input Type:ECommonInputType Enum=GetCurrentInputType(self=Get CommonInputSubsystem())) -> (Output_Get:ECommonInputType Enum)
   > On Input Method Changed!(execute, self, New Input Type:ECommonInputType Enum=@n36) -> (then) {is=title K2Node_CallFunction}
   > SetupMainMenuUI(self:Fort Sidebar Settings VM Object Reference=Get FortSidebarSettingsVM())
   > Update Settings Visibility!(execute:Exec) -> (then:Exec) {is=title K2Node_MacroInstance}
   Event OnInitialized()
   > UpdateButtonTimeLimitText()
  ]}
 GetControlsHelpButtonDesiredVisibility {nodes [
   GetControlsHelpButtonDesiredVisibility() {is=title K2Node_FunctionEntry}
   n1 = IsUsingCompactUI()
   n2 = GetRuntimeOptions()
   n3 = IsCurrentContextProductTagAllowedByLists!(
    AllowedTagList:"Gameplay Tag Container Structure (by ref)"=Get ControlHelp_AllowedProductTags!(self:Fort Runtime Options Object Reference=@n2) -> (ControlHelp_AllowedProductTags:Gameplay Tag Container Structure)
    DisallowedTagList:"Gameplay Tag Container Structure (by ref)"=Get ControlHelp_DisallowedProductTags!(self:Fort Runtime Options Object Reference=@n2) -> (ControlHelp_DisallowedProductTags:Gameplay Tag Container Structure)
    bMatchDefaultTag:bool=true
   ) -> (ReturnValue:bool)
   > Return Node!(
    execute
    Visibility:ESlateVisibility Enum=Select!(Option 0:ESlateVisibility Enum=Collapsed, Option 1:ESlateVisibility Enum=SelfHitTestInvisible, Index:bool=AND Boolean(A=@n1, B=@n3)) -> (ReturnValue)
   )
  ]}
 GetHLTButtonDesiredVisibility {nodes [
   GetHLTButtonDesiredVisibility() {is=title K2Node_FunctionEntry}
   n1 = IsUsingCompactUI()
   n2 = AND Boolean(A=ShouldForceHUDLayoutToolButtonVisible!() -> (ReturnValue:bool), B=@n1)
   n3 = GetRuntimeOptions()
   n4 = AND Boolean(
    A=@n1
    B=IsCurrentContextProductTagAllowedByLists(
     AllowedTagList=GetHUDLayoutTool_AllowedProductTags!(self:Fort Runtime Options Object Reference=@n3) -> (OutTagContainer:Gameplay Tag Container Structure)
     DisallowedTagList=GetHUDLayoutTool_DisallowedProductTags!(self:Fort Runtime Options Object Reference=@n3) -> (OutTagContainer:Gameplay Tag Container Structure)
     bMatchDefaultTag=true
    )
   )
   > Return Node(Visibility=Select(Option 0=Collapsed, Option 1=SelfHitTestInvisible, Index=OR Boolean(A=@n2, B=@n4)))
  ]}
 NavigateFromSettingsToTabGroup {nodes [
   NavigateFromSettingsToTabGroup!() -> (then, Navigation:EUINavigation Enum) {is=title K2Node_FunctionEntry}
   > n1 = FindParentWidgetOfType!(execute, StartingWidget:Widget Object Reference=Self-Reference(), Type:Widget Class Reference=/Sidebar/WBP_Sidebar.WBP_Sidebar_C) -> (then, ReturnValue:WBP Sidebar Object Reference)
   > n2 = NavigationToTabGroup!(execute, self:WBP Sidebar Object Reference=@n1, Navigation:EUINavigation Enum=Up) -> (then, ReturnValue:Widget Object Reference)
   > Return Node!(execute, ReturnValue:Widget Object Reference=@n2)
  ]}
 OnSetLastFocusedButton {nodes [
   n1 = OnSetLastFocusedButton!() -> (then, Type:EUINavigation Enum, OldFocusedWidget:Widget Object Reference, NewFocusedWidget:Widget Object Reference) {is=title K2Node_FunctionEntry}
   n2 = GetOuterUserWidget!(Widget:Widget Object Reference=@n1.NewFocusedWidget) -> (ReturnValue:User Widget Object Reference)
   > Branch(Condition=IsValid(Object=@n2))
   > Cast To WBP_Sidebar_ButtonGeneric(Object=@n2)
   > Set LastFocusedButton(LastFocusedButton:User Widget Object Reference=@n2) -> (Output_Get:User Widget Object Reference)
  ]}
 Set Button Visibilities From Actions {nodes [
   n1 = Set Button Visibilities From Actions!() -> (then, Sidebar Actions:Set of ESidebarAction Enums=SupportCreator) {is=title K2Node_FunctionEntry}
   > SetVisibility(
    self=Get Button_TimeLimits() -> (Button_TimeLimits:WBP Sidebar Button Generic Object Reference)
    InVisibility=Select(
     Option 0=Collapsed
     Option 1=Visible
     Index=Contains Item!(TargetSet:Set of ESidebarAction Enums=@n1, ItemToFind:"ESidebarAction Enum (by ref)"=Cilantro) -> (ReturnValue:bool) {is=title K2Node_CallFunction}
    )
   )
   > SetVisibility(
    self=Get Button_ManageUnlockedContent() -> (Button_ManageUnlockedContent:WBP Sidebar Button Generic Object Reference)
    InVisibility=Select(Option 0=Collapsed, Option 1=Visible, Index=Contains Item(TargetSet=@n1, ItemToFind=ManageUnlockedContent))
   )
   > Creative_Get Start Game Button State()
   > SetVisibility(
    self=Get Button_ControlsHelp() -> (Button_ControlsHelp:WBP Sidebar Button Generic Object Reference)
    InVisibility=GetControlsHelpButtonDesiredVisibility!() -> (Visibility:ESlateVisibility Enum=Visible) {is=title K2Node_CallFunction}
   )
   > SetVisibility(
    self=Get Button_HudLayoutTool() -> (Button_HudLayoutTool:WBP Sidebar Button Generic Object Reference)
    InVisibility=GetHLTButtonDesiredVisibility!() -> (Visibility:ESlateVisibility Enum=Visible) {is=title K2Node_CallFunction}
   )
   > RequestRefreshFocus(self:Common Activatable Widget Object Reference)
   > SidebarFeatureFinishedLoading(self:Fort Sidebar VM Object Reference=Get FortSidebarVM())
  ]}
 SetRegistryButtons {nodes [
   n1 = SetRegistryButtons!() -> (then, Registry Buttons:Array of Registry Button Structures) {is=title K2Node_FunctionEntry}
   > n2 = Set Registry Buttons(Registry Buttons:Array of Registry Button Structures=@n1) -> (Output_Get:Array of Registry Button Structures)
   > ClearChildren(self=Get Vbox_RegistryButtons() -> (Vbox_RegistryButtons:Vertical Box Object Reference))
   > n3 = Branch(Condition=NOT Boolean(A=Is Empty!(TargetArray:Array of Registry Button Structures=@n2) -> (ReturnValue:bool)))
   n4 = Get Vbox_RegistryButtons()
   > SetVisibility(self=@n4, InVisibility=SelfHitTestInvisible)
   > n5 = For Each Loop(Array=Get Registry Buttons())
   > n6 = Construct Object from Class!(
    execute
    Class=/Sidebar/Buttons/WBP_Sidebar_ButtonGeneric.WBP_Sidebar_ButtonGeneric_C
    self
    Style:Common Button Style Class Reference=/FortUILibrary/UIKit/Buttons/Blocks/ButtonStyle_UIKit_Empty.ButtonStyle_UIKit_Empty_C
    bApplyAlphaOnDisable:bool=False
    bLocked:bool=False
    bSelectable:bool=True
    bShouldSelectUponReceivingFocus:bool=True
    bInteractableWhenSelected:bool=True
    bToggleable:bool=False
    bTriggerClickedAfterSelection:bool=False
    bDisplayInputActionWhenNotInteractable:bool=True
    bHideInputActionWithKeyboard:bool=False
    bShouldUseFallbackDefaultInputAction:bool=True
    bRequiresHold:bool=False
    HoldData:Common UIHold Data Class Reference=&CommonUI_HoldData_Base_C
    ClickMethod:EButtonClickMethod Enum=DownAndUp
    InputPriority:int=0
    TriggeringInputAction:Data Table Row Handle Structure
   ) -> (then, ReturnValue:WBP Sidebar Button Generic Object Reference)
   > SetText(self=@n6, Text=@n5.Array Element_SettingRegistryDisplayName)
   > Bind Event to On Clicked(self:Common Button Base Object Reference=@n6, Delegate=Create Event()) {is=title K2Node_AddDelegate}
   > AddChild(self=Get Vbox_RegistryButtons(), Content=@n6)
   SetVisibility(execute=@n3.else, self=@n4, InVisibility=Collapsed)
  ]}
 Update Back to Hub Visibility {nodes [
   Update Back to Hub Visibility() {is=title K2Node_FunctionEntry}
   > n1 = Cast To FortGameStateAthena!(execute, Object=GetGameState!() -> (ReturnValue:Game State Base Object Reference)) -> (then, CastFailed, AsFort Game State Athena:Fort Game State Athena Object Reference) {is=title K2Node_DynamicCast}
   > Set Local_GameState(Local_GameState:Fort Game State Athena Object Reference=@n1) -> (Output_Get:Fort Game State Athena Object Reference)
   > n2 = Cast To FortPlayerControllerAthena(Object=GetOwningPlayer())
   > n3 = Set Local_PlayerController(Local_PlayerController:Fort Player Controller Athena Object Reference=@n2) -> (Output_Get:Fort Player Controller Athena Object Reference)
   > n4 = Cast To FortPlayerState!(execute, Object=Get PlayerState!(self:Controller Object Reference=@n3) -> (PlayerState:Player State Object Reference)) -> (then, CastFailed, AsFort Player State:Fort Player State Object Reference) {is=title K2Node_DynamicCast}
   > Set Local_PlayerState(Local_PlayerState:Fort Player State Object Reference=@n4) -> (Output_Get:Fort Player State Object Reference)
   n5 = Get Local_GameState() -> (Local_GameState:Fort Game State Athena Object Reference)
   n6 = IsTeleportToCreativeHubAllowed!(
    self:Fort Game State Athena Object Reference=@n5
    PlayerState:Fort Player State Object Reference=Get Local_PlayerState() -> (Local_PlayerState:Fort Player State Object Reference)
   ) -> (ReturnValue:bool)
   n7 = CanRespawnOnStarterIsland!(self:Fort Game State Athena Object Reference=@n5) -> (ReturnValue:bool)
   > SetVisibility(
    self=Get Button_BackToHub() -> (Button_BackToHub:WBP Sidebar Button Generic Object Reference)
    InVisibility=Select(Option 0=Collapsed, Option 1=SelfHitTestInvisible, Index=AND Boolean(A=@n7, B=@n6))
   )
   SetVisibility(execute=[@n4.CastFailed, @n2.CastFailed, @n1.CastFailed], self=Get Button_BackToHub(), InVisibility=Collapsed)
  ]}
 Update Minigame Visibility {nodes [
   Update Minigame Visibility() {is=title K2Node_FunctionEntry}
   > n1 = Cast To FortGameStateAthena(Object=GetGameState())
   > Set Local Game State(Local Game State:Fort Game State Athena Object Reference=@n1) -> (Output_Get:Fort Game State Athena Object Reference)
   > n2 = Cast To FortPlayerControllerAthena(Object=GetOwningPlayer())
   > Set Local Player Controller(Local Player Controller:Fort Player Controller Athena Object Reference=@n2) -> (Output_Get:Fort Player Controller Athena Object Reference)
   n3 = Get Local Player Controller() -> (Local Player Controller:Fort Player Controller Athena Object Reference)
   n4 = Get Local Game State() -> (Local Game State:Fort Game State Athena Object Reference)
   > SetVisibility(
    self=Get Button_StartGame()
    InVisibility=Select(
     Option 0=Collapsed
     Option 1=Visible
     Index=isMinigameVisible!(Player:Fort Player Controller Athena Object Reference=@n3, FortGameStateAthena:Fort Game State Athena Object Reference=@n4) -> (bResult:bool)
    )
   )
  ]}
 Update Moderator Mode Visibility {nodes [
   Update Moderator Mode Visibility() {is=title K2Node_FunctionEntry}
   > n1 = Cast To FortPlayerControllerAthena(Object=GetOwningPlayer())
   n2 = NOT Boolean(A=IsMobileGame())
   n3 = IsCreativeModeratorModeEnabled!(self:Fort Player Controller Gameplay Object Reference=@n1) -> (ReturnValue:bool)
   n4 = IsCreativeModeratorModeAllowed!(self:Fort Player Controller Gameplay Object Reference=@n1) -> (ReturnValue:bool)
   > SetVisibility(
    self=[
     Get Button_ModeratorMode() -> (Button_ModeratorMode:WBP Sidebar Button Generic Object Reference)
     Get Button_ModeratorVLACapture() -> (Button_ModeratorVLACapture:WBP Sidebar Button Generic Object Reference)
    ]
    InVisibility=Select(Option 0=Collapsed, Option 1=SelfHitTestInvisible, Index=AND Boolean!(A=@n2, B=@n4, C:bool=@n3) -> (ReturnValue))
   )
  ]}
 Update Moderator Mode Visuals {nodes [
   Update Moderator Mode Visuals!() -> (then, is Loading:bool=~, is Moderator Mode Active:bool=~) {is=title K2Node_FunctionEntry}
   n1 = Get is Moderator Mode Active() -> (is Moderator Mode Active:bool)
   n2 = Get Button_ModeratorMode()
   > SetText(
    self=@n2
    Text=SelectText(
     A=SelectText(A="LOCTEXT(\"Disabling Moderator mode\")", B="LOCTEXT(\"Enabling Moderator mode\")", bPickA=@n1)
     B=SelectText(A="LOCTEXT(\"Disable Moderator mode\")", B="LOCTEXT(\"Enable Moderator mode\")", bPickA=@n1)
     bPickA=Get is Loading() -> (is Loading:bool)
    )
   )
   > Set ButtonType!(
    execute
    ButtonType:E_UI_Sidebar_Button_Type Enum=Select(Option 0:E_UI_Sidebar_Button_Type Enum=NewEnumerator0, Option 1:E_UI_Sidebar_Button_Type Enum=NewEnumerator3, Index=Get is Loading()) -> (ReturnValue:E_UI_Sidebar_Button_Type Enum)
    self:WBP Sidebar Button Generic Object Reference=@n2
   ) -> (then, Output_Get:E_UI_Sidebar_Button_Type Enum=NewEnumerator0)
   > SetIsLocked!(execute, self:Common Button Base Object Reference=@n2, bInIsLocked:bool=Get is Loading()) -> (then)
   > SetIconMaterial!(
    execute
    self:WBP UIKit Button Generic Base Object Reference=@n2
    Material Instance:Material Instance Object Reference=Select(
     Option 0:Material Instance Object Reference
     Option 1:Material Instance Object Reference=/Game/UI/DiscoveryShared/Art/MI_UI_Discovery_Throbber
     Index=Get is Loading()
    ) -> (ReturnValue:Material Instance Object Reference)
   ) -> (then)
   > InitializeButtonVisuals(self:WBP Sidebar Button Generic Object Reference=@n2)
  ]}
 Update Moderator VLACapture Visuals {nodes [
   n1 = Update Moderator VLACapture Visuals!() -> (then, NewCaptureState:EVLACaptureStateInterface Enum=Idle) {is=title K2Node_FunctionEntry}
   n2 = Get Button_ModeratorVLACapture()
   > SetText(
    self=@n2
    Text=Select!(
     Idle:Text="LOCTEXT(\"Start Capture\")"
     Capturing:Text="LOCTEXT(\"Stop Capture\")"
     Stopping:Text="LOCTEXT(\"Stopping Capture\")"
     Unavailable:Text="LOCTEXT(\"Capture Unavailable\")"
     Index:EVLACaptureStateInterface Enum=@n1
    ) -> (ReturnValue:Text)
   )
   > Set ButtonType(
    ButtonType=Select(
     Idle:E_UI_Sidebar_Button_Type Enum=NewEnumerator0
     Capturing:E_UI_Sidebar_Button_Type Enum=NewEnumerator0
     Stopping:E_UI_Sidebar_Button_Type Enum=NewEnumerator3
     Unavailable:E_UI_Sidebar_Button_Type Enum=NewEnumerator3
     Index=@n1
    ) -> (ReturnValue:E_UI_Sidebar_Button_Type Enum)
    self=@n2
   ) -> (Output_Get=NewEnumerator0)
   > SetIsLocked(
    self=@n2
    bInIsLocked=Select(Idle:bool=false, Capturing:bool=false, Stopping:bool=true, Unavailable:bool=true, Index=@n1) -> (ReturnValue:bool)
   )
   > InitializeButtonVisuals(self=@n2)
  ]}
 Update Respawn Visibility {nodes [
   Update Respawn Visibility() {is=title K2Node_FunctionEntry}
   > n1 = Cast To FortGameStateAthena(Object=GetGameState())
   > n2 = Cast To FortRemoteControlledPawnAthena!(execute, Object=GetOwningPlayerPawn!(self:User Widget Object Reference) -> (ReturnValue:Pawn Object Reference)) -> (then, CastFailed, AsFort Remote Controlled Pawn Athena:Fort Remote Controlled Pawn Athena Object Reference) {is=title K2Node_DynamicCast}
   SetVisibility(
    execute=[@n5.CastFailed, @n4.Is Not Valid, @n2.then, @n1.CastFailed]
    self=Get Button_Respawn() -> (Button_Respawn:WBP Sidebar Button Generic Object Reference)
    InVisibility=Collapsed
   )
   n3 = GetOwningPlayer()
   n4 = Is Valid(exec=@n2.CastFailed, InputObject=@n3)
   > n5 = Cast To FortPlayerState(Object=Get PlayerState(self=@n3))
   > Set PlayerTriggeredSpawnAllowed(PlayerTriggeredSpawnAllowed:bool=IsPlayerTriggeredRespawnAllowed!(self:Fort Game State Athena Object Reference=@n1, PlayerState:Fort Player State Object Reference=@n5) -> (ReturnValue:bool)) -> (Output_Get:bool)
   n6 = IsRespawningAllowed!(self:Fort Game State Zone Object Reference=@n1, PlayerState:Fort Player State Object Reference=@n5) -> (ReturnValue:bool)
   n7 = Get PlayerTriggeredSpawnAllowed() -> (PlayerTriggeredSpawnAllowed:bool)
   > SetVisibility(
    self=Get Button_Respawn()
    InVisibility=Select!(
     Option 0:ESlateVisibility Enum=Collapsed
     Option 1:ESlateVisibility Enum=SelfHitTestInvisible
     Index:bool=AND Boolean!(A=@n6, B=@n7) -> (ReturnValue)
    ) -> (ReturnValue:ESlateVisibility Enum)
   )
  ]}
 Update Team Select Visibility {nodes [
   Update Team Select Visibility() {is=title K2Node_FunctionEntry}
   n1 = GetOwningPlayer()
   > n2 = Cast To FortPlayerControllerAthena(Object=@n1)
   > n3 = Cast To FortGameStateAthena(Object=GetGameState())
   n4 = IsCreativeModeratorModeActive(self=@n2)
   n5 = Get Controlled Pawn!(self:Controller Object Reference=@n1) -> (ReturnValue:Pawn Object Reference) {is=title K2Node_CallFunction}
   n6 = AND Boolean!(
    A=IsTeamSwitchAllowed!(self:Fort Game State Athena Object Reference=@n3, SwitchingPawn:Pawn Object Reference=@n5) -> (ReturnValue:bool)
    B=NOT Boolean(A=Cast To FortRemoteControlledPawnAthena!(Object=@n5) -> (AsFort Remote Controlled Pawn Athena, bSuccess:bool=~).bSuccess)
    C:bool=GetShowTeamSelectButton!() -> (ReturnValue:bool)
   ) -> (ReturnValue)
   > SetVisibility(
    self=Get Button_TeamSelect() -> (Button_TeamSelect:WBP Sidebar Button Generic Object Reference)
    InVisibility=Select(Option 0=Collapsed, Option 1=SelfHitTestInvisible, Index=OR Boolean(A=@n6, B=@n4))
   )
   SetVisibility(execute=[@n3.CastFailed, @n2.CastFailed], self=Get Button_TeamSelect(), InVisibility=Collapsed)
  ]}
 UpdateButtonTimeLimitText {nodes [
   UpdateButtonTimeLimitText() {is=title K2Node_FunctionEntry}
   n1 = GetRuntimeOptions()
   > SetText(
    self=Get Button_TimeLimits()
    Text=SelectText(
     A="LOCTEXT(\"Time limits && reports\")"
     B="LOCTEXT(\"Time limits\")"
     bPickA=Get bEnableThyme!(self:Fort Runtime Options Object Reference=@n1) -> (bEnableThyme:bool)
    )
   )
  ]}
 UpdateSettingsVisibility {nodes [
   Inputs()
   > Update Minigame Visibility() {is=title K2Node_CallFunction}
   > Update Moderator Mode Visibility() {is=title K2Node_CallFunction}
   > Update Respawn Visibility() {is=title K2Node_CallFunction}
   > Update Back to Hub Visibility() {is=title K2Node_CallFunction}
   > Update Team Select Visibility() {is=title K2Node_CallFunction}
   n1 = Get FortGlobalUIContext() -> (ReturnValue:Fort Global UIContext Object Reference) {is=title K2Node_GetSubsystem}
   > SetVisibility(
    self=Get Button_ParentalControls() -> (Button_ParentalControls:WBP Sidebar Button Generic Object Reference)
    InVisibility=Select(Option 0=Visible, Option 1=Collapsed, Index=IsInZone!(self:Fort Global UIContext Object Reference=@n1) -> (ReturnValue:bool))
   )
   > SetVisibility(
    self=[Get HeaderText() -> (HeaderText:Common Text Object Reference), Get Vbox_CreativeOptions()]
    InVisibility=Select(
     Option 0=Collapsed
     Option 1=Visible
     Index=Creative_isCreativeOptionsNotEmpty!() -> (isCreativeOptionsNotEmpty:bool) {is=title K2Node_CallFunction}
    )
   )
   > n2 = Cast To FortPlayerControllerAthena(Object=GetOwningPlayer())
   > n3 = SetVisibility(
    self=Get Button_IslandSettings() -> (Button_IslandSettings:WBP Sidebar Button Generic Object Reference)
    InVisibility=Select(
     Option 0=Collapsed
     Option 1=SelfHitTestInvisible
     Index=IsPlayerInAnExperienceTheyOwn!(self:Fort Player Controller Object Reference=@n2) -> (ReturnValue:bool)
    )
   )
   n4 = GetRuntimeOptions()
   SetVisibility(
    execute=[@n3.then, @n2.CastFailed]
    self=Get Vbox_Developer() -> (Vbox_Developer:Vertical Box Object Reference)
    InVisibility=Select(Option 0=Visible, Option 1=Collapsed, Index=IsShippingBuild!(self:Fort Runtime Options Object Reference=@n4) -> (ReturnValue:bool))
   )
   > n5 = Branch(Condition=IsInSubGame!(SubGame:ESubGame Enum=Campaign) -> (ReturnValue:bool))
   n6 = Get Vbox_SaveTheWorld() -> (Vbox_SaveTheWorld:Vertical Box Object Reference)
   > n7 = SetVisibility(self=@n6, InVisibility=Visible)
   Outputs!(then:Exec=[@n7.then, @n8.then])
   n8 = SetVisibility(execute=@n5.else, self=@n6, InVisibility=Collapsed)
  ]}
 isMinigameVisible {nodes [
   isMinigameVisible!() -> (then, Player:Fort Player Controller Athena Object Reference, FortGameStateAthena:Fort Game State Athena Object Reference) {is=title K2Node_FunctionEntry}
   > n1 = Sequence()
   > Get!(execute) -> (then, else, Player:Fort Player Controller Athena Object Reference)
   > Set HasPermissionToStart(HasPermissionToStart:bool=CanOperateOnMinigame!(PlayerController:Fort Player Controller Object Reference=Get Player() -> (Player:Fort Player Controller Athena Object Reference)) -> (ReturnValue:bool)) -> (Output_Get:bool)
   > n2 = Set Minigame(Minigame:Fort Minigame Object Reference=GetMinigame!(ContextActor:Actor Object Reference=GetOwningPlayer()) -> (ReturnValue)) -> (Output_Get:Fort Minigame Object Reference)
   > Is Valid(InputObject=@n2)
   n3 = Get Player()
   > n4 = Set bIsModerator(bIsModerator:bool=IsCreativeModeratorModeEnabled(self=@n3)) -> (Output_Get:bool)
   n5 = HasStarted!(self:Fort Minigame Object Reference=@n2) -> (ReturnValue:bool)
   > Branch(Condition=OR Boolean!(A=@n4, B=CanOperateOnMinigame(PlayerController=@n3), C:bool=NOT Boolean(A=@n5)) -> (ReturnValue))
   n6 = IsCuratedHub(self=Get Minigame() -> (Minigame:Fort Minigame Object Reference))
   n7 = NOT Boolean(A=IsInPublicMatchmadeCreativeMinigame!(self:Fort Player Controller Athena Object Reference=Get Player()) -> (ReturnValue:bool))
   n8 = Get bIsModerator() -> (bIsModerator:bool)
   > Set bShow(bShow:bool=AND Boolean(A=OR Boolean!(A=@n8, B=@n7) -> (ReturnValue), B=NOT Boolean(A=@n6), C=Get HasPermissionToStart() -> (HasPermissionToStart:bool))) -> (Output_Get:bool)
   Return Node!(execute=@n1.then_1, bResult:bool=Get bShow() -> (bShow:bool))
  ]}
]
variables [
 Creative_FortPlayerControllerAthena {type=FortPlayerControllerAthena, default=null}
 Creative_Machine {type=FortMinigameSettingsBuilding, default=null}
 Current Input Type {type=ECommonInputType, default=MouseAndKeyboard}
 Fort Player Controller Athena {type=FortPlayerControllerAthena, default=null}
 GameButtonState {type=MinigameButtonsState, default=Invalid}
 Is Creative Moderator Mode Active {type=boolean}
 LastFocusedButton {type=UserWidget, default=null}
 MobileControlsHelpClass {type=Class, default=/FortniteCoreUI/Frontend/BacchusControlsHelp.BacchusControlsHelp_C}
 Registry Buttons {type="array<RegistryButton>", default []}
 RespawnClass {type=Class, default=/FortniteCoreUI/Frontend/RespawnPromptWidget.RespawnPromptWidget_C}
 TeamSelectWidgetClass {type=Class, default=/FortniteCoreUI/Frontend/TeamSelect/TeamSelectWidget.TeamSelectWidget_C}
]
widgets {
 info {
  inheritedWidgetCount=0
  namedSlotCount=1
  parentClass=&CommonActivatableWidget
  rootWidgetClass=/Sidebar/Shared/WBP_Sidebar_PanelLayout.WBP_Sidebar_PanelLayout_C
  widgetCount=30
 }
 widgets [
  WBP_Sidebar_PanelLayout {
   widgetClassPath=/Sidebar/Shared/WBP_Sidebar_PanelLayout
   bIsVariable=true
   widget=WBP_Sidebar_PanelLayout
   properties {headingText=Menu, showBackButton=false}
  }
  SB_Content {
   widgetClassPath=&ScrollBox
   namedSlotHost=WBP_Sidebar_PanelLayout
   widget=SB_Content
   properties {
    bOverride_Cursor=true
    clipping=Inherit
    consumeMouseWheel=Always
    scrollBarVisibility=Hidden
    scrollWhenFocusChanges=AnimatedScroll
    visibility=SelfHitTestInvisible
    scrollbarPadding=<ltrb 0 0 0 0>
    scrollbarThickness=<0 0>
    widgetStyle {
     barThickness=9
     bottomShadowBrush {
      bIsDynamicallyLoaded=false
      drawAs=NoDrawType
      imageType=FullColor
      mirroring=NoMirror
      resourceName=../../../Engine/Content/Slate/Common/ScrollBoxShadowBottom.png
      resourceObject=null
      tiling=NoTile
      imageSize=<16 8>
      margin=<ltrb 0.5 0 0.5 1>
      outlineSettings {
       bUseBrushTransparency=false
       roundingType=HalfHeightRadius
       width=0
       color {colorUseRule=UseColor_Specified, specifiedColor=<rgba 0 0 0 0>}
       cornerRadii=<0 0 0 0>
      }
      tintColor {colorUseRule=UseColor_Specified, specifiedColor=<rgba 1 1 1 1>}
      uVRegion {bIsValid=false, max=<0 0>, min=<0 0>}
     }
     horizontalScrolledContentPadding=<ltrb 0 0 1 0>
     leftShadowBrush {
      bIsDynamicallyLoaded=false
      drawAs=NoDrawType
      imageType=FullColor
      mirroring=NoMirror
      resourceName=../../../Engine/Content/Slate/Common/ScrollBoxShadowLeft.png
      resourceObject=null
      tiling=NoTile
      imageSize=<8 16>
      margin=<ltrb 1 0.5 0 0.5>
      outlineSettings {
       bUseBrushTransparency=false
       roundingType=HalfHeightRadius
       width=0
       color {colorUseRule=UseColor_Specified, specifiedColor=<rgba 0 0 0 0>}
       cornerRadii=<0 0 0 0>
      }
      tintColor {colorUseRule=UseColor_Specified, specifiedColor=<rgba 1 1 1 1>}
      uVRegion {bIsValid=false, max=<0 0>, min=<0 0>}
     }
     rightShadowBrush {
      bIsDynamicallyLoaded=false
      drawAs=NoDrawType
      imageType=FullColor
      mirroring=NoMirror
      resourceName=../../../Engine/Content/Slate/Common/ScrollBoxShadowRight.png
      resourceObject=null
      tiling=NoTile
      imageSize=<8 16>
      margin=<ltrb 0 0.5 1 0.5>
      outlineSettings {
       bUseBrushTransparency=false
       roundingType=HalfHeightRadius
       width=0
       color {colorUseRule=UseColor_Specified, specifiedColor=<rgba 0 0 0 0>}
       cornerRadii=<0 0 0 0>
      }
      tintColor {colorUseRule=UseColor_Specified, specifiedColor=<rgba 1 1 1 1>}
      uVRegion {bIsValid=false, max=<0 0>, min=<0 0>}
     }
     topShadowBrush {
      bIsDynamicallyLoaded=false
      drawAs=NoDrawType
      imageType=FullColor
      mirroring=NoMirror
      resourceName=../../../Engine/Content/Slate/Common/ScrollBoxShadowTop.png
      resourceObject=null
      tiling=NoTile
      imageSize=<16 8>
      margin=<ltrb 0.5 1 0.5 0>
      outlineSettings {
       bUseBrushTransparency=false
       roundingType=HalfHeightRadius
       width=0
       color {colorUseRule=UseColor_Specified, specifiedColor=<rgba 0 0 0 0>}
       cornerRadii=<0 0 0 0>
      }
      tintColor {colorUseRule=UseColor_Specified, specifiedColor=<rgba 1 1 1 1>}
      uVRegion {bIsValid=false, max=<0 0>, min=<0 0>}
     }
     verticalScrolledContentPadding=<ltrb 0 0 0 1>
    }
   }
  }
  FortScrollNavigationPanel {
   widgetClassPath=/Script/SidebarUI.FortScrollNavigationPanel
   bIsVariable=true
   parent=SB_Content
   widget=FortScrollNavigationPanel
   slot {class=ScrollBoxSlot, properties {padding=<ltrb 18 12 18 12>}}
   uIComponents [{
     component=&self:UIComponentWidgetBlueprintExtension_0.ComponentContainer.NavigationUIComponent_0
     componentClassPath=/Script/UMG.NavigationUIComponent
    }]
  }
  VBox_Menu {widgetClassPath=&VerticalBox, bIsVariable=true, parent=FortScrollNavigationPanel, widget=VBox_Menu, slot {class=VerticalBoxSlot}}
  Button_Settings {
   widgetClassPath=/Sidebar/Buttons/WBP_Sidebar_ButtonGeneric
   bIsVariable=true
   parent=VBox_Menu
   widget=Button_Settings
   properties {
    icon=/FortUILibrary/Textures/Icons/T_UI_Icon_UIKit_GeneralSettings_MSDF
    minHeightOverride=60
    navigation=&self:WidgetTree.Button_Settings.WidgetNavigation_0
    text=Settings
    useIconBlock=true
    buttonPadding=<ltrb 24 4 24 4>
    iconPadding=<ltrb 0 0 4 0>
   }
   slot {class=VerticalBoxSlot, properties {padding=<ltrb 0 0 0 12>}}
  }
  Button_ParentalControls {
   widgetClassPath=&WBP_Sidebar_ButtonGeneric
   bIsVariable=true
   parent=VBox_Menu
   widget=Button_ParentalControls
   properties {minHeightOverride=60, text=Parental controls}
   slot {class=VerticalBoxSlot, properties {padding=<ltrb 0 0 0 12>}}
  }
  Button_ManageUnlockedContent {
   widgetClassPath=&WBP_Sidebar_ButtonGeneric
   bIsVariable=true
   parent=VBox_Menu
   widget=Button_ManageUnlockedContent
   properties {minHeightOverride=60, text=Manage unlocked content}
   slot {class=VerticalBoxSlot, properties {padding=<ltrb 0 0 0 12>}}
  }
  Button_TimeLimits {
   widgetClassPath=&WBP_Sidebar_ButtonGeneric
   bIsVariable=true
   parent=VBox_Menu
   widget=Button_TimeLimits
   properties {minHeightOverride=60, text=Time limits}
   slot {class=VerticalBoxSlot, properties {padding=<ltrb 0 0 0 12>}}
  }
  Button_SupportAndRefunds {
   widgetClassPath=&WBP_Sidebar_ButtonGeneric
   bIsVariable=true
   parent=VBox_Menu
   widget=Button_SupportAndRefunds
   properties {minHeightOverride=60, text=Support and refunds}
   slot {class=VerticalBoxSlot, properties {padding=<ltrb 0 0 0 12>}}
  }
  Button_Reporting {
   widgetClassPath=&WBP_Sidebar_ButtonGeneric
   bIsVariable=true
   parent=VBox_Menu
   widget=Button_Reporting
   properties {minHeightOverride=60, text=Report}
   slot {class=VerticalBoxSlot, properties {padding=<ltrb 0 0 0 12>}}
  }
  Button_Legal {
   widgetClassPath=&WBP_Sidebar_ButtonGeneric
   bIsVariable=true
   parent=VBox_Menu
   widget=Button_Legal
   properties {minHeightOverride=60, text=Legal}
   slot {class=VerticalBoxSlot, properties {padding=<ltrb 0 0 0 12>}}
  }
  Button_ControlsHelp {
   widgetClassPath=&WBP_Sidebar_ButtonGeneric
   bIsVariable=true
   parent=VBox_Menu
   widget=Button_ControlsHelp
   properties {minHeightOverride=60, text=Controls help}
   slot {class=VerticalBoxSlot, properties {padding=<ltrb 0 0 0 12>}}
  }
  Button_HudLayoutTool {
   widgetClassPath=&WBP_Sidebar_ButtonGeneric
   bIsVariable=true
   parent=VBox_Menu
   widget=Button_HudLayoutTool
   properties {minHeightOverride=60, text=HUD Layout Tool}
   slot {class=VerticalBoxSlot, properties {padding=<ltrb 0 0 0 12>}}
  }
  Vbox_SaveTheWorld {
   widgetClassPath=&VerticalBox
   bIsVariable=true
   parent=FortScrollNavigationPanel
   widget=Vbox_SaveTheWorld
   slot {class=VerticalBoxSlot}
  }
  Button_Help {
   widgetClassPath=&WBP_Sidebar_ButtonGeneric
   bIsVariable=true
   parent=Vbox_SaveTheWorld
   widget=Button_Help
   properties {minHeightOverride=60, text=Help}
   slot {class=VerticalBoxSlot, properties {padding=<ltrb 0 0 0 12>}}
  }
  Button_Boosts {
   widgetClassPath=&WBP_Sidebar_ButtonGeneric
   bIsVariable=true
   parent=Vbox_SaveTheWorld
   widget=Button_Boosts
   properties {minHeightOverride=60, text=Boosts}
   slot {class=VerticalBoxSlot, properties {padding=<ltrb 0 0 0 12>}}
  }
  Vbox_Developer {
   widgetClassPath=&VerticalBox
   bIsVariable=true
   parent=FortScrollNavigationPanel
   widget=Vbox_Developer
   slot {class=VerticalBoxSlot}
  }
  Button_Cheats {
   widgetClassPath=&WBP_Sidebar_ButtonGeneric
   bIsVariable=true
   parent=Vbox_Developer
   widget=Button_Cheats
   properties {minHeightOverride=60, text="Cheats (developer)"}
   slot {class=VerticalBoxSlot, properties {padding=<ltrb 0 0 0 12>}}
  }
  Button_Locker {
   widgetClassPath=&WBP_Sidebar_ButtonGeneric
   bIsVariable=true
   parent=Vbox_Developer
   widget=Button_Locker
   properties {minHeightOverride=60, text="Locker (developer)"}
   slot {class=VerticalBoxSlot, properties {padding=<ltrb 0 0 0 12>}}
  }
  InputField_Console {
   widgetClassPath=/Sidebar/Buttons/WBP_Sidebar_InputField
   bIsVariable=true
   parent=Vbox_Developer
   widget=InputField_Console
   properties {
    hint Text=Enter console commands...
    icon=/Game/ArtTools/Tech_Anim/shelf_ue4TechAnim/icons/consoleCmds_icon
    maxTextCharacters=200
    iconProperties {
     material=/FortUILibrary/Systems/InputField/MaterialInstances/MI_UIKit_InputFieldIcon
     image Size=<50 50>
     placement {
      horizontal Alignment=HAlign_Center
      vertical Alignment=VAlign_Center
      grid {column=0, column Span=0, layer=0, row=1, row Span=0, nudge=<0 0>}
      padding=<ltrb 10 0 0 0>
     }
    }
   }
   slot {class=VerticalBoxSlot}
  }
  HeaderText {
   widgetClassPath=&CommonTextBlock
   bIsVariable=true
   parent=FortScrollNavigationPanel
   widget=HeaderText
   properties {
    bIsEnabled=false
    text=Creative Options
    visibility=HitTestInvisible
    font {
     bForceMonospaced=false
     bMaterialIsStencil=false
     fontMaterial=null
     fontObject=&HeadingNow
     letterSpacing=0
     monospacedWidth=1
     size=18
     skewAmount=0
     typefaceFontName=75 Medium
     outlineSettings {
      bApplyOutlineToDropShadows=false
      bMiteredCorners=false
      bSeparateFillAlpha=false
      outlineMaterial=null
      outlineSize=0
      outlineColor=<rgba 0 0 0 1>
     }
    }
    shadowOffset=<0 0>
   }
   slot {class=VerticalBoxSlot, properties {padding=<ltrb 0 24 0 24>}}
  }
  Button_StartGame {
   widgetClassPath=&WBP_UIKit_Button_Generic
   bIsVariable=true
   parent=FortScrollNavigationPanel
   widget=Button_StartGame
   properties {
    bHideInputAction=true
    backgroundMaterial=/FortUILibrary/UIKit/Buttons/MaterialInstances/CTAButtons/MI_UIKit_BgLoud
    minHeightOverride=60
    mobileMinHeight=80
    text=Start game
    textTransformPolicy=None
    useIconBlock=false
    visibility=Collapsed
    buttonPadding=<ltrb 24 8 24 8>
    font {
     bForceMonospaced=false
     bMaterialIsStencil=false
     fontMaterial=/FortUILibrary/UIKit/Buttons/MaterialInstances/CTAButtons/MI_UIKit_TextQuiet
     fontObject=&HeadingNow
     letterSpacing=10
     monospacedWidth=1
     size=15
     skewAmount=0
     typefaceFontName=75 Medium
     outlineSettings {
      bApplyOutlineToDropShadows=false
      bMiteredCorners=false
      bSeparateFillAlpha=false
      outlineMaterial=null
      outlineSize=0
      outlineColor=<rgba 0 0 0 1>
     }
    }
   }
   slot {class=VerticalBoxSlot, properties {padding=<ltrb 0 0 0 8>}}
  }
  Button_ModeratorMode {
   widgetClassPath=&WBP_Sidebar_ButtonGeneric
   bIsVariable=true
   parent=FortScrollNavigationPanel
   widget=Button_ModeratorMode
   properties {
    iconMaterial=null
    minHeight=0
    minHeightOverride=60
    text=Moderator Mode
    useIconBlock=true
    visibility=Collapsed
    iconPadding=<ltrb 0 0 4 0>
   }
   slot {class=VerticalBoxSlot, properties {padding=<ltrb 0 0 0 6>}}
  }
  Button_ModeratorVLACapture {
   widgetClassPath=&WBP_Sidebar_ButtonGeneric
   bIsVariable=true
   parent=FortScrollNavigationPanel
   widget=Button_ModeratorVLACapture
   properties {
    iconMaterial=null
    minHeight=0
    minHeightOverride=60
    text=Moderator Mode VLACapture
    useIconBlock=true
    visibility=Collapsed
    iconPadding=<ltrb 0 0 4 0>
    renderTransform {angle=0, scale=<1 1>, shear=<0 0>, translation=<-1.1429 -1.1429>}
   }
   slot {class=VerticalBoxSlot, properties {padding=<ltrb 0 0 0 6>}}
  }
  Vbox_RegistryButtons {
   widgetClassPath=&VerticalBox
   bIsVariable=true
   parent=FortScrollNavigationPanel
   widget=Vbox_RegistryButtons
   properties {visibility=Collapsed}
   slot {class=VerticalBoxSlot, properties {padding=<ltrb 0 0 0 8>}}
  }
  Vbox_CreativeOptions {
   widgetClassPath=&VerticalBox
   bIsVariable=true
   parent=FortScrollNavigationPanel
   widget=Vbox_CreativeOptions
   slot {class=VerticalBoxSlot}
  }
  Button_Respawn {
   widgetClassPath=&WBP_Sidebar_ButtonGeneric
   bIsVariable=true
   parent=Vbox_CreativeOptions
   widget=Button_Respawn
   properties {minHeight=0, minHeightOverride=60, text=Respawn, visibility=Collapsed}
   slot {class=VerticalBoxSlot, properties {padding=<ltrb 0 0 0 6>}}
  }
  Button_TeamSelect {
   widgetClassPath=&WBP_Sidebar_ButtonGeneric
   bIsVariable=true
   parent=Vbox_CreativeOptions
   widget=Button_TeamSelect
   properties {minHeightOverride=60, text=Team select, visibility=Collapsed}
   slot {class=VerticalBoxSlot, properties {padding=<ltrb 0 0 0 6>}}
  }
  Button_IslandSettings {
   widgetClassPath=&WBP_Sidebar_ButtonGeneric
   bIsVariable=true
   parent=Vbox_CreativeOptions
   widget=Button_IslandSettings
   properties {minHeightOverride=60, text=Island settings, visibility=Collapsed}
   slot {class=VerticalBoxSlot, properties {padding=<ltrb 0 0 0 6>}}
  }
  Button_BackToHub {
   widgetClassPath=&WBP_Sidebar_ButtonGeneric
   bIsVariable=true
   parent=Vbox_CreativeOptions
   widget=Button_BackToHub
   properties {minHeightOverride=60, text=Back to hub, visibility=Collapsed}
   slot {class=VerticalBoxSlot, properties {padding=<ltrb 0 0 0 6>}}
  }
 ]
}
