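# WBP_WebSO_Panel is an activatable sidebar panel that hosts Fortnite's web social overlay.
# It places a FortWebSocialOverlayWidget inside the shared sidebar panel layout, with the back button hidden.
# SetCurrentTabByFeatureId puts the feature ID plus a "[web]" tag in the heading, then maps Social, Chat and AddFriend to the Friends, Chats and AddFriends tabs.
# On deactivation it resets the overlay tab to None.
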WidgetBlueprint /Sidebar/Panels/WBP_WebSO_Panel
graphs [
 EventGraph {nodes [
   Event On Deactivated()
   > SetCurrentTab!(
    execute
    self:Fort Web Social Overlay Widget Object Reference=Get FortWebSocialOverlayWidget() -> (FortWebSocialOverlayWidget:Fort Web Social Overlay Widget Object Reference)
    Tab:ESocialOverlayTab Enum=None
   ) -> (then)
  ]}
 SetCurrentTabByFeatureId {nodes [
   SetCurrentTabByFeatureId!() -> (then, FeatureId:Name) {is=title K2Node_FunctionEntry}
   > Set with Broadcast HeadingText!(
    execute
    HeadingText:Text=Format Text!(Format="LOCTEXT(\"{FeatureId} [web]\")", FeatureId:Name=Get FeatureId() -> (FeatureId:Name=None)) -> (Result)
    self:WBP Sidebar Panel Layout Object Reference=Get PanelLayout() -> (PanelLayout:WBP Sidebar Panel Layout Object Reference)
   ) -> (then, Output_Get:Text)
   > n1 = Switch on Name!(execute:Exec, Selection:Name=Get FeatureId() -> (FeatureId=None)) -> (Social:Exec, Chat:Exec, AddFriend:Exec) {is=title K2Node_SwitchName, bHasDefaultPin=false, pinNames [Social,Chat,AddFriend]}
   SetCurrentTab(execute=@n1.AddFriend, self=Get FortWebSocialOverlayWidget(), Tab=AddFriends)
   SetCurrentTab(execute=@n1.Chat, self=Get FortWebSocialOverlayWidget(), Tab=Chats)
   SetCurrentTab(execute=@n1.Social, self=Get FortWebSocialOverlayWidget(), Tab=Friends)
  ]}
]
variables []
widgets {
 info {
  inheritedWidgetCount=0
  namedSlotCount=1
  parentClass=&CommonActivatableWidget
  rootWidgetClass=/Sidebar/Shared/WBP_Sidebar_PanelLayout.WBP_Sidebar_PanelLayout_C
  widgetCount=2
 }
 widgets [
  PanelLayout {
   widgetClassPath=/Sidebar/Shared/WBP_Sidebar_PanelLayout
   bIsVariable=true
   widget=PanelLayout
   properties {headingText="Panel name [web]", showBackButton=false}
  }
  FortWebSocialOverlayWidget {
   widgetClassPath=/Script/FortniteUI.FortWebSocialOverlayWidget
   bIsVariable=true
   namedSlotHost=PanelLayout
   widget=FortWebSocialOverlayWidget
   properties {bAutoActivate=true}
  }
 ]
}
