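# UI StateTree for the sidebar overlay, switching between closed and open states, with Confirmation as a child of open.
# A root event transition enters Confirmation, and an event on OverlayOpen returns to the empty OverlayClosed state.
# Widget slotting tasks nest the panels: OverlayOpen puts a cached layout in UI.Sidebar.Overlay, and Confirmation puts an uncached confirmation panel in UI.Sidebar.OverlayPanel.
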StateTree /Sidebar/StateTree/ST_SidebarOverlay
states [
 Root {
  id=#1
  type=STATE
  children [#2, #3]
  tasks []
  transitions [{id=#4, trigger=ON_EVENT, state {id=#5, name=Confirmation}}]
 }
 OverlayClosed {id=#2, type=STATE, tasks [], transitions []}
 OverlayOpen {
  id=#3
  type=STATE
  children [#5]
  tasks [{
    id=#6
    description=UIState Tree Widget Slotting Task
    properties {
     iD=#7
     instance {
      _structType=/Script/UIStateTree.UIStateTreeWidgetSlottingTaskInstanceData
      bIsTaskActive=false
      widgets [{
        bAutoActivateWidget=true
        bCacheInstance=true
        bFill=true
        bIsPlayerSpecific=true
        bIsWidgetEnabled=true
        bKeepConstructed=true
        cacheEntryName=OverlayPanel
        memoryTag=None
        widget=/Sidebar/Panels/Overlay/WBP_Sidebar_OverlayPanelLayout.WBP_Sidebar_OverlayPanelLayout_C
        hostTag {tagName=UI.Sidebar.Overlay}
        padding=<ltrb 0 0 0 0>
       }]
     }
     node {
      name=None
      _structType=/Script/UIStateTree.UIStateTreeWidgetSlottingTask
      bCanEditConsideredForCompletion=true
      bConsideredForCompletion=false
      bHasTaskCompletionDelegateDispatcher=false
      bTaskEnabled=true
      transitionHandlingPriority=Normal
      bindingsBatch {value=65535}
      executionRuntimeTemplateIndex {value=65535}
      instanceDataHandle {index=65535, source=None, stateHandle {index=65535}}
      instanceTemplateIndex {value=65535}
      outputBindingsBatch {value=65535}
     }
    }
   }]
  transitions [{id=#8, trigger=ON_EVENT, state {id=#2, name=OverlayClosed}}]
 }
 Confirmation {
  id=#5
  type=STATE
  tasks [{
    id=#9
    description=UIState Tree Widget Slotting Task
    properties {
     iD=#10
     instance {
      _structType=&UIStateTreeWidgetSlottingTaskInstanceData
      bIsTaskActive=false
      widgets [{
        bAutoActivateWidget=true
        bCacheInstance=false
        bFill=true
        bIsPlayerSpecific=true
        bIsWidgetEnabled=true
        bKeepConstructed=true
        cacheEntryName=OverlayConfirmationPanel
        memoryTag=None
        widget=/Sidebar/Panels/Overlay/WBP_Confirmation_OverlayPanel.WBP_Confirmation_OverlayPanel_C
        hostTag {tagName=UI.Sidebar.OverlayPanel}
        padding=<ltrb 0 0 0 0>
       }]
     }
     node {
      name=None
      _structType=&UIStateTreeWidgetSlottingTask
      bCanEditConsideredForCompletion=true
      bConsideredForCompletion=false
      bHasTaskCompletionDelegateDispatcher=false
      bTaskEnabled=true
      transitionHandlingPriority=Normal
      bindingsBatch {value=65535}
      executionRuntimeTemplateIndex {value=65535}
      instanceDataHandle {index=65535, source=None, stateHandle {index=65535}}
      instanceTemplateIndex {value=65535}
      outputBindingsBatch {value=65535}
     }
    }
   }]
  transitions []
 }
]
